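# Material function that adds an animated Headband-series look to cosmetic item cards, passing its input through when bIsHeadbandSeries is off.
# A panned starburst streak pattern is added to the product of two powered, panning noise layers, all on rotated UVs.
# A vertical gradient darkens the sum, which SmoothStep remaps to index a color curve atlas row.
# Named reroutes define aspect-normalized UVs and scaled UVs with optional perspective warp.
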MaterialFunction /FortniteCoreUI/Frontend/CosmeticItemCard/Materials/MF_UI_HeadbandSeries
graph {nodes [
  n1 = ScalarParameter() {defaultValue=0.0857, group=Headband, parameterName=Min, sliderMax=1}
  n2 = ScalarParameter() {defaultValue=0.9524, group=Headband, parameterName=Max, sliderMax=1}
  n3 = VectorParameter() {group=Headband, parameterName=Streaks, channelNames {a="", b=Speed, g=StretchY, r=StretchX}, defaultValue=<rgba 0.7329 0.0052 0.1681 1>}
  n4 = MaterialFunctionCall(UVs=NamedRerouteUsage(), "Rotation Angle (0-1)"=ScalarParameter() {defaultValue=0.7173, group=Headband, parameterName=Rotation Angle}) {materialFunction=&CustomRotator}
  n5 = Panner(Coordinate=Multiply(A=@n4.Rotated Values, B=AppendVector(A=@n3.StretchX, B=@n3.StretchX)), Speed=AppendVector(A=Constant(), B=@n3.Speed))
  n6 = VectorParameter() {group=Headband, parameterName=Noise, channelNames {a=Power, b=Speed, g=StretchY, r=StretchX}, defaultValue=<rgba 3.6698 0.0826 0.5 -0.3543>}
  n7 = AppendVector(A=@n6.StretchX, B=@n6.StretchX)
  n8 = Constant()
  n9 = TextureSampleParameter2D(UVs=Panner(Coordinate=Multiply(A=@n4.Rotated Values, B=@n7), Speed=AppendVector(A=@n8, B=@n6.Speed))) {
   group=Headband
   parameterName=NoisePattern
   samplerType=SAMPLERTYPE_LinearColor
   texture=&HighContrastNoise
  }
  n10 = Power(Base=@n9.R, Exp=@n6.Power)
  n11 = VectorParameter() {
   group=Headband
   parameterName=Noise2
   channelNames {a=Power, b=Speed, g=StretchY, r=StretchX}
   defaultValue=<rgba 0.6505 0.0398 0.3619 0.3605>
  }
  n12 = AppendVector(A=@n11.StretchX, B=@n11.StretchX)
  n13 = Constant()
  n14 = TextureSampleParameter2D(UVs=Panner(Coordinate=Multiply(A=@n4.Rotated Values, B=@n12), Speed=AppendVector(A=@n13, B=@n11.Speed))) {
   group=Headband
   parameterName=NoisePattern
   samplerType=SAMPLERTYPE_LinearColor
   texture=&HighContrastNoise
  }
  n15 = Power(Base=@n14.R, Exp=@n11.Power)
  n16 = Constant()
  n17 = ScalarParameter() {defaultValue=0.1714, group=Headband, parameterName=GradientOffset}
  n18 = ComponentMask(None=Add(A=TextureCoordinate(), B=AppendVector(A=@n16, B=@n17))) {g=true}
  n19 = ScalarParameter() {defaultValue=0.8401, group=Headband, parameterName=DarkenGradient}
  n20 = Multiply(
   A=Add(
    A=TextureSampleParameter2D(UVs=@n5) {group=Headband, parameterName=Pattern, texture=/Game/Global/Textures/Default/PointSplatter/T_UI_StarBurst_Scatter}.R
    B=Multiply(A=@n10, B=@n15)
   )
   B=OneMinus(None=Multiply(A=@n18, B=@n19))
  )
  n21 = CurveAtlasRowParameter(CurveTime=ComponentMask(None=SmoothStep(Min=@n1, Max=@n2, Value=@n20)) {r=true}) {
   atlas=/Game/UI/Frontend/ItemShop/Art/ItemCardColorAtlas
   curve=/Game/UI/Frontend/ItemShop/Art/ItemCard_Headband
   group=Headband
   parameterName=CurveParamsBG
  }
  FunctionOutput(None=StaticSwitchParameter(True=@n21, False=FunctionInput()) {parameterName=bIsHeadbandSeries})
  LinearInterpolate(
   A=FunctionInput(Preview=VectorParameter() {group=CorruptSeries, parameterName=BG_Color, defaultValue=<rgba 1 1 1 1>}.RGB) {bUsePreviewValueAsDefault=true, inputName=RGB 1, sortPriority=1}
   B=FunctionInput(Preview=VectorParameter() {group=CorruptSeries, parameterName=CellColor, defaultValue=<rgba 0 0 0 1>}.RGB) {bUsePreviewValueAsDefault=true, inputName=RGB 2, sortPriority=1}
  )
  n22 = MaterialFunctionCall(Float2=MaterialFunctionCall() {materialFunction=&ScreenResolution}.Visible Resolution) {materialFunction=&BreakOutFloat2Components}
  n23 = Max(A=@n22.G)
  Multiply(
   A=AppendVector(A=Divide(A=@n22.R, B=@n23), B=Constant() {r=1})
   B=MaterialFunctionCall() {materialFunction=&ScreenAlignedUVs}."X 100%, Y 100%"
  )
  n24 = VectorParameter() {group=Headband, parameterName=HeadbandScale, channelNames {a=OffsetY, b=OffsetX, g=Y, r=X}, defaultValue=<rgba 1 1 0 0>}
  n25 = MaterialFunctionCall!(
   Texture Scale=AppendVector(A=@n24.X, B=@n24.X)
   UVs=NamedRerouteUsage()
   Offset XY Texturesize Z=MaterialFunctionCall(R=@n24.OffsetX, G=@n24.OffsetX, B=Constant() {r=1}) {materialFunction=&AppendMany}.RGB
  ) {materialFunction=/Engine/Functions/Engine_MaterialFunctions01/Texturing/ScaleUVsAroundPoint}
  n26 = VectorParameter() {group=Headband, parameterName=HeadbandPerspective, channelNames {a=YSlide, b=XSlide, g=Y, r=X}}
  n27 = MaterialFunctionCall!(UV=@n25.UVs, Texture, FixedScaleFactor_X, FixedScaleFactor_Y, Persp_Y=@n26.Y, Persp_X=@n26.Y, Slide_X=@n26.Y, Slide_Y=@n26.Y, Flip_Y, Flip_X) {materialFunction=/Game/UI/Foundation/Materials/MF_UI_BasicPerspective}
  NamedRerouteDeclaration(None=Add(A=StaticSwitchParameter(True=@n27.FinalUvs, False=@n25.UVs) {group=Headband, parameterName=HeadbandUsePerpsective}) {constB=0}) {name=NormalizedSSUVs, nodeColor=<rgba 1 0 0.7969 1>}
  n28 = MaterialFunctionCall(Float2=MaterialFunctionCall() {materialFunction=&GetUserInterfaceUV}.Pixel Size) {materialFunction=&BreakOutFloat2Components}
  n29 = Divide(A=@n28.G, B=@n28.G)
  n30 = Constant() {r=1}
  NamedRerouteDeclaration(None=MaterialFunctionCall(UVs=TextureCoordinate(), Texture Scale=AppendVector(A=@n29, B=@n30)) {materialFunction=&ScaleUVsByCenter}.UVs) {name=NormalizedUVs, nodeColor=<rgba 0.8047 1 0 1>}
  Reroute()
  Reroute()
  TextureCoordinate()
 ]}
