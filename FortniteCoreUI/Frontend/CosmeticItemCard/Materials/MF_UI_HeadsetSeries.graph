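# Material function for cosmetic item cards that swaps in the Headset series texture through the bIsHeadsetSeries static switch.
# When on, it outputs the RGB of T_UI_Headset, sampled with MF_UI_UVScaling's ScaleY UVs scaled about their center.
# When off, it passes a function input through.
# The RGB 1 and RGB 2 inputs and a pixel-size-based UV correction branch are not connected to the output.
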MaterialFunction /FortniteCoreUI/Frontend/CosmeticItemCard/Materials/MF_UI_HeadsetSeries
graph {nodes [
  FunctionInput() {inputName=RGB 1, sortPriority=1}
  FunctionInput() {inputName=RGB 2, sortPriority=1}
  n1 = MaterialFunctionCall(
   UVs=MaterialFunctionCall!(UV, TextureSize, ScalePivotPoint) {materialFunction=&MF_UI_UVScaling}.ScaleY
   Texture Scale=Constant() {r=0.85}
  ) {materialFunction=&ScaleUVsByCenter}
  FunctionOutput(None=StaticSwitchParameter(True=TextureSample(UVs=@n1.UVs) {texture=/OfferCatalog/Art/Headset/T_UI_Headset}.RGB, False=FunctionInput()) {parameterName=bIsHeadsetSeries})
  n2 = MaterialFunctionCall() {materialFunction=&GetUserInterfaceUV}
  n3 = MaterialFunctionCall(Float2=@n2.Pixel Size) {materialFunction=&BreakOutFloat2Components}
  n4 = Divide(A=@n3.G, B=@n3.G)
  n5 = AppendVector(A=Constant() {r=1}, B=@n4)
  n6 = Constant()
  n7 = Multiply(A=OneMinus(None=@n4)) {constB=0.5}
  n8 = MaterialFunctionCall(Float2=MaterialFunctionCall() {materialFunction=&GetUserInterfaceUV}.Pixel Size) {materialFunction=&BreakOutFloat2Components}
  n9 = Constant()
  n10 = MaterialFunctionCall() {materialFunction=&GetUserInterfaceUV}
  n11 = MaterialFunctionCall(Float2=@n10.Pixel Size) {materialFunction=&BreakOutFloat2Components}
  n12 = Divide(A=@n11.R, B=@n11.R)
  n13 = AppendVector(A=@n12, B=Constant() {r=1})
  n14 = Multiply(A=OneMinus(None=@n12)) {constB=0.5}
  n15 = Constant()
  LinearInterpolate(
   A=Add(A=Multiply(A=@n2.Normalized UV, B=@n5), B=AppendVector(A=@n6, B=@n7))
   B=Add(A=Multiply(A=@n10.Normalized UV, B=@n13), B=AppendVector(A=@n14, B=@n15))
   Alpha=If(A=@n8.R, B=@n8.R, "A > B"=@n9, "A == B"=@n9, "A < B"=Constant() {r=1})
  )
 ]}
