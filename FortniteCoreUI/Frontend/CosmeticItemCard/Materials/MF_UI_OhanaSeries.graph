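# Material function for Fortnite cosmetic item cards that adds the Ohana series background when bIsOhanaSeries is on, otherwise passing its input through.
# Aspect-corrected UVs and panning-noise distortion feed MF_UI_Confetti, producing sparks masked by a vertical gradient and tinted with two colors.
# The sparks, with a faint sine-time pulse and overdriven lerp, blend with a bottom-centered radial gradient between RGB 1 and RGB 2.
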MaterialFunction /FortniteCoreUI/Frontend/CosmeticItemCard/Materials/MF_UI_OhanaSeries
graph {nodes [
  n1 = OneMinus(None=Sine(None=Time()))
  n2 = VectorParameter() {group=Sparks, parameterName=Sparks Dark Color, defaultValue=<rgba 0.1928 1 0 1>}
  n3 = VectorParameter() {group=Sparks, parameterName=Sparks Light Color, defaultValue=<rgba 1 0.6174 0 1>}
  n4 = Add(A=NamedRerouteUsage(), B=NamedRerouteUsage())
  n5 = AppendVector(
   A=ScalarParameter() {defaultValue=8.7879, parameterName=L1-Patter-X-Offset}
   B=ScalarParameter() {defaultValue=18.2159, parameterName=L1-Patter-Y-Offset}
  )
  n6 = ScalarParameter() {defaultValue=11.4012, parameterName=L1-Pattern-X}
  n7 = ScalarParameter() {defaultValue=-3.2956, parameterName=L1-Pattern-Y}
  n8 = ScalarParameter() {defaultValue=117.7869, parameterName=L1-Confetti-Local-Tile-X}
  n9 = ScalarParameter() {defaultValue=27.9364, parameterName=L1-Confetti-Local-Tile-Y}
  n10 = ScalarParameter() {defaultValue=4.2917, parameterName=L1-SmoothStep-Max}
  n11 = ScalarParameter() {defaultValue=8.1826, parameterName=L1-SmoothStep-Min}
  n12 = TextureCoordinate()
  n13 = Subtract(A=OneMinus(None=ComponentMask(None=@n12) {g=true}), B=ScalarParameter() {defaultValue=0.3, parameterName=L1-Gradient-Mask})
  n14 = ScalarParameter() {defaultValue=2, parameterName=L1-Gradient-Falloff}
  n15 = Multiply(
   A=MaterialFunctionCall!(
    "UV(V2)"=Add(A=@n4, B=@n5)
    Global-Angle=ScalarParameter() {defaultValue=0.0391, parameterName=L1-Global Angle}
    "Pattern-Tiles(V2)"=AppendVector(A=@n6, B=@n7)
    RandomSeed=ScalarParameter() {defaultValue=628.0901, parameterName=L1-RandomSeed}
    Confetti-Speed-TimeFactor=ScalarParameter() {defaultValue=1.1381, parameterName=L1-Confetti-Speed-TimeFactor}
    Confetti-Rotation-TimeFactor=ScalarParameter() {parameterName=L1-Confetti-Rotation-TimeFactor}
    Confetti-Local-Rotation-Range=ScalarParameter() {defaultValue=304.3395, parameterName=L1-Confetti-Local-Rotation-Range}
    "Confetti-Local-Tiles(V2)"=AppendVector(A=@n8, B=@n9)
    "SmoothStepRanges(V2)"=AppendVector(A=@n10, B=@n11)
   ) {materialFunction=/Game/UI/Frontend/ItemShop/Art/Cactus/MF_UI_Confetti}.Confetti-Mask
   B=Saturate(None=Multiply(A=@n13, B=@n14))
  )
  n16 = ScalarParameter() {defaultValue=1, parameterName=L1-Intensity}
  n17 = LinearInterpolate(
   A=Clamp(None=Multiply(A=@n1) {constB=0.0035})
   B=LinearInterpolate(B=LinearInterpolate(A=@n2.RGB, B=@n3.RGB), Alpha=Multiply(A=@n15, B=@n16))
  ) {constAlpha=3.6765}
  n18 = AppendVector(A=ScalarParameter() {defaultValue=0.6, parameterName=Radial-X-Tile}, B=ScalarParameter() {defaultValue=0.95, parameterName=Radial-Y-Tile})
  n19 = Distance(A=Multiply(A=NamedRerouteUsage(), B=@n18), B=Multiply(A=@n18, B=Constant2Vector() {g=1, r=0.5}))
  n20 = ScalarParameter() {defaultValue=1.25, parameterName=Distance Power}
  n21 = LinearInterpolate(
   A=FunctionInput(Preview=VectorParameter() {parameterName=RGB1, defaultValue=<rgba 0.1912 1 0 1>}.RGB) {inputName=RGB 1, sortPriority=1}
   B=FunctionInput(Preview=Constant3Vector()) {inputName=RGB 2, sortPriority=1}
   Alpha=Saturate(None=Power(Base=@n19, Exp=@n20))
  )
  FunctionOutput(None=StaticSwitchParameter(True=LinearInterpolate(A=@n17, B=@n21), False=FunctionInput()) {parameterName=bIsOhanaSeries})
  n22 = MaterialFunctionCall() {materialFunction=&GetUserInterfaceUV}
  n23 = MaterialFunctionCall(Float2=@n22.Pixel Size) {materialFunction=&BreakOutFloat2Components}
  n24 = Divide(A=@n23.G, B=@n23.G)
  n25 = AppendVector(A=Constant() {r=0.9336}, B=@n24)
  n26 = Constant()
  n27 = Multiply(A=OneMinus(None=@n24)) {constB=0.5}
  n28 = MaterialFunctionCall(Float2=MaterialFunctionCall() {materialFunction=&GetUserInterfaceUV}.Pixel Size) {materialFunction=&BreakOutFloat2Components}
  n29 = Constant()
  n30 = MaterialFunctionCall() {materialFunction=&GetUserInterfaceUV}
  n31 = MaterialFunctionCall(Float2=@n30.Pixel Size) {materialFunction=&BreakOutFloat2Components}
  n32 = Divide(A=@n31.R, B=@n31.R)
  n33 = AppendVector(A=@n32, B=Constant() {r=1})
  n34 = Multiply(A=OneMinus(None=@n32)) {constB=0.5}
  n35 = Constant()
  LinearInterpolate(
   A=Add(A=Multiply(A=@n22.Normalized UV, B=@n25), B=AppendVector(A=@n26, B=@n27))
   B=Add(A=Multiply(A=@n30.Normalized UV, B=@n33), B=AppendVector(A=@n34, B=@n35))
   Alpha=If(A=@n28.R, B=@n28.R, "A > B"=@n29, "A == B"=@n29, "A < B"=Constant() {r=1})
  )
  n36 = MaterialFunctionCall() {materialFunction=&GetUserInterfaceUV}
  n37 = MaterialFunctionCall(Float2=@n36.Pixel Size) {materialFunction=&BreakOutFloat2Components}
  n38 = AppendVector(A=Divide(A=@n37.R, B=Max(A=@n37.G)), B=Constant() {r=1})
  n39 = OneMinus(None=@n38)
  NamedRerouteDeclaration(None=Add(A=Multiply(A=@n39) {constB=0.5}, B=Multiply(A=@n38, B=@n36.Normalized UV))) {name="Squared UV - (x/y,y)", nodeColor=<rgba 0 0 0 1>}
  n40 = TextureCoordinate()
  n41 = ScalarParameter() {defaultValue=0.5, parameterName=DupeScaler}
  n42 = Time()
  n43 = ScalarParameter() {defaultValue=0.6346, parameterName=TimeFactor}
  n44 = Power(
   Base=Panner(Coordinate=Multiply(A=@n40, B=@n41), Time=Multiply(A=@n42, B=@n43)) {speedX=0.058, speedY=0.096}
   Exp=ScalarParameter() {defaultValue=0.8088, parameterName=ConfettiParam2}
  )
  NamedRerouteDeclaration(None=Multiply(
    A=TextureSample(UVs=@n44) {samplerType=SAMPLERTYPE_Grayscale, texture=&LowResBlurredNoise}.R
    B=ScalarParameter() {defaultValue=0.35, parameterName=DistortionScaler}
   )) {name=UV-Distortion, nodeColor=<rgba 0 0 0 1>}
 ]}
