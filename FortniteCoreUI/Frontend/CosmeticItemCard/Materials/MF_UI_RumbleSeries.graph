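# Material function that draws the Rumble series background for item shop cosmetic cards, enabled by static switch bIsRumbleSeries.
# A scaled, rotated curve-atlas gradient gets half-strength spike and full lightning textures added.
# A tunable top-and-bottom edge mask then blends toward RGB 2; RGB 1 is unused.
# The textures' aspect-fit UV math compares and divides pixel sizes by themselves, so it reduces to a fixed horizontal scale.
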MaterialFunction /FortniteCoreUI/Frontend/CosmeticItemCard/Materials/MF_UI_RumbleSeries
graph {nodes [
  FunctionInput() {inputName=RGB 1, sortPriority=1}
  n1 = MaterialFunctionCall() {materialFunction=&GetUserInterfaceUV}
  n2 = MaterialFunctionCall(Float2=@n1.Pixel Size) {materialFunction=&BreakOutFloat2Components}
  n3 = Divide(A=@n2.G, B=@n2.G)
  n4 = AppendVector(A=Constant() {r=0.9336}, B=@n3)
  n5 = Constant()
  n6 = Multiply(A=OneMinus(None=@n3)) {constB=0.5}
  n7 = MaterialFunctionCall(Float2=MaterialFunctionCall() {materialFunction=&GetUserInterfaceUV}.Pixel Size) {materialFunction=&BreakOutFloat2Components}
  n8 = Constant()
  n9 = MaterialFunctionCall() {materialFunction=&GetUserInterfaceUV}
  n10 = MaterialFunctionCall(Float2=@n9.Pixel Size) {materialFunction=&BreakOutFloat2Components}
  n11 = Divide(A=@n10.R, B=@n10.R)
  n12 = AppendVector(A=@n11, B=Constant() {r=1})
  n13 = Multiply(A=OneMinus(None=@n11)) {constB=0.5}
  n14 = Constant()
  n15 = LinearInterpolate(
   A=Add(A=Multiply(A=@n1.Normalized UV, B=@n4), B=AppendVector(A=@n5, B=@n6))
   B=Add(A=Multiply(A=@n9.Normalized UV, B=@n12), B=AppendVector(A=@n13, B=@n14))
   Alpha=If(A=@n7.R, B=@n7.R, "A > B"=@n8, "A == B"=@n8, "A < B"=Constant() {r=1})
  )
  n16 = MaterialFunctionCall(Texture Scale=ScalarParameter() {defaultValue=1.1897, group=Gradient, parameterName=GradientScale}) {materialFunction=&ScaleUVsByCenter}
  n17 = ScalarParameter() {defaultValue=1.2318, group=Gradient, parameterName=GradientAngle, sliderMin=0.832}
  n18 = CurveAtlasRowParameter(CurveTime=ComponentMask(None=MaterialFunctionCall(UVs=@n16.UVs, "Rotation Angle (0-1)"=@n17) {materialFunction=&CustomRotator}.Rotated Values) {r=true}) {
   atlas=/Game/UI/Frontend/ItemShop/Art/ItemCardColorAtlas
   curve=/Game/UI/Frontend/ItemShop/Art/ItemCard_Rumble
   defaultValue=6
   group=Gradient
   parameterName=CurveParamsBG
  }
  n19 = Multiply(A=TextureSample(UVs=@n15) {texture=/OfferCatalog/Art/Rumble/T_UI_RumbleSpike}.RGB, B=Constant() {r=1})
  n20 = Add(A=TextureSample(UVs=@n15) {texture=/OfferCatalog/Art/Rumble/T_UI_Lightning}.RGB, B=LinearInterpolate(A=@n18, B=Add(A=@n19, B=@n18)))
  n21 = FunctionInput(Preview=VectorParameter() {parameterName=RGB2, defaultValue=<rgba 0.5711 0.2747 0 1>}.RGB) {inputName=RGB 2, sortPriority=1}
  n22 = ConstantBiasScale(None=TextureCoordinate()) {bias=-0.5, scale=2}
  n23 = Subtract(A=ComponentMask(None=Abs(None=@n22)) {g=true}, B=ScalarParameter() {defaultValue=0.7349, parameterName=Subtract})
  n24 = ScalarParameter() {defaultValue=1.4621, parameterName=Falloff}
  n25 = Multiply(A=Saturate(None=Multiply(A=@n23, B=@n24)), B=ScalarParameter() {defaultValue=2.5483, parameterName=Intensity})
  FunctionOutput(None=StaticSwitchParameter(True=LinearInterpolate(A=@n20, B=@n21, Alpha=@n25), False=FunctionInput()) {parameterName=bIsRumbleSeries})
  VectorParameter() {parameterName=RGB1, defaultValue=<rgba 0.807 0.4035 0 1>}
 ]}
