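# Material function that builds the Snowfall series background for cosmetic item cards, selected by the bIsSnowfallSeries static switch, which otherwise passes an input through.
# A radially blended scale texture plus a rotated, panning gradient ramp is mixed with RGB 1, using RGB 2 as blend weight.
# Tinted vignettes and a vertical gradient, masked by scales, blend this toward a constant before RGB5 is added.
# The scale UVs' aspect-fit math divides values by themselves, so it applies no correction.
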MaterialFunction /FortniteCoreUI/Frontend/CosmeticItemCard/Materials/MF_UI_SnowfallSeries
graph {nodes [
  n1 = VectorParameter() {parameterName=RGB5, defaultValue=<rgba 0.8073 0.2422 0 1>}
  n2 = VectorParameter() {parameterName=RGB1, defaultValue=<rgba 0.807 0.4035 0 1>}
  n3 = MaterialFunctionCall() {materialFunction=&GetUserInterfaceUV}
  n4 = MaterialFunctionCall(Float2=@n3.Pixel Size) {materialFunction=&BreakOutFloat2Components}
  n5 = Divide(A=@n4.G, B=@n4.G)
  n6 = AppendVector(A=Constant() {r=1}, B=@n5)
  n7 = Constant()
  n8 = Multiply(A=OneMinus(None=@n5)) {constB=0.5}
  n9 = MaterialFunctionCall(Float2=MaterialFunctionCall() {materialFunction=&GetUserInterfaceUV}.Pixel Size) {materialFunction=&BreakOutFloat2Components}
  n10 = Constant()
  n11 = MaterialFunctionCall() {materialFunction=&GetUserInterfaceUV}
  n12 = MaterialFunctionCall(Float2=@n11.Pixel Size) {materialFunction=&BreakOutFloat2Components}
  n13 = Divide(A=@n12.R, B=@n12.R)
  n14 = AppendVector(A=@n13, B=Constant() {r=1})
  n15 = Multiply(A=OneMinus(None=@n13)) {constB=0.5}
  n16 = Constant()
  n17 = LinearInterpolate(
   A=Add(A=Multiply(A=@n3.Normalized UV, B=@n6), B=AppendVector(A=@n7, B=@n8))
   B=Add(A=Multiply(A=@n11.Normalized UV, B=@n14), B=AppendVector(A=@n15, B=@n16))
   Alpha=If(A=@n9.R, B=@n9.R, "A > B"=@n10, "A == B"=@n10, "A < B"=Constant() {r=1})
  )
  n18 = LinearInterpolate(
   B=TextureSampleParameter2D(UVs=@n17) {parameterName=Scale Texture, texture=/OfferCatalog/Art/Snowfall/Textures/T_UI_Scales}.RGB
   Alpha=MaterialFunctionCall(Radius=Constant() {r=1.2843}) {materialFunction=&RadialGradientExponential}.RadialGradientExponential
  ) {constA=4.7963, constAlpha=0.8676, constB=1.0837}
  n19 = MaterialFunctionCall(
   UVs=MaterialFunctionCall(Texture Scale=ScalarParameter() {defaultValue=1.5678, group=Gradient, parameterName=GradientScale}) {materialFunction=&ScaleUVsByCenter}.UVs
   "Rotation Angle (0-1)"=ScalarParameter() {defaultValue=2.7, group=Gradient, parameterName=GradientAngle}
  ) {materialFunction=&CustomRotator}
  n20 = TextureSample(UVs=Panner(Coordinate=ComponentMask(None=@n19.Rotated Values) {r=true}, Speed=Constant2Vector() {g=0.15, r=0.15}) {speedX=0.816, speedY=0.6}) {texture=/OfferCatalog/Art/Snowfall/Textures/T_UI_GradientRamp}
  n21 = VectorParameter() {parameterName=RGB2, defaultValue=<rgba 0.807 0.3093 0 1>}
  n22 = TextureCoordinate()
  n23 = MaterialFunctionCall(UVs=@n22, Texture Scale=AppendVector(A=ScalarParameter() {defaultValue=1, parameterName=Vignette_G_Scale_X}, B=Constant() {r=1})) {materialFunction=&ScaleUVsByCenter}
  n24 = AppendVector(A=Constant() {r=0.5}, B=ScalarParameter() {defaultValue=0.5, parameterName=Vignette_G_Position_Y})
  n25 = ScalarParameter() {defaultValue=0.5, parameterName=Vignette_G_Radius}
  n26 = ScalarParameter() {defaultValue=2, parameterName=Vignette_G_Power}
  n27 = MaterialFunctionCall(UVs=@n22, Texture Scale=AppendVector(A=Constant() {r=1}, B=ScalarParameter() {defaultValue=1.5, parameterName=Vignette_R_Scale_Y})) {materialFunction=&ScaleUVsByCenter}
  n28 = Constant() {r=0.5}
  n29 = ScalarParameter() {defaultValue=1666.3051, parameterName=Vignette_R_Radius}
  n30 = ScalarParameter() {defaultValue=2, parameterName=Vignette_R_Power}
  n31 = Multiply(
   A=LinearInterpolate(
    B=MaterialFunctionCall(UVs=@n23.UVs, CenterPosition=@n24, Radius=@n25, Density=@n26) {materialFunction=&RadialGradientExponential}.RadialGradientExponential
    Alpha=MaterialFunctionCall(UVs=@n27.UVs, CenterPosition=@n28, Radius=@n29, Density=@n30) {materialFunction=&RadialGradientExponential}.RadialGradientExponential
   )
   B=VectorParameter() {parameterName=RGB3, defaultValue=<rgba 0.807 0.4035 0 1>}.RGB
  )
  n32 = TextureCoordinate()
  n33 = Multiply(
   A=Power(Base=ComponentMask(None=@n32) {g=true}, Exp=ScalarParameter() {defaultValue=7.0551, parameterName=ContrastBackground2})
   B=VectorParameter() {parameterName=RGB4, defaultValue=<rgba 1 1 1 1>}.RGB
  )
  n34 = LinearInterpolate(
   A=LinearInterpolate(
    A=FunctionInput(Preview=@n2.RGB) {inputName=RGB 1, sortPriority=1}
    B=Add(A=@n18, B=@n20.RGB)
    Alpha=FunctionInput(Preview=@n21.RGB) {inputName=RGB 2, sortPriority=1}
   ) {constA=0.1105}
   Alpha=Multiply(A=Add(A=@n31, B=@n33), B=TextureSample(UVs=@n17) {texture=&T_UI_Scales}.RGB)
  ) {constB=0.5009}
  FunctionOutput(None=StaticSwitchParameter(True=Add(A=@n1.RGB, B=@n34), False=FunctionInput()) {parameterName=bIsSnowfallSeries})
 ]}
