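# Material function that builds the Textile series pattern for cosmetic item card UI, passing its input through when bIsTextileSeries is off.
# Rerouted UVs are scaled around a point, optionally perspective-warped, then tiled and panned to sample the pattern texture.
# A rotated, offset linear gradient blends the pattern with its inverse, and the result is inverted and multiplied by Darken.
# Color-blend and pixel-size aspect-correction branches remain unconnected to the output.
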MaterialFunction /FortniteCoreUI/Frontend/CosmeticItemCard/Materials/MF_UI_TextileSeries
graph {nodes [
  n1 = VectorParameter() {group=Textile, parameterName=TextileScale, channelNames {a=OffsetY, b=OffsetX, g=Y, r=X}, defaultValue=<rgba 1 1 0 0>}
  n2 = MaterialFunctionCall!(
   Texture Scale=AppendVector(A=@n1.X, B=@n1.X)
   UVs=NamedRerouteUsage()
   Offset XY Texturesize Z=MaterialFunctionCall(R=@n1.OffsetX, G=@n1.OffsetX, B=Constant() {r=1}) {materialFunction=&AppendMany}.RGB
  ) {materialFunction=/Engine/Functions/Engine_MaterialFunctions01/Texturing/ScaleUVsAroundPoint}
  n3 = VectorParameter() {group=Textile, parameterName=TextilePerspective, channelNames {a=YSlide, b=XSlide, g=Y, r=X}}
  n4 = Add(A=StaticSwitchParameter(
    True=MaterialFunctionCall!(UV=@n2.UVs, Texture, FixedScaleFactor_X, FixedScaleFactor_Y, Persp_Y=@n3.Y, Persp_X=@n3.Y, Slide_X=@n3.Y, Slide_Y=@n3.Y, Flip_Y, Flip_X) {materialFunction=/Game/UI/Foundation/Materials/MF_UI_BasicPerspective}.FinalUvs
    False=@n2.UVs
   ) {group=Textile, parameterName=TextileUsePerpsective}) {constB=0}
  n5 = NamedRerouteDeclaration(None=@n4) {name=NormalizedSSUVs, nodeColor=<rgba 1 0 0.7969 1>}
  n6 = ScalarParameter() {defaultValue=4, group=Textile, parameterName=TextilePatternRepeat}
  n7 = VectorParameter() {group=Textile, parameterName=TextilePanSpeed, channelNames {a="", b="", g=Y, r=X}, defaultValue=<rgba 0 0 0 1>}
  n8 = TextureSampleParameter2D(UVs=Panner(Coordinate=Multiply(A=@n5, B=@n6), Speed=AppendVector(A=@n7.X, B=@n7.X))) {group=Textile, parameterName=TextilePattern, texture=/OfferCatalog/Art/SpecialShopBG/T_UI_ShopBG_BCG}
  n9 = ComponentMask(None=MaterialFunctionCall(UVs=TextureCoordinate(), "Rotation Angle (0-1)"=ScalarParameter() {defaultValue=0.25, group=Textile, parameterName=TextileInvertAngle}) {materialFunction=&CustomRotator}.Rotated Values) {g=true}
  n10 = ScalarParameter() {group=Textile, parameterName=TextileInvertAngleMuilt}
  n11 = TextureSample(UVs=Add(A=Multiply(A=@n9, B=@n10), B=ScalarParameter() {group=Textile, parameterName=TextileInvertOffset})) {samplerType=SAMPLERTYPE_LinearColor, texture=/Game/Global/Textures/Default/Masks_and_Gradients/linear_gradient}
  n12 = OneMinus(None=LinearInterpolate(A=OneMinus(None=@n8.RGB), B=@n8.RGB, Alpha=OneMinus(None=@n11.R)))
  n13 = ScalarParameter() {defaultValue=1, group=Textile, parameterName=Darken}
  FunctionOutput(None=StaticSwitchParameter(True=Multiply(A=@n12, B=@n13), False=FunctionInput()) {parameterName=bIsTextileSeries})
  LinearInterpolate(
   A=FunctionInput(Preview=VectorParameter() {group=CorruptSeries, parameterName=BG_Color, defaultValue=<rgba 1 1 1 1>}.RGB) {bUsePreviewValueAsDefault=true, inputName=RGB 1, sortPriority=1}
   B=FunctionInput(Preview=VectorParameter() {group=CorruptSeries, parameterName=CellColor, defaultValue=<rgba 0 0 0 1>}.RGB) {bUsePreviewValueAsDefault=true, inputName=RGB 2, sortPriority=1}
  )
  n14 = MaterialFunctionCall() {materialFunction=&GetUserInterfaceUV}
  n15 = MaterialFunctionCall(Float2=@n14.Pixel Size) {materialFunction=&BreakOutFloat2Components}
  n16 = Divide(A=@n15.G, B=@n15.G)
  n17 = AppendVector(A=Constant() {r=1}, B=@n16)
  n18 = Constant()
  n19 = Multiply(A=OneMinus(None=@n16)) {constB=0.5}
  n20 = MaterialFunctionCall(Float2=MaterialFunctionCall() {materialFunction=&GetUserInterfaceUV}.Pixel Size) {materialFunction=&BreakOutFloat2Components}
  n21 = Constant()
  n22 = MaterialFunctionCall() {materialFunction=&GetUserInterfaceUV}
  n23 = MaterialFunctionCall(Float2=@n22.Pixel Size) {materialFunction=&BreakOutFloat2Components}
  n24 = Divide(A=@n23.R, B=@n23.R)
  n25 = AppendVector(A=@n24, B=Constant() {r=1})
  n26 = Multiply(A=OneMinus(None=@n24)) {constB=0.5}
  n27 = Constant()
  LinearInterpolate(
   A=Add(A=Multiply(A=@n14.Normalized UV, B=@n17), B=AppendVector(A=@n18, B=@n19))
   B=Add(A=Multiply(A=@n22.Normalized UV, B=@n25), B=AppendVector(A=@n26, B=@n27))
   Alpha=If(A=@n20.R, B=@n20.R, "A > B"=@n21, "A == B"=@n21, "A < B"=Constant() {r=1})
  )
  MaterialFunctionCall() {materialFunction=&GetUserInterfaceUV}
  n28 = MaterialFunctionCall(Float2=MaterialFunctionCall() {materialFunction=&ScreenAlignedUVs}."X 100%, Y 100%") {materialFunction=&BreakOutFloat2Components}
  Multiply(A=@n4, B=AppendVector(A=Divide(A=@n28.R, B=@n28.R), B=Constant() {r=1}))
  n29 = MaterialFunctionCall(Float2=MaterialFunctionCall() {materialFunction=&ScreenResolution}.Visible Resolution) {materialFunction=&BreakOutFloat2Components}
  n30 = Divide(A=@n29.R, B=Max(A=@n29.G))
  n31 = Constant() {r=1}
  NamedRerouteDeclaration(None=Multiply(
    A=AppendVector(A=@n30, B=@n31)
    B=MaterialFunctionCall() {materialFunction=&ScreenAlignedUVs}."X 100%, Y 100%"
   )) {name=NormalizedUVs, nodeColor=<rgba 0.5469 1 0 1>}
 ]}
