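# Material function for the Alfredo series look in offer catalog UI, toggled by the AlfredoSeries static switch, which otherwise passes the input through.
# A feathered edge frame and exponential radial gradient, each tinted, are blended as the base glow.
# This blends with a tinted pattern masked by a panning tile texture, static squares and two scrolling cell-noise samples, using aspect-aware UI UVs.
# Several pixel-size, CorruptSeries color and second-frame branches are unconnected.
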MaterialFunction /OfferCatalog/Art/Alfredo/MF_UI_Alfredo_Series
graph {nodes [
  n1 = MaterialFunctionCall() {materialFunction=&GetUserInterfaceUV}
  n2 = Subtract(A=@n1.Pixel Size, B=Multiply(A=ScalarParameter() {defaultValue=2, parameterName=PixelSize}) {constB=2})
  Divide(A=ComponentMask(None=@n2) {r=true}, B=ComponentMask(None=@n1.Pixel Size) {r=true})
  Divide(A=ComponentMask(None=@n2) {g=true}, B=ComponentMask(None=@n1.Pixel Size) {g=true})
  n3 = MaterialFunctionCall() {materialFunction=&GetUserInterfaceUV}
  n4 = Subtract(A=@n3.Pixel Size, B=Multiply(A=ScalarParameter() {defaultValue=2, parameterName=PixelSize}) {constB=2})
  Divide(A=ComponentMask(None=@n4) {r=true}, B=ComponentMask(None=@n3.Pixel Size) {r=true})
  Divide(A=ComponentMask(None=@n4) {g=true}, B=ComponentMask(None=@n3.Pixel Size) {g=true})
  n5 = Multiply(
   A=AppendVector(A=ScalarParameter() {defaultValue=8, parameterName="U_Tiling(C)"}, B=ScalarParameter() {defaultValue=4, parameterName="V_Tiling(C)"})
   B=MaterialFunctionCall!(UV, TextureSize) {materialFunction=&MF_UI_UVScaling}.Scaled to Fit
  )
  n6 = TextureSampleParameter2D!(
   UVs=MaterialFunctionCall(UVs=Panner(Coordinate=@n5) {speedX=0.5}, "Rotation Angle (0-1)"=Constant()) {materialFunction=&CustomRotator}.Rotated Values
   Level
   Apply View MipBias
  ) {mipValueMode=TMVM_MipLevel, parameterName=Texture_Input_C, texture=/Game/Effects/Fort_Effects/Textures/Tile/Tests/T_Tile_Panner_Mask}
  n7 = ScalarParameter() {defaultValue=0.1, group=Tech, parameterName=Tech Noise Add}
  n8 = Multiply(A=Power(Base=Add(A=@n6.G, B=@n7)), B=ScalarParameter() {defaultValue=2, parameterName=Texture_Multiply_C}) {constB=2}
  n9 = ScalarParameter() {defaultValue=7, parameterName="U_Tiling(B)"}
  n10 = ScalarParameter() {defaultValue=7, parameterName="V_Tiling(B)"}
  n11 = TextureSampleParameter2D!(
   UVs=Multiply(A=AppendVector(A=@n9, B=@n10), B=MaterialFunctionCall() {materialFunction=&MF_UI_UVScaling}.ScaleY)
   Apply View MipBias
  ) {parameterName=Texture_Input_B, texture=&/T_UI_Squares}
  n12 = MaterialFunctionCall() {materialFunction=&MF_UI_UVScaling}
  n13 = TextureSample(UVs=Panner(Coordinate=@n12.Scaled to Fit, Speed=Constant2Vector() {g=-0.025, r=-0.025})) {samplerType=SAMPLERTYPE_Grayscale, texture=/Game/Weapons/FORT_Melee/Pickaxe_Cube_Ninja_Male/FX/T_CellNoise}
  n14 = TextureSample(UVs=Panner(Coordinate=@n12.Scaled to Fit, Speed=Constant2Vector() {g=0.025}) {constCoordinate=1}) {samplerType=SAMPLERTYPE_Grayscale, texture=&T_CellNoise}
  n15 = LinearInterpolate(
   B=VectorParameter() {parameterName=EmissiveTint, defaultValue=<rgba 0 0.5333 0.7682 1>}.RGB
   Alpha=Multiply(A=Multiply(A=@n8, B=@n11.R), B=Multiply(A=@n13.RGB, B=@n14.RGB))
  )
  n16 = ScalarParameter() {defaultValue=0.9093, parameterName=FrameFeather3}
  n17 = ScalarParameter() {defaultValue=1.9825, parameterName=FrameFeather4}
  n18 = Abs(None=ConstantBiasScale(None=TextureCoordinate()) {bias=-0.5, scale=2})
  n19 = ComponentMask(None=@n18) {r=true}
  n20 = ComponentMask(None=@n18) {g=true}
  n21 = Saturate(None=Add(A=SmoothStep(Min=@n16, Max=@n17, Value=@n19), B=SmoothStep(Min=@n16, Max=@n17, Value=@n20)))
  n22 = ScalarParameter() {defaultValue=6.0755, parameterName=FrameOpacity}
  n23 = MaterialFunctionCall() {materialFunction=&MF_UI_UVScaling}
  n24 = Constant2Vector() {g=0.4, r=0.5}
  n25 = ScalarParameter() {defaultValue=0.48, parameterName=Radius}
  n26 = ScalarParameter() {defaultValue=1.4128, parameterName=Density}
  n27 = LinearInterpolate(
   A=Multiply(A=Multiply(A=@n21, B=@n22), B=VectorParameter() {parameterName=Emissive Color, defaultValue=<rgba 0 0.1719 0.7682 1>}.RGB)
   B=Multiply(
    A=MaterialFunctionCall(UVs=@n23.ScaleY, CenterPosition=@n24, Radius=@n25, Density=@n26) {materialFunction=&RadialGradientExponential}.RadialGradientExponential
    B=VectorParameter() {parameterName=Emissive Colorradial, defaultValue=<rgba 0 0.1719 0.7682 1>}.RGB
   )
  )
  FunctionOutput(None=StaticSwitchParameter(True=LinearInterpolate(A=@n27, B=@n15), False=FunctionInput()) {parameterName=AlfredoSeries})
  LinearInterpolate(
   A=FunctionInput(Preview=VectorParameter() {group=CorruptSeries, parameterName=BG_Color, defaultValue=<rgba 1 1 1 1>}.RGB) {bUsePreviewValueAsDefault=true, inputName=RGB 1, sortPriority=1}
   B=FunctionInput(Preview=VectorParameter() {group=CorruptSeries, parameterName=CellColor, defaultValue=<rgba 0 0 0 1>}.RGB) {bUsePreviewValueAsDefault=true, inputName=RGB 2, sortPriority=1}
  )
  n28 = ScalarParameter() {defaultValue=0.767, parameterName=FrameFeather1}
  n29 = ScalarParameter() {defaultValue=-5.5135, parameterName=FrameFeather2}
  n30 = Abs(None=ConstantBiasScale(None=TextureCoordinate()) {bias=-0.5, scale=2})
  n31 = SmoothStep(Min=@n28, Max=@n29, Value=ComponentMask(None=@n30) {r=true})
  n32 = SmoothStep(Min=@n28, Max=@n29, Value=ComponentMask(None=@n30) {g=true})
  Multiply(A=Saturate(None=Add(A=@n31, B=@n32)), B=ScalarParameter() {defaultValue=6.0755, parameterName=FrameOpacity})
 ]}
