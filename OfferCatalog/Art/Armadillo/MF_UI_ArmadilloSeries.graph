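# Material function for offer catalog UI that applies the Armadillo series look when the bIsArmadilloSeries static switch is on, otherwise passing an input through.
# Skyfield and magenta LED smoke textures share aspect-corrected UVs, scaled about the center.
# A tinted diamond gradient blends with the skyfield at an alpha above one, extrapolating past the texture.
# A final 0.99 lerp leaves the smoke texture almost entirely dominant.
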MaterialFunction /OfferCatalog/Art/Armadillo/MF_UI_ArmadilloSeries
graph {nodes [
  FunctionInput() {inputName=RGB 1, sortPriority=1}
  n1 = Constant3Vector() {constant=<rgba 0.8594 0 0.2243 1>}
  n2 = Constant() {r=2.0896}
  n3 = MaterialFunctionCall(
   UVs=MaterialFunctionCall!(UV, TextureSize, ScalePivotPoint) {materialFunction=&MF_UI_UVScaling}.ScaleY
   Texture Scale=Constant() {r=1.3}
  ) {materialFunction=&ScaleUVsByCenter}
  n4 = LinearInterpolate(
   A=Multiply(
    A=FunctionInput(Preview=@n1) {inputName=RGB 2, sortPriority=1}
    B=MaterialFunctionCall!(Falloff=@n2) {materialFunction=/Engine/Functions/Engine_MaterialFunctions01/Gradient/DiamondGradient}.DiamondGradient
   )
   B=TextureSample(UVs=@n3.UVs) {texture=&/T_UI_Armadillo_Skyfield}.RGB
  ) {constAlpha=4.5232}
  n5 = TextureSample(UVs=@n3.UVs) {texture=&/T_UI_Armadillo_MagentaLEDSmoke}
  FunctionOutput(None=StaticSwitchParameter(True=LinearInterpolate(A=@n4, B=@n5.RGB) {constAlpha=0.99}, False=FunctionInput()) {parameterName=bIsArmadilloSeries})
 ]}
