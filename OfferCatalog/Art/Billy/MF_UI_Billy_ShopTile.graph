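# Material function that supplies the Billy series shop tile background for UI materials in the offer catalog.
# The tile texture, with UVs scaled to fill, blends toward a constant dark green through a rotated linear gradient.
# A static switch, bls_BillySeries, outputs this result when on and passes the function input through when off.
# A separate lerp of two CorruptSeries color inputs is left unconnected to the output.
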MaterialFunction /OfferCatalog/Art/Billy/MF_UI_Billy_ShopTile
graph {nodes [
  n1 = TextureSample(UVs=MaterialFunctionCall!(UV, TextureSize) {materialFunction=&MF_UI_UVScaling}.Scaled To Fill) {texture=&/T_UI_Billy_ShopTile}
  n2 = Constant3Vector() {constant=<rgba 0 0.0781 0.0213 0>}
  n3 = MaterialFunctionCall!(Texture, GradientStart=Constant() {r=0.5}, GradientEnd=Constant() {r=1}, RotationAngle=Constant() {r=0.75}, RotationCenter, GradientOffset) {materialFunction=&MF_UI_LinearGradient}
  FunctionOutput(None=StaticSwitchParameter(True=LinearInterpolate(A=@n1.RGB, B=@n2, Alpha=@n3.Result), False=FunctionInput()) {parameterName=bls_BillySeries})
  LinearInterpolate(
   A=FunctionInput(Preview=VectorParameter() {group=CorruptSeries, parameterName=BG_Color, defaultValue=<rgba 1 1 1 1>}.RGB) {bUsePreviewValueAsDefault=true, inputName=RGB 1, sortPriority=1}
   B=FunctionInput(Preview=VectorParameter() {group=CorruptSeries, parameterName=CellColor, defaultValue=<rgba 0 0 0 1>}.RGB) {bUsePreviewValueAsDefault=true, inputName=RGB 2, sortPriority=1}
  )
 ]}
