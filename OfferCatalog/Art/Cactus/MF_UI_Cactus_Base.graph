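# Material function for the Cactus series UI background in the offer catalog, drawing animated confetti over a radial two-color gradient.
# Four parameterized MF_UI_Confetti layers, each faded by a vertical gradient, merge into one mask, with L1 and L2 also driving the confetti color blend.
# Aspect-corrected UVs and panning-noise distortion are declared as named reroutes.
# The CactusSeries static switch selects this composite or passes through the function input.
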MaterialFunction /OfferCatalog/Art/Cactus/MF_UI_Cactus_Base
graph {nodes [
  n1 = NamedRerouteUsage()
  n2 = NamedRerouteUsage()
  n3 = ScalarParameter() {parameterName=L1-Patter-X-Offset}
  n4 = ScalarParameter() {defaultValue=0.5, parameterName=L1-Patter-Y-Offset}
  n5 = MaterialFunctionCall!(
   "UV(V2)"=Add(A=Add(A=@n1, B=@n2), B=AppendVector(A=@n3, B=@n4))
   Global-Angle=ScalarParameter() {defaultValue=-0.125, parameterName=L1-Global Angle}
   "Pattern-Tiles(V2)"=AppendVector(A=ScalarParameter() {defaultValue=50, parameterName=L1-Pattern-X}, B=ScalarParameter() {defaultValue=0.75, parameterName=L1-Pattern-Y})
   RandomSeed=ScalarParameter() {defaultValue=12345, parameterName=L1-RandomSeed}
   Confetti-Speed-TimeFactor=ScalarParameter() {defaultValue=-0.05, parameterName=L1-Confetti-Speed-TimeFactor}
   Confetti-Rotation-TimeFactor=ScalarParameter() {defaultValue=0.25, parameterName=L1-Confetti-Rotation-TimeFactor}
   Confetti-Local-Rotation-Range=ScalarParameter() {defaultValue=0.0075, parameterName=L1-Confetti-Local-Rotation-Range}
   "Confetti-Local-Tiles(V2)"=AppendVector(
    A=ScalarParameter() {defaultValue=1, parameterName=L1-Confetti-Local-Tile-X}
    B=ScalarParameter() {defaultValue=90, parameterName=L1-Confetti-Local-Tile-Y}
   )
   "SmoothStepRanges(V2)"=AppendVector(
    A=ScalarParameter() {defaultValue=0.9, parameterName=L1-SmoothStep-Max}
    B=ScalarParameter() {defaultValue=0.5, parameterName=L1-SmoothStep-Min}
   )
  ) {materialFunction=/Game/UI/Frontend/ItemShop/Art/Cactus/MF_UI_Confetti}
  n6 = NamedRerouteUsage()
  n7 = NamedRerouteUsage()
  n8 = ScalarParameter() {parameterName=L2-Patter-X-Offset}
  n9 = ScalarParameter() {defaultValue=0.15, parameterName=L2-Patter-Y-Offset}
  n10 = MaterialFunctionCall(
   "UV(V2)"=Add(A=Add(A=@n6, B=@n7), B=AppendVector(A=@n8, B=@n9))
   Global-Angle=ScalarParameter() {defaultValue=-0.085, parameterName=L2-Global Angle}
   "Pattern-Tiles(V2)"=AppendVector(A=ScalarParameter() {defaultValue=60, parameterName=L2-Pattern-X}, B=ScalarParameter() {defaultValue=2, parameterName=L2-Pattern-Y})
   RandomSeed=ScalarParameter() {defaultValue=2345, parameterName=L2-RandomSeed}
   Confetti-Speed-TimeFactor=ScalarParameter() {defaultValue=-0.05, parameterName=L2-Confetti-Speed-TimeFactor}
   Confetti-Rotation-TimeFactor=ScalarParameter() {defaultValue=0.125, parameterName=L2-Confetti-Rotation-TimeFactor}
   Confetti-Local-Rotation-Range=ScalarParameter() {defaultValue=0.005, parameterName=L2-Confetti-Local-Rotation-Range}
   "Confetti-Local-Tiles(V2)"=AppendVector(
    A=ScalarParameter() {defaultValue=1, parameterName=L2-Confetti-Local-Tile-X}
    B=ScalarParameter() {defaultValue=54.3, parameterName=L2-Confetti-Local-Tile-Y}
   )
   "SmoothStepRanges(V2)"=AppendVector(
    A=ScalarParameter() {defaultValue=1, parameterName=L2-SmoothStep-Max}
    B=ScalarParameter() {defaultValue=0.5, parameterName=L2-SmoothStep-Min}
   )
  ) {materialFunction=&MF_UI_Confetti}
  n11 = LinearInterpolate(
   A=VectorParameter() {parameterName=Small Confetti, defaultValue=<rgba 0.8854 0.5486 0.5045 1>}.RGB
   B=VectorParameter() {parameterName=Big Confetti, defaultValue=<rgba 0.0535 0.5326 0.7344 1>}.RGB
   Alpha=LinearInterpolate(A=@n10.Confetti-Mask, B=@n5.Confetti-Mask, Alpha=@n5.Confetti-Mask)
  )
  n12 = Add(A=NamedRerouteUsage(), B=NamedRerouteUsage())
  n13 = ScalarParameter() {defaultValue=-0.075, parameterName=L4-Global Angle}
  n14 = AppendVector(A=ScalarParameter() {defaultValue=89.0677, parameterName=L4-Pattern-X}, B=ScalarParameter() {defaultValue=7, parameterName=L4-Pattern-Y})
  n15 = ScalarParameter() {defaultValue=3863.1282, parameterName=L4-RandomSeed}
  n16 = ScalarParameter() {defaultValue=-0.171, parameterName=L4-Confetti-Speed-TimeFactor}
  n17 = ScalarParameter() {defaultValue=0.101, parameterName=L4-Confetti-Rotation-TimeFactor}
  n18 = ScalarParameter() {defaultValue=0.005, parameterName=L4-Confetti-Local-Rotation-Range}
  n19 = AppendVector(
   A=ScalarParameter() {defaultValue=1, parameterName=L4-Confetti-Local-Tile-X}
   B=ScalarParameter() {defaultValue=33.2071, parameterName=L4-Confetti-Local-Tile-Y}
  )
  n20 = AppendVector(
   A=ScalarParameter() {defaultValue=-0.075, parameterName=L4-SmoothStep-Max}
   B=ScalarParameter() {defaultValue=0.484, parameterName=L4-SmoothStep-Min}
  )
  n21 = TextureCoordinate()
  n22 = Subtract(A=OneMinus(None=ComponentMask(None=@n21) {g=true}), B=ScalarParameter() {defaultValue=0.5, parameterName=L4-GradientMask})
  n23 = Multiply(
   A=Multiply(
    A=MaterialFunctionCall(
     "UV(V2)"=@n12
     Global-Angle=@n13
     "Pattern-Tiles(V2)"=@n14
     RandomSeed=@n15
     Confetti-Speed-TimeFactor=@n16
     Confetti-Rotation-TimeFactor=@n17
     Confetti-Local-Rotation-Range=@n18
     "Confetti-Local-Tiles(V2)"=@n19
     "SmoothStepRanges(V2)"=@n20
    ) {materialFunction=&MF_UI_Confetti}.Confetti-Mask
    B=Saturate(None=@n22)
   )
   B=ScalarParameter() {defaultValue=0.75, parameterName=L4-Intensity}
  )
  n24 = Add(A=NamedRerouteUsage(), B=NamedRerouteUsage())
  n25 = ScalarParameter() {defaultValue=-0.085, parameterName=L3-Global Angle}
  n26 = AppendVector(A=ScalarParameter() {defaultValue=60.48, parameterName=L3-Pattern-X}, B=ScalarParameter() {defaultValue=5, parameterName=L3-Pattern-Y})
  n27 = ScalarParameter() {defaultValue=3456, parameterName=L3-RandomSeed}
  n28 = ScalarParameter() {defaultValue=-0.123, parameterName=L3-Confetti-Speed-TimeFactor}
  n29 = ScalarParameter() {defaultValue=0.013, parameterName=L3-Confetti-Rotation-TimeFactor}
  n30 = ScalarParameter() {defaultValue=0.005, parameterName=L3-Confetti-Local-Rotation-Range}
  n31 = AppendVector(
   A=ScalarParameter() {defaultValue=1, parameterName=L3-Confetti-Local-Tile-X}
   B=ScalarParameter() {defaultValue=20.2015, parameterName=L3-Confetti-Local-Tile-Y}
  )
  n32 = AppendVector(
   A=ScalarParameter() {defaultValue=0.005, parameterName=L3-SmoothStep-Max}
   B=ScalarParameter() {defaultValue=0.5, parameterName=L3-SmoothStep-Min}
  )
  n33 = TextureCoordinate()
  n34 = Subtract(A=OneMinus(None=ComponentMask(None=@n33) {g=true}), B=ScalarParameter() {defaultValue=0.25, parameterName=L3-GradientMask})
  n35 = Multiply(
   A=Multiply(
    A=MaterialFunctionCall(
     "UV(V2)"=@n24
     Global-Angle=@n25
     "Pattern-Tiles(V2)"=@n26
     RandomSeed=@n27
     Confetti-Speed-TimeFactor=@n28
     Confetti-Rotation-TimeFactor=@n29
     Confetti-Local-Rotation-Range=@n30
     "Confetti-Local-Tiles(V2)"=@n31
     "SmoothStepRanges(V2)"=@n32
    ) {materialFunction=&MF_UI_Confetti}.Confetti-Mask
    B=Saturate(None=@n34)
   )
   B=ScalarParameter() {defaultValue=0.75, parameterName=L3-Intensity}
  )
  n36 = OneMinus(None=ComponentMask(None=TextureCoordinate()) {g=true})
  n37 = ScalarParameter() {defaultValue=0.35, parameterName=L2-GradientMask}
  n38 = Multiply(A=@n10.Confetti-Mask, B=Saturate(None=Subtract(A=@n36, B=@n37)))
  n39 = ScalarParameter() {defaultValue=1.378, parameterName=L2-Intensity}
  n40 = OneMinus(None=ComponentMask(None=TextureCoordinate()) {g=true})
  n41 = ScalarParameter() {defaultValue=0.3, parameterName=L1-Gradient-Mask}
  n42 = Saturate(None=Multiply(A=Subtract(A=@n40, B=@n41), B=ScalarParameter() {defaultValue=2, parameterName=L1-Gradient-Falloff}))
  n43 = LinearInterpolate(
   A=LinearInterpolate(A=LinearInterpolate(A=@n23, Alpha=@n35), Alpha=Multiply(A=@n38, B=@n39))
   Alpha=Multiply(A=Multiply(A=@n5.Confetti-Mask, B=@n42), B=ScalarParameter() {defaultValue=1, parameterName=L1-Intensity})
  )
  n44 = AppendVector(A=ScalarParameter() {defaultValue=0.6, parameterName=Radial-X-Tile}, B=ScalarParameter() {defaultValue=0.95, parameterName=Radial-Y-Tile})
  n45 = Distance(A=Multiply(A=NamedRerouteUsage(), B=@n44), B=Multiply(A=@n44, B=Constant2Vector() {g=1, r=0.5}))
  n46 = ScalarParameter() {defaultValue=1.25, parameterName=Distance Power}
  n47 = LinearInterpolate(
   A=VectorParameter() {parameterName=Inner Gradient, defaultValue=<rgba 0.7083 0.1954 0.1254 1>}.RGB
   B=VectorParameter() {parameterName=Outer Gradient, defaultValue=<rgba 0.0684 0.0126 0.5104 1>}.RGB
   Alpha=Saturate(None=Power(Base=@n45, Exp=@n46))
  )
  FunctionOutput(None=StaticSwitchParameter(True=LinearInterpolate(A=@n47, B=@n11, Alpha=@n43), False=FunctionInput()) {parameterName=CactusSeries})
  LinearInterpolate(
   A=FunctionInput(Preview=VectorParameter() {group=CorruptSeries, parameterName=BG_Color, defaultValue=<rgba 1 1 1 1>}.RGB) {bUsePreviewValueAsDefault=true, inputName=RGB 1, sortPriority=1}
   B=FunctionInput(Preview=VectorParameter() {group=CorruptSeries, parameterName=CellColor, defaultValue=<rgba 0 0 0 1>}.RGB) {bUsePreviewValueAsDefault=true, inputName=RGB 2, sortPriority=1}
  )
  n48 = MaterialFunctionCall() {materialFunction=&GetUserInterfaceUV}
  n49 = MaterialFunctionCall(Float2=@n48.Pixel Size) {materialFunction=&BreakOutFloat2Components}
  n50 = AppendVector(A=Divide(A=@n49.R, B=Max(A=@n49.G)), B=Constant() {r=1})
  n51 = OneMinus(None=@n50)
  NamedRerouteDeclaration(None=Add(A=Multiply(A=@n51) {constB=0.5}, B=Multiply(A=@n50, B=@n48.Normalized UV))) {name="Squared UV - (x/y,y)", nodeColor=<rgba 0 0 0 1>}
  n52 = TextureCoordinate()
  n53 = ScalarParameter() {defaultValue=0.5, parameterName=DupeScaler}
  n54 = Time()
  n55 = ScalarParameter() {defaultValue=0.6346, parameterName=TimeFactor}
  n56 = Power(
   Base=Panner(Coordinate=Multiply(A=@n52, B=@n53), Time=Multiply(A=@n54, B=@n55)) {speedX=0.058, speedY=0.096}
   Exp=ScalarParameter() {defaultValue=0.8088, parameterName=ConfettiParam2}
  )
  NamedRerouteDeclaration(None=Multiply(
    A=TextureSample(UVs=@n56) {samplerType=SAMPLERTYPE_Grayscale, texture=&LowResBlurredNoise}.R
    B=ScalarParameter() {defaultValue=0.35, parameterName=DistortionScaler}
   )) {name=UV-Distortion, nodeColor=<rgba 0 0 0 1>}
 ]}
