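# Material function that supplies the Ch5S1 gold and orange background for shop tiles, enabled by the bls_Ch5S1 static switch, which otherwise passes its input through.
# An offset, scaled radial gradient and an angled linear gradient each tint a color pair, then blend by the red channel of a grunge texture.
# UV scaling provides Fit and Fill reroutes.
# A CorruptSeries BG_Color/CellColor blend is left unconnected to the output.
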MaterialFunction /OfferCatalog/Art/Ch5S1/MF_UI_Ch5S1_ShopTiles
graph {nodes [
  n1 = Constant()
  n2 = Constant() {r=0.1}
  n3 = MaterialFunctionCall(UVs=Add(A=NamedRerouteUsage(), B=AppendVector(A=@n1, B=@n2)), Texture Scale=Constant() {r=1.4}) {materialFunction=&ScaleUVsByCenter}
  n4 = Constant()
  n5 = Constant() {r=0.95}
  n6 = LinearInterpolate(
   A=Constant3Vector() {constant=<rgba 0.5972 0.3185 0.004 1>}
   B=Constant3Vector() {constant=<rgba 0.9911 0.8148 0.2423 1>}
   Alpha=SmoothStep(Value=MaterialFunctionCall(UV=@n3.UVs, GradientStart=@n4, GradientEnd=@n5) {materialFunction=&MF_UI_RadialGradient}.Result) {constMax=2}
  )
  n7 = AppendVector(A=Constant(), B=Constant() {r=-0.2})
  n8 = NamedRerouteUsage()
  n9 = LinearInterpolate(
   A=Constant3Vector() {constant=<rgba 0.5029 0.1384 0.0003 1>}
   B=Constant3Vector() {constant=<rgba 0.599 0.3239 0.0135 1>}
   Alpha=MaterialFunctionCall(UVs=Add(A=@n7, B=@n8), GradientStart=Constant() {r=0.25}, GradientEnd=Constant() {r=0.5}, RotationAngle=Constant() {r=0.25}) {materialFunction=&MF_UI_LinearGradient}.Result
  )
  n10 = Constant() {r=0.1}
  n11 = Constant() {r=0.054}
  n12 = NamedRerouteUsage()
  n13 = TextureSample(UVs=Add(A=AppendVector(A=@n10, B=@n11), B=Multiply(A=@n12) {constB=0.6})) {samplerType=SAMPLERTYPE_Normal, texture=&/T_UI_Ch5S1Grunge}
  FunctionOutput(None=StaticSwitchParameter(True=LinearInterpolate(A=@n6, B=@n9, Alpha=@n13.R), False=FunctionInput()) {parameterName=bls_Ch5S1})
  LinearInterpolate(
   A=FunctionInput(Preview=VectorParameter() {group=CorruptSeries, parameterName=BG_Color, defaultValue=<rgba 1 1 1 1>}.RGB) {bUsePreviewValueAsDefault=true, inputName=RGB 1, sortPriority=1}
   B=FunctionInput(Preview=VectorParameter() {group=CorruptSeries, parameterName=CellColor, defaultValue=<rgba 0 0 0 1>}.RGB) {bUsePreviewValueAsDefault=true, inputName=RGB 2, sortPriority=1}
  )
  n14 = MaterialFunctionCall!(UV, TextureSize) {materialFunction=&MF_UI_UVScaling}
  NamedRerouteDeclaration(None=@n14.Scaled to Fit) {name=Fit, nodeColor=<rgba 0.5703 1 0 1>}
  NamedRerouteDeclaration(None=@n14.Scaled To Fill) {name=Fill, nodeColor=<rgba 0.5703 1 0 1>}
 ]}
