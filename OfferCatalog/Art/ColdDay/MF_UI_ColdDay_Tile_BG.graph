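# UI material function that supplies the ColdDay series tile background, switched by the static parameter blsColdDay_Series, otherwise passing its input through.
# The series branch builds a tan, brown and rust gradient from a smoothstepped radial gradient, then blends it using an inverted rerouted mask.
# Named reroutes organize screen-space or UI UVs, scaled variants and exponential gradient values.
# The wave texture sample and CorruptSeries color lerp do not feed the output.
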MaterialFunction /OfferCatalog/Art/ColdDay/MF_UI_ColdDay_Tile_BG
graph {nodes [
  n1 = NamedRerouteUsage()
  n2 = LinearInterpolate(A=Constant3Vector() {constant=<rgba 0.7682 0.6584 0.4969 1>}, B=Constant3Vector() {constant=<rgba 0.7135 0.5964 0.4349 1>})
  n3 = MaterialFunctionCall(UV=NamedRerouteUsage(), GradientStart=Constant() {r=-0.5}, GradientEnd=Constant() {r=1}) {materialFunction=&MF_UI_RadialGradient}
  n4 = LinearInterpolate(
   A=LinearInterpolate(A=Constant3Vector() {constant=<rgba 0.1979 0.1219 0.0072 1>}, B=@n2, Alpha=SmoothStep(Value=@n3.Result) {constMin=-0.25})
   B=Constant3Vector() {constant=<rgba 0.2582 0.0595 0.0065 1>}
  )
  n5 = OneMinus(None=NamedRerouteUsage())
  FunctionOutput(None=StaticSwitchParameter(True=LinearInterpolate(A=@n1, B=@n4, Alpha=@n5), False=FunctionInput()) {parameterName=blsColdDay_Series})
  LinearInterpolate(
   A=FunctionInput(Preview=VectorParameter() {group=CorruptSeries, parameterName=BG_Color, defaultValue=<rgba 1 1 1 1>}.RGB) {bUsePreviewValueAsDefault=true, inputName=RGB 1, sortPriority=1}
   B=FunctionInput(Preview=VectorParameter() {group=CorruptSeries, parameterName=CellColor, defaultValue=<rgba 0 0 0 1>}.RGB) {bUsePreviewValueAsDefault=true, inputName=RGB 2, sortPriority=1}
  )
  n6 = MaterialFunctionCall(UVs=NamedRerouteUsage(), Radius=Constant() {r=1.1}, Density=Constant() {r=2}) {materialFunction=&RadialGradientExponential}
  NamedRerouteDeclaration(None=@n6.RadialGradientExponential) {name=Gradient Alpha, nodeColor=<rgba 1 0 0.0938 1>}
  NamedRerouteDeclaration(None=Multiply(A=MaterialFunctionCall!(UV, TextureSize, ScalePivotPoint) {materialFunction=&MF_UI_UVScaling}.Scaled To Fill)) {name=UVFill, nodeColor=<rgba 1 0 0.7266 1>}
  NamedRerouteDeclaration(None=StaticSwitchParameter(
    True=ScreenPosition().ViewportUV
    False=MaterialFunctionCall() {materialFunction=&GetUserInterfaceUV}.Normalized UV
   ) {group=UV Space, parameterName=Is3DSpace}) {name=Normal UV, nodeColor=<rgba 0 0 0 1>}
  n7 = MaterialFunctionCall!(Screen UV=NamedRerouteUsage(), Texture Size) {materialFunction=/FortUILibrary/MaterialFunctions/Basics/MF_UI_ScreenSpaceUVScaling}
  NamedRerouteDeclaration(None=@n7.Scaled to Fill) {name=Normal UV Scaled to Fill, nodeColor=<rgba 0 0 0 1>}
  NamedRerouteDeclaration(None=Multiply(A=@n2, B=@n6.RadialGradientExponential)) {name=Gradient, nodeColor=<rgba 0.3828 0 1 1>}
  NamedRerouteDeclaration(None=@n7.Scaled to Fit) {name=Normal UV Scaled to Fit, nodeColor=<rgba 0 0 0 1>}
  NamedRerouteUsage()
  NamedRerouteUsage()
  TextureSample(UVs=MaterialFunctionCall(UVs=NamedRerouteUsage(), Texture Scale=Constant() {r=0.3}) {materialFunction=&ScaleUVsByCenter}.UVs) {samplerType=SAMPLERTYPE_LinearColor, texture=/Game/Athena/UI/Events/ColdDay/Materials/T_UI_ColdDay_Waves}
 ]}
