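# Material function that builds the Coyote Trail shop tile background, enabled by static switch blsCoyoteTrail_Series, otherwise passing its input through.
# A scrolling, rotated sine band, gated by a slower time-based sine pulse, lights green through a rotated pattern texture.
# A second pattern layer adds dark green, then the result blends with a radial gradient.
# Parameterised horizontal and vertical edge masks fade the tile to black.
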MaterialFunction /OfferCatalog/Art/Coyote_Trail/MF_UI_CoyoteTrail_ShopTiles
graph {nodes [
  n1 = Time()
  n2 = ScalarParameter() {defaultValue=1, parameterName=tenminutes}
  n3 = Multiply(A=Sine(None=Multiply(A=@n1, B=@n2)) {period=5}) {constB=3}
  Ceil(None=@n3)
  Constant3Vector() {constant=<rgba 0.0312 0.0356 0.0434 0>}
  n4 = Saturate(None=SmoothStep(Value=@n3) {constMax=8, constMin=0.5})
  n5 = MaterialFunctionCall!(UV, TextureSize) {materialFunction=&MF_UI_UVScaling}
  n6 = Time()
  n7 = Multiply(A=Frac(None=OneMinus(None=@n6)))
  n8 = Constant()
  n9 = MaterialFunctionCall(
   UVs=Add(A=ConstantBiasScale(None=@n5.Scaled To Fill) {scale=0.24}, B=AppendVector(A=@n7, B=@n8))
   "Rotation Angle (0-1)"=Constant() {r=0.358}
  ) {materialFunction=&CustomRotator}
  n10 = Multiply(A=ComponentMask(None=Subtract(A=@n9.Rotated Values) {constB=0}) {r=true}) {constB=0.5}
  n11 = Constant() {r=0.0025}
  n12 = Multiply(A=Subtract(A=Sine(None=@n10), B=OneMinus(None=@n11)) {constB=0.25}, B=Divide(B=@n11) {constB=0.5})
  n13 = Multiply(A=Saturate(None=Multiply(A=@n12) {constB=2})) {constB=10}
  n14 = MaterialFunctionCall() {materialFunction=&MF_UI_UVScaling}
  n15 = Add(
   A=@n14.Scaled To Fill
   B=MaterialFunctionCall(X=Constant() {r=1.038}, Y=Constant()) {materialFunction=&MakeFloat2}.Result
  )
  n16 = ScalarParameter() {defaultValue=1.1, parameterName=PatternSize}
  n17 = MaterialFunctionCall(
   UVs=MaterialFunctionCall(UVs=@n15, Texture Scale=@n16) {materialFunction=&ScaleUVsByCenter}.UVs
   "Rotation Angle (0-1)"=Constant() {r=0.25}
  ) {materialFunction=&CustomRotator}
  n18 = TextureSample(UVs=@n17.Rotated Values) {texture=&/T_UI_CoyoteTrailPattern}
  n19 = LinearInterpolate(B=LinearInterpolate(B=Constant3Vector() {constant=<rgba 0.0106 0.2708 0.1837 1>}, Alpha=Multiply(A=@n4, B=@n13)), Alpha=@n18.R)
  n20 = LinearInterpolate(
   A=@n19
   B=Constant3Vector() {constant=<rgba 0.0068 0.0365 0.026 0>}
   Alpha=TextureSample(UVs=@n17.Rotated Values) {texture=&/T_UI_CoyoteTrailPattern_02}.R
  )
  n21 = Constant3Vector() {constant=<rgba 0.0105 0.224 0.1247 0>}
  n22 = LinearInterpolate(A=Multiply(A=@n19) {constB=7.6538}, B=@n21, Alpha=@n18.R)
  n23 = Multiply(
   A=MaterialFunctionCall(X=Constant() {r=1}, Y=Constant() {r=1}) {materialFunction=&MakeFloat2}.Result
   B=@n14.Scaled To Fill
  )
  n24 = Constant()
  n25 = Constant() {r=2}
  n26 = LinearInterpolate(
   A=Multiply(A=LinearInterpolate(A=@n20, B=@n22)) {constB=4}
   B=LinearInterpolate(
    A=Constant3Vector() {constant=<rgba 0.0087 0.0087 0.0087 0>}
    B=@n21
    Alpha=MaterialFunctionCall(UV=@n23, GradientStart=@n24, GradientEnd=@n25) {materialFunction=&MF_UI_RadialGradient}.Result
   )
  )
  n27 = Constant3Vector()
  n28 = MaterialFunctionCall() {materialFunction=&MF_UI_UVScaling}
  n29 = Abs(None=ConstantBiasScale(None=@n28.Scaled To Fill) {bias=-0.5, scale=2})
  n30 = Multiply(
   A=Subtract(A=ComponentMask(None=@n29) {g=true}, B=ScalarParameter() {defaultValue=0.0511, parameterName=Subtract})
   B=ScalarParameter() {defaultValue=1.0721, parameterName=Falloff}
  )
  n31 = Abs(None=ConstantBiasScale(None=@n28.Scaled To Fill) {bias=-0.5, scale=2})
  n32 = Multiply(
   A=Subtract(A=ComponentMask(None=@n31) {r=true}, B=ScalarParameter() {defaultValue=0.302, parameterName=SubtractH})
   B=ScalarParameter() {defaultValue=0.8595, parameterName=FalloffH}
  )
  n33 = LinearInterpolate(
   A=Multiply(A=Saturate(None=@n30), B=ScalarParameter() {defaultValue=1, parameterName=Intensity})
   B=Multiply(A=Saturate(None=@n32), B=ScalarParameter() {defaultValue=2.3306, parameterName=IntensityH})
  )
  FunctionOutput(None=StaticSwitchParameter(True=LinearInterpolate(A=@n26, B=@n27, Alpha=@n33), False=FunctionInput()) {parameterName=blsCoyoteTrail_Series})
  LinearInterpolate(
   A=FunctionInput(Preview=VectorParameter() {group=CorruptSeries, parameterName=BG_Color, defaultValue=<rgba 1 1 1 1>}.RGB) {bUsePreviewValueAsDefault=true, inputName=RGB 1, sortPriority=1}
   B=FunctionInput(Preview=VectorParameter() {group=CorruptSeries, parameterName=CellColor, defaultValue=<rgba 0 0 0 1>}.RGB) {bUsePreviewValueAsDefault=true, inputName=RGB 2, sortPriority=1}
  )
  MaterialFunctionCall(UVs=@n15, Texture Scale=@n16) {materialFunction=&ScaleUVsByCenter}
  MaterialFunctionCall() {materialFunction=&GetUserInterfaceUV}
  Multiply()
  OneMinus(None=@n33)
 ]}
