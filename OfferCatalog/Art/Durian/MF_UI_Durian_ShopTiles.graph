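# UI material function that draws an animated Durian series background for shop tiles, or passes its input through when the Durian_Series static switch is off.
# Two palettes, red-orange and crimson-pink, layer exponential radial gradients over a near-black base and crossfade on a time-driven sine.
# Named reroutes supply screen-scaled UVs, a linear gradient, pixelized radial UVs and an LED dot mask.
# The dots are tiled SDF circles sized by a panning, mirrored ping-pong pattern.
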MaterialFunction /OfferCatalog/Art/Durian/MF_UI_Durian_ShopTiles
graph {nodes [
  n1 = Time()
  n2 = ScalarParameter() {defaultValue=0.2, parameterName=tenminutes}
  n3 = Multiply(A=Sine(None=Multiply(A=@n1, B=@n2)) {period=10}) {constB=20}
  Ceil(None=@n3)
  n4 = Constant3Vector() {constant=<rgba 0.0208 0.003 0.0054 1>}
  n5 = NamedRerouteUsage()
  n6 = LinearInterpolate(
   A=Constant3Vector() {constant=<rgba 0.4948 0 0.0154 1>}
   B=Constant3Vector() {constant=<rgba 1 0.3534 0.0076 1>}
   Alpha=MaterialFunctionCall(UVs=@n5, Radius=Constant() {r=0.45}, Density=Constant() {r=6}) {materialFunction=&RadialGradientExponential}.RadialGradientExponential
  )
  n7 = Constant3Vector() {constant=<rgba 0.724 0.2213 0.0264 1>}
  n8 = MaterialFunctionCall(UVs=@n5, Radius=Constant() {r=0.43}, Density=Constant() {r=18}) {materialFunction=&RadialGradientExponential}
  n9 = NamedRerouteUsage()
  n10 = LinearInterpolate(
   A=LinearInterpolate(A=@n4, B=LinearInterpolate(A=@n6, B=@n7, Alpha=@n8.RadialGradientExponential), Alpha=NamedRerouteUsage())
   B=Multiply(A=Constant3Vector() {constant=<rgba 0.2188 0.0023 0.0023 0>}, B=@n9)
   Alpha=@n9
  )
  n11 = NamedRerouteUsage()
  n12 = Constant() {r=0.5}
  n13 = Constant() {r=6}
  n14 = LinearInterpolate(
   A=Constant3Vector() {constant=<rgba 0.4271 0 0.0507 1>}
   B=Constant3Vector() {constant=<rgba 1 0.099 0.1591 1>}
   Alpha=SmoothStep(Value=MaterialFunctionCall(UVs=@n11, Radius=@n12, Density=@n13) {materialFunction=&RadialGradientExponential}.RadialGradientExponential)
  )
  n15 = Constant3Vector() {constant=<rgba 0.5521 0.1679 0.0401 1>}
  n16 = MaterialFunctionCall(UVs=NamedRerouteUsage(), Radius=Constant() {r=0.37}, Density=Constant() {r=18}) {materialFunction=&RadialGradientExponential}
  n17 = NamedRerouteUsage()
  n18 = LinearInterpolate(
   A=LinearInterpolate(A=@n4, B=LinearInterpolate(A=@n14, B=@n15, Alpha=@n16.RadialGradientExponential), Alpha=NamedRerouteUsage())
   B=Multiply(A=Constant3Vector() {constant=<rgba 0.1771 0.0091 0.0144 1>}, B=@n17)
   Alpha=@n17
  )
  n19 = Saturate(None=SmoothStep(Value=@n3) {constMax=8, constMin=0.2})
  FunctionOutput(None=StaticSwitchParameter(True=LinearInterpolate(A=@n10, B=@n18, Alpha=@n19), False=FunctionInput()) {parameterName=Durian_Series})
  n20 = NamedRerouteUsage()
  n21 = Constant() {r=0.75}
  n22 = ComponentMask(None=OneMinus(None=MaterialFunctionCall(UVs=@n20, "Rotation Angle (0-1)"=@n21) {materialFunction=&CustomRotator}.Rotated Values)) {r=true}
  LinearInterpolate(A=Constant() {r=1}, B=Constant() {r=1}, Alpha=@n22)
  LinearInterpolate(
   A=FunctionInput(Preview=VectorParameter() {group=CorruptSeries, parameterName=BG_Color, defaultValue=<rgba 1 1 1 1>}.RGB) {bUsePreviewValueAsDefault=true, inputName=RGB 1, sortPriority=1}
   B=FunctionInput(Preview=VectorParameter() {group=CorruptSeries, parameterName=CellColor, defaultValue=<rgba 0 0 0 1>}.RGB) {bUsePreviewValueAsDefault=true, inputName=RGB 2, sortPriority=1}
  )
  n23 = MaterialFunctionCall!(ScreenUV, TextureSize, ScalePivotPoint) {materialFunction=/FortUILibrary/MaterialFunctions/Basics/MF_UI_ScreenSpaceUVScaling}
  NamedRerouteDeclaration(None=@n23.Scaled to Fill) {name=UVFIll, nodeColor=<rgba 0 0.625 1 1>}
  NamedRerouteDeclaration(None=@n23.ScaleY) {name=Scale Y, nodeColor=<rgba 0 1 0.9062 1>}
  n24 = NamedRerouteUsage()
  n25 = Constant() {r=-0.2}
  n26 = Constant() {r=0.5}
  n27 = Constant() {r=-0.25}
  NamedRerouteDeclaration(None=SmoothStep(Value=MaterialFunctionCall(UVs=@n24, GradientStart=@n25, GradientEnd=@n26, RotationAngle=@n27) {materialFunction=&MF_UI_LinearGradient}.Result) {constMax=1.25}) {name=LinearGradient, nodeColor=<rgba 0.5 1 0 1>}
  n28 = MaterialFunctionCall() {materialFunction=&MF_UI_UVScaling}
  NamedRerouteDeclaration(None=@n28.Scaled To Fill) {name=ScalingUVFill, nodeColor=<rgba 0 1 0.8594 1>}
  n29 = AppendVector(A=Constant(), B=Constant() {r=-0.36})
  NamedRerouteDeclaration(None=MaterialFunctionCall!(UV=Add(A=@n24, B=@n29), Resolution=Constant() {r=200}) {materialFunction=/Game/Weapons/FORT_Melee/Pickaxe_Veiled/FX/MF_UVPixelizer}.Result) {name=RadialGradientUVs, nodeColor=<rgba 0 1 0.4141 1>}
  NamedRerouteDeclaration(None=@n28.ScaleY) {name=UVFit, nodeColor=<rgba 1 0 0.6094 1>}
  n30 = MaterialFunctionCall!(UV=NamedRerouteUsage(), Offset=Constant2Vector(), Rotation, Scale=Constant() {r=0.75}, MirrorY=StaticBool() {value=true}) {materialFunction=/FortUILibrary/MaterialFunctions/Basics/MF_UI_Symmetry_2x1_DEPRECATED}
  n31 = Time() {period=10}
  n32 = MaterialFunctionCall!(
   Value=MaterialFunctionCall(UVs=Panner(Coordinate=@n30.Result, Time=@n31) {speedX=0.03}, "Rotation Angle (0-1)"=Constant() {r=0.35}) {materialFunction=&CustomRotator}.Rotated Values
   Scale=Constant() {r=8}
   Height
  ) {materialFunction=/FortUILibrary/MaterialFunctions/Basics/MF_UI_PingPong}
  n33 = OneMinus(None=ComponentMask(None=Clamp(None=@n32.Result) {minDefault=0.35}) {r=true})
  n34 = SmoothStep(Value=@n33) {constMax=0.5}
  n35 = Frac(None=Multiply(A=NamedRerouteUsage(), B=Constant() {r=110}))
  n36 = MaterialFunctionCall!(
   x=SmoothStep(
    Min=@n33
    Max=Add(A=@n33) {constB=0.001}
    Value=MaterialFunctionCall!(In Texcoords=@n35) {materialFunction=/Game/Effects/Fort_Effects/MaterialFunctions/MF_Circle_SDF}.Result
   )
   size=@n22
  ) {materialFunction=/Game/Effects/Fort_Effects/MaterialFunctions/MF_Fill}
  NamedRerouteDeclaration(None=Multiply(A=Min(A=@n34, B=@n36.Result)) {constB=0.15}) {name=LED Dots, nodeColor=<rgba 0.2188 1 0 1>}
  NamedRerouteUsage()
 ]}
