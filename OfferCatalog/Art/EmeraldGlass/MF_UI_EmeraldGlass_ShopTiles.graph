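# UI material function that adds an Emerald Glass series gradient treatment to item shop tiles, or passes its input through when that static switch is off.
# A static switch picks a blue or orange-to-gold rotated linear gradient.
# A rounded mask, using shop collection corner settings, clips the gradient, and its smoothstepped outline SDF blends it against black.
# Skew-box border and color-blend nodes are present but never reach the output.
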MaterialFunction /OfferCatalog/Art/EmeraldGlass/MF_UI_EmeraldGlass_ShopTiles
graph {nodes [
  n1 = Constant()
  n2 = Constant3Vector() {constant=<rgba 0.0168 0.0482 0.552 1>}
  n3 = Constant3Vector() {constant=<rgba 0.1337 0.16 0.5833 1>}
  n4 = MaterialFunctionCall(GradientStart=Constant() {r=-0.144}, GradientEnd=Constant() {r=0.9047}, RotationAngle=Constant() {r=0.25}) {materialFunction=&MF_UI_LinearGradient}
  n5 = Constant3Vector() {constant=<rgba 0.776 0.23 0 1>}
  n6 = Constant3Vector() {constant=<rgba 0.9844 0.7538 0.0345 1>}
  n7 = MaterialFunctionCall(GradientStart=Constant() {r=-0.144}, GradientEnd=Constant() {r=0.9047}, RotationAngle=Constant() {r=0.25}) {materialFunction=&MF_UI_LinearGradient}
  n8 = MaterialFunctionCall!(
   UV
   Size
   Corner Radius=CollectionParameter() {collection=/FortUILibrary/Systems/ItemShopTile/MPC_ItemShop, parameterName=CornerRadius}
   Radius in Pixels=CollectionParameter() {collection=&MPC_ItemShop, parameterName=bRadiusInPixels}
   Skew Angle
   Outline Thickness=ScalarParameter() {defaultValue=6, parameterName=Border Thickness}
  ) {materialFunction=&MF_UI_CoreUI_RoundedMask}
  n9 = LinearInterpolate(
   B=StaticSwitchParameter(True=LinearInterpolate(A=@n2, B=@n3, Alpha=@n4.Result), False=LinearInterpolate(A=@n5, B=@n6, Alpha=@n7.Result)) {parameterName=bIsBlue}
   Alpha=@n8.Shape Mask
  )
  n10 = SmoothStep(Value=@n8.Outline SDF) {constMax=0.008}
  FunctionOutput(None=StaticSwitchParameter(True=LinearInterpolate(A=@n1, B=@n9, Alpha=@n10), False=FunctionInput()) {parameterName=blsEmeraldGlass_Series})
  LinearInterpolate(
   A=FunctionInput(Preview=VectorParameter() {group=CorruptSeries, parameterName=BG_Color, defaultValue=<rgba 1 1 1 1>}.RGB) {bUsePreviewValueAsDefault=true, inputName=RGB 1, sortPriority=1}
   B=FunctionInput(Preview=VectorParameter() {group=CorruptSeries, parameterName=CellColor, defaultValue=<rgba 0 0 0 1>}.RGB) {bUsePreviewValueAsDefault=true, inputName=RGB 2, sortPriority=1}
  )
  n11 = Constant()
  n12 = Multiply(A=ScalarParameter() {defaultValue=0.01, parameterName=Smoothness}, B=NamedRerouteUsage())
  n13 = ScalarParameter() {parameterName=MinWidth}
  n14 = ScalarParameter() {group=Inner Skew Box, parameterName=I_Angle_Top}
  n15 = ScalarParameter() {group=Inner Skew Box, parameterName=I_Angle_Right}
  n16 = ScalarParameter() {group=Inner Skew Box, parameterName=I_Angle_Bottom}
  n17 = ScalarParameter() {group=Inner Skew Box, parameterName=I_Angle_Left}
  n18 = ScalarParameter() {defaultValue=0.25, group=Inner Skew Box, parameterName=I_Offset_Right}
  n19 = ScalarParameter() {defaultValue=0.25, group=Inner Skew Box, parameterName=I_Offset_Left}
  n20 = MaterialFunctionCall!(UV, TextureSize) {materialFunction=&MF_UI_UVScaling}
  n21 = ComponentMask(None=@n20.ScaleY) {g=true}
  n22 = ComponentMask(None=@n20.ScaleY) {r=true}
  n23 = Divide(A=@n21, B=@n22)
  n24 = Divide(A=@n22, B=@n21)
  n25 = AppendVector(A=@n23, B=@n24)
  n26 = OneMinus(None=Multiply(A=ScalarParameter() {defaultValue=0.4, group=Inner Skew Box, parameterName=I_SkewScale}, B=Saturate(None=@n25)))
  n27 = StaticBool()
  n28 = SmoothStep(
   Min=@n11
   Max=@n12
   Value=MaterialFunctionCall!(
    MinWidth=@n13
    Angle Top=ScalarParameter() {group=Outer Skew Box, parameterName=O_Angle_Top}
    Angle Right=ScalarParameter() {group=Outer Skew Box, parameterName=O_Angle_Right}
    Angle Bottom=ScalarParameter() {group=Outer Skew Box, parameterName=O_Angle_Bottom}
    Angle Left=ScalarParameter() {group=Outer Skew Box, parameterName=O_Angle_Left}
    Scale=ScalarParameter() {defaultValue=1, group=Outer Skew Box, parameterName=O_SkewScale}
    Flip=StaticBool()
   ) {materialFunction=/Game/UI/Sidebar/Materials/MF_BoxSkew}.Result
  )
  n29 = Subtract(
   A=@n28
   B=SmoothStep(
    Min=@n11
    Max=@n12
    Value=MaterialFunctionCall!(
     MinWidth=@n13
     Angle Top=@n14
     Angle Right=@n15
     Angle Bottom=@n16
     Angle Left=@n17
     "Offset  Right"=@n18
     "Offset  Left"=@n19
     Scale=@n26
     Flip=@n27
    ) {materialFunction=/Game/UI/Sidebar/Materials/MF_BoxSkew1}.Result
   )
  )
  LinearInterpolate(A=Constant3Vector() {constant=<rgba 0.9844 0.8659 0.0003 1>}, Alpha=@n29) {constB=0}
  n30 = VectorParameter() {parameterName=InnerColor, defaultValue=<rgba 0.3372 0.3372 0.3372 1>}
  n31 = VectorParameter() {parameterName=OuterColor, defaultValue=<rgba 1 1 1 1>}
  LinearInterpolate(A=Multiply(A=@n30.RGB, B=@n28), B=@n31.RGB, Alpha=@n29)
  NamedRerouteDeclaration(None=@n23) {name=ScaleX, nodeColor=<rgba 0 0 0 1>}
  NamedRerouteDeclaration(None=@n24) {name=ScaleY, nodeColor=<rgba 0 0 0 1>}
  Saturate(None=LinearInterpolate(A=Multiply(A=@n30.A, B=@n28), B=@n31.A, Alpha=@n29))
  TextureCoordinate() {coordinateIndex=3}
 ]}
