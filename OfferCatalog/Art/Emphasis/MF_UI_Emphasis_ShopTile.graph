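# Material function that gives offer catalog shop tiles an animated dark green emphasis background, switched by the static parameter blsEmphasis_Series.
# Scaled, rotated, panned UVs drive a flow-mapped flag texture that blends two green tints, reusing the Item Shop water flow map.
# An unclamped vertical gradient blends this toward a darker tint.
# With the switch off, the input passes through; the BG_Color/CellColor blend is not connected to the output.
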MaterialFunction /OfferCatalog/Art/Emphasis/MF_UI_Emphasis_ShopTile
graph {nodes [
  n1 = MaterialFunctionCall() {materialFunction=&MF_UI_UVScaling}
  n2 = Panner(Coordinate=MaterialFunctionCall(UVs=Multiply(A=@n1.Scaled To Fill) {constB=1.25}, Rotation Center=Constant() {r=0.5}, "Rotation Angle (0-1)"=Constant() {r=-0.9187}) {materialFunction=&CustomRotator}.Rotated Values) {speedX=0.15, speedY=0.05}
  n3 = LinearInterpolate(
   A=Constant3Vector() {constant=<rgba 0.116 0.1683 0.1188 1>}
   B=Constant3Vector() {constant=<rgba 0.054 0.0677 0.053 1>}
   Alpha=MaterialFunctionCall!(
    Texture 2D=TextureObject() {samplerType=SAMPLERTYPE_Alpha, texture=/OfferCatalog/Art/FNCS22/T_UI_FlagTexture}
    Texture UVs=Multiply(A=@n2) {constB=0.4}
    FlowMap Texture 2D=TextureObject() {samplerType=SAMPLERTYPE_LinearColor, texture=/Game/UI/Frontend/ItemShop/Art/WaterBackground/T_UI_WaterFlowMap}
    FlowMap UVs
    FlowIntensity=Constant() {r=0.3}
    FlowSpeed=Constant() {r=0.5}
   ) {materialFunction=/FortUILibrary/MaterialFunctions/Effects/MF_UI_FlowMap}.Result
  )
  n4 = TextureCoordinate()
  n5 = Multiply(A=Subtract(A=ComponentMask(None=@n4) {g=true}) {constB=0.35}) {constB=2}
  n6 = Multiply(A=@n5, B=Constant3Vector() {constant=<rgba 0.0075 0.0122 0.0072 1>})
  FunctionOutput(None=StaticSwitchParameter(True=LinearInterpolate(A=@n3, B=@n6, Alpha=@n5), False=FunctionInput()) {parameterName=blsEmphasis_Series})
  LinearInterpolate(
   A=FunctionInput(Preview=VectorParameter() {group=CorruptSeries, parameterName=BG_Color, defaultValue=<rgba 1 1 1 1>}.RGB) {bUsePreviewValueAsDefault=true, inputName=RGB 1, sortPriority=1}
   B=FunctionInput(Preview=VectorParameter() {group=CorruptSeries, parameterName=CellColor, defaultValue=<rgba 0 0 0 1>}.RGB) {bUsePreviewValueAsDefault=true, inputName=RGB 2, sortPriority=1}
  )
 ]}
