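# Material function that supplies the Red Oasis series background coloring for shop tile UI materials.
# The Red Oasis shop tile texture's RGB drives lerps between constant color pairs, with static switches Black and Green selecting dark, green or near-white variants.
# The blsRedOasis_Series switch outputs that result, or passes the function input through when off.
# A lerp of CorruptSeries BG_Color and CellColor inputs sits unconnected to the output.
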MaterialFunction /OfferCatalog/Art/FFC/MF_UI_ShopTile_RedOasis
graph {nodes [
  n1 = LinearInterpolate(
   A=Constant3Vector()
   B=Constant3Vector() {constant=<rgba 0.069 0.0729 0.0681 0>}
   Alpha=TextureSample() {texture=&/T_UI_RedOasisShopTile}.RGB
  )
  n2 = StaticSwitchParameter(
   True=LinearInterpolate(
    A=Constant3Vector() {constant=<rgba 0.0179 0.8594 0.0576 0>}
    B=Constant3Vector() {constant=<rgba 0.5578 0.9167 0.4345 1>}
    Alpha=TextureSample() {texture=&T_UI_RedOasisShopTile}.RGB
   )
   False=LinearInterpolate(
    A=Constant3Vector() {constant=<rgba 0.9624 1 0.9789 0>}
    B=Constant3Vector() {constant=<rgba 0.7135 0.7135 0.7135 1>}
    Alpha=TextureSample() {texture=&T_UI_RedOasisShopTile}.RGB
   )
  ) {parameterName=Green}
  FunctionOutput(None=StaticSwitchParameter(True=StaticSwitchParameter(True=@n1, False=@n2) {parameterName=Black}, False=FunctionInput()) {parameterName=blsRedOasis_Series})
  LinearInterpolate(
   A=FunctionInput(Preview=VectorParameter() {group=CorruptSeries, parameterName=BG_Color, defaultValue=<rgba 1 1 1 1>}.RGB) {bUsePreviewValueAsDefault=true, inputName=RGB 1, sortPriority=1}
   B=FunctionInput(Preview=VectorParameter() {group=CorruptSeries, parameterName=CellColor, defaultValue=<rgba 0 0 0 1>}.RGB) {bUsePreviewValueAsDefault=true, inputName=RGB 2, sortPriority=1}
  )
 ]}
