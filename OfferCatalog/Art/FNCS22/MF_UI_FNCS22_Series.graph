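# UI material function for the FNCS22 series offer art, building an animated purple-blue background behind the blsFNCS22_Series static switch, which otherwise passes its input through.
# A rotated, panning flag texture run through a flow map blends the two base colours.
# A waterfall texture's green channel, with UVs nudged by the flag, masks an oscillating, noise-flowed shine pattern.
# The flag flow and waterfall mask together composite the shine over the colours.
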MaterialFunction /OfferCatalog/Art/FNCS22/MF_UI_FNCS22_Series
graph {nodes [
  n1 = MaterialFunctionCall() {materialFunction=&MF_UI_UVScaling}
  n2 = MaterialFunctionCall(UVs=@n1.Scaled To Fill, Rotation Center=Constant() {r=0.5}, "Rotation Angle (0-1)"=Constant() {r=-0.9187}) {materialFunction=&CustomRotator}
  n3 = MaterialFunctionCall!(
   Texture 2D=TextureObject() {samplerType=SAMPLERTYPE_Alpha, texture=&/T_UI_FlagTexture}
   Texture UVs=Multiply(A=Panner(Coordinate=@n2.Rotated Values) {speedX=0.15, speedY=0.15}) {constB=0.354}
   FlowMap Texture 2D=TextureObject() {samplerType=SAMPLERTYPE_LinearColor, texture=/Game/UI/Frontend/ItemShop/Art/WaterBackground/T_UI_WaterFlowMap}
   FlowMap UVs
   FlowIntensity=Constant() {r=0.3}
   FlowSpeed=Constant() {r=0.5}
  ) {materialFunction=/FortUILibrary/MaterialFunctions/Effects/MF_UI_FlowMap}
  n4 = LinearInterpolate(A=Constant3Vector() {constant=<rgba 0.138 0.0742 0.75 1>}, B=Constant3Vector() {constant=<rgba 0.0343 0 0.2344 1>}, Alpha=@n3.Result)
  n5 = TextureCoordinate()
  n6 = Constant() {r=0.1187}
  n7 = Frac(None=Multiply(
    A=MaterialFunctionCall(UVs=@n5, "Rotation Angle (0-1)"=@n6) {materialFunction=&CustomRotator}.Rotated Values
    B=Constant2Vector() {g=0.25, r=0.5}
   ))
  n8 = Constant() {r=0.2}
  n9 = Multiply(A=Time(), B=Constant() {r=0.02})
  n10 = Sine(None=Frac(None=Multiply(A=@n8, B=@n9)))
  n11 = Multiply(A=@n10, B=@n10)
  n12 = Add(
   A=MaterialFunctionCall(UVs=Add(A=@n7), Texture Scale=Constant() {r=1}) {materialFunction=&ScaleUVsByCenter}.UVs
   B=AppendVector(A=Constant() {r=1.5}, B=ConstantBiasScale(None=@n11) {bias=-0.5, scale=2})
  )
  n13 = MaterialFunctionCall(
   Texture 2D=TextureObject() {samplerType=SAMPLERTYPE_Grayscale, texture=&LowResBlurredNoise}
   Texture UVs=Multiply(A=MaterialFunctionCall() {materialFunction=&MF_UI_UVScaling}.Scaled To Fill) {constB=0.01}
   FlowMap Texture 2D=TextureObject() {samplerType=SAMPLERTYPE_LinearColor, texture=&T_UI_WaterFlowMap}
   FlowIntensity=Constant() {r=0.3}
   FlowSpeed=Constant() {r=0.25}
  ) {materialFunction=&MF_UI_FlowMap}
  n14 = TextureSample(UVs=Panner(Coordinate=@n1.Scaled To Fill) {speedX=-0.1, speedY=0.15}) {samplerType=SAMPLERTYPE_Alpha, texture=&T_UI_FlagTexture}
  n15 = Multiply(A=TextureCoordinate()) {constB=1.25}
  n16 = Constant() {r=0.95}
  n17 = Saturate(None=TextureSample(UVs=LinearInterpolate(A=@n14.R, B=@n15, Alpha=@n16)) {samplerType=SAMPLERTYPE_Normal, texture=&/T_UI_FNCSwaterfall}.G)
  n18 = LinearInterpolate(B=TextureSample(UVs=LinearInterpolate(A=@n12, Alpha=@n13.Result)) {texture=/OfferCatalog/Art/PinkLemonade/T_UI_ShinePattern}.RGB, Alpha=@n17) {constA=1}
  n19 = LinearInterpolate(B=@n17, Alpha=@n3.Result)
  FunctionOutput(None=StaticSwitchParameter(True=LinearInterpolate(A=@n4, B=@n18, Alpha=@n19), False=FunctionInput()) {parameterName=blsFNCS22_Series})
  LinearInterpolate(
   A=FunctionInput(Preview=VectorParameter() {group=CorruptSeries, parameterName=BG_Color, defaultValue=<rgba 1 1 1 1>}.RGB) {bUsePreviewValueAsDefault=true, inputName=RGB 1, sortPriority=1}
   B=FunctionInput(Preview=VectorParameter() {group=CorruptSeries, parameterName=CellColor, defaultValue=<rgba 0 0 0 1>}.RGB) {bUsePreviewValueAsDefault=true, inputName=RGB 2, sortPriority=1}
  )
  LinearInterpolate(Alpha=@n13.Result)
  LinearInterpolate(A=@n18, B=Saturate(None=Add(B=@n3.Result) {constA=0.1}), Alpha=@n3.Result) {constB=0.25}
  Multiply()
  Reroute(None=@n17)
 ]}
