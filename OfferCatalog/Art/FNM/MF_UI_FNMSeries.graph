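# UI material function that applies a Fortnitemares (FNM) series storm look when bIsFNMSeries is on, otherwise passing its input through.
# IsAnim picks between RGB 1 blended with panning clouds and the storm texture, or that texture lightly tinted by RGB 2.
# Storm UVs run through aspect-fit math that compares pixel-size components to themselves, so it leaves them unchanged.
# A radial gradient node feeds nothing.
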MaterialFunction /OfferCatalog/Art/FNM/MF_UI_FNMSeries
graph {nodes [
  n1 = VectorParameter() {parameterName=RGB1, defaultValue=<rgba 0.5711 0.3663 0.1248 1>}
  n2 = TextureCoordinate()
  n3 = Constant() {r=1}
  n4 = Constant() {r=1}
  n5 = Time()
  n6 = LinearInterpolate(A=Panner(Coordinate=Multiply(A=@n2, B=@n3), Time=Multiply(A=@n4, B=@n5)) {speedX=0.2, speedY=0.2})
  n7 = MaterialFunctionCall() {materialFunction=&GetUserInterfaceUV}
  n8 = MaterialFunctionCall(Float2=@n7.Pixel Size) {materialFunction=&BreakOutFloat2Components}
  n9 = Divide(A=@n8.G, B=@n8.G)
  n10 = AppendVector(A=Constant() {r=1}, B=@n9)
  n11 = Constant()
  n12 = Multiply(A=OneMinus(None=@n9)) {constB=0.5}
  n13 = MaterialFunctionCall(Float2=MaterialFunctionCall() {materialFunction=&GetUserInterfaceUV}.Pixel Size) {materialFunction=&BreakOutFloat2Components}
  n14 = Constant()
  n15 = MaterialFunctionCall() {materialFunction=&GetUserInterfaceUV}
  n16 = MaterialFunctionCall(Float2=@n15.Pixel Size) {materialFunction=&BreakOutFloat2Components}
  n17 = Divide(A=@n16.R, B=@n16.R)
  n18 = AppendVector(A=@n17, B=Constant() {r=1})
  n19 = Multiply(A=OneMinus(None=@n17)) {constB=0.5}
  n20 = Constant()
  n21 = LinearInterpolate(
   A=Add(A=Multiply(A=@n7.Normalized UV, B=@n10), B=AppendVector(A=@n11, B=@n12))
   B=Add(A=Multiply(A=@n15.Normalized UV, B=@n18), B=AppendVector(A=@n19, B=@n20))
   Alpha=If(A=@n13.R, B=@n13.R, "A > B"=@n14, "A == B"=@n14, "A < B"=Constant() {r=1})
  )
  n22 = LinearInterpolate(
   A=LinearInterpolate(
    A=FunctionInput(Preview=@n1.RGB) {inputName=RGB 1, sortPriority=1}
    B=TextureSampleParameter2D(UVs=@n6) {parameterName=CloudTex, texture=&T_Smooth_tile}.RGB
   ) {constAlpha=0.6}
   B=TextureSampleParameter2D(UVs=@n21) {parameterName=FNMTexture, texture=&/T_UI_Fortnitemares_Storm}.RGB
  ) {constAlpha=0.7}
  n23 = LinearInterpolate(
   A=TextureSampleParameter2D(UVs=@n21) {parameterName=FNMTexture, texture=&T_UI_Fortnitemares_Storm}.RGB
   B=FunctionInput(Preview=VectorParameter() {parameterName=RGB2, defaultValue=<rgba 0 0 0.5469 1>}.RGB) {inputName=RGB 2, sortPriority=1}
  ) {constAlpha=0.04}
  FunctionOutput(None=StaticSwitchParameter(True=StaticSwitchParameter(True=@n22, False=@n23) {parameterName=IsAnim}, False=FunctionInput()) {parameterName=bIsFNMSeries})
  MaterialFunctionCall(
   UVs=MaterialFunctionCall() {materialFunction=&ScreenAlignedUVs}."X 100%, Y 100%"
   CenterPosition=Constant2Vector() {g=0.5, r=0.5}
   Radius=ScalarParameter() {defaultValue=0.5, parameterName=RadialRadius_02}
   Density=ScalarParameter() {defaultValue=1.0452, parameterName=RadialDensity_02}
  ) {materialFunction=&RadialGradientExponential}
 ]}
