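# Material function that builds the animated Fortnitemares 2022 item shop tile background, enabled by the blsFNMares_2022_Series static switch, which otherwise passes its input through.
# It sums masked arrows, radial color gradients and procedural confetti warped by a panning swirl flowmap, then blends in a background texture.
# Panned screen static with timed glitch bursts distorts a radial gradient that, combined with the background texture, masks the composite.
# Several branches, including frame and emissive color, are unconnected to the output.
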MaterialFunction /OfferCatalog/Art/FNM/S22/MF_UI_FNMares_ShopTile
graph {nodes [
  AppendVector(A=ScalarParameter() {parameterName=L1-Patter-X-Offset}, B=ScalarParameter() {defaultValue=0.5, parameterName=L1-Patter-Y-Offset})
  n1 = Multiply()
  AppendVector(A=Multiply(A=@n1, B=LinearInterpolate()), B=Constant() {r=0.15})
  n2 = MaterialFunctionCall() {materialFunction=&GetUserInterfaceUV}
  n3 = Subtract(A=@n2.Pixel Size, B=Multiply(A=ScalarParameter() {defaultValue=6, parameterName=PixelSize}) {constB=2})
  Divide(A=ComponentMask(None=@n3) {r=true}, B=ComponentMask(None=@n2.Pixel Size) {r=true})
  Divide(A=ComponentMask(None=@n3) {g=true}, B=ComponentMask(None=@n2.Pixel Size) {g=true})
  n4 = TextureCoordinate()
  n5 = Multiply(A=Time(), B=Constant() {r=1.5})
  n6 = Constant() {r=0.5}
  n7 = TextureCoordinate()
  n8 = ComponentMask(None=TextureSample(
    UVs=Panner(Coordinate=@n7) {speedX=0.1, speedY=0.1}
    Tex=TextureObject() {samplerType=SAMPLERTYPE_LinearColor, texture=/Game/Global/Textures/Default/Flowmaps/T_SwirlFlow_02}
   ).R) {g=true, r=true}
  n9 = Add(A=Multiply(A=Constant2Vector() {g=1, r=-1}, B=Multiply(A=@n6, B=@n8)), B=TextureCoordinate())
  n10 = ScalarParameter() {defaultValue=2, parameterName=Layer2Scale}
  n11 = Multiply(
   A=Multiply(
    A=Rotator(Coordinate=@n4, Time=@n5)
    B=MaterialFunctionCall(UVs=@n9, Texture Scale=@n10) {materialFunction=&ScaleUVsByCenter}.UVs
   )
   B=ScalarParameter() {defaultValue=7.3891, parameterName=GSize}
  )
  n12 = Constant() {r=1}
  n13 = Constant() {r=5}
  n14 = MaterialFunctionCall(
   UVs=MaterialFunctionCall() {materialFunction=&MF_UI_UVScaling}.Scaled To Fill
   Texture Scale=ScalarParameter() {defaultValue=1.25, parameterName=Scale_Arrows}
  ) {materialFunction=&ScaleUVsByCenter}
  n15 = MaterialFunctionCall(X=ScalarParameter() {defaultValue=0.1, parameterName=arrowX}, Y=ScalarParameter() {defaultValue=0.04, parameterName=ArrowY}) {materialFunction=&MakeFloat2}
  n16 = LinearInterpolate(
   B=LinearInterpolate(
    A=Constant3Vector() {constant=<rgba 0 0 0 1>}
    B=Constant3Vector() {constant=<rgba 0.7913 0 0 1>}
    Alpha=MaterialFunctionCall(Texture=@n11, GradientStart=@n12, GradientEnd=@n13) {materialFunction=&MF_UI_LinearGradient}.Result
   )
   Alpha=TextureSample(UVs=Add(A=@n14.UVs, B=@n15.Result)) {texture=&/T_UI_Arrows}.R
  )
  n17 = Constant3Vector() {constant=<rgba 0.0184 0.0325 0.3177 0>}
  n18 = MaterialFunctionCall() {materialFunction=&MF_UI_UVScaling}
  n19 = ScalarParameter() {defaultValue=1.3802, parameterName=Scale_G2}
  n20 = ScalarParameter() {defaultValue=0.005, parameterName=X_G2}
  n21 = ScalarParameter() {defaultValue=-0.296, parameterName=Y_G2}
  n22 = MaterialFunctionCall(
   UV=Add(
    A=MaterialFunctionCall(UVs=@n18.Scaled to Fit, Texture Scale=@n19) {materialFunction=&ScaleUVsByCenter}.UVs
    B=MaterialFunctionCall(X=@n20, Y=@n21) {materialFunction=&MakeFloat2}.Result
   )
   GradientStart=Constant() {r=-0.008}
   GradientEnd=Constant() {r=0.612}
  ) {materialFunction=&MF_UI_RadialGradient}
  n23 = Constant3Vector() {constant=<rgba 0.5521 0 0.1072 0>}
  n24 = LinearInterpolate(A=LinearInterpolate(B=LinearInterpolate(B=@n17), Alpha=@n22.Result), B=LinearInterpolate(B=LinearInterpolate(B=@n23), Alpha=@n22.Result)) {constAlpha=0}
  n25 = Constant() {r=0.5}
  n26 = TextureCoordinate()
  n27 = ComponentMask(None=TextureSample(
    UVs=Panner(Coordinate=@n26) {speedX=0.07, speedY=0.0132}
    Tex=TextureObject() {samplerType=SAMPLERTYPE_LinearColor, texture=&T_SwirlFlow_02}
   ).R) {g=true, r=true}
  n28 = Multiply(A=Multiply(A=Constant2Vector() {g=1, r=-1}, B=Multiply(A=@n25, B=@n27)))
  n29 = TextureCoordinate()
  n30 = MaterialFunctionCall!(
   "UV(V2)"=MaterialFunctionCall(UVs=Add(A=@n28, B=@n29), Texture Scale=ScalarParameter() {defaultValue=1.9466, parameterName=Layer1Scale}) {materialFunction=&ScaleUVsByCenter}.UVs
   Global-Angle=ScalarParameter() {defaultValue=-0.5, parameterName=L1-Global Angle}
   "Pattern-Tiles(V2)"=AppendVector(A=ScalarParameter() {defaultValue=30, parameterName=L1-Pattern-X}, B=ScalarParameter() {defaultValue=300, parameterName=L1-Pattern-Y})
   RandomSeed=ScalarParameter() {defaultValue=12345, parameterName=L1-RandomSeed}
   Confetti-Speed-TimeFactor=ScalarParameter() {defaultValue=-0.05, parameterName=L1-Confetti-Speed-TimeFactor}
   Confetti-Rotation-TimeFactor=ScalarParameter() {defaultValue=0.25, parameterName=L1-Confetti-Rotation-TimeFactor}
   Confetti-Local-Rotation-Range=ScalarParameter() {defaultValue=0.0075, parameterName=L1-Confetti-Local-Rotation-Range}
   "Confetti-Local-Tiles(V2)"=AppendVector(
    A=ScalarParameter() {defaultValue=50, parameterName=L1-Confetti-Local-Tile-X}
    B=ScalarParameter() {defaultValue=30, parameterName=L1-Confetti-Local-Tile-Y}
   )
   "SmoothStepRanges(V2)"=AppendVector(
    A=ScalarParameter() {defaultValue=1, parameterName=L1-SmoothStep-Max}
    B=ScalarParameter() {defaultValue=0.5, parameterName=L1-SmoothStep-Min}
   )
  ) {materialFunction=/Game/UI/Frontend/ItemShop/Art/Cactus/MF_UI_Confetti}
  n31 = LinearInterpolate(A=@n30.Confetti-Mask, B=Constant3Vector() {constant=<rgba 0.3594 0.2418 0 0>}, Alpha=@n30.Confetti-Mask)
  n32 = MaterialFunctionCall(
   UVs=MaterialFunctionCall() {materialFunction=&MF_UI_UVScaling}.Scaled to Fit
   Texture Scale=ScalarParameter() {defaultValue=1.1583, parameterName=Scale}
  ) {materialFunction=&ScaleUVsByCenter}
  n33 = MaterialFunctionCall(X=ScalarParameter() {defaultValue=0.005, parameterName=X}, Y=ScalarParameter() {defaultValue=0.192, parameterName=Y}) {materialFunction=&MakeFloat2}
  n34 = LinearInterpolate(
   A=Constant3Vector()
   B=Constant3Vector() {constant=<rgba 0.0312 0.0003 0 0>}
   Alpha=MaterialFunctionCall(
    UV=Add(A=@n32.UVs, B=@n33.Result)
    GradientStart=ScalarParameter() {defaultValue=0.029, parameterName=GradientStart}
    GradientEnd=ScalarParameter() {defaultValue=0.7818, parameterName=GradientEnd}
   ) {materialFunction=&MF_UI_RadialGradient}.Result
  )
  n35 = TextureSample(UVs=MaterialFunctionCall() {materialFunction=&MF_UI_UVScaling}.Scaled To Fill) {texture=&/T_UI_FNMares_BGTexture}
  n36 = LinearInterpolate(A=Add(A=Add(A=@n16, B=@n24), B=LinearInterpolate(A=@n31, B=@n34)), B=@n35.RGB)
  n37 = TextureCoordinate()
  n38 = Multiply(A=Time(), B=ScalarParameter() {defaultValue=0.08, parameterName=BigGlitchSpeed})
  n39 = Round(None=Power(Base=Frac(None=@n38)))
  n40 = LinearInterpolate(A=@n37, B=Multiply(A=@n37) {constB=0.75}, Alpha=@n39)
  n41 = Multiply(A=TextureSample(UVs=Panner(Coordinate=@n40) {speedX=5, speedY=0.1}) {samplerType=SAMPLERTYPE_Masks, texture=/Game/Missions/Primary/RepairTheShelter/Textures/Cinematic_Screen_Static}.B) {constB=50}
  n42 = Multiply(
   A=TextureSample(UVs=Panner(Coordinate=@n40) {speedX=6, speedY=-0.05}) {samplerType=SAMPLERTYPE_Masks, texture=&Cinematic_Screen_Static}.G
   B=TextureSample(UVs=Panner(Coordinate=@n40) {speedX=-3, speedY=-0.03}) {samplerType=SAMPLERTYPE_Masks, texture=&Cinematic_Screen_Static}.R
  )
  n43 = ScalarParameter() {defaultValue=0.25, group=Static, parameterName=Static Scale}
  n44 = ScalarParameter() {defaultValue=0.5, group=Static, parameterName=Static Big Scale}
  n45 = LinearInterpolate(
   A=ScalarParameter() {defaultValue=0.344, group=Static, parameterName=StaticLowValue}
   B=ScalarParameter() {defaultValue=1, group=Static, parameterName=StaticHighValue}
   Alpha=Multiply(A=Add(A=@n41, B=@n42), B=LinearInterpolate(A=@n43, B=@n44) {constAlpha=1})
  )
  n46 = Multiply(
   A=@n42
   B=LinearInterpolate(
    A=ScalarParameter() {defaultValue=1.5, group=Static, parameterName=Static UV Distort}
    B=ScalarParameter() {defaultValue=3, group=Static, parameterName=Static UV Big Distort}
    Alpha=@n39
   )
  )
  n47 = Constant() {r=0.05}
  n48 = Add(
   A=MaterialFunctionCall() {materialFunction=&MF_UI_UVScaling}.Scaled To Fill
   B=Multiply(A=Add(A=@n45), B=AppendVector(A=@n46, B=@n47))
  )
  n49 = MaterialFunctionCall(UV=@n48, GradientStart=Constant() {r=1.5}, GradientEnd=Constant() {r=0.1}) {materialFunction=&MF_UI_RadialGradient}
  n50 = LinearInterpolate(B=@n49.Result)
  n51 = Add(A=@n35.RGB, B=Multiply(A=OneMinus(None=@n50)) {constB=2})
  FunctionOutput(None=StaticSwitchParameter(True=LinearInterpolate(B=@n36, Alpha=@n51), False=FunctionInput()) {parameterName=blsFNMares_2022_Series})
  LinearInterpolate(
   A=FunctionInput(Preview=VectorParameter() {group=CorruptSeries, parameterName=BG_Color, defaultValue=<rgba 1 1 1 1>}.RGB) {bUsePreviewValueAsDefault=true, inputName=RGB 1, sortPriority=1}
   B=FunctionInput(Preview=VectorParameter() {group=CorruptSeries, parameterName=CellColor, defaultValue=<rgba 0 0 0 1>}.RGB) {bUsePreviewValueAsDefault=true, inputName=RGB 2, sortPriority=1}
  )
  n52 = Add(
   A=MaterialFunctionCall(
    UVs=MaterialFunctionCall() {materialFunction=&MF_UI_UVScaling}.Scaled to Fit
    Texture Scale=ScalarParameter() {defaultValue=0.6005, parameterName=Scale g2}
   ) {materialFunction=&ScaleUVsByCenter}.UVs
   B=MaterialFunctionCall(X=ScalarParameter() {defaultValue=0.005, parameterName=X g2}, Y=ScalarParameter() {defaultValue=-0.056, parameterName=Y g2}) {materialFunction=&MakeFloat2}.Result
  )
  n53 = ScalarParameter() {defaultValue=0.101, parameterName=GradientStart g2}
  n54 = ScalarParameter() {defaultValue=1.3253, parameterName=GradientEnd g2}
  LinearInterpolate(
   A=LinearInterpolate(
    A=Constant3Vector()
    B=Constant3Vector()
    Alpha=MaterialFunctionCall(UV=@n52, GradientStart=@n53, GradientEnd=@n54) {materialFunction=&MF_UI_RadialGradient}.Result
   )
   B=@n34
  )
  n55 = VectorParameter() {parameterName=Emissive Color, defaultValue=<rgba 1 1 1 0>}
  LinearInterpolate(A=Multiply(A=@n55.RGB), B=Multiply(A=@n49.Result, B=@n55.RGB)) {constAlpha=0.3}
  n56 = ScalarParameter() {defaultValue=0.8377, parameterName=FrameFeather1}
  n57 = ScalarParameter() {defaultValue=1.2794, parameterName=FrameFeather2}
  n58 = Abs(None=ConstantBiasScale(None=MaterialFunctionCall() {materialFunction=&MF_UI_UVScaling}.ScaleY) {bias=-0.5, scale=2})
  n59 = Add(
   A=SmoothStep(Min=@n56, Max=@n57, Value=ComponentMask(None=@n58) {r=true})
   B=SmoothStep(Min=@n56, Max=@n57, Value=ComponentMask(None=@n58) {g=true})
  )
  Multiply(A=Multiply(A=Saturate(None=@n59), B=ScalarParameter() {defaultValue=10, parameterName=FrameOpacity}), B=@n55.RGB)
  Multiply(
   A=Add(B=@n1)
   B=LinearInterpolate(
    A=ScalarParameter() {defaultValue=0.25, group=Static, parameterName=Static Scale}
    B=ScalarParameter() {defaultValue=0.5, group=Static, parameterName=Static Big Scale}
   ) {constAlpha=1}
  )
  TextureCoordinate()
  TextureSample(UVs=@n48) {texture=/OfferCatalog/Art/Alfredo/T_UI_Frame}
 ]}
