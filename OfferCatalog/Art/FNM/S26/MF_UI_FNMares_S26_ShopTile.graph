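# Material function for Fortnitemares S26 item shop tile backgrounds, adding animated purple fog over a background texture.
# Fog combines flowmap-distorted blurred noise with two panning noise layers, tinted and masked by a rotatable smoothstepped linear gradient.
# A static switch selects background UVs via named reroutes; blsFNMares_2023_Series outputs either the effect or the passed-in input.
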MaterialFunction /OfferCatalog/Art/FNM/S26/MF_UI_FNMares_S26_ShopTile
graph {nodes [
  n1 = TextureSampleParameter2D(UVs=StaticSwitchParameter(True=NamedRerouteUsage(), False=NamedRerouteUsage()) {group=Texture, parameterName=UVScaling}) {parameterName=Background Texture, texture=&/T_UI_MarbleFloor}
  n2 = Constant3Vector() {constant=<rgba 0.5029 0 1 1>}
  n3 = TextureObjectParameter() {
   group=Fog Flowmap
   parameterName=Fog Shape - Flowmap
   samplerType=SAMPLERTYPE_LinearColor
   texture=/Game/Global/Textures/Default/Flowmaps/T_SwirlFlow_02
  }
  n4 = ScalarParameter() {defaultValue=3, group=Fog Flowmap, parameterName=Flowmap Size}
  n5 = MaterialFunctionCall() {materialFunction=&MF_UI_UVScaling}
  n6 = ScalarParameter() {defaultValue=0.02, group=Fog Flowmap, parameterName=Flowmap Movement X}
  n7 = ScalarParameter() {defaultValue=0.02, group=Fog Flowmap, parameterName=Flowmap Movement Y}
  n8 = MaterialFunctionCall!(
   Texture 2D=TextureObject() {samplerType=SAMPLERTYPE_Grayscale, texture=&LowResBlurredNoise}
   Texture UVs
   FlowMap Texture 2D=@n3
   FlowMap UVs=Panner(Coordinate=Multiply(A=@n5.Scaled To Fill, B=@n4) {constB=0.75}, Speed=AppendVector(A=@n6, B=@n7)) {speedY=0.03}
   FlowIntensity=ScalarParameter() {defaultValue=0.2, group=Fog Flowmap, parameterName=Flowmap Intensity}
   FlowSpeed
  ) {materialFunction=/FortUILibrary/MaterialFunctions/Effects/MF_UI_FlowMap}
  n9 = Multiply(A=@n5.Scaled To Fill) {constB=2}
  n10 = Multiply(A=Time(), B=ScalarParameter() {defaultValue=1, group=Fog, parameterName=Cloud Movement})
  n11 = Add(
   A=TextureSample(UVs=Panner(Coordinate=@n9, Time=@n10) {speedX=0.015, speedY=0.013}) {samplerSource=SSM_Wrap_WorldGroupSettings, texture=/Game/UI/Frontend/ItemShop/Art/IWDTakeover/Texture/T-Noise-IWD}.R
   B=TextureSample(UVs=Panner(Coordinate=@n9, Time=@n10) {speedX=0.05, speedY=0.015}) {samplerType=SAMPLERTYPE_Grayscale, texture=&LowResBlurredNoise}.R
  )
  n12 = LinearInterpolate(B=LinearInterpolate(A=@n8.Result, Alpha=Multiply(A=@n11) {constB=1.4}), Alpha=@n8.Result)
  n13 = ScalarParameter() {group=Fog, parameterName=Smoke Endpoint}
  n14 = ScalarParameter() {defaultValue=0.8, group=Fog, parameterName=Smoke Intensity}
  n15 = Constant()
  n16 = MaterialFunctionCall(
   Input=ScalarParameter() {defaultValue=270, group=Fog, parameterName=Smoke Angle}
   Input Low=@n15
   Input High=Constant() {r=-360}
   Target Low=@n15
   Target High=Constant() {r=1}
  ) {materialFunction=&RemapValueRange}
  n17 = Multiply(
   A=Multiply(A=Multiply(A=@n2, B=@n12), B=ScalarParameter() {defaultValue=1.7, group=Fog, parameterName=Smoke Opacity}) {constB=0.3}
   B=SmoothStep(Value=MaterialFunctionCall(GradientStart=@n13, GradientEnd=@n14, RotationAngle=@n16.Result) {materialFunction=&MF_UI_LinearGradient}.Result) {constMax=1.5}
  )
  FunctionOutput(None=StaticSwitchParameter(True=Add(A=@n1.RGB, B=@n17), False=FunctionInput()) {parameterName=blsFNMares_2023_Series})
  LinearInterpolate(
   A=FunctionInput(Preview=VectorParameter() {group=CorruptSeries, parameterName=BG_Color, defaultValue=<rgba 1 1 1 1>}.RGB) {bUsePreviewValueAsDefault=true, inputName=RGB 1, sortPriority=1}
   B=FunctionInput(Preview=VectorParameter() {group=CorruptSeries, parameterName=CellColor, defaultValue=<rgba 0 0 0 1>}.RGB) {bUsePreviewValueAsDefault=true, inputName=RGB 2, sortPriority=1}
  )
  NamedRerouteDeclaration(None=@n5.Scaled to Fit) {name=UVFit, nodeColor=<rgba 1 0 0.7266 1>}
  NamedRerouteDeclaration(None=MaterialFunctionCall() {materialFunction=&ScreenAlignedUVs}."X 100%, Y 100%") {name=ScreenAlignedUVs, nodeColor=<rgba 0 0.6484 1 1>}
 ]}
