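# Material function for the Fearless Flight shop tile art, switched by bls_FearlessFlight and passing through its input when off.
# It builds background layers from a purple gradient, flow-mapped CRT lines, a starburst with radiating lines, and panned radial-coordinate streaks.
# Four portal-texture samples, with rotating UVs at different scales and animated hexagon highlights, form a masked layer maxed against an orange radial gradient and lerped over a base.
# Named reroutes link the layers.
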MaterialFunction /OfferCatalog/Art/FearlessFlight/MF_UI_ShopTile_FearlessFlight
graph {nodes [
  ComponentMask() {g=true, r=true}
  n1 = NamedRerouteUsage()
  n2 = Constant3Vector() {constant=<rgba 1 0.5128 0.0085 1>}
  n3 = Constant3Vector() {constant=<rgba 1 0.1592 0 1>}
  n4 = MaterialFunctionCall(UV=NamedRerouteUsage(), GradientStart=Constant(), GradientEnd=Constant() {r=1}) {materialFunction=&MF_UI_RadialGradient}
  n5 = TextureSample(UVs=NamedRerouteUsage(), Tex=NamedRerouteUsage())
  n6 = NamedRerouteUsage()
  n7 = TextureSample(UVs=@n6, Tex=NamedRerouteUsage())
  n8 = NamedRerouteUsage()
  n9 = TextureSample(UVs=@n8, Tex=NamedRerouteUsage())
  n10 = NamedRerouteUsage()
  n11 = TextureSample(UVs=@n10, Tex=NamedRerouteUsage())
  n12 = Saturate(None=Add(A=Add(A=@n5.A, B=@n7.A), B=Add(A=@n9.A, B=@n11.A)))
  n13 = Constant() {r=0.005}
  n14 = Multiply(
   A=MaterialFunctionCall!(time, UV=@n8, time_gain, UV_Static, hexTimeGain, PatternHexIntensity=@n13, flip) {materialFunction=/OfferCatalog/Art/_MaterialFunctions_ShopTheming/MF_UI_Hexagon}.Result
   B=NamedRerouteUsage()
  )
  n15 = Multiply(
   A=Multiply(A=MaterialFunctionCall(UV=@n10, PatternHexIntensity=@n13) {materialFunction=&MF_UI_Hexagon}.Result) {constB=5}
   B=NamedRerouteUsage()
  )
  n16 = Multiply(
   A=Multiply(A=MaterialFunctionCall(UV=@n6, PatternHexIntensity=@n13) {materialFunction=&MF_UI_Hexagon}.Result) {constB=3}
   B=NamedRerouteUsage()
  )
  n17 = Multiply(A=@n7.RGB) {constB=0.65}
  n18 = NamedRerouteUsage()
  n19 = Multiply(
   A=NamedRerouteUsage()
   B=Multiply(A=MaterialFunctionCall(UV=@n18, PatternHexIntensity=@n13) {materialFunction=&MF_UI_Hexagon}.Result) {constB=5}
  )
  n20 = Multiply(A=@n5.RGB) {constB=0.8}
  n21 = Add(A=Add(A=Add(A=@n16, B=@n17), B=Add(A=@n19, B=@n20)), B=Add(A=Add(A=@n14, B=@n9.RGB), B=Add(A=@n15, B=@n11.RGB)))
  n22 = Max(
   A=MaterialFunctionCall(A=@n2, B=@n3, C=Constant3Vector() {constant=<rgba 1 0.0079 0 1>}, Alpha=@n4.Result) {materialFunction=&3ColorBlend}.Result
   B=Multiply(A=@n12, B=Multiply(A=@n21) {constB=3})
  )
  n23 = Subtract(A=@n12, B=Multiply(A=NamedRerouteUsage()) {constB=2})
  FunctionOutput(None=StaticSwitchParameter(True=LinearInterpolate(A=@n1, B=@n22, Alpha=@n23), False=FunctionInput()) {parameterName=bls_FearlessFlight})
  LinearInterpolate(
   A=FunctionInput(Preview=VectorParameter() {group=CorruptSeries, parameterName=BG_Color, defaultValue=<rgba 1 1 1 1>}.RGB) {bUsePreviewValueAsDefault=true, inputName=RGB 1, sortPriority=1}
   B=FunctionInput(Preview=VectorParameter() {group=CorruptSeries, parameterName=CellColor, defaultValue=<rgba 0 0 0 1>}.RGB) {bUsePreviewValueAsDefault=true, inputName=RGB 2, sortPriority=1}
  )
  LinearInterpolate(A=@n2, B=@n3, Alpha=@n4.Result)
  MaterialFunctionCall() {materialFunction=&MF_UI_UVScaling}
  n24 = NamedRerouteUsage()
  n25 = Constant2Vector() {g=0.85, r=1}
  n26 = TextureSample(
   UVs=MaterialFunctionCall(UVs=Multiply(A=@n24, B=@n25), Texture Scale=Constant() {r=0.5}) {materialFunction=&ScaleUVsByCenter}.UVs
   Tex=TextureObject() {samplerType=SAMPLERTYPE_Alpha, texture=&/T_UI_RadiatingLines}
  ) {samplerType=SAMPLERTYPE_LinearColor, texture=&T_UI_RadiatingLines}
  Multiply(A=@n26.R) {constB=1.25}
  NamedRerouteDeclaration(None=MaterialFunctionCall(
    UVs=MaterialFunctionCall!(ScreenUV, TextureSize, ScalePivotPoint) {materialFunction=/FortUILibrary/MaterialFunctions/Basics/MF_UI_ScreenSpaceUVScaling}.Scaled to Fill
    Texture Scale=Constant() {r=0.5}
   ) {materialFunction=&ScaleUVsByCenter}.UVs) {name=ScreenUVs, nodeColor=<rgba 1 0.867 0 1>}
  n27 = NamedRerouteUsage()
  n28 = Time()
  NamedRerouteDeclaration(None=Rotator(Coordinate=@n27, Time=Multiply(A=@n28) {constB=0.496})) {name=Spins, nodeColor=<rgba 1 0 0.375 1>}
  n29 = Multiply(A=NamedRerouteUsage(), B=Constant2Vector() {g=200, r=84}) {constB=75}
  n30 = MaterialFunctionCall!(
   Texture 2D=TextureObject() {texture=/OfferCatalog/Art/Sunburst/T_UI_CRTTexture}
   Texture UVs=MaterialFunctionCall(UVs=Frac(None=@n29), Texture Scale=Constant2Vector() {g=0.5, r=2}) {materialFunction=&ScaleUVsByCenter}.UVs
   FlowMap Texture 2D=TextureObject() {texture=/OfferCatalog/Art/Sunburst/T_UI_CirclesFlowmap}
   FlowMap UVs
   FlowIntensity=Constant() {r=0.75}
   FlowSpeed=Constant() {r=0.484}
  ) {materialFunction=/FortUILibrary/MaterialFunctions/Effects/MF_UI_FlowMap}
  n31 = Max(A=Max(A=ComponentMask(None=@n30.Result) {r=true}, B=ComponentMask(None=@n30.Result) {g=true}), B=ComponentMask(None=@n30.Result) {b=true})
  NamedRerouteDeclaration(None=LinearInterpolate(
    A=Constant3Vector() {constant=<rgba 0.0395 0.0009 0.0343 1>}
    B=Constant3Vector() {constant=<rgba 0.1221 0 0.1144 1>}
    Alpha=Multiply(A=@n31) {constB=1.25}
   )) {name=CRT Lines, nodeColor=<rgba 1 0 0.8672 1>}
  NamedRerouteDeclaration(None=TextureObjectParameter() {parameterName=Tex, texture=&/T_UI_PortalHaze}) {name=HexTexture, nodeColor=<rgba 1 0.2403 0 1>}
  NamedRerouteDeclaration(None=TextureObject() {samplerType=SAMPLERTYPE_Alpha, texture=&/T_UI_PortalMask}) {name=HexMask, nodeColor=<rgba 1 1 1 1>}
  NamedRerouteDeclaration(None=Rotator(Coordinate=@n27, Time=Multiply(A=@n28) {constB=0.493})) {name=MedHexRotation, nodeColor=<rgba 1 0 0.4688 1>}
  NamedRerouteDeclaration(None=MaterialFunctionCall(UVs=NamedRerouteUsage(), Texture Scale=Constant() {r=0.35}) {materialFunction=&ScaleUVsByCenter}.UVs) {name=SmallHexUVs, nodeColor=<rgba 1 0.7734 0 1>}
  NamedRerouteDeclaration(None=MaterialFunctionCall(UVs=NamedRerouteUsage(), Texture Scale=Constant() {r=0.6}) {materialFunction=&ScaleUVsByCenter}.UVs) {name=MediumHexUVs, nodeColor=<rgba 0.625 0.4154 0 1>}
  NamedRerouteDeclaration(None=Rotator(Coordinate=@n27, Time=Multiply(A=@n28) {constB=0.49})) {name=SmallHexRotation, nodeColor=<rgba 0 0.8828 1 1>}
  NamedRerouteDeclaration(None=Constant3Vector() {constant=<rgba 1 0.6657 0 1>}) {name=HexRunnerColor, nodeColor=<rgba 0 0.9531 1 1>}
  n32 = Constant3Vector() {constant=<rgba 0.1413 0.0027 0.2016 1>}
  n33 = Constant3Vector() {constant=<rgba 0.4735 0.0006 0.5647 1>}
  n34 = MaterialFunctionCall(UV=NamedRerouteUsage(), GradientStart=Constant(), GradientEnd=Constant() {r=1}) {materialFunction=&MF_UI_RadialGradient}
  n35 = NamedRerouteUsage()
  n36 = ComponentMask(None=Abs(None=ConstantBiasScale(None=@n35) {bias=-0.5, scale=2})) {g=true}
  n37 = ScalarParameter() {defaultValue=0.2, parameterName=Subtract}
  n38 = Saturate(None=Multiply(A=Subtract(A=@n36, B=@n37), B=ScalarParameter() {defaultValue=3, parameterName=Falloff}))
  n39 = Constant3Vector() {constant=<rgba 0.0648 0.0006 0.3372 1>}
  NamedRerouteDeclaration(None=LinearInterpolate(A=LinearInterpolate(A=@n32, B=@n33, Alpha=@n34.Result), B=Multiply(A=@n38, B=@n39), Alpha=@n38)) {name=GradientBG, nodeColor=<rgba 1 0 0.2109 1>}
  n40 = Add(A=Multiply(A=NamedRerouteUsage()) {constB=0.745}, B=NamedRerouteUsage())
  n41 = Multiply(A=NamedRerouteUsage()) {constB=0.25}
  NamedRerouteDeclaration(None=Add(
    A=@n40
    B=MaterialFunctionCall!(Base=@n40, Blend=@n41) {materialFunction=/Engine/Functions/Engine_MaterialFunctions03/Blends/Blend_HardLight}.Result
   )) {name=FinalBG, nodeColor=<rgba 0.6846 0 1 1>}
  n42 = Constant2Vector() {g=1.25, r=1}
  n43 = TextureSample(UVs=MaterialFunctionCall(UVs=Multiply(A=@n24, B=@n42), Texture Scale=Constant() {r=0.8}) {materialFunction=&ScaleUVsByCenter}.UVs) {samplerType=SAMPLERTYPE_LinearColor, texture=&/T_UI_Starburst}
  n44 = Constant3Vector() {constant=<rgba 0 0.1293 1 1>}
  n45 = Multiply(A=@n43.G) {constB=5}
  n46 = Constant3Vector() {constant=<rgba 0 0.1449 0.5226 1>}
  n47 = Add(
   A=Add(A=Multiply(A=@n43.R, B=@n44), B=Multiply(A=@n45, B=@n46))
   B=Multiply(A=@n43.B, B=Constant3Vector() {constant=<rgba 0 0.0806 0.5694 1>})
  )
  n48 = Multiply(A=Multiply(A=@n26.R, B=Constant3Vector() {constant=<rgba 1 0.9962 0 1>})) {constB=0.75}
  NamedRerouteDeclaration(None=Add(A=NamedRerouteUsage(), B=Add(A=@n47, B=@n48))) {name=StarburstBG, nodeColor=<rgba 0 0.9766 1 1>}
  NamedRerouteDeclaration(None=@n26.R) {name=BurstAlpha, nodeColor=<rgba 0 0.2266 1 1>}
  NamedRerouteDeclaration(None=Rotator(Coordinate=@n27, Time=Multiply(A=@n28) {constB=0.52})) {name=OuterHexSpins, nodeColor=<rgba 1 0 0.375 1>}
  NamedRerouteDeclaration(None=MaterialFunctionCall(UVs=NamedRerouteUsage(), Texture Scale=Constant() {r=1.7}) {materialFunction=&ScaleUVsByCenter}.UVs) {name=OuterMostHexUVs, nodeColor=<rgba 1 0.7734 0 1>}
  n49 = NamedRerouteUsage()
  n50 = Constant3Vector() {constant=<rgba 1 0.5501 0 0>}
  n51 = MaterialFunctionCall(Vector or UVs=Subtract(A=NamedRerouteUsage()) {constB=0.5}) {materialFunction=&VectorToRadialValue}
  n52 = Panner(
   Coordinate=@n51.Radial Coordinates
   Speed=AppendVector(A=ScalarParameter() {defaultValue=0.0125, parameterName=U-Speed}, B=ScalarParameter() {defaultValue=-0.5, parameterName=V-Speed})
  )
  n53 = Constant() {r=1}
  n54 = ScalarParameter() {defaultValue=0.009, parameterName=R-V-Tiles}
  n55 = TextureSample(UVs=Multiply(A=@n52, B=AppendVector(A=@n53, B=@n54))) {texture=/Game/Global/Textures/Default/PointSplatter/T_Ui_ColumbusSeries_512}
  n56 = ScalarParameter() {defaultValue=3, parameterName=G-U-Tiles}
  n57 = ScalarParameter() {defaultValue=0.03, parameterName=G-V-Tiles}
  n58 = TextureSample(UVs=Multiply(A=@n52, B=AppendVector(A=@n56, B=@n57))) {texture=&T_Ui_ColumbusSeries_512}
  n59 = Multiply(
   A=Power(Base=@n51.Linear Distance, Exp=ScalarParameter() {defaultValue=1, parameterName=MaskPower})
   B=Add(A=Multiply(A=@n55.B) {constB=2}, B=Multiply(A=@n58.B) {constB=2})
  )
  NamedRerouteDeclaration(None=Add(
    A=LinearInterpolate(A=@n49, B=@n50, Alpha=@n59)
    B=MaterialFunctionCall(Base=@n59, Blend=@n50) {materialFunction=&Blend_Overlay}.Result
   )) {name=Streaks, nodeColor=<rgba 0 0.2031 1 1>}
  NamedRerouteDeclaration(None=@n59) {name=StreaksAlpha, nodeColor=<rgba 0 1 0.5078 1>}
  NamedRerouteDeclaration(None=@n43.B) {name=StarburstAlpha, nodeColor=<rgba 1 0 0.9844 1>}
  NamedRerouteUsage()
  ScalarParameter() {defaultValue=2, parameterName=Intensity}
  TextureCoordinate()
 ]}
