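# Material function for the Greenhouse offer-catalog shop tile that draws an animated moonlit ocean backdrop, switched against a pass-through input by bls_GreenhousesSeries.
# It layers radial background, glowing moon, bobbing ships, horizon, caustics, panning fog, spotlights and edge scratches from gradient helpers and textures.
# Premium and free palettes feature throughout, with PaidToFree driving a smoke-distorted transition mask.
# Levels, hue shift and noise provide finishing adjustments.
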MaterialFunction /OfferCatalog/Art/Greenhouse/MF_UI_Greenhouse_ShopTile
graph {nodes [
  FunctionOutput(None=StaticSwitchParameter(True=NamedRerouteUsage(), False=FunctionInput()) {parameterName=bls_GreenhousesSeries})
  LinearInterpolate(
   A=FunctionInput(Preview=VectorParameter() {group=CorruptSeries, parameterName=BG_Color, defaultValue=<rgba 1 1 1 1>}.RGB) {bUsePreviewValueAsDefault=true, inputName=RGB 1, sortPriority=1}
   B=FunctionInput(Preview=VectorParameter() {group=CorruptSeries, parameterName=CellColor, defaultValue=<rgba 0 0 0 1>}.RGB) {bUsePreviewValueAsDefault=true, inputName=RGB 2, sortPriority=1}
  )
  n1 = NamedRerouteUsage()
  n2 = NamedRerouteUsage()
  n3 = Add(A=NamedRerouteUsage(), B=NamedRerouteUsage())
  n4 = StaticSwitch(
   True=NamedRerouteUsage()
   False=NamedRerouteDeclaration(None=Saturate(None=@n3)) {name=ApplyLevels, nodeColor=<rgba 0 0 0 1>}
   Value=NamedRerouteUsage()
  )
  n5 = ScalarParameter() {defaultValue=1, parameterName=ShowCentralProps, sliderMax=1}
  n6 = LinearInterpolate(A=@n4, B=Saturate(None=Add(A=@n4, B=@n2)), Alpha=@n5)
  n7 = Saturate(None=Divide(A=@n6, B=OneMinus(None=@n1)))
  n8 = Saturate(None=Add(A=LinearInterpolate(A=@n7, B=@n6, Alpha=@n5), B=NamedRerouteUsage()))
  n9 = NamedRerouteUsage()
  NamedRerouteDeclaration(None=Saturate(None=MaterialFunctionCall!(Base=@n8, Add=@n9) {materialFunction=/Engine/Functions/Engine_MaterialFunctions02/Math/MultiplyAdd}.Result)) {name=FINAL, nodeColor=<rgba 1 0 0.6562 1>}
  n10 = NamedRerouteUsage()
  n11 = MaterialFunctionCall(
   UVs=NamedRerouteUsage()
   GradientStart=Constant()
   GradientEnd=Constant() {r=1}
   RotationAngle=Constant() {r=0.25}
   GradientOffset=Constant() {r=0.15}
  ) {materialFunction=&MF_UI_LinearGradient}
  NamedRerouteDeclaration(None=Multiply(A=NamedRerouteUsage(), B=Multiply(A=@n10, B=@n11.Result))) {name=OceanSpotlight, nodeColor=<rgba 0 0 0 1>}
  n12 = NamedRerouteUsage()
  n13 = LinearInterpolate(B=NamedRerouteUsage(), Alpha=NamedRerouteUsage())
  NamedRerouteDeclaration(None=Saturate(None=Add(A=@n12, B=@n13))) {name=MergedOcean, nodeColor=<rgba 0 0 0 1>}
  NamedRerouteDeclaration(None=VectorParameter() {group=Color, parameterName=Ship1Color, defaultValue=<rgba 0.0204 0.0417 0.0099 1>}.RGB) {name=Ship1Color, nodeColor=<rgba 0 0 0 1>}
  n14 = NamedRerouteUsage()
  n15 = NamedRerouteUsage()
  n16 = NamedRerouteUsage()
  NamedRerouteDeclaration(None=LinearInterpolate(
    A=NamedRerouteUsage()
    B=LinearInterpolate(A=@n14, B=@n15, Alpha=@n16)
    Alpha=ScalarParameter() {defaultValue=1, parameterName=ShowCentralProps, sliderMax=1}
   )) {name=OceanAndBaseFog, nodeColor=<rgba 0 0 0 1>}
  n17 = LinearInterpolate(
   A=MaterialFunctionCall(UVs=NamedRerouteUsage(), Texture Scale=Constant2Vector() {g=0.5, r=1}) {materialFunction=&ScaleUVsByCenter}.UVs
   B=NamedRerouteUsage()
  ) {constAlpha=0.2}
  n18 = MaterialFunctionCall(
   UVs=NamedRerouteUsage()
   GradientStart=Constant()
   GradientEnd=Constant() {r=0.05}
   RotationAngle=Constant() {r=0.25}
   GradientOffset=Constant() {r=0.525}
  ) {materialFunction=&MF_UI_LinearGradient}
  NamedRerouteDeclaration(None=Subtract(
    A=MaterialFunctionCall(UV=@n17) {materialFunction=&MF_UI_RadialGradient}.Result
    B=NamedRerouteDeclaration(None=@n18.Result) {name=OceanMask, nodeColor=<rgba 0 0 0 1>}
   )) {name=OceanSpotlightMask, nodeColor=<rgba 0 0 0 1>}
  NamedRerouteDeclaration(None=StaticSwitch(
    True=ScreenPosition().ViewportUV
    False=MaterialFunctionCall() {materialFunction=&GetUserInterfaceUV}.Normalized UV
    Value=NamedRerouteUsage()
   )) {name=Screen UV, nodeColor=<rgba 0 0 0 1>}
  n19 = Constant() {r=0.4}
  n20 = NamedRerouteUsage()
  n21 = ScalarParameter() {defaultValue=1, group=Fg_Fog, parameterName=Fg_Fog_Scale}
  n22 = MaterialFunctionCall(
   UVs=Panner(Coordinate=Multiply(A=@n20, B=@n19), Time=NamedRerouteUsage()) {bFractionalPart=true, speedX=-0.05, speedY=0.053}
   Texture Scale=@n21
  ) {materialFunction=&ScaleUVsByCenter}
  n23 = AppendVector(A=Constant() {r=0.38}, B=Constant() {r=0.22})
  n24 = Multiply(A=Time(), B=ScalarParameter() {defaultValue=0.25, group=Fg_Fog, parameterName=Fg_Fog_PanSpeed})
  n25 = MaterialFunctionCall(
   UVs=Panner(Coordinate=Multiply(A=@n20, B=@n23), Time=NamedRerouteDeclaration(None=@n24) {name=Fg_Smoke_Time, nodeColor=<rgba 0 0 0 1>}) {bFractionalPart=true, speedX=-0.05, speedY=0.05}
   Texture Scale=@n21
  ) {materialFunction=&ScaleUVsByCenter}
  n26 = Clamp(
   None=Add(
    A=TextureSample(UVs=@n22.UVs) {texture=/Game/Global/Textures/Default/Tileable_Patterns/T_Smoke_Puff_Tile_Up}.RGB
    B=TextureSample(UVs=@n25.UVs) {texture=&T_Smoke_Puff_Tile_Up}.RGB
   )
   Min=Constant() {r=-20}
   Max=Constant() {r=6.5}
  ) {maxDefault=6.5, minDefault=-20}
  n27 = NamedRerouteUsage()
  n28 = LinearInterpolate(A=Multiply(A=NamedRerouteUsage(), B=@n27), B=@n27) {constAlpha=0.7}
  n29 = Power(Base=Multiply(A=Add(A=@n26, B=@n28)) {constB=1.5}, Exp=ScalarParameter() {defaultValue=0.5, group=Fg_Fog, parameterName=Fg_Fog_PowerExp}) {constExponent=3}
  n30 = NamedRerouteUsage()
  NamedRerouteDeclaration(None=Saturate(None=Multiply(A=@n29, B=@n30))) {name=Fog_Foreground, nodeColor=<rgba 0 0 0 1>}
  n31 = Constant() {r=0.1}
  n32 = Add(
   A=NamedRerouteUsage()
   B=AppendVector(
    A=ScalarParameter() {defaultValue=-0.11, group=Moon, parameterName=MoonX}
    B=ScalarParameter() {defaultValue=0.22, group=Moon, parameterName=MoonY}
   )
  )
  n33 = MaterialFunctionCall(
   UVs=@n32
   Texture Scale=MaterialFunctionCall(
    Input=MaterialFunctionCall(Speed=@n31) {materialFunction=&DebugTimeSine}.Result
    Input Low=Constant()
    Input High=Constant() {r=1}
    Target Low=ScalarParameter() {defaultValue=0.37, group=Moon, parameterName=MoonGlowScale}
    Target High=Constant() {r=0.405}
   ) {materialFunction=&RemapValueRange}.Result
  ) {materialFunction=&ScaleUVsByCenter}
  n34 = ScalarParameter() {group=Moon, parameterName=MoonGlowGradStart}
  n35 = ScalarParameter() {defaultValue=0.65, group=Moon, parameterName=MoonGlowGradEnd}
  n36 = Multiply(
   A=Subtract(
    A=MaterialFunctionCall(UV=@n33.UVs, GradientStart=@n34, GradientEnd=@n35) {materialFunction=&MF_UI_RadialGradient}.Result
    B=NamedRerouteUsage()
   )
   B=NamedRerouteUsage()
  )
  n37 = NamedRerouteUsage()
  n38 = Saturate(None=Subtract(A=Multiply(A=@n36, B=@n37), B=NamedRerouteUsage()))
  NamedRerouteDeclaration(None=LinearInterpolate(A=NamedRerouteUsage(), B=@n38, Alpha=@n38)) {name=GlowingMoon, nodeColor=<rgba 0 0 0 1>}
  n39 = VectorParameter() {group=Color, parameterName=SpotlightPreviewReward_Additional_PremiumColor, defaultValue=<rgba 1 0.8698 0.141 1>}
  n40 = VectorParameter() {group=Color, parameterName=SpotlightPreviewReward_Additional_FreeColor, defaultValue=<rgba 1 0.939 0.8333 1>}
  NamedRerouteDeclaration(None=LinearInterpolate(A=@n39.RGB, B=@n40.RGB, Alpha=NamedRerouteUsage())) {name=Additional_RewardPreviewSpotlightColor, nodeColor=<rgba 0 0 0 1>}
  NamedRerouteDeclaration(None=MaterialFunctionCall(UV=NamedRerouteUsage()) {materialFunction=&MF_UI_UVScaling}.Scaled to Fit) {name=Screen UV Scaled to Fit, nodeColor=<rgba 0 0 0 1>}
  n41 = NamedRerouteUsage()
  n42 = NamedRerouteUsage()
  n43 = Add(
   A=NamedRerouteUsage()
   B=AppendVector(A=ScalarParameter() {group=Radial, parameterName=Radial_X}, B=ScalarParameter() {defaultValue=0.175, group=Radial, parameterName=Radial_Y})
  )
  n44 = ScalarParameter() {defaultValue=1.6, group=Radial, parameterName=Radial_ScaleMultiplier}
  n45 = SmoothStep(
   Min=Constant() {r=-0.05}
   Max=Constant() {r=0.95}
   Value=MaterialFunctionCall(
    UV=MaterialFunctionCall(UVs=@n43, Texture Scale=@n44) {materialFunction=&ScaleUVsByCenter}.UVs
    GradientStart=ScalarParameter() {defaultValue=0.96, group=Radial, parameterName=Radial_GradStart}
    GradientEnd=ScalarParameter() {group=Radial, parameterName=Radial_GradEnd}
   ) {materialFunction=&MF_UI_RadialGradient}.Result
  )
  n46 = NamedRerouteUsage()
  n47 = NamedRerouteUsage()
  NamedRerouteDeclaration(None=LinearInterpolate(A=LinearInterpolate(A=@n41, B=@n42, Alpha=@n45), B=LinearInterpolate(A=@n46, B=@n47, Alpha=@n45), Alpha=NamedRerouteUsage())) {name=Radial, nodeColor=<rgba 0 0 0 1>}
  NamedRerouteDeclaration(None=VectorParameter() {group=Color, parameterName=EdgeOfScreen_PremiumColor, defaultValue=<rgba 0.1042 0.0484 0.01 1>}.RGB) {name=EdgeOfScreen_PremiumColor, nodeColor=<rgba 0 0 0 1>}
  n48 = NamedRerouteUsage()
  n49 = NamedRerouteUsage()
  n50 = NamedRerouteUsage()
  n51 = NamedRerouteUsage()
  n52 = NamedRerouteUsage()
  n53 = LinearInterpolate(
   A=NamedRerouteUsage()
   B=LinearInterpolate(A=LinearInterpolate(A=@n48, B=@n49, Alpha=@n50), B=LinearInterpolate(A=@n51, B=@n52, Alpha=@n50), Alpha=NamedRerouteUsage())
   Alpha=NamedRerouteUsage()
  )
  n54 = NamedRerouteUsage()
  NamedRerouteDeclaration(None=Saturate(None=MaterialFunctionCall(Base=@n53, Add=@n54) {materialFunction=&MultiplyAdd}.Result)) {name=Merged Background, nodeColor=<rgba 0 0 0 1>}
  n55 = MaterialFunctionCall(UV=MaterialFunctionCall(UVs=NamedRerouteUsage(), Texture Scale=Constant() {r=10}) {materialFunction=&ScaleUVsByCenter}.UVs) {materialFunction=&MF_UI_RadialGradient}
  n56 = Multiply(A=NamedRerouteUsage(), B=ScalarParameter() {defaultValue=0.5, group=Spotlight, parameterName=SpotlightFog_Scale})
  n57 = Multiply(A=@n56, B=AppendVector(A=Constant() {r=1}, B=Constant() {r=0.9}))
  n58 = NamedRerouteDeclaration(None=Multiply(A=Time(), B=Constant() {r=0.5})) {name=Smoke_Bkg_Time, nodeColor=<rgba 0 0 0 1>}
  n59 = Constant() {r=1.4}
  n60 = Constant() {r=1.32}
  n61 = Panner(Coordinate=Multiply(A=@n56, B=AppendVector(A=@n59, B=@n60)), Time=NamedRerouteUsage()) {bFractionalPart=true, speedX=-0.016, speedY=0.02}
  n62 = LinearInterpolate(
   A=TextureSample(UVs=Panner(Coordinate=@n57, Time=@n58) {bFractionalPart=true, speedX=0.012, speedY=0.053}) {samplerType=SAMPLERTYPE_LinearGrayscale, texture=/Game/Global/Textures/Default/Tileable_Patterns/T_StormCloudDetail01_M}.G
   B=TextureSample(UVs=NamedRerouteDeclaration(None=@n61) {name=Smoke_Bkg_AnimUvs, nodeColor=<rgba 0 0 0 1>}) {texture=/Game/Athena/SafeZone/S11/T_TilingClouds_D}.G
  )
  n63 = Multiply(A=@n62, B=@n62)
  NamedRerouteDeclaration(None=Multiply(A=Multiply(A=@n55.Result, B=@n63), B=ScalarParameter() {defaultValue=1, group=Spotlight, parameterName=SpotlightFog_Opacity, sliderMax=1})) {name=Fog_for_Spotlight, nodeColor=<rgba 0 0 0 1>}
  n64 = NamedRerouteUsage()
  NamedRerouteDeclaration(None=LinearInterpolate(A=NamedRerouteUsage(), B=@n64, Alpha=@n64)) {name=Fog_Merged, nodeColor=<rgba 0 0 0 1>}
  n65 = Multiply(A=NamedRerouteUsage()) {constB=0.5}
  n66 = ScalarParameter() {defaultValue=0.17, group=Ships, parameterName=Ship2X}
  n67 = Multiply(A=ScalarParameter() {defaultValue=0.1, group=Ships, parameterName=Ship2Y}, B=NamedRerouteUsage())
  n68 = NamedRerouteUsage()
  n69 = MaterialFunctionCall(UVs=Add(A=@n68, B=AppendVector(A=@n66, B=@n67)), Texture Scale=ScalarParameter() {defaultValue=0.28, group=Ships, parameterName=Ship2Scale}) {materialFunction=&ScaleUVsByCenter}
  n70 = NamedRerouteDeclaration(None=MaterialFunctionCall(UVs=@n69.UVs, Rotation Center=Constant2Vector() {r=0.5}, "Rotation Angle (0-1)"=OneMinus(None=@n65)) {materialFunction=&CustomRotator}.Rotated Values) {name=Ship2Uvs, nodeColor=<rgba 0 0 0 1>}
  n71 = LinearInterpolate(
   A=TextureSample(UVs=@n70) {texture=/OfferCatalog/Art/Greenhouse/Textures/T_UI_GreenHouse_Ship2Blur}.RGB
   B=TextureSample(UVs=@n70) {texture=/OfferCatalog/Art/Greenhouse/Textures/T_UI_GreenHouse_ShipPacked}.G
   Alpha=MaterialFunctionCall(UVs=@n69.UVs, GradientStart=Constant(), GradientEnd=Constant() {r=1}, RotationAngle=Constant() {r=0.4}) {materialFunction=&MF_UI_LinearGradient}.Result
  )
  n72 = MaterialFunctionCall(
   UVs=@n69.UVs
   GradientStart=ScalarParameter() {defaultValue=0.8, group=Ships, parameterName=Ship2FadeGrad_Start}
   GradientEnd=ScalarParameter() {defaultValue=1, group=Ships, parameterName=Ship2FadeGrad_End}
   RotationAngle=Constant() {r=0.755}
   GradientOffset=Multiply(A=ScalarParameter() {defaultValue=0.075, group=Ships, parameterName=Ship2FadeGradientOffset, sliderMax=1}) {constB=-1}
  ) {materialFunction=&MF_UI_LinearGradient}
  NamedRerouteDeclaration(None=Saturate(None=Subtract(A=@n71, B=@n72.Result))) {name=Ship2 Mask, nodeColor=<rgba 0 0 0 1>}
  n73 = Add(A=NamedRerouteUsage(), B=NamedRerouteUsage())
  NamedRerouteDeclaration(None=LinearInterpolate(Alpha=Multiply(A=@n73) {constB=0})) {name=ColoredShips, nodeColor=<rgba 0 0 0 1>}
  n74 = NamedRerouteUsage()
  n75 = ScalarParameter() {defaultValue=-0.265, group=Ships, parameterName=Ship3X}
  n76 = Multiply(A=ScalarParameter() {defaultValue=0.105, group=Ships, parameterName=Ship3Y}, B=NamedRerouteUsage())
  n77 = MaterialFunctionCall(UVs=Add(A=@n68, B=AppendVector(A=@n75, B=@n76)), Texture Scale=ScalarParameter() {defaultValue=0.375, group=Ships, parameterName=Ship3Scale}) {materialFunction=&ScaleUVsByCenter}
  n78 = NamedRerouteDeclaration(None=MaterialFunctionCall(UVs=@n77.UVs, Rotation Center=Constant2Vector() {r=0.5}, "Rotation Angle (0-1)"=Multiply(A=@n74) {constB=0.5}) {materialFunction=&CustomRotator}.Rotated Values) {name=Ship3Uvs, nodeColor=<rgba 0 0 0 1>}
  n79 = LinearInterpolate(
   A=TextureSample(UVs=@n78) {texture=/OfferCatalog/Art/Greenhouse/Textures/T_UI_GreenHouse_Ship3Blur}.RGB
   B=TextureSample(UVs=@n78) {texture=&T_UI_GreenHouse_ShipPacked}.B
   Alpha=MaterialFunctionCall(UVs=@n77.UVs, GradientStart=Constant(), GradientEnd=Constant() {r=1}, RotationAngle=Constant() {r=0.5}, GradientOffset=Constant() {r=0.4}) {materialFunction=&MF_UI_LinearGradient}.Result
  ) {constAlpha=0.9}
  n80 = MaterialFunctionCall(
   UVs=@n77.UVs
   GradientStart=Constant() {r=0.8}
   GradientEnd=Constant() {r=1}
   RotationAngle=Constant() {r=0.7475}
   GradientOffset=ScalarParameter() {defaultValue=-0.2, group=Ships, parameterName=Ship3FadeGradientOffset, sliderMax=1}
  ) {materialFunction=&MF_UI_LinearGradient}
  NamedRerouteDeclaration(None=Saturate(None=Subtract(A=@n79, B=@n80.Result))) {name=Ship3 Mask, nodeColor=<rgba 0 0 0 1>}
  n81 = NamedRerouteUsage()
  n82 = LinearInterpolate(A=@n81, B=Add(A=@n81, B=NamedRerouteUsage())) {constAlpha=0.05}
  n83 = NamedRerouteUsage()
  NamedRerouteDeclaration(None=TextureSample(UVs=Panner(Coordinate=@n82, Time=@n83) {speedX=-0.05}) {samplerType=SAMPLERTYPE_LinearColor, texture=/Game/Athena/UI/Events/LinedNotebook/Textures/Smoke512}.RGB) {name=Fg_Fogtexture, nodeColor=<rgba 0 0 0 1>}
  n84 = ScalarParameter() {defaultValue=90, group=Fg_Fog, parameterName=Fg_Fog_Angle}
  n85 = Constant()
  n86 = Constant() {r=-360}
  n87 = Constant() {r=1}
  NamedRerouteDeclaration(None=MaterialFunctionCall(
    UVs=NamedRerouteUsage()
    GradientStart=ScalarParameter() {defaultValue=-0.3, group=Fg_Fog, parameterName=Fg_Fog_Endpoint}
    GradientEnd=ScalarParameter() {defaultValue=1.6, group=Fg_Fog, parameterName=Fg_Fog_Intensity}
    RotationAngle=MaterialFunctionCall(Input=@n84, Input Low=@n85, Input High=@n86, Target Low=@n85, Target High=@n87) {materialFunction=&RemapValueRange}.Result
   ) {materialFunction=&MF_UI_LinearGradient}.Result) {name=ForegroundFog_LinearMask, nodeColor=<rgba 0 0 0 1>}
  n88 = ScalarParameter() {defaultValue=1, group=Spotlight, parameterName=Spotlight_XSize}
  n89 = ScalarParameter() {defaultValue=0.6, group=Spotlight, parameterName=Spotlight_YSize}
  n90 = LinearInterpolate(
   A=MaterialFunctionCall(UVs=NamedRerouteUsage(), Texture Scale=AppendVector(A=@n88, B=@n89)) {materialFunction=&ScaleUVsByCenter}.UVs
   B=NamedRerouteUsage()
   Alpha=Constant() {r=0.35}
  ) {constAlpha=0.2}
  n91 = ScalarParameter() {defaultValue=0.05, group=Spotlight, parameterName=Spotlight_XTransform, sliderMax=0.5, sliderMin=-0.5}
  n92 = ScalarParameter() {defaultValue=0.1, group=Spotlight, parameterName=Spotlight_YTransform, sliderMax=0.5, sliderMin=-0.5}
  n93 = MaterialFunctionCall(
   UVs=Add(A=NamedRerouteDeclaration(None=@n90) {name=FogSpotlightUVs, nodeColor=<rgba 0 0 0 1>}, B=AppendVector(A=@n91, B=@n92))
   Size=Constant()
   GlowMin=Constant()
   GlowMax=ScalarParameter() {defaultValue=0.8, group=Spotlight, parameterName=Spotlight_GlowMax, sliderMax=2}
  ) {materialFunction=&MF_UI_Circle}
  n94 = Add(A=NamedRerouteUsage(), B=AppendVector(A=Constant(), B=Constant() {r=0.25}))
  n95 = Constant() {r=0.8}
  n96 = Multiply(
   A=LinearInterpolate(A=NamedRerouteUsage(), B=NamedRerouteUsage(), Alpha=NamedRerouteUsage())
   B=MaterialFunctionCall(
    UVs=MaterialFunctionCall(UVs=@n94, Texture Scale=@n95) {materialFunction=&ScaleUVsByCenter}.UVs
    GradientStart=Constant()
    GradientEnd=Constant() {r=1}
    RotationAngle=Constant() {r=0.25}
   ) {materialFunction=&MF_UI_LinearGradient}.Result
  )
  NamedRerouteDeclaration(None=Multiply(
    A=NamedRerouteDeclaration(None=@n93.Fill) {name=FogSpotlightMask, nodeColor=<rgba 0 0 0 1>}
    B=NamedRerouteDeclaration(None=@n96) {name=FogColor, nodeColor=<rgba 0 0 0 1>}
   )) {name=FogSpotlight, nodeColor=<rgba 0 0 0 1>}
  n97 = NamedRerouteDeclaration(None=ScalarParameter() {defaultValue=1, parameterName=PaidToFree, sliderMax=1}) {name=PaidToFreeAnim, nodeColor=<rgba 0 0 0 1>}
  n98 = NamedRerouteUsage()
  n99 = Constant() {r=2}
  n100 = TextureSample(UVs=Panner(Coordinate=MaterialFunctionCall(UVs=@n98, Texture Scale=@n99) {materialFunction=&ScaleUVsByCenter}.UVs) {speedX=0.01, speedY=0.01}) {texture=&T_Smoke_Puff_Tile_Up}
  n101 = ComponentMask(None=Add(A=ComponentMask(None=@n98) {g=true, r=true}, B=Multiply(A=@n100.R) {constB=0.3})) {g=true, r=true}
  n102 = Constant2Vector() {g=0.67, r=0.67}
  n103 = Add(
   A=Multiply(A=Add(A=@n97) {constB=-0.6}) {constB=1.5}
   B=MaterialFunctionCall("Value (0-1)"=Distance(A=@n101, B=@n102)) {materialFunction=&MF_UI_EaseCurves}.EaseInOutSine
  )
  NamedRerouteDeclaration(None=SmoothStep(Value=Saturate(None=@n103)) {constMax=0.6, constMin=0.4}) {name=PaidToFreeFX, nodeColor=<rgba 0 0 0 1>}
  n104 = Constant()
  n105 = Subtract(A=NamedRerouteUsage(), B=@n104)
  n106 = Constant()
  n107 = ScalarParameter() {defaultValue=1, group=Levels, parameterName=LevelsMax}
  n108 = Min(A=Divide(A=Max(A=@n105, B=@n106), B=Subtract(A=@n107, B=@n104)), B=Constant() {r=1}) {desc=LevelsControlInputRange}
  n109 = Divide(A=Constant() {r=1}, B=ScalarParameter() {defaultValue=1, group=Levels, parameterName=LevelsGamma})
  NamedRerouteDeclaration(None=MaterialFunctionCall(
    Hue Shift Percentage=ScalarParameter() {defaultValue=-0.005, group=Levels, parameterName=LevelsHueShift, sliderMax=0.05, sliderMin=-0.05}
    Texture=Power(Base=@n108, Exp=@n109) {desc=LevelsControlInput}
   ) {materialFunction=&HueShift}.Result) {name=LevelsApplied, nodeColor=<rgba 0 0 0 1>}
  n110 = MaterialFunctionCall(UVs=NamedRerouteUsage(), Texture Scale=ScalarParameter() {defaultValue=0.125, group=Noise, parameterName=NoiseScale}) {materialFunction=&ScaleUVsByCenter}
  n111 = NamedRerouteUsage()
  n112 = Subtract(
   A=TextureSample(UVs=Panner(Coordinate=@n110.UVs, Time=@n111) {speedX=0.012, speedY=0.05}) {texture=/Game/Environments/FrontEnd/Textures/Noise}.R
   B=NamedRerouteUsage()
  )
  NamedRerouteDeclaration(None=Multiply(A=Saturate(None=@n112), B=ScalarParameter() {defaultValue=0.5, group=Noise, parameterName=NoiseIntensity, sliderMax=1})) {name=Noise, nodeColor=<rgba 0 0 0 1>}
  n113 = CurveAtlasRowParameter(CurveTime=MaterialFunctionCall(UVs=NamedRerouteUsage(), GradientStart=Constant() {r=0.5}, GradientEnd=Constant() {r=1}, RotationAngle=Constant() {r=0.2}) {materialFunction=&MF_UI_LinearGradient}.Result) {
   atlas=&/CA_UI_Greenhouse_EventScreenColors
   curve=&/C_UI_Greenhouse_ShipColor
   defaultValue=1
   group=Color
   parameterName=ShipColorCurve
  }
  NamedRerouteDeclaration(None=Multiply(A=@n113, B=NamedRerouteUsage())) {name=Ship2 Colored, nodeColor=<rgba 0 0 0 1>}
  n114 = CurveAtlasRowParameter(CurveTime=MaterialFunctionCall(UVs=NamedRerouteUsage(), RotationAngle=Constant() {r=0.3}, GradientOffset=Constant() {r=-0.1}) {materialFunction=&MF_UI_LinearGradient}.Result) {
   atlas=&CA_UI_Greenhouse_EventScreenColors
   curve=&C_UI_Greenhouse_ShipColor
   defaultValue=1
   group=Color
   parameterName=ShipColorCurve
  }
  NamedRerouteDeclaration(None=Multiply(A=@n114, B=NamedRerouteUsage())) {name=Ship3 Colored, nodeColor=<rgba 0 0 0 1>}
  NamedRerouteDeclaration(None=VectorParameter() {group=Color, parameterName=BkgRadial_PremiumLightColor, defaultValue=<rgba 0.1384 0.0723 0.0369 1>}.RGB) {name=BkgRadial_PremiumLightColor, nodeColor=<rgba 0 0 0 1>}
  NamedRerouteDeclaration(None=VectorParameter() {group=Color, parameterName=BkgRadial_PremiumDarkColor, defaultValue=<rgba 0.0152 0.0137 0.0137 1>}.RGB) {name=BkgRadial_PremiumDarkColor, nodeColor=<rgba 0 0 0 1>}
  NamedRerouteDeclaration(None=VectorParameter() {group=Color, parameterName=BkgRadial_FreeLightColor, defaultValue=<rgba 0.0252 0.0561 0.008 1>}.RGB) {name=BkgRadial_FreeLightColor, nodeColor=<rgba 0 0 0 1>}
  NamedRerouteDeclaration(None=VectorParameter() {group=Color, parameterName=BkgRadial_FreeDarkColor, defaultValue=<rgba 0.0015 0.004 0.0003 1>}.RGB) {name=BkgRadial_FreeDarkColor, nodeColor=<rgba 0.0015 0.004 0.0003 1>}
  n115 = Time()
  n116 = Multiply(A=@n115, B=Multiply(A=NamedRerouteUsage()) {constB=2})
  NamedRerouteDeclaration(None=MaterialFunctionCall(
    Input=Sine(None=@n116)
    Input Low=Constant() {r=-1}
    Input High=Constant() {r=1}
    Target Low=Constant() {r=0.95}
    Target High=Constant() {r=1.05}
   ) {materialFunction=&RemapValueRange}.Result) {name=BobAnim, nodeColor=<rgba 0 0 0 1>}
  n117 = MaterialFunctionCall!(
   Texture 2D=TextureObject() {texture=/OfferCatalog/Art/Greenhouse/Textures/T_UI_Greenhouse_EdgeNoise}
   Texture UVs=MaterialFunctionCall(UVs=NamedRerouteUsage(), Texture Scale=Constant() {r=1.025}) {materialFunction=&ScaleUVsByCenter}.UVs
   FlowMap Texture 2D=TextureObject() {texture=/OfferCatalog/Art/Greenhouse/Textures/T_UI_GreenHouse_Moon}
   FlowMap UVs=MaterialFunctionCall(UVs=NamedRerouteUsage(), Texture Scale=Constant() {r=0.75}) {materialFunction=&ScaleUVsByCenter}.UVs
   FlowIntensity=Constant() {r=0.01}
   FlowSpeed=Constant() {r=0.1}
  ) {materialFunction=/FortUILibrary/MaterialFunctions/Effects/MF_UI_FlowMap}
  n118 = Time()
  n119 = NamedRerouteUsage()
  n120 = MaterialFunctionCall(Input=Sine(None=Multiply(A=@n118, B=@n119)), Input Low=Constant() {r=-1}, Target Low=Constant() {r=0.8}, Target High=Constant() {r=1}) {materialFunction=&RemapValueRange}
  n121 = LinearInterpolate(A=NamedRerouteUsage(), B=NamedRerouteUsage(), Alpha=NamedRerouteUsage())
  NamedRerouteDeclaration(None=Multiply(A=Multiply(A=@n117.Result, B=@n120.Result), B=Multiply(A=@n121) {constB=10})) {name=EdgeOfScreenScratchs, nodeColor=<rgba 0 0 0 1>}
  NamedRerouteDeclaration(None=VectorParameter() {group=Color, parameterName=Spotlight_PremiumColor, defaultValue=<rgba 0.6667 0.3099 0.0638 1>}.RGB) {name=Spotlight_PremiumColor, nodeColor=<rgba 0 0 0 1>}
  n122 = MaterialFunctionCall(UVs=NamedRerouteUsage(), Texture Scale=Constant() {r=1}) {materialFunction=&ScaleUVsByCenter}
  NamedRerouteDeclaration(None=LinearInterpolate(
    A=VectorParameter() {group=Color, parameterName=Spotlight_FreeColorDark, defaultValue=<rgba 0.0529 0.107 0.0252 1>}.RGB
    B=VectorParameter() {group=Color, parameterName=Spotlight_FreeColorLight, defaultValue=<rgba 0.2705 0.402 0.0319 1>}.RGB
    Alpha=MaterialFunctionCall(UV=@n122.UVs) {materialFunction=&MF_UI_RadialGradient}.Result
   )) {name=Spotlight_FreeColor, nodeColor=<rgba 0 0 0 1>}
  n123 = MaterialFunctionCall(UV=MaterialFunctionCall(UVs=NamedRerouteUsage()) {materialFunction=&ScaleUVsByCenter}.UVs) {materialFunction=&MF_UI_RadialGradient}
  n124 = NamedRerouteUsage()
  NamedRerouteDeclaration(None=Multiply(A=Multiply(A=@n123.Result, B=@n124)) {constB=0.5}) {name=PreviewReward_AdditonalSpotlightRadial, nodeColor=<rgba 0 0 0 1>}
  n125 = ScalarParameter() {defaultValue=-0.03, group=Horizon, parameterName=HorizonYTransform, sliderMax=0.2, sliderMin=-0.2}
  n126 = Multiply(A=Time(), B=NamedRerouteUsage())
  n127 = Multiply(A=MaterialFunctionCall(Input=Sine(None=@n126), Input Low=Constant() {r=-1}, Target Low=Constant() {r=-1.01}, Target High=Constant() {r=1.01}) {materialFunction=&RemapValueRange}.Result) {constB=0.0025}
  n128 = AppendVector(
   A=ScalarParameter() {defaultValue=0.2, group=Horizon, parameterName=HorizonXTransform}
   B=NamedRerouteDeclaration(None=Add(A=@n125, B=@n127)) {name=HorizonYTransform, nodeColor=<rgba 0 0 0 1>}
  )
  n129 = NamedRerouteUsage()
  n130 = NamedRerouteUsage()
  NamedRerouteDeclaration(None=MaterialFunctionCall(UVs=Add(A=@n128, B=@n129), "Rotation Angle (0-1)"=Multiply(A=@n130) {constB=0.15}) {materialFunction=&CustomRotator}.Rotated Values) {name=HorizonUV Scaled to Fit, nodeColor=<rgba 0 0 0 1>}
  NamedRerouteDeclaration(None=StaticBoolParameter() {parameterName=IsVaultWorld}) {name=Bool Is Vault World, nodeColor=<rgba 0 0 0 1>}
  n131 = MaterialFunctionCall(UVs=@n32, Texture Scale=ScalarParameter() {defaultValue=0.16, group=Moon, parameterName=MoonScale}) {materialFunction=&ScaleUVsByCenter}
  n132 = Saturate(None=Subtract(A=TextureSample(UVs=@n131.UVs) {texture=&T_UI_GreenHouse_Moon}.R, B=NamedRerouteUsage()))
  NamedRerouteDeclaration(None=Multiply(A=NamedRerouteUsage(), B=NamedRerouteDeclaration(None=@n132) {name=Moon, nodeColor=<rgba 0 0 0 1>})) {name=ColoredMoon, nodeColor=<rgba 0 0 0 1>}
  n133 = Multiply(A=@n115, B=NamedRerouteUsage())
  NamedRerouteDeclaration(None=MaterialFunctionCall(
    Input=Sine(None=@n133)
    Input Low=Constant() {r=-1}
    Input High=Constant() {r=1}
    Target Low=Constant() {r=0.01}
    Target High=Constant() {r=-0.01}
   ) {materialFunction=&RemapValueRange}.Result) {name=RotateAnim, nodeColor=<rgba 0 0 0 1>}
  NamedRerouteDeclaration(None=ScalarParameter() {defaultValue=0.1, group=Animation, parameterName=AnimTimingMultiply, sliderMax=1}) {name=AnimTiming, nodeColor=<rgba 0 0 0 1>}
  NamedRerouteDeclaration(None=VectorParameter() {group=Color, parameterName=OceanSpotlightColor, defaultValue=<rgba 0.1022 0.2874 0.0168 1>}.RGB) {name=OceanSpotlightColor, nodeColor=<rgba 0 0 0 1>}
  n134 = VectorParameter() {group=Color, parameterName=MoonGlowColor, defaultValue=<rgba 0.5776 0.9301 0.1301 0.502>}
  NamedRerouteDeclaration(None=@n134.RGB) {name=MoonGlowColor, nodeColor=<rgba 0 0 0 1>}
  NamedRerouteDeclaration(None=@n134.A) {name=MoonGlowAlpha, nodeColor=<rgba 0 0 0 1>}
  n135 = MaterialFunctionCall(UVs=NamedRerouteUsage(), Texture Scale=AppendVector(A=Constant() {r=2}, B=Constant() {r=0.2})) {materialFunction=&ScaleUVsByCenter}
  n136 = AppendVector(A=Constant(), B=Constant() {r=-0.3})
  n137 = TextureSample(UVs=Panner(Coordinate=Add(A=@n135.UVs, B=@n136), Time=NamedRerouteUsage()) {bFractionalPart=true, speedX=-0.01}) {samplerType=SAMPLERTYPE_Alpha, texture=/OfferCatalog/Art/Greenhouse/Textures/T_UI_GreenHouse_Caustics}
  NamedRerouteDeclaration(None=Multiply(A=NamedRerouteUsage(), B=Multiply(A=@n137.R))) {name=OceanCaustics, nodeColor=<rgba 0 0 0 1>}
  NamedRerouteDeclaration(None=VectorParameter() {group=Color, parameterName=OceanCausticsColor, defaultValue=<rgba 0.0784 0.2 0.0157 1>}.RGB) {name=OceanCausticsColor, nodeColor=<rgba 0 0 0 1>}
 ]}
