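# UI material function for the IWD takeover art in the offer catalog, switching series artwork on through bIsIWDSeries, otherwise outputting a function input.
# Pixel-size ratios scale and center two UV sets, blended by a size comparison, though each ratio and comparison uses one component against itself.
# These UVs sample the Texture parameter, where IsBFS selects the HeartSky default or the Frame default tinted slightly toward RGB 2.
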MaterialFunction /OfferCatalog/Art/IWDTakeover/MF_UI_IWDSeries
graph {nodes [
  FunctionInput(Preview=VectorParameter() {parameterName=RGB1, defaultValue=<rgba 0.5711 0.3663 0.1248 1>}.RGB) {inputName=RGB 1, sortPriority=1}
  n1 = MaterialFunctionCall() {materialFunction=&GetUserInterfaceUV}
  n2 = MaterialFunctionCall(Float2=@n1.Pixel Size) {materialFunction=&BreakOutFloat2Components}
  n3 = Divide(A=@n2.G, B=@n2.G)
  n4 = AppendVector(A=Constant() {r=1}, B=@n3)
  n5 = Constant()
  n6 = Multiply(A=OneMinus(None=@n3)) {constB=0.5}
  n7 = MaterialFunctionCall(Float2=MaterialFunctionCall() {materialFunction=&GetUserInterfaceUV}.Pixel Size) {materialFunction=&BreakOutFloat2Components}
  n8 = Constant()
  n9 = MaterialFunctionCall() {materialFunction=&GetUserInterfaceUV}
  n10 = MaterialFunctionCall(Float2=@n9.Pixel Size) {materialFunction=&BreakOutFloat2Components}
  n11 = Divide(A=@n10.R, B=@n10.R)
  n12 = AppendVector(A=@n11, B=Constant() {r=1})
  n13 = Multiply(A=OneMinus(None=@n11)) {constB=0.5}
  n14 = Constant()
  n15 = LinearInterpolate(
   A=Add(A=Multiply(A=@n1.Normalized UV, B=@n4), B=AppendVector(A=@n5, B=@n6))
   B=Add(A=Multiply(A=@n9.Normalized UV, B=@n12), B=AppendVector(A=@n13, B=@n14))
   Alpha=If(A=@n7.R, B=@n7.R, "A > B"=@n8, "A == B"=@n8, "A < B"=Constant() {r=1})
  )
  n16 = TextureSampleParameter2D(UVs=@n15) {parameterName=Texture, texture=/OfferCatalog/Art/IWDTakeover/Texture/T_UI_IWD_HeartSky}
  n17 = LinearInterpolate(
   A=TextureSampleParameter2D(UVs=@n15) {parameterName=Texture, texture=/OfferCatalog/Art/IWDTakeover/Texture/T_UI_IWD_Frame}.RGB
   B=FunctionInput(Preview=VectorParameter() {parameterName=RGB2, defaultValue=<rgba 0 0 0.5469 1>}.RGB) {inputName=RGB 2, sortPriority=1}
  ) {constAlpha=0.04}
  FunctionOutput(None=StaticSwitchParameter(True=StaticSwitchParameter(True=@n16.RGB, False=@n17) {parameterName=IsBFS}, False=FunctionInput()) {parameterName=bIsIWDSeries})
 ]}
