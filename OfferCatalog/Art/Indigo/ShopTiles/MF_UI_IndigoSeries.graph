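# UI material function for Indigo series shop tiles that optionally swaps its incoming value for an artwork texture.
# The blsIndigoSeries static switch outputs either the function input or the RGB of IndigoTexture, defaulting to T_UI_Sunset.
# Texture UVs come from the ScaleY output of MF_UI_UVScaling, then pass through ScaleUVsByCenter.
# A pixel-size-based UV blend and a CorruptSeries two-color lerp are also present but do not feed the output.
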MaterialFunction /OfferCatalog/Art/Indigo/ShopTiles/MF_UI_IndigoSeries
graph {nodes [
  n1 = MaterialFunctionCall(
   UVs=MaterialFunctionCall!(UV, TextureSize) {materialFunction=&MF_UI_UVScaling}.ScaleY
   Texture Scale=Constant() {r=1.1}
  ) {materialFunction=&ScaleUVsByCenter}
  FunctionOutput(None=StaticSwitchParameter(
    True=TextureSampleParameter2D(UVs=@n1.UVs) {parameterName=IndigoTexture, texture=/OfferCatalog/Art/Indigo/ShopTiles/Textures/T_UI_Sunset}.RGB
    False=FunctionInput()
   ) {parameterName=blsIndigoSeries})
  n2 = MaterialFunctionCall() {materialFunction=&GetUserInterfaceUV}
  n3 = MaterialFunctionCall(Float2=@n2.Pixel Size) {materialFunction=&BreakOutFloat2Components}
  n4 = Divide(A=@n3.G, B=@n3.G)
  n5 = AppendVector(A=Constant() {r=1}, B=@n4)
  n6 = Constant()
  n7 = Multiply(A=OneMinus(None=@n4)) {constB=0.5}
  n8 = MaterialFunctionCall(Float2=MaterialFunctionCall() {materialFunction=&GetUserInterfaceUV}.Pixel Size) {materialFunction=&BreakOutFloat2Components}
  n9 = Constant()
  n10 = MaterialFunctionCall() {materialFunction=&GetUserInterfaceUV}
  n11 = MaterialFunctionCall(Float2=@n10.Pixel Size) {materialFunction=&BreakOutFloat2Components}
  n12 = Divide(A=@n11.R, B=@n11.R)
  n13 = AppendVector(A=@n12, B=Constant() {r=1})
  n14 = Multiply(A=OneMinus(None=@n12)) {constB=0.5}
  n15 = Constant()
  LinearInterpolate(
   A=Add(A=Multiply(A=@n2.Normalized UV, B=@n5), B=AppendVector(A=@n6, B=@n7))
   B=Add(A=Multiply(A=@n10.Normalized UV, B=@n13), B=AppendVector(A=@n14, B=@n15))
   Alpha=If(A=@n8.R, B=@n8.R, "A > B"=@n9, "A == B"=@n9, "A < B"=Constant() {r=1})
  )
  LinearInterpolate(
   A=FunctionInput(Preview=VectorParameter() {group=CorruptSeries, parameterName=BG_Color, defaultValue=<rgba 1 1 1 1>}.RGB) {bUsePreviewValueAsDefault=true, inputName=RGB 1, sortPriority=1}
   B=FunctionInput(Preview=VectorParameter() {group=CorruptSeries, parameterName=CellColor, defaultValue=<rgba 0 0 0 1>}.RGB) {bUsePreviewValueAsDefault=true, inputName=RGB 2, sortPriority=1}
  )
 ]}
