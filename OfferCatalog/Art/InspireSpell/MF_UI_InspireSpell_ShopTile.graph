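# Material function for the InspireSpell item shop tile background, used when the blsInspireSpell_Series static switch is on, otherwise passing its input through.
# Two crossed linear gradients blend a dark tone into a grunge-textured warm base.
# Dark SDF stars tile over it in two horizontally panning rows, offset vertically per half by a rounded-UV mask.
# A random value from the column index varies each star's scale and rotation.
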MaterialFunction /OfferCatalog/Art/InspireSpell/MF_UI_InspireSpell_ShopTile
graph {nodes [
  Constant3Vector() {constant=<rgba 0.026 0.0173 0.0066 0>}
  n1 = MaterialFunctionCall!(UV, TextureSize) {materialFunction=&MF_UI_UVScaling}
  n2 = MaterialFunctionCall!(
   Texture
   GradientStart=Constant() {r=-0.15}
   GradientEnd=Constant() {r=0.5}
   RotationAngle=Constant() {r=0.65}
   RotationCenter
   GradientOffset
  ) {materialFunction=&MF_UI_LinearGradient}
  n3 = MaterialFunctionCall(GradientStart=Constant() {r=-0.05}, GradientEnd=Constant() {r=0.5}, RotationAngle=Constant() {r=0.15}) {materialFunction=&MF_UI_LinearGradient}
  n4 = LinearInterpolate(
   A=Constant3Vector() {constant=<rgba 0.0573 0.0503 0.0373 0>}
   B=LinearInterpolate(
    A=Constant3Vector() {constant=<rgba 0.3073 0.2721 0.2065 0>}
    B=Constant3Vector() {constant=<rgba 0.5573 0.5017 0.4412 1>}
    Alpha=TextureSample(UVs=@n1.Scaled To Fill) {texture=&/T_UI_LightGrunge}.R
   )
   Alpha=Multiply(A=Multiply(A=@n2.Result, B=@n3.Result)) {constB=2}
  )
  n5 = Constant3Vector() {constant=<rgba 0.0087 0.0087 0.0087 0>}
  n6 = MaterialFunctionCall(UVs=TextureCoordinate(), Texture Scale=Constant() {r=2.05}) {materialFunction=&ScaleUVsByCenter}
  n7 = Constant() {desc=pan speed, r=0.02}
  n8 = Round(None=ComponentMask(None=TextureCoordinate()) {g=true})
  n9 = AppendVector(
   A=LinearInterpolate(A=@n7, B=Multiply(A=@n7), Alpha=NamedRerouteDeclaration(None=@n8) {name=up and down mask, nodeColor=<rgba 1 0.1641 0 1>}) {constA=-0.1, constB=0.1}
   B=Constant()
  )
  n10 = Constant()
  n11 = Constant() {r=-0.0198}
  n12 = LinearInterpolate(A=@n11, B=Multiply(A=@n11) {constB=-4.5}, Alpha=NamedRerouteUsage()) {constA=0.2, constB=-0.2}
  n13 = Frac(None=Add(A=Panner(Coordinate=@n6.UVs, Speed=@n9), B=AppendVector(A=@n10, B=@n12)))
  n14 = Constant() {desc=num stars in a row, r=11}
  n15 = MaterialFunctionCall!(UV=@n13, PatternX=@n14, PatternY=Constant() {r=2}) {materialFunction=/FortUILibrary/MaterialFunctions/Basics/MF_UI_Pattern}
  n16 = Multiply(A=ComponentMask(None=ComponentMask(None=@n13) {r=true}) {r=true}, B=@n14)
  n17 = MaterialFunctionCall!(UV=Divide(A=Truncate!(None=@n16) {is=class MaterialExpressionTruncate}, B=Constant() {r=4}), Seed=Constant() {r=0.1}) {materialFunction=&MF_UI_Random}
  n18 = LinearInterpolate(Alpha=@n17.Result) {constA=0.3, constB=1.1}
  n19 = Frac(None=Add(A=@n17.Result) {constB=0.123})
  n20 = TextureSample(UVs=MaterialFunctionCall(
    UVs=MaterialFunctionCall(UVs=@n15.ScaledPatternUVs, Texture Scale=@n18) {materialFunction=&ScaleUVsByCenter}.UVs
    "Rotation Angle (0-1)"=LinearInterpolate(Alpha=@n19) {constB=0.03}
   ) {materialFunction=&CustomRotator}.Rotated Values) {samplerSource=SSM_Clamp_WorldGroupSettings, samplerType=SAMPLERTYPE_Alpha, texture=/Game/UI/Frontend/ItemShop/Art/InspireSpell/T_UI_SDFstar}
  n21 = SmoothStep(Value=@n20.R) {constMax=0.5, constMin=0.25}
  FunctionOutput(None=StaticSwitchParameter(True=LinearInterpolate(A=@n4, B=@n5, Alpha=@n21), False=FunctionInput()) {parameterName=blsInspireSpell_Series})
  LinearInterpolate(
   A=FunctionInput(Preview=VectorParameter() {group=CorruptSeries, parameterName=BG_Color, defaultValue=<rgba 1 1 1 1>}.RGB) {bUsePreviewValueAsDefault=true, inputName=RGB 1, sortPriority=1}
   B=FunctionInput(Preview=VectorParameter() {group=CorruptSeries, parameterName=CellColor, defaultValue=<rgba 0 0 0 1>}.RGB) {bUsePreviewValueAsDefault=true, inputName=RGB 2, sortPriority=1}
  )
 ]}
