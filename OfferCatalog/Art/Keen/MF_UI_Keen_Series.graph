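# UI material function that, when static switch bIsKeenSeries is on, outputs the Keen_BG texture darkened to 70% brightness, otherwise passing its input through.
# Texture UVs blend two aspect-correction-style branches built from UI pixel size, though each divide and compare uses one component against itself, leaving UVs unscaled as wired.
# A CorruptSeries color lerp and a radial gradient are present but unconnected to the output.
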MaterialFunction /OfferCatalog/Art/Keen/MF_UI_Keen_Series
graph {nodes [
  n1 = MaterialFunctionCall() {materialFunction=&GetUserInterfaceUV}
  n2 = MaterialFunctionCall(Float2=@n1.Pixel Size) {materialFunction=&BreakOutFloat2Components}
  n3 = Divide(A=@n2.G, B=@n2.G)
  n4 = Multiply(A=@n1.Normalized UV, B=AppendVector(A=Constant() {r=1}, B=@n3))
  n5 = AppendVector(A=Constant(), B=Multiply(A=OneMinus(None=@n3)) {constB=0.5})
  n6 = MaterialFunctionCall(Float2=MaterialFunctionCall() {materialFunction=&GetUserInterfaceUV}.Pixel Size) {materialFunction=&BreakOutFloat2Components}
  n7 = Constant()
  n8 = Constant() {r=1}
  n9 = MaterialFunctionCall() {materialFunction=&GetUserInterfaceUV}
  n10 = MaterialFunctionCall(Float2=@n9.Pixel Size) {materialFunction=&BreakOutFloat2Components}
  n11 = Divide(A=@n10.R, B=@n10.R)
  n12 = Multiply(A=@n9.Normalized UV, B=AppendVector(A=@n11, B=Constant() {r=1}))
  n13 = AppendVector(A=Multiply(A=OneMinus(None=@n11)) {constB=0.5}, B=Constant())
  n14 = TextureSampleParameter2D(UVs=LinearInterpolate(A=Add(A=@n4, B=@n5), B=Add(A=@n12, B=@n13), Alpha=If(A=@n6.R, B=@n6.R, "A > B"=@n7, "A == B"=@n7, "A < B"=@n8))) {parameterName=Keen_BG, texture=&/T_UI_Keen}
  FunctionOutput(None=StaticSwitchParameter(True=LinearInterpolate(B=@n14.RGB) {constAlpha=0.7}, False=FunctionInput()) {parameterName=bIsKeenSeries})
  LinearInterpolate(
   A=FunctionInput(Preview=VectorParameter() {group=CorruptSeries, parameterName=BG_Color, defaultValue=<rgba 1 1 1 1>}.RGB) {bUsePreviewValueAsDefault=true, inputName=RGB 1, sortPriority=1}
   B=FunctionInput(Preview=VectorParameter() {group=CorruptSeries, parameterName=CellColor, defaultValue=<rgba 0 0 0 1>}.RGB) {bUsePreviewValueAsDefault=true, inputName=RGB 2, sortPriority=1}
  )
  MaterialFunctionCall(
   UVs=MaterialFunctionCall() {materialFunction=&ScreenAlignedUVs}."X 100%, Y 100%"
   CenterPosition=Constant2Vector() {g=0.5, r=0.5}
   Radius=ScalarParameter() {defaultValue=0.5, parameterName=RadialRadius_02}
   Density=ScalarParameter() {defaultValue=1.0452, parameterName=RadialDensity_02}
  ) {materialFunction=&RadialGradientExponential}
 ]}
