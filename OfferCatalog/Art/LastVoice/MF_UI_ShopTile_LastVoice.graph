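# Material function for LastVoice series item shop tile backgrounds, enabled by the bls_LastVoice_Series static switch, which otherwise passes its input through.
# It builds two animated cloudscape variants from flat sky colors, gradient-masked background and foreground cloud layers, panning noise and flow-map-distorted cloud textures, connected through named reroutes.
# The DN scalar blends between two tinted background results.
# A feathered frame tint and several cloud and gradient calls remain unconnected to the output.
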MaterialFunction /OfferCatalog/Art/LastVoice/MF_UI_ShopTile_LastVoice
graph {nodes [
  Constant3Vector() {constant=<rgba 0.5457 0.4406 0.5349 1>}
  Constant3Vector() {constant=<rgba 0.5457 0.4406 0.5349 1>}
  Constant3Vector() {constant=<rgba 0.5397 0.2349 0.9047 1>}
  Constant3Vector() {constant=<rgba 0.5397 0.2349 0.9047 1>}
  n1 = Add(A=NamedRerouteUsage(), B=AppendVector(A=Constant(), B=Constant() {r=0.5}))
  n2 = Constant()
  n3 = Constant() {r=0.5}
  n4 = Constant() {r=0.75}
  n5 = Multiply(A=SmoothStep(Value=MaterialFunctionCall(UVs=@n1, GradientStart=@n2, GradientEnd=@n3, RotationAngle=@n4) {materialFunction=&MF_UI_LinearGradient}.Result) {constMax=1.1}) {constB=0.9}
  n6 = LinearInterpolate(
   A=LinearInterpolate(A=NamedRerouteUsage()) {constAlpha=0.25}
   B=LinearInterpolate(A=NamedRerouteUsage(), B=Constant3Vector() {constant=<rgba 0.0212 0.008 0.1384 1>}, Alpha=Saturate(None=@n5)) {constAlpha=0.9}
   Alpha=ScalarParameter() {desc=1.0 for DN Activation, group=Activation, parameterName=DN}
  ) {constAlpha=1}
  FunctionOutput(None=StaticSwitchParameter(True=@n6, False=FunctionInput()) {parameterName=bls_LastVoice_Series})
  n7 = ScalarParameter() {defaultValue=0.75, parameterName=FrameFeather1}
  n8 = ScalarParameter() {defaultValue=1.5, parameterName=FrameFeather2}
  n9 = Abs(None=ConstantBiasScale(None=MaterialFunctionCall() {materialFunction=&GetUserInterfaceUV}.Normalized UV) {bias=-0.5, scale=2})
  n10 = Add(
   A=SmoothStep(Min=@n7, Max=@n8, Value=ComponentMask(None=@n9) {r=true})
   B=SmoothStep(Min=@n7, Max=@n8, Value=ComponentMask(None=@n9) {g=true})
  )
  LinearInterpolate(
   A=@n6
   B=Constant3Vector() {constant=<rgba 0.4844 0.7603 1 1>}
   Alpha=Multiply(A=Saturate(None=@n10), B=ScalarParameter() {defaultValue=0.3, parameterName=FrameOpacity})
  )
  LinearInterpolate(
   A=FunctionInput(Preview=VectorParameter() {group=CorruptSeries, parameterName=BG_Color, defaultValue=<rgba 1 1 1 1>}.RGB) {bUsePreviewValueAsDefault=true, inputName=RGB 1, sortPriority=1}
   B=FunctionInput(Preview=VectorParameter() {group=CorruptSeries, parameterName=CellColor, defaultValue=<rgba 0 0 0 1>}.RGB) {bUsePreviewValueAsDefault=true, inputName=RGB 2, sortPriority=1}
  )
  MaterialFunctionCall!(
   Mask
   CloudIntensity=Constant() {r=1}
   Density=Constant() {r=1}
   Scale=Constant() {r=1}
   Speed
   CloudColorLow
   CloudColorHigh=Constant3Vector() {constant=<rgba 0.8963 0.8796 0.9047 1>}
   CloudColorDark=Constant3Vector() {constant=<rgba 0.5457 0.1594 0.3718 1>}
   CloudDarkIntensity=Constant() {r=2}
   FoggyAlphaMin=Constant() {r=5}
   NoiseTexture=TextureObject() {samplerType=SAMPLERTYPE_Alpha, texture=/Game/UI/Frontend/ItemShop/Art/LastVoice/T_UI_CloudLV}
  ) {materialFunction=/FortUILibrary/MaterialFunctions/Effects/MF_UI_Clouds}
  n11 = NamedRerouteUsage()
  MaterialFunctionCall(UVs=@n11, Texture Scale=Constant() {r=0.8}) {materialFunction=&ScaleUVsByCenter}
  MaterialFunctionCall() {materialFunction=&MF_UI_UVScaling}
  MaterialFunctionCall(
   CloudIntensity=Constant() {r=1}
   Density=Constant() {r=1}
   Scale=Constant() {r=1}
   CloudColorHigh=Constant3Vector() {constant=<rgba 0.8963 0.8796 0.9047 1>}
   CloudColorDark=Constant3Vector() {constant=<rgba 0.5457 0.1594 0.3718 1>}
   CloudDarkIntensity=Constant() {r=2}
   FoggyAlphaMin=Constant() {r=5}
   NoiseTexture=TextureObject() {samplerType=SAMPLERTYPE_Alpha, texture=&T_UI_CloudLV}
  ) {materialFunction=&MF_UI_Clouds}
  MaterialFunctionCall() {materialFunction=&MF_UI_UVScaling}
  n12 = Constant3Vector() {constant=<rgba 0.0203 0.0103 0.147 1>}
  MaterialFunctionCall(
   A=@n12
   B=Constant3Vector() {constant=<rgba 0.0382 0.2961 0.6939 1>}
   C=Constant3Vector() {constant=<rgba 0.8178 0.8459 0.9047 1>}
   Alpha=MaterialFunctionCall(GradientStart=Constant() {r=-0.1}, GradientEnd=Constant() {r=0.45}, RotationAngle=Constant() {r=0.72}, GradientOffset=Constant() {r=0.1}) {materialFunction=&MF_UI_LinearGradient}.Result
  ) {materialFunction=&3ColorBlend}
  MaterialFunctionCall() {materialFunction=&MF_UI_UVScaling}
  n13 = NamedRerouteUsage()
  MaterialFunctionCall(UVs=@n13, Texture Scale=Constant() {r=0.8}) {materialFunction=&ScaleUVsByCenter}
  n14 = Constant3Vector() {constant=<rgba 0.0482 0.3419 0.7991 1>}
  MaterialFunctionCall(
   A=@n14
   B=Constant3Vector() {constant=<rgba 0.0382 0.2961 0.6939 1>}
   C=Constant3Vector() {constant=<rgba 0.8178 0.8459 0.9047 1>}
   Alpha=MaterialFunctionCall(GradientStart=Constant() {r=-0.1}, GradientEnd=Constant() {r=0.45}, RotationAngle=Constant() {r=0.72}, GradientOffset=Constant() {r=0.1}) {materialFunction=&MF_UI_LinearGradient}.Result
  ) {materialFunction=&3ColorBlend}
  n15 = NamedRerouteUsage()
  n16 = Constant() {r=0.75}
  n17 = TextureSample(UVs=Panner(Coordinate=Multiply(A=@n15, B=@n16)) {speedX=0.01, speedY=0.01}) {samplerType=SAMPLERTYPE_Grayscale, texture=&LowResBlurredNoise}
  n18 = Multiply(A=@n17.R) {constB=1.5}
  n19 = Constant()
  n20 = Constant() {r=4}
  n21 = MaterialFunctionCall!(
   UVs=Multiply(A=AppendVector(A=@n19, B=@n20), B=NamedRerouteUsage())
   Subtract=Constant()
   Falloff=Constant() {r=0.1}
   Intensity=Constant() {r=2}
  ) {materialFunction=/Game/UI/Frontend/ItemShop/Art/_MaterialFunctions_ShopTheming/MF_UI_TopBottomGradient}
  n22 = NamedRerouteUsage()
  n23 = Multiply(A=OneMinus(None=SmoothStep(Value=@n21.Result)), B=Max(A=Multiply(A=@n22) {constB=1.25}, B=@n17.R))
  n24 = SmoothStep(Value=OneMinus(None=Add(A=@n18, B=@n23))) {constMax=0.35}
  n25 = Constant()
  n26 = Constant() {r=0.1}
  n27 = MaterialFunctionCall(
   UVs=Add(A=NamedRerouteUsage(), B=AppendVector(A=@n25, B=@n26))
   GradientStart=Constant() {r=-0.01}
   GradientEnd=Constant() {r=0.96}
   RotationAngle=Constant() {r=0.75}
   GradientOffset=Constant() {r=0.1}
  ) {materialFunction=&MF_UI_LinearGradient}
  n28 = Saturate(None=Multiply(A=@n24, B=Min(A=@n27.Result, B=@n24)))
  n29 = Constant3Vector() {constant=<rgba 0.107 0.0319 0.4125 1>}
  n30 = NamedRerouteUsage()
  n31 = Constant() {r=0.15}
  n32 = Constant() {r=1}
  n33 = Constant() {r=0.75}
  n34 = Constant() {r=0.25}
  n35 = LinearInterpolate(
   A=MaterialFunctionCall(
    A=Constant3Vector() {constant=<rgba 0.8423 0.3128 0.9688 1>}
    B=Constant3Vector() {constant=<rgba 0.3046 0.0744 0.7031 1>}
    C=@n29
    Alpha=NamedRerouteUsage()
   ) {materialFunction=&3ColorBlend}.Result
   B=@n29
   Alpha=SmoothStep(Value=MaterialFunctionCall(UVs=@n30, GradientStart=@n31, GradientEnd=@n32, RotationAngle=@n33, GradientOffset=@n34) {materialFunction=&MF_UI_LinearGradient}.Result)
  )
  Multiply(A=@n28, B=@n35)
  n36 = NamedRerouteUsage()
  n37 = Constant() {r=0.75}
  n38 = TextureSample(UVs=Panner(Coordinate=Multiply(A=@n36, B=@n37)) {speedX=0.01, speedY=0.01}) {samplerType=SAMPLERTYPE_Grayscale, texture=&LowResBlurredNoise}
  n39 = Multiply(A=@n38.R) {constB=1.5}
  n40 = Multiply(A=AppendVector(A=Constant(), B=Constant() {r=4.5}), B=NamedRerouteUsage())
  n41 = Constant()
  n42 = Constant() {r=0.2}
  n43 = Constant() {r=2}
  n44 = NamedRerouteUsage()
  n45 = Multiply(
   A=OneMinus(None=MaterialFunctionCall(UVs=@n40, Subtract=@n41, Falloff=@n42, Intensity=@n43) {materialFunction=&MF_UI_TopBottomGradient}.Result)
   B=Max(A=Multiply(A=@n44) {constB=1.25}, B=@n38.R)
  )
  n46 = SmoothStep(Value=OneMinus(None=Add(A=@n39, B=@n45))) {constMax=0.35}
  n47 = MaterialFunctionCall(GradientStart=Constant() {r=-0.01}, GradientEnd=Constant() {r=0.6}, RotationAngle=Constant() {r=0.75}, GradientOffset=Constant() {r=0.1}) {materialFunction=&MF_UI_LinearGradient}
  n48 = Saturate(None=Multiply(A=@n46, B=Min(A=@n47.Result, B=@n46)))
  n49 = Constant3Vector() {constant=<rgba 0.1279 0.268 0.5833 1>}
  n50 = LinearInterpolate(
   A=MaterialFunctionCall(
    A=Constant3Vector() {constant=<rgba 0.9911 0.893 0.9558 1>}
    B=Constant3Vector() {constant=<rgba 0.4007 0.375 0.9115 1>}
    C=@n49
    Alpha=NamedRerouteUsage()
   ) {materialFunction=&3ColorBlend}.Result
   B=@n49
   Alpha=MaterialFunctionCall(GradientStart=Constant() {r=0.15}, GradientEnd=Constant() {r=1}, RotationAngle=Constant() {r=0.75}, GradientOffset=Constant() {r=0.25}) {materialFunction=&MF_UI_LinearGradient}.Result
  )
  Multiply(A=@n48, B=@n50)
  NamedRerouteDeclaration(None=MaterialFunctionCall(
    UVs=MaterialFunctionCall!(ScreenUV, TextureSize, ScalePivotPoint) {materialFunction=/FortUILibrary/MaterialFunctions/Basics/MF_UI_ScreenSpaceUVScaling}.Scaled to Fill
    Texture Scale=Constant() {r=0.27}
   ) {materialFunction=&ScaleUVsByCenter}.UVs) {name=UVScaling_1, nodeColor=<rgba 0 0.9297 1 1>}
  n51 = Multiply(A=NamedRerouteUsage()) {constB=1.05}
  n52 = Panner(Coordinate=@n51) {speedX=0.01, speedY=0.001}
  n53 = TextureObject() {samplerType=SAMPLERTYPE_Alpha, texture=&T_UI_CloudLV}
  n54 = TextureSample(
   UVs=Panner(Coordinate=@n51) {speedX=0.001, speedY=0.001}
   Tex=TextureObject() {samplerType=SAMPLERTYPE_Grayscale, texture=/FortUILibrary/MaterialFunctions/ExampleContent/T_Cloud_R}
  )
  n55 = Constant() {r=0.5}
  n56 = Constant() {r=0.07}
  n57 = Constant() {r=0.25}
  NamedRerouteDeclaration(None=ComponentMask(None=MaterialFunctionCall!(UVs=@n52, Texture Object=@n53, Grayscale Flow Map=@n54.RGB, Flow Direction=@n55, Flow Intensity=@n56, Flow Speed=@n57) {materialFunction=/FortUILibrary/MaterialFunctions/Effects/MF_UI_GrayscaleFlowMap_New}.Distorted Texture) {r=true}) {name=FlowMap_1, nodeColor=<rgba 1 0.4922 0 1>}
  n58 = Multiply(A=NamedRerouteUsage()) {constB=1.05}
  n59 = Panner(Coordinate=@n58) {speedX=0.01, speedY=0.001}
  n60 = TextureObject() {samplerType=SAMPLERTYPE_Alpha, texture=&T_UI_CloudLV}
  n61 = TextureSample(
   UVs=Panner(Coordinate=@n58) {speedX=0.001, speedY=0.001}
   Tex=TextureObject() {samplerType=SAMPLERTYPE_Grayscale, texture=&T_Cloud_R}
  )
  n62 = Constant() {r=0.5}
  n63 = Constant() {r=0.07}
  n64 = Constant() {r=0.25}
  NamedRerouteDeclaration(None=ComponentMask(None=MaterialFunctionCall(UVs=@n59, Texture Object=@n60, Grayscale Flow Map=@n61.RGB, Flow Direction=@n62, Flow Intensity=@n63, Flow Speed=@n64) {materialFunction=&MF_UI_GrayscaleFlowMap_New}.Distorted Texture) {r=true}) {name=FlowMap_2, nodeColor=<rgba 1 0.4922 0 1>}
  n65 = Multiply(A=@n13) {constB=0.7}
  n66 = Panner(Coordinate=@n65) {speedX=0.001, speedY=0.01}
  n67 = TextureObject() {samplerType=SAMPLERTYPE_Alpha, texture=&T_UI_CloudLV}
  n68 = TextureSample(
   UVs=Panner(Coordinate=@n65) {speedX=0.01, speedY=0.01}
   Tex=TextureObject() {samplerType=SAMPLERTYPE_Grayscale, texture=&T_Cloud_R}
  )
  n69 = Constant() {r=0.5}
  n70 = Constant() {r=0.01}
  n71 = Constant() {r=0.25}
  NamedRerouteDeclaration(None=ComponentMask(None=MaterialFunctionCall(UVs=@n66, Texture Object=@n67, Grayscale Flow Map=@n68.RGB, Flow Direction=@n69, Flow Intensity=@n70, Flow Speed=@n71) {materialFunction=&MF_UI_GrayscaleFlowMap_New}.Distorted Texture) {r=true}) {name=FlowMap_FG_2, nodeColor=<rgba 0 0 0 1>}
  NamedRerouteDeclaration(None=LinearInterpolate(
    A=Constant3Vector() {constant=<rgba 0.2712 0.0735 0.651 1>}
    B=Constant3Vector() {constant=<rgba 0.6194 0.1823 1 1>}
    Alpha=NamedRerouteUsage()
   )) {name=foreground clouds_2, nodeColor=<rgba 0 1 0.7188 1>}
  n72 = Multiply(A=AppendVector(A=Constant(), B=Constant() {r=0.85}), B=NamedRerouteUsage())
  n73 = Constant() {r=0.05}
  n74 = Constant() {r=2}
  n75 = Constant() {r=2}
  n76 = NamedRerouteUsage()
  n77 = Constant() {r=1.25}
  n78 = TextureSample(UVs=Panner(Coordinate=MaterialFunctionCall(UVs=@n76, Texture Scale=@n77) {materialFunction=&ScaleUVsByCenter}.UVs) {speedX=0.03, speedY=0.02}) {samplerType=SAMPLERTYPE_Grayscale, texture=&LowResBlurredNoise}
  n79 = Max(A=NamedRerouteUsage(), B=@n78.R)
  n80 = Multiply(
   A=OneMinus(None=MaterialFunctionCall(UVs=@n72, Subtract=@n73, Falloff=@n74, Intensity=@n75) {materialFunction=&MF_UI_TopBottomGradient}.Result)
   B=@n79
  )
  n81 = Max(A=@n80, B=@n79)
  n82 = Max(
   A=MaterialFunctionCall(
    UVs=NamedRerouteUsage()
    GradientStart=Constant() {r=-0.1}
    GradientEnd=Constant() {r=1.5}
    RotationAngle=Constant() {r=0.8}
    GradientOffset=Constant() {r=0.5}
   ) {materialFunction=&MF_UI_LinearGradient}.Result
   B=MaterialFunctionCall(
    UVs=NamedRerouteUsage()
    GradientStart=Constant() {r=-0.1}
    GradientEnd=Constant() {r=1.25}
    RotationAngle=Constant() {r=0.65}
    GradientOffset=Constant() {r=0.55}
   ) {materialFunction=&MF_UI_LinearGradient}.Result
  )
  n83 = Saturate(None=Multiply(A=Max(A=@n81, B=@n82), B=@n82))
  NamedRerouteDeclaration(None=SmoothStep(Value=Multiply(A=@n83) {constB=4}) {constMax=0.54}) {name=foreground cloud alpha_2, nodeColor=<rgba 0 1 0.8359 1>}
  n84 = Multiply(A=@n11) {constB=0.65}
  n85 = Panner(Coordinate=@n84) {speedX=0.001, speedY=0.01}
  n86 = TextureObject() {samplerType=SAMPLERTYPE_Alpha, texture=&T_UI_CloudLV}
  n87 = TextureSample(
   UVs=Panner(Coordinate=@n84) {speedX=0.01, speedY=0.01}
   Tex=TextureObject() {samplerType=SAMPLERTYPE_Grayscale, texture=&T_Cloud_R}
  )
  n88 = Constant() {r=0.5}
  n89 = Constant() {r=0.01}
  n90 = Constant() {r=0.25}
  NamedRerouteDeclaration(None=ComponentMask(None=MaterialFunctionCall(UVs=@n85, Texture Object=@n86, Grayscale Flow Map=@n87.RGB, Flow Direction=@n88, Flow Intensity=@n89, Flow Speed=@n90) {materialFunction=&MF_UI_GrayscaleFlowMap_New}.Distorted Texture) {r=true}) {name=FlowMap_FG_1, nodeColor=<rgba 0 0 0 1>}
  NamedRerouteDeclaration(None=@n12) {name=sky gradient_2, nodeColor=<rgba 0 0.4609 1 1>}
  NamedRerouteDeclaration(None=Multiply(A=@n28) {constB=1.5}) {name=BG Cloud Alpha_2, nodeColor=<rgba 0.6641 1 0 1>}
  n91 = NamedRerouteUsage()
  n92 = Constant3Vector() {constant=<rgba 0.8551 0.183 1 1>}
  n93 = Constant3Vector() {constant=<rgba 0.7969 0.2075 0.792 1>}
  n94 = Constant3Vector() {constant=<rgba 0.2155 0.1629 0.8021 1>}
  n95 = NamedRerouteUsage()
  n96 = NamedRerouteUsage()
  n97 = Constant() {r=0.15}
  n98 = Constant() {r=1}
  n99 = Constant() {r=0.75}
  n100 = Constant() {r=0.25}
  n101 = NamedRerouteUsage()
  n102 = Constant() {r=1.25}
  n103 = TextureSample(UVs=Panner(Coordinate=Multiply(A=@n101, B=@n102)) {speedX=0.01, speedY=0.01}) {samplerType=SAMPLERTYPE_Grayscale, texture=&LowResBlurredNoise}
  n104 = Multiply(A=@n103.R) {constB=1.1}
  n105 = Multiply(A=AppendVector(A=Constant(), B=Constant() {r=2}))
  n106 = Constant() {r=-0.2}
  n107 = Constant() {r=0.9}
  n108 = Constant() {r=2}
  n109 = Multiply(
   A=OneMinus(None=MaterialFunctionCall(UVs=@n105, Subtract=@n106, Falloff=@n107, Intensity=@n108) {materialFunction=&MF_UI_TopBottomGradient}.Result)
   B=Max(A=NamedRerouteUsage(), B=@n103.R)
  )
  n110 = SmoothStep(Value=OneMinus(None=Add(A=@n104, B=@n109))) {constMax=0.5, constMin=0.1}
  n111 = Multiply(A=MaterialFunctionCall(
    UVs=NamedRerouteUsage()
    GradientStart=Constant() {r=0.2}
    GradientEnd=Constant() {r=0.65}
    RotationAngle=Constant() {r=0.75}
    GradientOffset=Constant() {r=0.2}
   ) {materialFunction=&MF_UI_LinearGradient}.Result) {constB=0.2}
  n112 = Saturate(None=Multiply(A=@n110, B=Min(A=@n111, B=@n110)))
  n113 = LinearInterpolate(
   A=LinearInterpolate(A=NamedRerouteUsage(), B=@n35, Alpha=Multiply(A=@n91) {constB=1.2})
   B=LinearInterpolate(
    A=MaterialFunctionCall(A=@n92, B=@n93, C=@n94, Alpha=@n95) {materialFunction=&3ColorBlend}.Result
    B=@n94
    Alpha=MaterialFunctionCall(UVs=@n96, GradientStart=@n97, GradientEnd=@n98, RotationAngle=@n99, GradientOffset=@n100) {materialFunction=&MF_UI_LinearGradient}.Result
   )
   Alpha=Multiply(A=SmoothStep(Value=@n112)) {constB=1.25}
  )
  n114 = NamedRerouteUsage()
  n115 = NamedRerouteUsage()
  n116 = MaterialFunctionCall(GradientStart=Constant() {r=0.05}, GradientEnd=Constant() {r=0.25}, RotationAngle=Constant() {r=0.25}, GradientOffset=Constant() {r=-0.07}) {materialFunction=&MF_UI_LinearGradient}
  NamedRerouteDeclaration(None=LinearInterpolate(
    A=Constant3Vector() {constant=<rgba 0.0203 0.0103 0.147 1>}
    B=LinearInterpolate(A=@n113, B=@n114, Alpha=@n115)
    Alpha=SmoothStep(Value=@n116.Result)
   )) {name=BestBG_DN, nodeColor=<rgba 1 0.7969 0 1>}
  NamedRerouteDeclaration(None=LinearInterpolate(
    A=Constant3Vector() {constant=<rgba 0.8469 0.4543 0.6425 1>}
    B=Constant3Vector() {constant=<rgba 0.9301 0.8632 0.9131 1>}
    Alpha=NamedRerouteUsage()
   )) {name=foreground clouds_1, nodeColor=<rgba 0 1 0.7188 1>}
  n117 = Multiply(A=AppendVector(A=Constant(), B=Constant() {r=0.85}), B=NamedRerouteUsage())
  n118 = Constant() {r=0.05}
  n119 = Constant() {r=1}
  n120 = Constant() {r=2}
  n121 = NamedRerouteUsage()
  n122 = Constant() {r=1.25}
  n123 = TextureSample(UVs=Panner(Coordinate=MaterialFunctionCall(UVs=@n121, Texture Scale=@n122) {materialFunction=&ScaleUVsByCenter}.UVs) {speedX=0.03, speedY=0.02}) {samplerType=SAMPLERTYPE_Grayscale, texture=&LowResBlurredNoise}
  n124 = Max(A=NamedRerouteUsage(), B=@n123.R)
  n125 = Multiply(
   A=OneMinus(None=MaterialFunctionCall(UVs=@n117, Subtract=@n118, Falloff=@n119, Intensity=@n120) {materialFunction=&MF_UI_TopBottomGradient}.Result)
   B=@n124
  )
  n126 = Max(A=@n125, B=@n124)
  n127 = Max(
   A=MaterialFunctionCall(GradientStart=Constant() {r=-0.1}, GradientEnd=Constant() {r=1.25}, RotationAngle=Constant() {r=0.8}, GradientOffset=Constant() {r=0.5}) {materialFunction=&MF_UI_LinearGradient}.Result
   B=MaterialFunctionCall(GradientStart=Constant() {r=-0.1}, GradientEnd=Constant() {r=1.25}, RotationAngle=Constant() {r=0.65}, GradientOffset=Constant() {r=0.55}) {materialFunction=&MF_UI_LinearGradient}.Result
  )
  n128 = Saturate(None=Multiply(A=Max(A=@n126, B=@n127), B=@n127))
  NamedRerouteDeclaration(None=SmoothStep(Value=Multiply(A=@n128) {constB=4}) {constMax=0.54}) {name=foreground cloud alpha_1, nodeColor=<rgba 0 1 0.8359 1>}
  NamedRerouteDeclaration(None=@n14) {name=sky gradient_1, nodeColor=<rgba 0 0.4609 1 1>}
  NamedRerouteDeclaration(None=Multiply(A=@n48) {constB=1.5}) {name=BG Cloud Alpha_1, nodeColor=<rgba 0.6641 1 0 1>}
  n129 = NamedRerouteUsage()
  n130 = Constant3Vector() {constant=<rgba 0.7801 0.5833 1 1>}
  n131 = Constant3Vector() {constant=<rgba 0.7969 0.4441 0.6856 1>}
  n132 = Constant3Vector() {constant=<rgba 0.2772 0.2302 0.8021 1>}
  n133 = NamedRerouteUsage()
  n134 = Constant() {r=0.15}
  n135 = Constant() {r=1}
  n136 = Constant() {r=0.75}
  n137 = Constant() {r=0.25}
  n138 = NamedRerouteUsage()
  n139 = Constant() {r=1.25}
  n140 = TextureSample(UVs=Panner(Coordinate=Multiply(A=@n138, B=@n139)) {speedX=0.01, speedY=0.01}) {samplerType=SAMPLERTYPE_Grayscale, texture=&LowResBlurredNoise}
  n141 = Multiply(A=@n140.R) {constB=1.1}
  n142 = Multiply(A=AppendVector(A=Constant(), B=Constant() {r=5}), B=NamedRerouteUsage())
  n143 = Constant()
  n144 = Constant() {r=0.2}
  n145 = Constant() {r=2}
  n146 = Multiply(
   A=OneMinus(None=MaterialFunctionCall(UVs=@n142, Subtract=@n143, Falloff=@n144, Intensity=@n145) {materialFunction=&MF_UI_TopBottomGradient}.Result)
   B=Max(A=NamedRerouteUsage(), B=@n140.R)
  )
  n147 = SmoothStep(Value=OneMinus(None=Add(A=@n141, B=@n146))) {constMax=0.5, constMin=0.1}
  n148 = Constant() {r=0.2}
  n149 = Constant() {r=0.65}
  n150 = Constant() {r=0.75}
  n151 = Constant() {r=0.2}
  n152 = Multiply(
   A=@n147
   B=Min(
    A=MaterialFunctionCall(GradientStart=@n148, GradientEnd=@n149, RotationAngle=@n150, GradientOffset=@n151) {materialFunction=&MF_UI_LinearGradient}.Result
    B=@n147
   )
  )
  n153 = LinearInterpolate(
   A=LinearInterpolate(A=NamedRerouteUsage(), B=@n50, Alpha=Multiply(A=@n129) {constB=1.2})
   B=LinearInterpolate(
    A=MaterialFunctionCall(A=@n130, B=@n131, C=@n132, Alpha=@n133) {materialFunction=&3ColorBlend}.Result
    B=@n132
    Alpha=MaterialFunctionCall(GradientStart=@n134, GradientEnd=@n135, RotationAngle=@n136, GradientOffset=@n137) {materialFunction=&MF_UI_LinearGradient}.Result
   )
   Alpha=Multiply(A=Saturate(None=@n152)) {constB=1.25}
  )
  n154 = NamedRerouteUsage()
  n155 = NamedRerouteUsage()
  n156 = Constant() {r=0.05}
  n157 = Constant() {r=0.25}
  n158 = Constant() {r=0.25}
  n159 = Constant() {r=-0.07}
  NamedRerouteDeclaration(None=LinearInterpolate(
    A=Constant3Vector() {constant=<rgba 0.1878 0.0802 0.4852 1>}
    B=LinearInterpolate(A=@n153, B=@n154, Alpha=@n155)
    Alpha=MaterialFunctionCall(GradientStart=@n156, GradientEnd=@n157, RotationAngle=@n158, GradientOffset=@n159) {materialFunction=&MF_UI_LinearGradient}.Result
   )) {name=BestBG, nodeColor=<rgba 1 0.7969 0 1>}
  NamedRerouteDeclaration(None=MaterialFunctionCall() {materialFunction=&MF_UI_ScreenSpaceUVScaling}.Scaled to Fit) {name=UVScaling_2, nodeColor=<rgba 0 0.9297 1 1>}
  OneMinus(None=Add(A=@n123.R, B=@n125))
  OneMinus(None=Add(A=@n78.R, B=@n80))
 ]}
