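# Material function that draws the animated Lettuce series background for shop tiles in the offer catalog UI.
# A hexagon SDF on scaled, offset UVs has scrolling time subtracted, then is repeated and quantized into stepped three-color pink-to-dark-red ripples.
# A linear gradient fades the ripples into flat pink, and a static switch chooses between this result and a passthrough input.
# Cosine, background texture and two-color blend nodes are left unconnected.
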MaterialFunction /OfferCatalog/Art/Lettuce/MF_UI_Lettuce_ShopTiles
graph {nodes [
  n1 = MaterialFunctionCall(X=Constant() {r=-1.2}, Y=Constant() {r=-1.2}) {materialFunction=&MakeFloat2}
  n2 = MaterialFunctionCall!(UV, TextureSize) {materialFunction=&MF_UI_UVScaling}
  MaterialExpressionConstantBiasScale_5 = ConstantBiasScale(None=MaterialFunctionCall(UVs=Add(A=@n1.Result, B=@n2.Scaled To Fill), Texture Scale=Constant() {r=0.8}) {materialFunction=&ScaleUVsByCenter}.UVs)
  MaterialExpressionConstant_54 = Constant() {r=1}
  n3 = Custom!(uv=@MaterialExpressionConstantBiasScale_5, radius=@MaterialExpressionConstant_54) {
   code="float3 k = float3(-0.866025404,0.5,0.577350269);\r\nuv = abs(uv);\r\nuv -= 2.0*min(dot(k.xy,uv),0.0)*k.xy;\r\nuv -= float2(clamp(uv.x, -k.z*radius, k.z*radius), radius);\r\nreturn length(uv)*sign(uv.y);"
   description=HexagonSDF
   inputs [
    {inputName=uv, input {expression=&self:MaterialExpressionConstantBiasScale_5}}
    {inputName=radius, input {expression=&self:MaterialExpressionConstant_54}}
   ]
  }
  n4 = Frac(None=Multiply(A=Time(), B=Constant() {r=0.009}))
  n5 = Frac(None=Multiply(A=Time(), B=Constant() {desc=RippleFrequency, r=0.009}))
  n6 = Multiply(A=Subtract(A=@n3, B=Add(A=@n4, B=@n5)), B=Constant() {desc=RippleFrequency, r=3})
  Cosine(None=@n6) {period=3.2}
  Divide() {constB=1}
  n7 = Constant3Vector() {constant=<rgba 1 0.0729 0.2438 1>}
  n8 = MaterialFunctionCall!(
   Texture
   GradientStart=Constant() {r=0.328}
   GradientEnd=Constant() {r=0.9156}
   RotationAngle=Constant() {r=0.75}
   RotationCenter
   GradientOffset
  ) {materialFunction=&MF_UI_LinearGradient}
  n9 = Frac(None=@n6)
  n10 = Constant() {r=3}
  n11 = Round(None=Multiply(A=Abs(None=@n9), B=@n10))
  n12 = Multiply(A=MaterialFunctionCall(
    A=Constant3Vector() {constant=<rgba 1 0.0729 0.2438 1>}
    B=Constant3Vector() {constant=<rgba 0.7913 0.003 0.1221 1>}
    C=Constant3Vector() {constant=<rgba 0.2874 0.0006 0.0467 1>}
    Alpha=Divide(A=@n11, B=@n10)
   ) {materialFunction=&Lerp_3Color}.Result) {constB=1.5}
  FunctionOutput(None=StaticSwitchParameter(True=LinearInterpolate(A=@n12, B=@n7, Alpha=@n8.Result), False=FunctionInput()) {parameterName=blsLettuce_Series})
  LinearInterpolate(
   A=FunctionInput(Preview=VectorParameter() {group=CorruptSeries, parameterName=BG_Color, defaultValue=<rgba 1 1 1 1>}.RGB) {bUsePreviewValueAsDefault=true, inputName=RGB 1, sortPriority=1}
   B=FunctionInput(Preview=VectorParameter() {group=CorruptSeries, parameterName=CellColor, defaultValue=<rgba 0 0 0 1>}.RGB) {bUsePreviewValueAsDefault=true, inputName=RGB 2, sortPriority=1}
  )
  TextureSample(UVs=MaterialFunctionCall(
    UVs=MaterialFunctionCall() {materialFunction=&MF_UI_UVScaling}.Scaled To Fill
    Texture Scale=Constant() {r=1.1}
   ) {materialFunction=&ScaleUVsByCenter}.UVs) {texture=&/T_UI_LettuceBG}
 ]}
