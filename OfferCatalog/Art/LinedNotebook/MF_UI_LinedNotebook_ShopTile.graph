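# Material function that builds the Lined Notebook series background for UI shop tiles, switched by the blsLinedNotebook_Series static parameter, which otherwise returns its input.
# A brick texture, tinted dark green, blends with a yellow-green glow over a dark teal base.
# UV scaling and three offset exponential radial gradient masks are declared as named reroutes for the layering.
# A lerp between BG_Color and CellColor inputs is left unconnected to the output.
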MaterialFunction /OfferCatalog/Art/LinedNotebook/MF_UI_LinedNotebook_ShopTile
graph {nodes [
  n1 = Constant3Vector() {constant=<rgba 0.0063 0.0625 0.0477 1>}
  n2 = NamedRerouteUsage()
  n3 = Divide(A=TextureSample(UVs=Multiply(A=@n2) {constB=0.75}) {texture=/OfferCatalog/Art/A_Shop_Tiles_Textures/T_UI_LNB_Bricks}.R) {constB=0.07}
  n4 = Constant3Vector() {constant=<rgba 0 0.1146 0.0098 1>}
  n5 = Constant3Vector() {constant=<rgba 0.7967 1 0 1>}
  n6 = NamedRerouteUsage()
  n7 = LinearInterpolate(
   A=LinearInterpolate(A=Constant3Vector() {constant=<rgba 0 0 0 1>}, B=Multiply(A=@n3, B=@n4), Alpha=NamedRerouteUsage()) {constB=0}
   B=Divide(A=Multiply(A=@n5, B=@n6)) {constB=0.4}
   Alpha=Divide(A=SmoothStep(Value=@n6) {constMax=2.2}) {constB=0.35}
  )
  n8 = Multiply(A=NamedRerouteUsage())
  FunctionOutput(None=StaticSwitchParameter(True=LinearInterpolate(A=@n1, B=@n7, Alpha=@n8), False=FunctionInput()) {parameterName=blsLinedNotebook_Series})
  LinearInterpolate(
   A=FunctionInput(Preview=VectorParameter() {group=CorruptSeries, parameterName=BG_Color, defaultValue=<rgba 1 1 1 1>}.RGB) {bUsePreviewValueAsDefault=true, inputName=RGB 1, sortPriority=1}
   B=FunctionInput(Preview=VectorParameter() {group=CorruptSeries, parameterName=CellColor, defaultValue=<rgba 0 0 0 1>}.RGB) {bUsePreviewValueAsDefault=true, inputName=RGB 2, sortPriority=1}
  )
  NamedRerouteDeclaration(None=MaterialFunctionCall!(UV, TextureSize, ScalePivotPoint) {materialFunction=&MF_UI_UVScaling}.ScaleY) {name=UVScaling_Y, nodeColor=<rgba 0 0.3203 1 1>}
  n9 = NamedRerouteUsage()
  n10 = AppendVector(A=Constant(), B=Constant() {r=0.5})
  NamedRerouteDeclaration(None=MaterialFunctionCall(UVs=Add(A=@n9, B=@n10), Radius=Constant() {r=1.25}, Density=Constant() {r=0.75}) {materialFunction=&RadialGradientExponential}.RadialGradientExponential) {name="Green Gradient  Mask", nodeColor=<rgba 0.4726 1 0 1>}
  n11 = NamedRerouteUsage()
  n12 = AppendVector(A=Constant(), B=Constant() {r=0.5})
  NamedRerouteDeclaration(None=MaterialFunctionCall(UVs=Add(A=@n11, B=@n12), Radius=Constant() {r=1.25}, Density=Constant() {r=2}) {materialFunction=&RadialGradientExponential}.RadialGradientExponential) {name=Purple Gradient MAsk, nodeColor=<rgba 0.3392 0 1 1>}
  n13 = Add(A=NamedRerouteUsage(), B=AppendVector(A=Constant(), B=Constant() {r=0.2}))
  n14 = Constant() {r=1}
  n15 = Constant() {r=0.7}
  NamedRerouteDeclaration(None=SmoothStep(Value=MaterialFunctionCall(UVs=@n13, Radius=@n14, Density=@n15) {materialFunction=&RadialGradientExponential}.RadialGradientExponential) {constMax=0.6}) {name="Shading Gradient  Mask", nodeColor=<rgba 0.0131 0.0278 0 1>}
 ]}
