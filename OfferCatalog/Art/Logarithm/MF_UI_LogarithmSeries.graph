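# UI material function that, through the blsLogarithmSeries static switch, outputs either the Logarithm_BG texture or its input unchanged.
# Texture UVs come from GetUserInterfaceUV, blending two centered per-axis rescales chosen by a pixel-size comparison.
# As wired, each ratio and comparison pits a component against itself, so the rescale is effectively neutral.
# A BG_Color/CellColor lerp feeds no output.
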MaterialFunction /OfferCatalog/Art/Logarithm/MF_UI_LogarithmSeries
graph {nodes [
  n1 = MaterialFunctionCall() {materialFunction=&GetUserInterfaceUV}
  n2 = MaterialFunctionCall(Float2=@n1.Pixel Size) {materialFunction=&BreakOutFloat2Components}
  n3 = Divide(A=@n2.G, B=@n2.G)
  n4 = AppendVector(A=Constant() {r=1}, B=@n3)
  n5 = Constant()
  n6 = Multiply(A=OneMinus(None=@n3)) {constB=0.5}
  n7 = MaterialFunctionCall(Float2=MaterialFunctionCall() {materialFunction=&GetUserInterfaceUV}.Pixel Size) {materialFunction=&BreakOutFloat2Components}
  n8 = Constant()
  n9 = MaterialFunctionCall() {materialFunction=&GetUserInterfaceUV}
  n10 = MaterialFunctionCall(Float2=@n9.Pixel Size) {materialFunction=&BreakOutFloat2Components}
  n11 = Divide(A=@n10.R, B=@n10.R)
  n12 = AppendVector(A=@n11, B=Constant() {r=1})
  n13 = Multiply(A=OneMinus(None=@n11)) {constB=0.5}
  n14 = Constant()
  n15 = LinearInterpolate(
   A=Add(A=Multiply(A=@n1.Normalized UV, B=@n4), B=AppendVector(A=@n5, B=@n6))
   B=Add(A=Multiply(A=@n9.Normalized UV, B=@n12), B=AppendVector(A=@n13, B=@n14))
   Alpha=If(A=@n7.R, B=@n7.R, "A > B"=@n8, "A == B"=@n8, "A < B"=Constant() {r=1})
  )
  FunctionOutput(None=StaticSwitchParameter(
    True=TextureSampleParameter2D(UVs=@n15) {parameterName=Logarithm_BG, texture=&/T_UI_LogarithmBG}.RGB
    False=FunctionInput()
   ) {parameterName=blsLogarithmSeries})
  LinearInterpolate(
   A=FunctionInput(Preview=VectorParameter() {group=CorruptSeries, parameterName=BG_Color, defaultValue=<rgba 1 1 1 1>}.RGB) {bUsePreviewValueAsDefault=true, inputName=RGB 1, sortPriority=1}
   B=FunctionInput(Preview=VectorParameter() {group=CorruptSeries, parameterName=CellColor, defaultValue=<rgba 0 0 0 1>}.RGB) {bUsePreviewValueAsDefault=true, inputName=RGB 2, sortPriority=1}
  )
 ]}
