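# Material function that draws the green Lucky series item shop tile background when the blsLuckySeries static switch is on, otherwise passing its input through.
# A drifting, rotated MF_UI_ParticleGrid tiles clover textures, feeding its random outputs back through reroutes to vary cells.
# The clover mask blends gradient-shaded clovers, lit by a periodically pulsed diagonal shine sweep, over a radial gradient.
# A CorruptSeries-grouped color lerp is left unconnected.
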MaterialFunction /OfferCatalog/Art/Lucky/MF_UI_Lucky_ShopTile
graph {nodes [
  n1 = LinearInterpolate(
   A=Constant3Vector() {constant=<rgba 0 0.0573 0.0333 1>}
   B=Constant3Vector() {constant=<rgba 0 0.6372 0.1029 1>}
   Alpha=MaterialFunctionCall(Radius=Constant() {r=0.9}, Density=Constant() {r=1}) {materialFunction=&RadialGradientExponential}.RadialGradientExponential
  )
  n2 = Constant3Vector() {constant=<rgba 0 0.0503 0.0189 1>}
  n3 = Constant3Vector() {constant=<rgba 0 0.8698 0.1788 1>}
  n4 = MaterialFunctionCall(Radius=Constant() {r=0.65}, Density=Constant() {r=1.25}) {materialFunction=&RadialGradientExponential}
  n5 = Time()
  n6 = ScalarParameter() {defaultValue=0.5, parameterName=tenminutes}
  n7 = Multiply(A=Sine(None=Multiply(A=@n5, B=@n6)) {period=3}) {constB=3}
  n8 = ConstantBiasScale(None=MaterialFunctionCall!(UV, TextureSize) {materialFunction=&MF_UI_UVScaling}.Scaled To Fill) {scale=0.24}
  n9 = OneMinus(None=Time())
  n10 = AppendVector(A=Multiply(A=Frac(None=@n9)), B=Constant())
  n11 = Subtract(A=MaterialFunctionCall(UVs=Add(A=@n8, B=@n10), Rotation Center=Constant() {r=0.5}, "Rotation Angle (0-1)"=Constant() {r=0.35}) {materialFunction=&CustomRotator}.Rotated Values) {constB=0}
  n12 = Sine(None=Multiply(A=ComponentMask(None=@n11) {r=true}) {constB=0.5})
  n13 = Constant() {r=0.005}
  n14 = OneMinus(None=@n13)
  n15 = Multiply(A=Multiply(A=Subtract(A=@n12, B=@n14) {constB=0.25}, B=Divide(B=@n13) {constB=0.5})) {constB=2}
  n16 = Multiply(A=Saturate(None=SmoothStep(Value=@n7) {constMax=8, constMin=2}), B=Multiply(A=Saturate(None=@n15)) {constB=10})
  n17 = NamedRerouteUsage()
  n18 = Frac(None=Multiply(A=Time()) {constB=0.015})
  n19 = MaterialFunctionCall!(
   UV=MaterialFunctionCall() {materialFunction=&MF_UI_UVScaling}.Scaled To Fill
   Global Offset=AppendVector(A=Add(A=@n17, B=@n18), B=Constant())
   Global Rotation=Constant() {r=0.35}
   Tiling=Constant() {r=6}
   Cell Offset=Constant()
   Cell Scale=LinearInterpolate(Alpha=NamedRerouteUsage()) {constA=0.25, constB=1.1}
   Cell Rotation=Add(A=NamedRerouteUsage(), B=@n18)
   Random Seed
  ) {materialFunction=/FortUILibrary/MaterialFunctions/Basics/MF_UI_ParticleGrid}
  n20 = TextureSample(UVs=@n19.Cells UV) {samplerType=SAMPLERTYPE_Alpha, texture=/Game/UI/Frontend/ItemShop/Art/Lucky/T_UI_LuckyClover-01}
  n21 = LinearInterpolate(
   A=@n1
   B=LinearInterpolate(
    A=LinearInterpolate(A=@n2, B=@n3, Alpha=@n4.RadialGradientExponential)
    B=Constant3Vector() {constant=<rgba 0.0135 0.5191 0.1175 1>}
    Alpha=Multiply(A=@n16) {constB=0.5}
   )
   Alpha=@n20.G
  )
  FunctionOutput(None=StaticSwitchParameter(True=LinearInterpolate(A=@n1, B=@n21, Alpha=@n20.G), False=FunctionInput()) {parameterName=blsLuckySeries})
  LinearInterpolate(
   A=FunctionInput(Preview=VectorParameter() {group=CorruptSeries, parameterName=BG_Color, defaultValue=<rgba 1 1 1 1>}.RGB) {bUsePreviewValueAsDefault=true, inputName=RGB 1, sortPriority=1}
   B=FunctionInput(Preview=VectorParameter() {group=CorruptSeries, parameterName=CellColor, defaultValue=<rgba 0 0 0 1>}.RGB) {bUsePreviewValueAsDefault=true, inputName=RGB 2, sortPriority=1}
  )
  NamedRerouteDeclaration(None=@n19.Random Rows) {name=Random Rows, nodeColor=<rgba 0.7344 0 1 1>}
  NamedRerouteDeclaration(None=@n19.Random Cell Values) {name=Random Cell Values, nodeColor=<rgba 0.7344 0 1 1>}
  NamedRerouteDeclaration(None=@n19.Random Columns) {name=Columns, nodeColor=<rgba 1 0.4688 0 1>}
 ]}
