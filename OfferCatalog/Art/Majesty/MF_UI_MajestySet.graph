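# Material function for the Majesty set look in offer catalog UI, switched by bIsMajestySet against a pass-through input.
# A two-color radial gradient, with hardness raised by HoverBurst, receives a tinted, glowing, transformable graphic mask (crystal by default) inside its own radial falloff.
# A radial spotlight color-dodges that result.
# Screen-aligned, aspect-corrected gold dust is added on top, masked by noise and bottom, left and right falloffs.
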MaterialFunction /OfferCatalog/Art/Majesty/MF_UI_MajestySet
graph {nodes [
  n1 = MaterialFunctionCall(Float2=MaterialFunctionCall() {materialFunction=&ScreenResolution}.Visible Resolution) {materialFunction=&BreakOutFloat2Components}
  n2 = Divide(A=@n1.R, B=Max(A=@n1.G))
  n3 = Constant() {r=1}
  n4 = MaterialFunctionCall!(UVs=Multiply(
    A=AppendVector(A=@n2, B=@n3)
    B=MaterialFunctionCall() {materialFunction=&ScreenAlignedUVs}."X 100%, Y 100%"
   )) {materialFunction=&/MF_UI_GoldDust}
  n5 = TextureCoordinate()
  n6 = OneMinus(None=ComponentMask(None=@n5) {g=true})
  n7 = ScalarParameter() {defaultValue=-0.3, group=Falloff, parameterName=FalloffShift_Bottom}
  n8 = Rotator(Coordinate=@n5, Time=ScalarParameter() {defaultValue=-0.6, group=Falloff, parameterName=FalloffLeft_Rotation}) {speed=6.2831}
  n9 = Multiply(
   A=Add(A=ComponentMask(None=@n8) {g=true}, B=ScalarParameter() {defaultValue=-0.3, group=Falloff, parameterName=FalloffShift_Left})
   B=ScalarParameter() {defaultValue=1, group=Falloff, parameterName=FalloffEdgeSoft_Left}
  )
  n10 = Rotator(Coordinate=@n5, Time=ScalarParameter() {defaultValue=0.65, group=Falloff, parameterName="FalloffRight_Rotation "}) {speed=6.2831}
  n11 = Multiply(
   A=Add(A=ComponentMask(None=@n10) {g=true}, B=ScalarParameter() {defaultValue=-0.2, group=Falloff, parameterName=FalloffShift_Right})
   B=ScalarParameter() {defaultValue=0.5, group=Falloff, parameterName=FalloffEdgeSoft_Right}
  )
  n12 = DotProduct(A=Multiply(A=Add(A=@n6, B=@n7)), B=DotProduct(A=Saturate(None=@n9), B=Saturate(None=@n11)))
  n13 = LinearInterpolate(B=@n4.GoldDust, Alpha=LinearInterpolate(B=@n4.Noise, Alpha=Saturate(None=@n12)))
  n14 = AppendVector(
   A=Multiply(A=ScalarParameter() {defaultValue=50, group=Background Gradient, parameterName=Spotlight_Position_X}) {constB=0.01}
   B=Multiply(A=ScalarParameter() {defaultValue=25, group=Background Gradient, parameterName=Spotlight_Position_Y}) {constB=0.01}
  )
  n15 = AppendVector(
   A=Multiply(A=ScalarParameter() {group=Background Gradient, parameterName=Gradient_Position_X}) {constB=0.01}
   B=Multiply(A=ScalarParameter() {group=Background Gradient, parameterName=Gradient_Position_Y}) {constB=0.01}
  )
  n16 = Multiply(A=ScalarParameter() {defaultValue=66.67, group=Background Gradient, parameterName=Spotlight_Size}) {constB=0.01}
  n17 = Multiply(A=ScalarParameter() {group=Background Gradient, parameterName=Gradient_Size}) {constB=0.01}
  n18 = Multiply(A=ScalarParameter() {defaultValue=16, group=Background Gradient, parameterName=Spotlight_Hardness}) {constB=0.1}
  n19 = Multiply(A=ScalarParameter() {group=Background Gradient, parameterName=Gradient_Hardness}) {constB=0.1}
  n20 = MaterialFunctionCall() {materialFunction=&GetUserInterfaceUV}
  n21 = MaterialFunctionCall(Float2=@n20.Pixel Size) {materialFunction=&BreakOutFloat2Components}
  n22 = Divide(A=@n21.G, B=@n21.G)
  n23 = AppendVector(A=Constant() {r=1}, B=@n22)
  n24 = Constant()
  n25 = Multiply(A=OneMinus(None=@n22)) {constB=0.5}
  n26 = MaterialFunctionCall(Float2=MaterialFunctionCall() {materialFunction=&GetUserInterfaceUV}.Pixel Size) {materialFunction=&BreakOutFloat2Components}
  n27 = Constant()
  n28 = MaterialFunctionCall() {materialFunction=&GetUserInterfaceUV}
  n29 = MaterialFunctionCall(Float2=@n28.Pixel Size) {materialFunction=&BreakOutFloat2Components}
  n30 = Divide(A=@n29.R, B=@n29.R)
  n31 = AppendVector(A=@n30, B=Constant() {r=1})
  n32 = Multiply(A=OneMinus(None=@n30)) {constB=0.5}
  n33 = Constant()
  n34 = LinearInterpolate(
   A=Add(A=Multiply(A=@n20.Normalized UV, B=@n23), B=AppendVector(A=@n24, B=@n25))
   B=Add(A=Multiply(A=@n28.Normalized UV, B=@n31), B=AppendVector(A=@n32, B=@n33))
   Alpha=If(A=@n26.R, B=@n26.R, "A > B"=@n27, "A == B"=@n27, "A < B"=Constant() {r=1})
  )
  n35 = LinearInterpolate(
   B=Multiply(A=ScalarParameter() {group=Background Gradient, parameterName=Spotlight_Intensity}) {constB=0.01}
   Alpha=MaterialFunctionCall(
    UVs=@n34
    CenterPosition=StaticSwitchParameter(True=@n14, False=@n15) {group=Background Gradient, parameterName=UseCustomSpotlightLocation}
    Radius=StaticSwitchParameter(True=@n16, False=@n17) {group=Background Gradient, parameterName=UseCustomSpotlightRadius}
    Density=StaticSwitchParameter(True=@n18, False=@n19) {group=Background Gradient, parameterName=UseCustomSpotlight_Hardness}
   ) {materialFunction=&RadialGradientExponential}.RadialGradientExponential
  )
  n36 = Multiply(A=ScalarParameter() {defaultValue=50, group=Background Gradient, parameterName=Gradient_Position_X}) {constB=0.01}
  n37 = Multiply(A=ScalarParameter() {defaultValue=25, group=Background Gradient, parameterName=Gradient_Position_Y}) {constB=0.01}
  n38 = ScalarParameter() {defaultValue=66.67, group=Background Gradient, parameterName=Gradient_Size}
  n39 = ScalarParameter() {defaultValue=16, group=Background Gradient, parameterName=Gradient_Hardness}
  n40 = LinearInterpolate(A=@n39, B=Add(A=@n39, B=Multiply(A=@n39) {constB=0.25}), Alpha=ScalarParameter() {group=Animation, parameterName=HoverBurst})
  n41 = LinearInterpolate(
   A=FunctionInput(Preview=Constant3Vector() {constant=<rgba 0.0658 0.0352 0.1146 1>}) {inputName=RGB 1, sortPriority=1}
   B=FunctionInput(Preview=Constant3Vector() {constant=<rgba 0.3275 0.1615 0.3333 1>}) {inputName=RGB 2, sortPriority=1}
   Alpha=MaterialFunctionCall(UVs=@n34, CenterPosition=AppendVector(A=@n36, B=@n37), Radius=Multiply(A=@n38) {constB=0.01}, Density=Multiply(A=@n40) {constB=0.1}) {materialFunction=&RadialGradientExponential}.RadialGradientExponential
  )
  n42 = VectorParameter() {group=Graphic, parameterName=PrimaryColor, defaultValue=<rgba 0.6354 0.2723 0.0331 1>}
  n43 = MaterialFunctionCall(Float2=@n34) {materialFunction=&BreakOutFloat2Components}
  n44 = OneMinus(None=@n43.R)
  n45 = ScalarParameter() {group=Graphic, parameterName=InverseX}
  n46 = MaterialFunctionCall(
   UVs=AppendVector(A=LinearInterpolate(A=@n43.R, B=@n44, Alpha=@n45), B=@n43.G)
   Texture Scale=ScalarParameter() {defaultValue=1, group=Graphic, parameterName=GraphicScale}
  ) {materialFunction=&ScaleUVsByCenter}
  n47 = AppendVector(
   A=ScalarParameter() {defaultValue=0.5, group=Graphic, parameterName=GraphicPosition_X}
   B=ScalarParameter() {defaultValue=0.5, group=Graphic, parameterName=GraphicPosition_Y}
  )
  n48 = TextureSampleParameter2D(UVs=Rotator(Coordinate=Add(A=@n46.UVs, B=@n47), Time=ScalarParameter() {group=Graphic, parameterName=GraphicRotation}) {speed=6.2831}) {
   group=Graphic
   parameterName=Graphic
   samplerSource=SSM_Clamp_WorldGroupSettings
   samplerType=SAMPLERTYPE_Masks
   texture=&/T_UI_Crystal
  }
  n49 = VectorParameter() {group=Graphic, parameterName=GlowColor, defaultValue=<rgba 0.7969 0.2032 0.0498 1>}
  n50 = Multiply(A=ScalarParameter() {defaultValue=50, group=Graphic, parameterName=Falloff_Position_X}) {constB=0.01}
  n51 = Multiply(A=ScalarParameter() {defaultValue=25, group=Graphic, parameterName=Falloff_Position_Y}) {constB=0.01}
  n52 = ScalarParameter() {defaultValue=66.67, group=Graphic, parameterName=Falloff_Size}
  n53 = ScalarParameter() {defaultValue=16, group=Graphic, parameterName=Falloff_Hardness}
  n54 = LinearInterpolate(
   B=Add(A=LinearInterpolate(B=@n42.RGB, Alpha=@n48.G), B=LinearInterpolate(B=@n49.RGB, Alpha=@n48.R))
   Alpha=MaterialFunctionCall(UVs=@n34, CenterPosition=AppendVector(A=@n50, B=@n51), Radius=Multiply(A=@n52) {constB=0.01}, Density=Multiply(A=@n53) {constB=0.1}) {materialFunction=&RadialGradientExponential}.RadialGradientExponential
  )
  n55 = Saturate(None=MaterialFunctionCall!(Blend=Saturate(None=@n35), Base=Add(A=@n41, B=@n54)) {materialFunction=/Engine/Functions/Engine_MaterialFunctions03/Blends/Blend_ColorDodge}.Result)
  FunctionOutput(None=StaticSwitchParameter(True=Add(A=@n13, B=@n55), False=FunctionInput()) {parameterName=bIsMajestySet})
 ]}
