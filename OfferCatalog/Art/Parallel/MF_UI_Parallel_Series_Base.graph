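# UI material function for the Parallel series background, which passes its input through unless the blsParallelSeries static switch is on.
# When on, it adds a glowing smoke layer to the SeriesTexture sample.
# The smoke is a panning honeycomb texture masked by an exponential radial gradient, tinted between two colors and scaled by intensity.
# Corner icon, pattern and extra color-blend branches are present but do not reach the output.
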MaterialFunction /OfferCatalog/Art/Parallel/MF_UI_Parallel_Series_Base
graph {nodes [
  n1 = TextureCoordinate()
  Add(A=@n1, B=Multiply(B=ScalarParameter() {defaultValue=0.1, group=Smoke, parameterName=Smoke-Distortion-Intensity}))
  n2 = MaterialFunctionCall() {materialFunction=&GetUserInterfaceUV}
  n3 = MaterialFunctionCall(Float2=@n2.Pixel Size) {materialFunction=&BreakOutFloat2Components}
  n4 = Divide(A=@n3.G, B=@n3.G)
  n5 = Multiply(A=@n2.Normalized UV, B=AppendVector(A=Constant() {r=1}, B=@n4))
  n6 = AppendVector(A=Constant(), B=Multiply(A=OneMinus(None=@n4)) {constB=0.5})
  n7 = MaterialFunctionCall(Float2=MaterialFunctionCall() {materialFunction=&GetUserInterfaceUV}.Pixel Size) {materialFunction=&BreakOutFloat2Components}
  n8 = Constant()
  n9 = Constant() {r=1}
  n10 = MaterialFunctionCall() {materialFunction=&GetUserInterfaceUV}
  n11 = MaterialFunctionCall(Float2=@n10.Pixel Size) {materialFunction=&BreakOutFloat2Components}
  n12 = Divide(A=@n11.R, B=@n11.R)
  n13 = Multiply(A=@n10.Normalized UV, B=AppendVector(A=@n12, B=Constant() {r=1}))
  n14 = AppendVector(A=Multiply(A=OneMinus(None=@n12)) {constB=0.5}, B=Constant())
  n15 = MaterialFunctionCall(Float2=LinearInterpolate(A=Add(A=@n5, B=@n6), B=Add(A=@n13, B=@n14), Alpha=If(A=@n7.R, B=@n7.R, "A > B"=@n8, "A == B"=@n8, "A < B"=@n9))) {materialFunction=&BreakOutFloat2Components}
  n16 = Divide(A=@n15.R, B=Max(A=@n15.G))
  n17 = AppendVector(A=@n16, B=Constant() {r=1})
  n18 = TextureCoordinate() {coordinateIndex=4, uTiling=0}
  n19 = Multiply(A=OneMinus(None=@n16))
  n20 = Constant()
  n21 = TextureSampleParameter2D!(UVs=Add(A=Multiply(A=@n17, B=@n18), B=AppendVector(A=@n19, B=@n20)), Bias, Apply View MipBias) {
   group=Series
   mipValueMode=TMVM_MipBias
   parameterName=SeriesTexture
   samplerSource=SSM_Clamp_WorldGroupSettings
   texture=&/T_UI_Parallel_Background
  }
  n22 = MaterialFunctionCall() {materialFunction=&GetUserInterfaceUV}
  n23 = MaterialFunctionCall(Float2=@n22.Pixel Size) {materialFunction=&BreakOutFloat2Components}
  n24 = Divide(A=@n23.G, B=@n23.G)
  n25 = AppendVector(A=Constant() {r=1}, B=@n24)
  n26 = Constant()
  n27 = Multiply(A=OneMinus(None=@n24)) {constB=0.5}
  n28 = MaterialFunctionCall(Float2=MaterialFunctionCall() {materialFunction=&GetUserInterfaceUV}.Pixel Size) {materialFunction=&BreakOutFloat2Components}
  n29 = Constant()
  n30 = MaterialFunctionCall() {materialFunction=&GetUserInterfaceUV}
  n31 = MaterialFunctionCall(Float2=@n30.Pixel Size) {materialFunction=&BreakOutFloat2Components}
  n32 = Divide(A=@n31.R, B=@n31.R)
  n33 = AppendVector(A=@n32, B=Constant() {r=1})
  n34 = Multiply(A=OneMinus(None=@n32)) {constB=0.5}
  n35 = Constant()
  n36 = LinearInterpolate(
   A=Add(A=Multiply(A=@n22.Normalized UV, B=@n25), B=AppendVector(A=@n26, B=@n27))
   B=Add(A=Multiply(A=@n30.Normalized UV, B=@n33), B=AppendVector(A=@n34, B=@n35))
   Alpha=If(A=@n28.R, B=@n28.R, "A > B"=@n29, "A == B"=@n29, "A < B"=Constant() {r=1})
  )
  n37 = TextureSample(UVs=Multiply(A=Panner(Coordinate=@n36) {speedX=0.0048, speedY=0.0033}, B=ScalarParameter() {defaultValue=21.4696, group=Smoke, parameterName=Suit Scale})) {texture=&/T_UI_Honeycomb}
  n38 = ScalarParameter() {defaultValue=0.2596, group=Smoke, parameterName=Smoke-Offset-X}
  n39 = ScalarParameter() {defaultValue=0.4078, group=Smoke, parameterName=Smoke-Offset-Y}
  n40 = Multiply(
   A=@n37.R
   B=MaterialFunctionCall(
    UVs=@n1
    CenterPosition=AppendVector(A=@n38, B=@n39)
    Radius=ScalarParameter() {defaultValue=0.47, group=Smoke, parameterName=Radius-Smoke, sortPriority=1}
    Density=ScalarParameter() {defaultValue=1.97, group=Smoke, parameterName=Density-Smoke, sortPriority=1}
   ) {materialFunction=&RadialGradientExponential}.RadialGradientExponential
  )
  n41 = VectorParameter() {parameterName=RGB2, defaultValue=<rgba 1 0.3632 0.403 1>}
  n42 = VectorParameter() {parameterName=RGB1, defaultValue=<rgba 55.7248 35.0522 6.1263 1>}
  n43 = Multiply(
   A=Multiply(A=@n40, B=LinearInterpolate(A=@n41.RGB, B=@n42.RGB, Alpha=@n40))
   B=ScalarParameter() {defaultValue=1, group=Smoke, parameterName=Smoke-Intensity}
  )
  FunctionOutput(None=StaticSwitchParameter(True=Add(A=@n21.RGB, B=@n43), False=FunctionInput()) {parameterName=blsParallelSeries})
  n44 = VectorParameter() {parameterName=UpperLeftIconSize, channelNames {a="", b="", g=Height, r=Width}, defaultValue=<rgba 241.3318 267.5551 0 1>}
  n45 = Divide(A=@n44.Width, B=@n44.Width)
  n46 = MaterialFunctionCall(Float2=MaterialFunctionCall() {materialFunction=&GetUserInterfaceUV}.Pixel Size) {materialFunction=&BreakOutFloat2Components}
  n47 = NamedRerouteDeclaration(None=@n46.G) {name=PixelSizeY, nodeColor=<rgba 0.4531 1 0 1>}
  n48 = NamedRerouteDeclaration(None=@n46.R) {name=PixelSizeX, nodeColor=<rgba 0 1 1 1>}
  n49 = Divide(A=@n47, B=@n48)
  n50 = Divide(A=@n44.Height, B=@n44.Height)
  n51 = Divide(A=@n48, B=@n47)
  n52 = VectorParameter() {parameterName=UpperLeftIconSize, channelNames {a="", b="", g=Height, r=Width}, defaultValue=<rgba 241.3318 267.5551 0 1>}
  n53 = Divide(
   A=AppendVector(
    A=NamedRerouteDeclaration(None=@n52.Width) {name=IconSizeX, nodeColor=<rgba 1 0.1406 0 1>}
    B=NamedRerouteDeclaration(None=@n52.Height) {name=IconSizeY, nodeColor=<rgba 0 1 0.3438 1>}
   )
   B=MaterialFunctionCall() {materialFunction=&GetUserInterfaceUV}.Pixel Size
  )
  n54 = If(
   A=NamedRerouteUsage()
   B=NamedRerouteUsage()
   "A > B"=AppendVector(A=Constant() {r=1}, B=Multiply(A=@n50, B=@n51))
   "A == B"=@n53
   "A < B"=@n53
  )
  If(
   A=NamedRerouteUsage()
   B=NamedRerouteUsage()
   "A > B"=AppendVector(A=Multiply(A=@n45, B=@n49), B=Constant() {r=1})
   "A == B"=@n54
   "A < B"=@n54
  )
  LinearInterpolate(
   A=FunctionInput(Preview=VectorParameter() {group=CorruptSeries, parameterName=BG_Color, defaultValue=<rgba 1 1 1 1>}.RGB) {bUsePreviewValueAsDefault=true, inputName=RGB 1, sortPriority=1}
   B=FunctionInput(Preview=VectorParameter() {group=CorruptSeries, parameterName=CellColor, defaultValue=<rgba 0 0 0 1>}.RGB) {bUsePreviewValueAsDefault=true, inputName=RGB 2, sortPriority=1}
  )
  n55 = MaterialFunctionCall() {materialFunction=&GetUserInterfaceUV}
  n56 = MaterialFunctionCall(Float2=@n55.Pixel Size) {materialFunction=&BreakOutFloat2Components}
  n57 = Divide(A=@n56.G, B=@n56.G)
  n58 = Multiply(A=@n55.Normalized UV, B=AppendVector(A=Constant() {r=1}, B=@n57))
  n59 = AppendVector(A=Constant(), B=Multiply(A=OneMinus(None=@n57)) {constB=0.5})
  n60 = MaterialFunctionCall(Float2=MaterialFunctionCall() {materialFunction=&GetUserInterfaceUV}.Pixel Size) {materialFunction=&BreakOutFloat2Components}
  n61 = Constant()
  n62 = Constant() {r=1}
  n63 = MaterialFunctionCall() {materialFunction=&GetUserInterfaceUV}
  n64 = MaterialFunctionCall(Float2=@n63.Pixel Size) {materialFunction=&BreakOutFloat2Components}
  n65 = Divide(A=@n64.R, B=@n64.R)
  n66 = Multiply(A=@n63.Normalized UV, B=AppendVector(A=@n65, B=Constant() {r=1}))
  n67 = AppendVector(A=Multiply(A=OneMinus(None=@n65)) {constB=0.5}, B=Constant())
  n68 = MaterialFunctionCall(Float2=LinearInterpolate(A=Add(A=@n58, B=@n59), B=Add(A=@n66, B=@n67), Alpha=If(A=@n60.R, B=@n60.R, "A > B"=@n61, "A == B"=@n61, "A < B"=@n62))) {materialFunction=&BreakOutFloat2Components}
  n69 = Divide(A=@n68.R, B=Max(A=@n68.G))
  n70 = Multiply(A=AppendVector(A=@n69, B=Constant() {r=1}), B=TextureCoordinate() {coordinateIndex=4, uTiling=0})
  n71 = AppendVector(A=Multiply(A=OneMinus(None=@n69)), B=Constant())
  LinearInterpolate(
   B=Constant3Vector() {constant=<rgba 0.0319 0.1413 0.5395 1>}
   Alpha=TextureSampleParameter2D!(UVs=Add(A=@n70, B=@n71), Apply View MipBias) {parameterName=Param, texture=&/T_Parallel_MG}.A
  )
  n72 = NamedRerouteUsage()
  n73 = NamedRerouteUsage()
  n74 = NamedRerouteUsage()
  n75 = NamedRerouteUsage()
  n76 = MaterialFunctionCall(Float2=MaterialFunctionCall() {materialFunction=&GetUserInterfaceUV}.Pixel Size) {materialFunction=&BreakOutFloat2Components}
  n77 = NamedRerouteDeclaration(None=@n76.G) {name=PixelSizeY_1, nodeColor=<rgba 0.4531 1 0 1>}
  n78 = NamedRerouteDeclaration(None=@n76.R) {name=PixelSizeX_1, nodeColor=<rgba 0 1 1 1>}
  n79 = AppendVector(A=Multiply(A=Divide(A=@n74, B=@n75), B=Divide(A=@n77, B=@n78)), B=Constant() {r=1})
  n80 = Divide(A=@n75, B=@n74)
  n81 = Divide(A=@n78, B=@n77)
  n82 = VectorParameter() {parameterName=UpperLeftIconSize, channelNames {a="", b="", g=Height, r=Width}, defaultValue=<rgba 241.3318 227.7446 0 1>}
  n83 = Divide(
   A=AppendVector(
    A=NamedRerouteDeclaration(None=@n82.Width) {name=UpperIconSizeX, nodeColor=<rgba 1 0.1406 0 1>}
    B=NamedRerouteDeclaration(None=@n82.Height) {name=UpperIconSizeY, nodeColor=<rgba 0 1 0.3438 1>}
   )
   B=MaterialFunctionCall() {materialFunction=&GetUserInterfaceUV}.Pixel Size
  )
  n84 = If(
   A=NamedRerouteUsage()
   B=NamedRerouteUsage()
   "A > B"=AppendVector(A=Constant() {r=1}, B=Multiply(A=@n80, B=@n81))
   "A == B"=@n83
   "A < B"=@n83
  )
  n85 = VectorParameter() {parameterName=UpperLeftPosition, channelNames {a="", b="", g=Y, r=X}, defaultValue=<rgba 0 0 0 1>}
  n86 = TextureSample(UVs=MaterialFunctionCall!(Texture Scale=If(A=@n72, B=@n73, "A > B"=@n79, "A == B"=@n84, "A < B"=@n84), UVs, Offset XY Texturesize Z=@n85) {materialFunction=/Engine/Functions/Engine_MaterialFunctions01/Texturing/ScaleUVsAroundPoint}.UVs) {samplerSource=SSM_Clamp_WorldGroupSettings, texture=&/T_UI_Parallel_Corner2}
  n87 = VectorParameter() {parameterName=LowerRightPosition, channelNames {a="", b="", g=Y, r=X}, defaultValue=<rgba 1 1 1 1>}
  n88 = NamedRerouteUsage()
  n89 = NamedRerouteUsage()
  n90 = NamedRerouteUsage()
  n91 = NamedRerouteUsage()
  n92 = NamedRerouteUsage()
  n93 = NamedRerouteUsage()
  n94 = AppendVector(A=Multiply(A=Divide(A=@n90, B=@n91), B=Divide(A=@n92, B=@n93)), B=Constant() {r=1})
  n95 = Divide(A=@n91, B=@n90)
  n96 = Divide(A=@n93, B=@n92)
  n97 = VectorParameter() {parameterName=LowerRightIconSize, channelNames {a="", b="", g=Height, r=Width}, defaultValue=<rgba 378.4796 508.7373 0 1>}
  n98 = Divide(
   A=AppendVector(
    A=NamedRerouteDeclaration(None=@n97.Width) {name=LowerIconSizeX, nodeColor=<rgba 1 0.1406 0 1>}
    B=NamedRerouteDeclaration(None=@n97.Height) {name=LowerIconSizeY, nodeColor=<rgba 0 1 0.3438 1>}
   )
   B=MaterialFunctionCall() {materialFunction=&GetUserInterfaceUV}.Pixel Size
  )
  n99 = If(
   A=NamedRerouteUsage()
   B=NamedRerouteUsage()
   "A > B"=AppendVector(A=Constant() {r=1}, B=Multiply(A=@n95, B=@n96))
   "A == B"=@n98
   "A < B"=@n98
  )
  n100 = TextureSample(UVs=MaterialFunctionCall(Texture Scale=If(A=@n88, B=@n89, "A > B"=@n94, "A == B"=@n99, "A < B"=@n99), Offset XY Texturesize Z=@n87) {materialFunction=&ScaleUVsAroundPoint}.UVs) {samplerSource=SSM_Clamp_WorldGroupSettings, texture=&/T_UI_Parallel_Corner1}
  LinearInterpolate(A=LinearInterpolate(B=@n86.RGB, Alpha=@n86.RGB), B=@n100.RGB, Alpha=@n100.RGB)
  MaterialFunctionCall() {materialFunction=&GetUserInterfaceUV}
  MaterialFunctionCall() {materialFunction=&GetUserInterfaceUV}
  n101 = MaterialFunctionCall(
   UVs=TextureCoordinate()
   Texture Scale=Divide(
    A=ScalarParameter() {defaultValue=577.6138, parameterName=PatternSize}
    B=MaterialFunctionCall() {materialFunction=&GetUserInterfaceUV}.Pixel Size
   )
  ) {materialFunction=&ScaleUVsByCenter}
  Multiply(
   A=Desaturation(None=TextureSample(UVs=@n101.UVs) {texture=/BRCosmetics/Creative/Toys/Basketball/Textures/SportingGoods_Basketball_Creative_Pattern_TriangleBars_01}.RGB)
   B=Constant3Vector() {constant=<rgba 1 0 0.0398 0>}
  )
  OneMinus(None=@n37.G)
 ]}
