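# Material function that colors PastaSauce series item shop tiles when the blsPastaSauce_Series static switch is on, otherwise passing its input through.
# Offset radial gradients blend fixed pinks and dark plum, a top-bottom gradient darkens toward black, and the pink glow is reapplied on top.
# Named reroutes build scanlines from striped UVs, a box edge mask and a panning cracks texture.
# A CorruptSeries BG/Cell color blend is left unconnected.
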MaterialFunction /OfferCatalog/Art/PastaSauce/MF_UI_PastaSauce_ShopTiles
graph {nodes [
  n1 = Constant3Vector() {constant=<rgba 0.6308 0.0602 0.1695 1>}
  n2 = Constant3Vector() {constant=<rgba 1 0.6376 0.8388 1>}
  n3 = NamedRerouteUsage()
  n4 = AppendVector(A=Constant(), B=Constant())
  n5 = SmoothStep(Value=MaterialFunctionCall(UV=Add(A=@n3, B=@n4), GradientStart=Constant() {r=-0.25}, GradientEnd=Constant() {r=1}) {materialFunction=&MF_UI_RadialGradient}.Smooth) {constMax=1.5}
  n6 = NamedRerouteUsage()
  n7 = NamedRerouteUsage()
  n8 = AppendVector(A=Constant(), B=Constant())
  n9 = SmoothStep(Value=MaterialFunctionCall(UV=Add(A=@n7, B=@n8), GradientStart=Constant() {r=-0.5}, GradientEnd=Constant() {r=1}) {materialFunction=&MF_UI_RadialGradient}.Smooth) {constMax=1.25}
  n10 = Subtract(A=Saturate(None=@n6), B=Power(Base=Multiply(A=@n9) {constB=2}) {constExponent=0.25})
  n11 = NamedRerouteUsage()
  n12 = Constant() {r=0.2}
  n13 = Constant() {r=1}
  n14 = Constant() {r=1.25}
  n15 = LinearInterpolate(
   A=LinearInterpolate(A=LinearInterpolate(A=@n1, B=@n2, Alpha=@n5), B=Constant3Vector() {constant=<rgba 0.3177 0 0.1717 1>}, Alpha=Saturate(None=@n10))
   B=Constant3Vector() {constant=<rgba 0 0 0 1>}
   Alpha=SmoothStep(Value=MaterialFunctionCall!(UVs=@n11, Subtract=@n12, Falloff=@n13, Intensity=@n14) {materialFunction=/Game/UI/Frontend/ItemShop/Art/_MaterialFunctions_ShopTheming/MF_UI_TopBottomGradient}.Result) {constMax=1.2}
  ) {constB=0}
  FunctionOutput(None=StaticSwitchParameter(True=LinearInterpolate(A=@n15, B=@n2, Alpha=@n5), False=FunctionInput()) {parameterName=blsPastaSauce_Series})
  LinearInterpolate(
   A=FunctionInput(Preview=VectorParameter() {group=CorruptSeries, parameterName=BG_Color, defaultValue=<rgba 1 1 1 1>}.RGB) {bUsePreviewValueAsDefault=true, inputName=RGB 1, sortPriority=1}
   B=FunctionInput(Preview=VectorParameter() {group=CorruptSeries, parameterName=CellColor, defaultValue=<rgba 0 0 0 1>}.RGB) {bUsePreviewValueAsDefault=true, inputName=RGB 2, sortPriority=1}
  )
  Multiply(A=@n6) {constB=0.2}
  n16 = Constant2Vector() {g=0.65, r=0.25}
  n17 = TextureCoordinate() {coordinateIndex=1}
  Multiply(
   A=@n16
   B=ComponentMask(None=MaterialFunctionCall!(UV=@n17, Seed) {materialFunction=&MF_UI_Random}.Result) {g=true}
  )
  n18 = Time()
  n19 = ConstantBiasScale(None=Sine(None=Multiply(A=@n18) {constB=0}))
  n20 = TextureCoordinate() {coordinateIndex=3}
  n21 = ComponentMask(None=@n20) {r=true}
  n22 = ComponentMask(None=@n20) {g=true}
  n23 = Constant() {r=0.7}
  n24 = Multiply(A=OneMinus(None=Divide(A=@n21, B=@n22)), B=@n23)
  n25 = AppendVector(A=@n23, B=Add(A=Multiply(A=@n24) {constB=0.5}, B=@n23))
  n26 = Constant()
  n27 = Constant() {r=0.65}
  n28 = StaticBool() {value=true}
  n29 = OneMinus(None=SmoothStep(
    Max=LinearInterpolate(Alpha=@n19) {constA=0.8}
    Value=MaterialFunctionCall(Size=@n25, CornerRoundness=@n26, GlowMax=@n27, KeepAspectRatio=@n28) {materialFunction=&MF_UI_Box}.Fill
   ))
  Multiply(A=@n29, B=Constant3Vector() {constant=<rgba 1 0 0 0>})
  NamedRerouteDeclaration(None=MaterialFunctionCall() {materialFunction=&MF_UI_UVScaling}.Scaled To Fill) {name=UVFill_1, nodeColor=<rgba 0.9922 0 1 1>}
  n30 = Constant()
  n31 = Multiply(A=ComponentMask(None=@n17) {g=true}) {constB=3}
  n32 = ScalarParameter() {defaultValue=0.12, parameterName=ScanlineSpeed}
  n33 = TextureSample(UVs=Panner(Coordinate=AppendVector(A=@n30, B=@n31), Speed=Multiply(A=@n16, B=@n32) {constB=0.2}) {speedX=0.257, speedY=0.651}) {samplerType=SAMPLERTYPE_LinearColor, texture=/Game/Global/Textures/Default/Masks_and_Gradients/T_Cracks}
  n34 = Multiply(A=@n33.R, B=@n33.R)
  n35 = Multiply(A=@n34, B=@n34)
  n36 = ScalarParameter() {defaultValue=0.25, parameterName=ScanlineOpacity}
  NamedRerouteDeclaration(None=Saturate(None=Multiply(A=@n35, B=@n36) {constB=0.25})) {name=MovingScanlines, nodeColor=<rgba 0 0.1562 1 1>}
  n37 = NamedRerouteUsage()
  n38 = Multiply(A=MaterialFunctionCall(Base=Multiply(A=@n29) {constB=0.1}, Blend=@n37) {materialFunction=&Blend_Overlay}.Result) {constB=100}
  n39 = TextureCoordinate()
  n40 = Frac(None=Multiply(A=ComponentMask(None=@n39) {g=true}) {constB=42.5, desc=Number of lines})
  n41 = Add(A=Floor(None=Divide(A=@n40) {constB=0.66, desc="empty space sizing (percentage)"}), B=Multiply(A=@n37) {constB=3})
  NamedRerouteDeclaration(None=Saturate(None=Add(A=@n38, B=@n41))) {name=ScanlinesAlpha, nodeColor=<rgba 1 0.8672 0 1>}
  TextureCoordinate() {coordinateIndex=3}
 ]}
