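# Shop tile background material function for the Pencil series, drawing animated diamond patterns in blues, switched by bls_PencilSeries against a passthrough input.
# Slowly panning UVs feed two particle grids with alternating offsets, whose diamond distance-field textures become antialiased masks.
# Linear gradients tint and fade the layers, while diagonal and radial center-focus masks are also built.
# Named reroutes carry these colors and masks into the final lerps.
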MaterialFunction /OfferCatalog/Art/Pencil/MF_UI_ShopTile_Pencil
graph {nodes [
  Constant3Vector() {constant=<rgba 0.0703 0.4834 0.75 1>}
  Constant3Vector() {constant=<rgba 0.4963 0.3279 1 1>}
  Constant3Vector() {constant=<rgba 0.0035 0.1532 0.6198 1>}
  Constant() {r=0.5}
  Constant() {r=0.01}
  n1 = NamedRerouteUsage()
  n2 = LinearInterpolate(
   A=LinearInterpolate(A=NamedRerouteUsage(), B=NamedRerouteUsage(), Alpha=NamedRerouteUsage())
   B=NamedRerouteUsage()
   Alpha=NamedRerouteUsage()
  )
  n3 = Max(A=NamedRerouteUsage(), B=NamedRerouteUsage())
  n4 = OneMinus(None=NamedRerouteUsage())
  n5 = NamedRerouteUsage()
  n6 = Multiply(A=Min(A=Min(A=@n3, B=@n4), B=OneMinus(None=@n5))) {constB=0.75}
  FunctionOutput(None=StaticSwitchParameter(True=LinearInterpolate(A=@n1, B=@n2, Alpha=@n6), False=FunctionInput()) {parameterName=bls_PencilSeries})
  n7 = Constant3Vector() {constant=<rgba 0.0987 0.3085 0.7656 1>}
  LinearInterpolate(
   A=@n7
   Alpha=MaterialFunctionCall(UVs=NamedRerouteUsage(), GradientStart=Constant() {r=0.3}, GradientEnd=Constant() {r=0.75}, RotationAngle=Constant() {r=0.35}) {materialFunction=&MF_UI_LinearGradient}.Result
  )
  LinearInterpolate(
   A=VectorParameter() {parameterName=BG Upper Gradient, defaultValue=<rgba 0.4963 0.3279 1 1>}.RGB
   B=VectorParameter() {parameterName=SecondaryDiamonds, defaultValue=<rgba 0.8802 0.7075 0 1>}.RGB
  ) {constAlpha=0.75}
  LinearInterpolate(
   A=FunctionInput(Preview=VectorParameter() {group=CorruptSeries, parameterName=BG_Color, defaultValue=<rgba 1 1 1 1>}.RGB) {bUsePreviewValueAsDefault=true, inputName=RGB 1, sortPriority=1}
   B=FunctionInput(Preview=VectorParameter() {group=CorruptSeries, parameterName=CellColor, defaultValue=<rgba 0 0 0 1>}.RGB) {bUsePreviewValueAsDefault=true, inputName=RGB 2, sortPriority=1}
  )
  MaterialFunctionCall!(ScreenUV, TextureSize, ScalePivotPoint) {materialFunction=/FortUILibrary/MaterialFunctions/Basics/MF_UI_ScreenSpaceUVScaling}
  Multiply(A=NamedRerouteUsage(), B=NamedRerouteUsage())
  n8 = Panner(Coordinate=Multiply(A=NamedRerouteUsage()) {constB=0.45}, Time=Time(), Speed=Constant2Vector() {g=0.007, r=0.009})
  n9 = Constant() {r=15}
  n10 = Multiply(A=ComponentMask(None=@n8) {r=true}, B=@n9)
  n11 = LinearInterpolate(Alpha=Saturate(None=MaterialFunctionCall!(Value=@n10) {materialFunction=/FortUILibrary/MaterialFunctions/Basics/MF_UI_EvenOrOdd}.Result)) {constB=0.25}
  n12 = Constant()
  n13 = MaterialFunctionCall!(
   UV=@n8
   Global Offset=AppendVector(A=Constant(), B=@n11)
   Global Rotation
   Tiling=@n9
   Cell Offset=@n12
   Cell Scale=LinearInterpolate(Alpha=NamedRerouteUsage()) {constA=0.3, constB=1.25}
   Cell Rotation=@n12
   Random Seed
  ) {materialFunction=/FortUILibrary/MaterialFunctions/Basics/MF_UI_ParticleGrid}
  NamedRerouteDeclaration(None=@n13.Random Rows) {name=Rows, nodeColor=<rgba 0 1 0.5547 1>}
  NamedRerouteDeclaration(None=@n13.Random Cell Values) {name=RandomCell, nodeColor=<rgba 0 0.7422 1 1>}
  n14 = Constant()
  n15 = MaterialFunctionCall(
   UV=@n8
   Global Offset=AppendVector(A=@n11, B=Constant2Vector() {r=0.005})
   Tiling=@n9
   Cell Offset=@n14
   Cell Scale=LinearInterpolate(Alpha=NamedRerouteUsage()) {constA=0.3}
   Cell Rotation=@n14
  ) {materialFunction=&MF_UI_ParticleGrid}
  NamedRerouteDeclaration(None=@n15.Random Rows) {name=Rows_1, nodeColor=<rgba 0 1 0.5547 1>}
  NamedRerouteDeclaration(None=@n15.Random Cell Values) {name=RandomCell_1, nodeColor=<rgba 0 0.7422 1 1>}
  n16 = MaterialFunctionCall(UVs=NamedRerouteUsage(), GradientStart=Constant() {r=0.3}, GradientEnd=Constant() {r=0.75}, RotationAngle=Constant() {r=0.35}) {materialFunction=&MF_UI_LinearGradient}
  NamedRerouteDeclaration(None=LinearInterpolate(
    A=Constant3Vector() {constant=<rgba 0.0035 0.1532 0.6198 1>}
    B=Constant3Vector() {constant=<rgba 0.0561 0.5528 0.8733 1>}
    Alpha=SmoothStep(Value=@n16.Result) {constMax=1.5, constMin=-0.5}
   )) {name=BackgroundGradient, nodeColor=<rgba 1 0 0.6797 1>}
  NamedRerouteDeclaration(None=@n13.Random Columns) {name=Columns, nodeColor=<rgba 1 0 0.6094 1>}
  n17 = MaterialFunctionCall() {materialFunction=&MF_UI_UVScaling}
  NamedRerouteDeclaration(None=MaterialFunctionCall!(UVs=@n17.Scaled To Fill, Scale=Constant() {r=0.5}, X Position=Constant(), Y Position=Constant()) {materialFunction=/OfferCatalog/Art/_MaterialFunctions_ShopTheming/MF_UI_ScalePosition}.Result) {name=UVScaling, nodeColor=<rgba 0.2891 1 0 1>}
  n18 = MaterialFunctionCall(
   DistanceField=MaterialFunctionCall!(
    RGBA Channels=TextureSample(UVs=@n15.Cells UV) {samplerType=SAMPLERTYPE_LinearColor, texture=&/T_UI_Pencil_Diamond}.R
    Glow Min
    Glow Max
    Outline Thickness
   ) {materialFunction=/FortUILibrary/MaterialFunctions/Basics/MF_UI_ChannelPackedDistanceField}.DF
   Threshold=Constant() {r=0.4}
  ) {materialFunction=&MF_UI_AntiAliasing}
  n19 = MaterialFunctionCall(UVs=NamedRerouteUsage(), GradientStart=Constant() {r=0.25}, GradientEnd=Constant() {r=0.65}, RotationAngle=Constant() {r=0.35}) {materialFunction=&MF_UI_LinearGradient}
  NamedRerouteDeclaration(None=Multiply(A=Multiply(A=@n19.Result, B=@n18.Outer)) {constB=0.55}) {name=fading diamonds alpha, nodeColor=<rgba 0 1 0.4375 1>}
  NamedRerouteDeclaration(None=@n7) {name=SecondaryDiamondsColor, nodeColor=<rgba 0 1 0.4844 1>}
  n20 = TextureSample(UVs=@n13.Cells UV) {samplerType=SAMPLERTYPE_LinearColor, texture=&T_UI_Pencil_Diamond}
  NamedRerouteDeclaration(None=MaterialFunctionCall(
    DistanceField=MaterialFunctionCall(RGBA Channels=@n20.R) {materialFunction=&MF_UI_ChannelPackedDistanceField}.DF
    Threshold=Constant() {r=0.4}
   ) {materialFunction=&MF_UI_AntiAliasing}.Outer) {name=diamonds alpha, nodeColor=<rgba 0.3594 1 0 1>}
  n21 = MaterialFunctionCall(UVs=NamedRerouteUsage(), GradientStart=Constant() {r=0.1}, GradientEnd=Constant() {r=0.75}, RotationAngle=Constant() {r=0.9}) {materialFunction=&MF_UI_LinearGradient}
  NamedRerouteDeclaration(None=LinearInterpolate(
    A=Constant3Vector() {constant=<rgba 0.0203 0.1812 0.5271 1>}
    B=Constant3Vector() {constant=<rgba 0.0054 0.707 0.967 1>}
    Alpha=SmoothStep(Value=@n21.Result)
   )) {name=diamonds color, nodeColor=<rgba 0.2422 1 0 1>}
  n22 = NamedRerouteUsage()
  n23 = Constant() {r=1}
  n24 = Constant() {r=0.3}
  n25 = Constant()
  n26 = ConstantBiasScale(None=MaterialFunctionCall(
    UVs=MaterialFunctionCall(UVs=@n22, Scale=@n23, X Position=@n24, Y Position=@n25) {materialFunction=&MF_UI_ScalePosition}.Result
    Rotation Center=Constant() {r=0.5}
    "Rotation Angle (0-1)"=Constant() {r=0.6}
   ) {materialFunction=&CustomRotator}.Rotated Values) {bias=-0.5, scale=2}
  n27 = Subtract(A=ComponentMask(None=Abs(None=@n26)) {g=true}, B=ScalarParameter() {defaultValue=0.01, parameterName=Subtract})
  n28 = ScalarParameter() {defaultValue=0.85, parameterName=Falloff}
  n29 = Multiply(A=Saturate(None=Multiply(A=@n27, B=@n28)), B=ScalarParameter() {defaultValue=1, parameterName=Intensity})
  NamedRerouteDeclaration(None=SmoothStep(Value=Saturate(None=@n29)) {constMax=1.2}) {name=diagonal fade, nodeColor=<rgba 0 0.4844 1 1>}
  n30 = MaterialFunctionCall(
   UVs=Multiply(A=MaterialFunctionCall() {materialFunction=&MF_UI_UVScaling}.ScaleX, B=Constant2Vector() {g=1.5, r=1})
   Texture Scale=Constant() {r=1.25}
  ) {materialFunction=&ScaleUVsByCenter}
  n31 = Constant() {r=0.25}
  n32 = Constant() {r=1}
  NamedRerouteDeclaration(None=Saturate(None=MaterialFunctionCall(UV=@n30.UVs, GradientStart=@n31, GradientEnd=@n32) {materialFunction=&MF_UI_RadialGradient}.Smooth)) {name=center focus, nodeColor=<rgba 0 1 0.0391 1>}
  NamedRerouteDeclaration(None=@n17.Scaled to Fit) {name=UVFit, nodeColor=<rgba 0.1484 1 0 1>}
  NamedRerouteUsage()
  NamedRerouteUsage()
 ]}
