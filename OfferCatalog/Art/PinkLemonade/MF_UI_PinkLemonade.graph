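# UI material function that supplies the Pink Lemonade series gradient background, apparently for offer catalog art.
# Two exponential radial gradients centred at bottom middle, beyond the UV edge, layer blue, pink and cream with parameterised radius and density.
# The PinkLemonadeSeries static switch outputs this gradient or passes the function input through.
# Other branches, including panning swirl-distorted shine masked by tile frames, shine bars and a CorruptSeries colour blend, never reach the output.
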MaterialFunction /OfferCatalog/Art/PinkLemonade/MF_UI_PinkLemonade
graph {nodes [
  n1 = Constant2Vector() {g=1.7116, r=0.5}
  n2 = ScalarParameter() {defaultValue=1.7117, parameterName=FirstLayerRadius}
  n3 = ScalarParameter() {defaultValue=3.6979, parameterName=FirstLayerDensity}
  n4 = LinearInterpolate(
   A=LinearInterpolate(
    A=Constant3Vector() {constant=<rgba 0 0.3848 0.8385 1>}
    B=Constant3Vector() {constant=<rgba 0.8148 0.3438 0.5743 1>}
    Alpha=MaterialFunctionCall(CenterPosition=@n1, Radius=@n2, Density=@n3) {materialFunction=&RadialGradientExponential}.RadialGradientExponential
   )
   B=Constant3Vector() {constant=<rgba 0.9911 0.855 0.6654 1>}
   Alpha=MaterialFunctionCall(
    CenterPosition=Constant2Vector() {g=1.509, r=0.5}
    Radius=ScalarParameter() {defaultValue=1.0572, parameterName=SecondGradientRadius}
    Density=ScalarParameter() {defaultValue=3.3322, parameterName=2ndRadius}
   ) {materialFunction=&RadialGradientExponential}.RadialGradientExponential
  )
  FunctionOutput(None=StaticSwitchParameter(True=@n4, False=FunctionInput()) {parameterName=PinkLemonadeSeries})
  LinearInterpolate(
   A=FunctionInput(Preview=VectorParameter() {group=CorruptSeries, parameterName=BG_Color, defaultValue=<rgba 1 1 1 1>}.RGB) {bUsePreviewValueAsDefault=true, inputName=RGB 1, sortPriority=1}
   B=FunctionInput(Preview=VectorParameter() {group=CorruptSeries, parameterName=CellColor, defaultValue=<rgba 0 0 0 1>}.RGB) {bUsePreviewValueAsDefault=true, inputName=RGB 2, sortPriority=1}
  )
  n5 = Constant3Vector() {constant=<rgba 0.2344 0.3875 1 1>}
  n6 = Constant3Vector() {constant=<rgba 0.9387 0.2298 0.6756 1>}
  n7 = MaterialFunctionCall(
   CenterPosition=Constant2Vector() {g=1.7116, r=0.5}
   Radius=ScalarParameter() {defaultValue=1.7117, parameterName=FirstLayerRadius}
   Density=ScalarParameter() {defaultValue=3.6979, parameterName=FirstLayerDensity}
  ) {materialFunction=&RadialGradientExponential}
  n8 = Constant2Vector() {g=1.509, r=0.5}
  n9 = ScalarParameter() {defaultValue=1.0572, parameterName=SecondGradientRadius}
  n10 = ScalarParameter() {defaultValue=3.3322, parameterName=2ndRadius}
  n11 = TextureSample(
   UVs=Panner(Coordinate=TextureCoordinate()) {speedX=0.07, speedY=0.132}
   Tex=TextureObject() {samplerType=SAMPLERTYPE_LinearColor, texture=/Game/Global/Textures/Default/Flowmaps/T_SwirlFlow_02}
  )
  n12 = Multiply(A=Constant2Vector() {g=1, r=-1}, B=Multiply(A=Constant() {r=0.5}, B=ComponentMask(None=@n11.R) {g=true, r=true}))
  n13 = MaterialFunctionCall(UVs=Add(A=Multiply(A=@n12), B=TextureCoordinate()), Texture Scale=ScalarParameter() {defaultValue=2.0164, parameterName=Layer1Scale}) {materialFunction=&ScaleUVsByCenter}
  n14 = MaterialFunctionCall() {materialFunction=&MF_UI_UVScaling}
  LinearInterpolate(
   A=LinearInterpolate(
    A=LinearInterpolate(A=@n5, B=@n6, Alpha=@n7.RadialGradientExponential)
    B=Constant3Vector() {constant=<rgba 0.9911 0.8632 0.6654 1>}
    Alpha=MaterialFunctionCall(CenterPosition=@n8, Radius=@n9, Density=@n10) {materialFunction=&RadialGradientExponential}.RadialGradientExponential
   )
   B=Saturate(None=TextureSample(UVs=@n13.UVs) {texture=&/T_UI_ShinePattern}.RGB)
   Alpha=StaticSwitchParameter(
    True=TextureSampleParameter2D(UVs=@n14.ScaleX) {parameterName=WideTile, texture=&/T_UI_FrameWide}.R
    False=TextureSampleParameter2D(UVs=@n14.ScaleY) {parameterName=NormalTile, texture=&/T_UI_Frame}.R
   ) {parameterName=TileSize}
  )
  n15 = TextureSample(
   UVs=Panner(Coordinate=TextureCoordinate()) {speedX=0.07, speedY=0.132}
   Tex=TextureObject() {samplerType=SAMPLERTYPE_LinearColor, texture=&T_SwirlFlow_02}
  )
  n16 = Multiply(A=Constant2Vector() {g=1, r=-1}, B=Multiply(A=Constant() {r=0.5}, B=ComponentMask(None=@n15.R) {g=true, r=true}))
  n17 = ScalarParameter() {defaultValue=1.0497, parameterName=size}
  n18 = MaterialFunctionCall(UVs=Add(A=Multiply(A=@n16, B=@n17), B=TextureCoordinate()), Texture Scale=ScalarParameter() {defaultValue=2.0164, parameterName=Layer1Scale}) {materialFunction=&ScaleUVsByCenter}
  n19 = MaterialFunctionCall() {materialFunction=&MF_UI_UVScaling}
  LinearInterpolate(
   A=@n4
   B=Saturate(None=TextureSample(UVs=@n18.UVs) {texture=&T_UI_ShinePattern}.RGB)
   Alpha=StaticSwitchParameter(
    True=TextureSampleParameter2D(UVs=@n19.ScaleX) {parameterName=WideTile, texture=&T_UI_FrameWide}.R
    False=TextureSampleParameter2D(UVs=@n19.ScaleY) {parameterName=NormalTile, texture=&T_UI_Frame}.R
   ) {parameterName=TileSize}
  )
  LinearInterpolate(
   A=LinearInterpolate(
    A=Constant3Vector() {constant=<rgba 0.1384 0.1812 0.3663 1>}
    B=Constant3Vector() {constant=<rgba 0.0185 0.0252 0.0704 1>}
    Alpha=TextureSample() {texture=&/T_UI_DualColor}.G
   )
   Alpha=TextureSample() {texture=&/T_UI_ShineBars}.R
  )
 ]}
