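# Material function for RareDelight shop tile theming: when static switch bls_RareDelight is on it outputs a purple background, otherwise it passes its input through.
# An eased linear gradient and an exponential radial gradient are blended by a rerouted mask.
# A "triangles" mask builds two animated layers of randomly scaled, spinning cells on rotating radial UVs, faded by screen-space gradient.
# Its shapes come from MF_UI_Box SDFs, as the triangle-function calls are unconnected.
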MaterialFunction /OfferCatalog/Art/RareDelight/MF_UI_RareDelight_ShopTiles
graph {nodes [
  n1 = Constant3Vector() {constant=<rgba 0.4092 0.1541 0.8455 1>}
  n2 = NamedRerouteUsage()
  n3 = Constant() {r=0.5}
  n4 = Constant() {r=0.25}
  n5 = LinearInterpolate(
   A=Constant3Vector() {constant=<rgba 0.5841 0.3564 0.956 1>}
   B=@n1
   Alpha=MaterialFunctionCall("Value (0-1)"=MaterialFunctionCall(UVs=@n2, GradientStart=@n3, RotationAngle=@n4) {materialFunction=&MF_UI_LinearGradient}.Result) {materialFunction=&MF_UI_EaseCurves}.EaseOutSine
  )
  n6 = NamedRerouteUsage()
  n7 = Constant() {r=1}
  n8 = Constant()
  n9 = Constant() {r=0.17}
  n10 = LinearInterpolate(
   A=@n1
   B=Constant3Vector() {constant=<rgba 0.7726 0.3229 0.9911 1>}
   Alpha=MaterialFunctionCall(
    UVs=MaterialFunctionCall!(UVs=@n6, Scale=@n7, X Position=@n8, Y Position=@n9) {materialFunction=/OfferCatalog/Art/_MaterialFunctions_ShopTheming/MF_UI_ScalePosition}.Result
    Radius=Constant() {r=1}
    Density=Constant() {r=3}
   ) {materialFunction=&RadialGradientExponential}.RadialGradientExponential
  )
  n11 = NamedRerouteUsage()
  FunctionOutput(None=StaticSwitchParameter(True=LinearInterpolate(A=@n5, B=@n10, Alpha=@n11), False=FunctionInput()) {parameterName=bls_RareDelight})
  LinearInterpolate(
   A=FunctionInput(Preview=VectorParameter() {group=CorruptSeries, parameterName=BG_Color, defaultValue=<rgba 1 1 1 1>}.RGB) {bUsePreviewValueAsDefault=true, inputName=RGB 1, sortPriority=1}
   B=FunctionInput(Preview=VectorParameter() {group=CorruptSeries, parameterName=CellColor, defaultValue=<rgba 0 0 0 1>}.RGB) {bUsePreviewValueAsDefault=true, inputName=RGB 2, sortPriority=1}
  )
  n12 = ScalarParameter() {parameterName=X-Offset}
  n13 = ScalarParameter() {parameterName=Y-Offset}
  n14 = Rotator(Coordinate=Add(A=NamedRerouteUsage(), B=AppendVector(A=@n12, B=@n13))) {speed=0.069}
  n15 = Subtract(A=@n14) {constB=0.5}
  n16 = Multiply(A=@n15) {constB=-1}
  n17 = Ceil(None=ComponentMask(None=Add(A=@n14) {constB=-0.5}) {g=true})
  n18 = NamedRerouteDeclaration(None=LinearInterpolate(
    A=MaterialFunctionCall(Vector or UVs=@n15) {materialFunction=&VectorToRadialValue}.Radial Coordinates
    B=MaterialFunctionCall(Vector or UVs=@n16) {materialFunction=&VectorToRadialValue}.Radial Coordinates
    Alpha=Saturate(None=@n17)
   )) {name=UVMask, nodeColor=<rgba 0 0 0 1>}
  n19 = ComponentMask(None=@n18) {r=true}
  n20 = ComponentMask(None=@n18) {g=true}
  n21 = Multiply(
   A=MaterialFunctionCall(X=Divide(A=@n19, B=@n20), Y=@n19) {materialFunction=&MakeFloat2}.Result
   B=Constant2Vector() {g=2, r=0.25}
  ) {constB=3}
  n22 = Panner(Coordinate=@n21, Time=Multiply(A=Time(), B=ScalarParameter() {defaultValue=0.55, parameterName=Time-Multiply-Triangles-b})) {speedX=0.069, speedY=0.013}
  n23 = Constant() {r=-0.125}
  n24 = Multiply(A=Add(A=Rotator(Coordinate=@n22, Time=@n23) {speed=6.2832}, B=Constant2Vector() {g=0.125, r=0.125})) {constB=10}
  n25 = Frac(None=@n24)
  n26 = MaterialFunctionCall!(
   PrimaryTextureCoordinates=@n24
   2ndOrderTextureCoordinates
   RandomSeed=Constant() {r=98765}
   NoiseDensityBias
   Bypass Noise Influence Crosstalk
  ) {materialFunction=/Game/Effects/Fort_Effects/Materials/_Functions/MF_RandomUVSample}
  n27 = Multiply(A=LinearInterpolate(Alpha=@n26.NoiseWithRange) {constA=-0.25, constB=0.25}, B=Time())
  n28 = MaterialFunctionCall(
   UVs=Add(A=Rotator(Coordinate=@n25, Time=@n27) {speed=6.2832}, B=LinearInterpolate(Alpha=@n26.NoiseWithRange) {constA=-0.35, constB=0})
   Texture Scale=LinearInterpolate(Alpha=@n26.Noise) {constA=0.125, constB=0.5}
  ) {materialFunction=&ScaleUVsByCenter}
  MaterialFunctionCall!(UVs=@n28.UVs, Corner 01, Corner 02, Corner 03, Glow Min, Glow Max, Outline Thickness) {materialFunction=/FortUILibrary/MaterialFunctions/Shapes/MF_UI_WIP_Triangle}
  n29 = Time()
  n30 = ScalarParameter() {defaultValue=0.35, parameterName=Time-Multiply-Triangles-a}
  n31 = Multiply(A=Panner(Coordinate=@n21, Time=Multiply(A=@n29, B=@n30)) {speedX=0.038, speedY=0.013})
  n32 = Constant() {r=0.125}
  n33 = Multiply(A=Add(A=Rotator(Coordinate=@n31, Time=@n32) {speed=6.2832}, B=Constant2Vector() {g=0.25, r=0.35}) {constB=0.25}) {constB=5}
  n34 = Frac(None=@n33)
  n35 = MaterialFunctionCall(PrimaryTextureCoordinates=@n33, RandomSeed=Constant() {r=1234}) {materialFunction=&MF_RandomUVSample}
  n36 = Multiply(A=LinearInterpolate(Alpha=@n35.NoiseWithRange) {constA=-0.125, constB=0.125}, B=Time())
  n37 = MaterialFunctionCall(
   UVs=Add(A=Rotator(Coordinate=@n34, Time=@n36) {speed=6.2832}, B=LinearInterpolate(Alpha=@n35.NoiseWithRange) {constA=-0.125, constB=0.125})
   Texture Scale=LinearInterpolate(Alpha=@n35.Noise) {constA=0.125, constB=0.5}
  ) {materialFunction=&ScaleUVsByCenter}
  MaterialFunctionCall(UVs=@n37.UVs) {materialFunction=&MF_UI_WIP_Triangle}
  n38 = MaterialFunctionCall(Float2=MaterialFunctionCall() {materialFunction=&ScreenResolution}.Visible Resolution) {materialFunction=&BreakOutFloat2Components}
  Multiply(
   A=MaterialFunctionCall() {materialFunction=&ScreenAlignedUVs}."X 100%, Y 100%"
   B=AppendVector(A=Divide(A=@n38.R, B=@n38.R), B=Constant() {r=1})
  )
  NamedRerouteDeclaration(None=MaterialFunctionCall() {materialFunction=&MF_UI_UVScaling}.Scaled to Fit) {name=UVFit, nodeColor=<rgba 1 0 0.8203 1>}
  n39 = Multiply(A=NamedRerouteUsage(), B=NamedRerouteUsage())
  n40 = Power(
   Base=Multiply(A=OneMinus(None=@n39), B=ScalarParameter() {defaultValue=1.75, parameterName=Mask-Intensity}) {constB=4}
   Exp=ScalarParameter() {defaultValue=6, parameterName=Mask-Power}
  )
  n41 = OneMinus(None=MaterialFunctionCall() {materialFunction=&ScreenAlignedUVs}."X 100%, Y 100%")
  n42 = Saturate(None=Subtract(A=ComponentMask(None=@n41) {g=true}, B=ScalarParameter() {defaultValue=0.25, parameterName=Gradient Offset}))
  n43 = ComponentMask(None=NamedRerouteUsage()) {g=true}
  n44 = Multiply(A=@n35.NoiseWithRange, B=Saturate(None=OneMinus(None=@n43)))
  n45 = LinearInterpolate(Alpha=@n43) {constA=0.02, constB=0.16}
  n46 = MaterialFunctionCall(UVs=@n37.UVs) {materialFunction=&MF_UI_Box}
  n47 = OneMinus(None=Clamp(None=@n46.SDFOutline))
  n48 = Add(
   A=LinearInterpolate(Alpha=SmoothStep(Max=@n45, Value=@n46.SDFOutline) {constMax=0.02}) {constA=1, constB=0}
   B=Multiply(A=Subtract(A=@n47) {constB=0.85}) {constB=0.5}
  )
  n49 = ComponentMask(None=NamedRerouteUsage()) {g=true}
  n50 = Multiply(A=@n26.NoiseWithRange, B=Saturate(None=OneMinus(None=@n49)))
  n51 = LinearInterpolate(Alpha=@n49) {constA=0.02, constB=0.16}
  n52 = MaterialFunctionCall(UVs=@n28.UVs) {materialFunction=&MF_UI_Box}
  n53 = OneMinus(None=Clamp(None=@n52.SDFOutline))
  n54 = Add(
   A=LinearInterpolate(Alpha=SmoothStep(Max=@n51, Value=@n52.SDFOutline) {constMax=0.02}) {constA=1, constB=0}
   B=Multiply(A=Subtract(A=@n53) {constB=0.85}) {constB=0.5}
  )
  n55 = Max(A=Multiply(A=Saturate(None=@n44), B=Saturate(None=@n48)), B=Multiply(A=Saturate(None=@n50), B=Saturate(None=@n54)))
  NamedRerouteDeclaration(None=Multiply(A=Saturate(None=@n40), B=Multiply(A=@n42, B=@n55))) {name=triangles, nodeColor=<rgba 1 0 0.9609 1>}
 ]}
