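# Material function for Item Shop refeature tiles that renders an animated texture pattern over gradient colours, switchable via blsRefeatureSeries.
# A particle grid scatters a scrolling, rotating texture with randomised cells, using screen-space UVs when Uniform Across Tiles is on.
# Radial gradients supply background and pattern colours, plus an optional delayed sine-band sheen.
# The Cells Fade switch applies time-based per-cell fading only when false, despite its name.
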MaterialFunction /OfferCatalog/Art/Refeatures/MF_UI_Refeature_ShopTile
graph {
 properties {description=Texture Vibrancy}
 nodes [
  n1 = NamedRerouteUsage()
  n2 = LinearInterpolate(
   A=NamedRerouteUsage()
   B=VectorParameter() {parameterName=Sheen Color, defaultValue=<rgba 1 1 1 1>}.RGB
   Alpha=StaticSwitchParameter(True=NamedRerouteUsage(), False=Constant()) {defaultValue=true, parameterName=Sheen}
  )
  n3 = Multiply(A=NamedRerouteUsage(), B=ScalarParameter() {defaultValue=0.55, parameterName=Global Pattern Opacity})
  FunctionOutput(None=StaticSwitchParameter(True=LinearInterpolate(A=@n1, B=@n2, Alpha=@n3), False=FunctionInput()) {parameterName=blsRefeatureSeries})
  LinearInterpolate(
   A=FunctionInput(Preview=VectorParameter() {group=CorruptSeries, parameterName=BG_Color, defaultValue=<rgba 1 1 1 1>}.RGB) {bUsePreviewValueAsDefault=true, inputName=RGB 1, sortPriority=1}
   B=FunctionInput(Preview=VectorParameter() {group=CorruptSeries, parameterName=CellColor, defaultValue=<rgba 0 0 0 1>}.RGB) {bUsePreviewValueAsDefault=true, inputName=RGB 2, sortPriority=1}
  )
  n4 = MaterialFunctionCall!(ScreenUV, TextureSize, ScalePivotPoint) {materialFunction=/FortUILibrary/MaterialFunctions/Basics/MF_UI_ScreenSpaceUVScaling}
  n5 = ScalarParameter() {defaultValue=0.5, parameterName=Pattern Volume}
  n6 = MaterialFunctionCall() {materialFunction=&MF_UI_UVScaling}
  n7 = NamedRerouteUsage()
  n8 = Time()
  n9 = Frac(None=Multiply(A=@n8) {constB=0.015})
  n10 = Multiply(A=@n8, B=ScalarParameter() {defaultValue=0.015, parameterName=Rotation Speed})
  n11 = MaterialFunctionCall!(
   UV=StaticSwitchParameter(True=Multiply(A=@n4.Scaled to Fit, B=@n5), False=Multiply(A=@n6.Scaled To Fill, B=@n5)) {parameterName=Uniform Across Tiles}
   Global Offset=AppendVector(A=Add(A=@n7, B=@n9), B=Constant())
   Global Rotation=ScalarParameter() {defaultValue=0.35, parameterName=Global Rotation}
   Tiling=ScalarParameter() {defaultValue=5, parameterName=Tiling}
   Cell Offset=Constant()
   Cell Scale=LinearInterpolate(
    A=ScalarParameter() {defaultValue=0.5, parameterName=Minimum Texture Size}
    B=ScalarParameter() {defaultValue=1, parameterName=Maximum Texture Size}
    Alpha=NamedRerouteUsage()
   ) {constA=0.5}
   Cell Rotation=Add(A=NamedRerouteUsage(), B=Frac(None=@n10))
   Random Seed
  ) {materialFunction=/FortUILibrary/MaterialFunctions/Basics/MF_UI_ParticleGrid}
  NamedRerouteDeclaration(None=@n11.Random Rows) {name=Random Rows, nodeColor=<rgba 0.7344 0 1 1>}
  NamedRerouteDeclaration(None=@n11.Random Cell Values) {name=Random Cell Values, nodeColor=<rgba 0.7344 0 1 1>}
  n12 = Multiply(A=MaterialFunctionCall() {materialFunction=&GetUserInterfaceUV}.Normalized UV) {constB=0.5}
  n13 = Constant() {r=0.5}
  n14 = Constant() {r=0.35}
  n15 = Constant() {r=1}
  n16 = Constant() {r=3}
  n17 = Subtract(A=Frac(None=MaterialFunctionCall!(Animation Duration=@n15, Delay=@n16, Reset to Default) {materialFunction=/FortUILibrary/MaterialFunctions/Time/MF_UI_AnimationDelay}.Result)) {constB=0.125}
  n18 = Constant()
  n19 = Sine(None=Add(
    A=MaterialFunctionCall(UVs=@n12, Rotation Center=@n13, "Rotation Angle (0-1)"=@n14) {materialFunction=&CustomRotator}.Rotated Values
    B=AppendVector(A=@n17, B=@n18)
   ))
  n20 = ComponentMask(None=Multiply(A=Add(A=@n19)) {constB=0.5}) {r=true}
  n21 = Constant() {r=0.9}
  MaterialExpressionRemap_0 = MaterialXRemap!(None=Saturate(None=Subtract(A=@n20, B=@n21) {constB=0.75}), InputLow, InputHigh=OneMinus(None=@n21), TargetLow, TargetHigh) {inputHighDefault=0.25}
  NamedRerouteDeclaration(None=Multiply(A=@MaterialExpressionRemap_0) {constB=0.75}) {name=Sheen Mask, nodeColor=<rgba 0 0 0 1>}
  NamedRerouteDeclaration(None=@n11.Random Columns) {name=Columns, nodeColor=<rgba 1 0.4688 0 1>}
  n22 = MaterialFunctionCall() {materialFunction=&MF_UI_UVScaling}
  n23 = Constant()
  n24 = Constant() {r=0.5}
  NamedRerouteDeclaration(None=LinearInterpolate(
    A=VectorParameter() {parameterName=Pattern Outer Gradient, defaultValue=<rgba 0 0.0832 0.5226 1>}.RGB
    B=VectorParameter() {parameterName=Pattern Inner Gradient, defaultValue=<rgba 0 0.3689 0.776 1>}.RGB
    Alpha=MaterialFunctionCall(UV=@n22.Scaled to Fit, GradientStart=@n23, GradientEnd=@n24) {materialFunction=&MF_UI_RadialGradient}.Smooth
   )) {name=Pattern Gradient, nodeColor=<rgba 0 0 0 1>}
  n25 = MaterialFunctionCall(
   UVs=MaterialFunctionCall() {materialFunction=&GetUserInterfaceUV}.9-Slice UV
   Texture Scale=Constant() {r=1.5}
  ) {materialFunction=&ScaleUVsByCenter}
  n26 = Constant()
  n27 = Constant() {r=1.2}
  NamedRerouteDeclaration(None=LinearInterpolate(
    A=VectorParameter() {parameterName=Background Outer Gradient, defaultValue=<rgba 0 0.2754 0.5851 1>}.RGB
    B=VectorParameter() {parameterName=Background Inner Gradient, defaultValue=<rgba 0.2324 0.5303 0.842 1>}.RGB
    Alpha=MaterialFunctionCall(UV=@n25.UVs, GradientStart=@n26, GradientEnd=@n27) {materialFunction=&MF_UI_RadialGradient}.Smooth
   )) {name=Background Gradient, nodeColor=<rgba 0 0 0 1>}
  n28 = TextureSampleParameter2D(UVs=@n11.Cells UV) {
   desc="Strive for a 64x64 texture here, but 128x128 is permissible."
   parameterName=Texture
   samplerType=SAMPLERTYPE_Alpha
   texture=/Game/UI/Frontend/ItemShop/Art/Lucky/T_UI_LuckyClover-01
  }
  n29 = NamedRerouteUsage()
  n30 = Frac(None=Multiply(A=Time()) {constB=0.15})
  n31 = Multiply(A=Min(A=@n30, B=OneMinus(None=@n30))) {constB=2}
  n32 = NamedRerouteUsage()
  n33 = LinearInterpolate(Alpha=Max(A=Multiply(A=@n29, B=@n31), B=Round(None=@n32))) {constA=0.15}
  NamedRerouteDeclaration(None=StaticSwitchParameter(True=@n28.G, False=Multiply(A=@n28.G, B=@n33)) {parameterName=Cells Fade}) {name=Pattern Mask, nodeColor=<rgba 1 0 0.9141 1>}
 ]
}
