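# Material function for the Rufus series UI shop tile, producing a rarity-colored background when the blsRufus_Series static switch is on, otherwise passing its input through.
# Green, blue, purple and brown backgrounds blend constant color pairs through offset exponential radial gradients, which reuse the same parameter names.
# Chained lerps on the Rarity vector's channels select among rerouted colors, though the red-channel lerp has identical inputs.
# A separate BG/Cell color lerp feeds nothing.
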MaterialFunction /OfferCatalog/Art/Rufus/MF_UI_Rufus_ShopTile
graph {nodes [
  n1 = NamedRerouteUsage()
  n2 = NamedRerouteUsage()
  n3 = VectorParameter() {group=Rarity, parameterName="Rarity [set to 0 for color]", defaultValue=<rgba 1 1 1 1>}
  n4 = LinearInterpolate(
   A=NamedRerouteUsage()
   B=LinearInterpolate(A=LinearInterpolate(A=@n2, B=@n2, Alpha=@n3.R) {constAlpha=1}, B=NamedRerouteUsage(), Alpha=@n3.G)
   Alpha=@n3.B
  )
  FunctionOutput(None=StaticSwitchParameter(True=LinearInterpolate(A=@n1, B=@n4, Alpha=@n3.A), False=FunctionInput()) {parameterName=blsRufus_Series})
  LinearInterpolate(
   A=FunctionInput(Preview=VectorParameter() {group=CorruptSeries, parameterName=BG_Color, defaultValue=<rgba 1 1 1 1>}.RGB) {bUsePreviewValueAsDefault=true, inputName=RGB 1, sortPriority=1}
   B=FunctionInput(Preview=VectorParameter() {group=CorruptSeries, parameterName=CellColor, defaultValue=<rgba 0 0 0 1>}.RGB) {bUsePreviewValueAsDefault=true, inputName=RGB 2, sortPriority=1}
  )
  n5 = MaterialFunctionCall() {materialFunction=&MF_UI_UVScaling}
  NamedRerouteDeclaration(None=@n5.Scaled to Fit) {name=UVFit, nodeColor=<rgba 0 0.6719 1 1>}
  NamedRerouteDeclaration(None=@n5.Scaled To Fill) {name=UVFill, nodeColor=<rgba 1 0.6562 0 1>}
  n6 = ScalarParameter() {group=Rufus Gradient, parameterName=X Location}
  n7 = ScalarParameter() {defaultValue=0.1, group=Rufus Gradient, parameterName=Y Location}
  n8 = MaterialFunctionCall(
   UVs=Add(A=NamedRerouteUsage(), B=AppendVector(A=@n6, B=@n7))
   Radius=ScalarParameter() {defaultValue=0.55, group=Rufus Gradient, parameterName=Gradient Radius}
   Density=ScalarParameter() {defaultValue=2.8, group=Rufus Gradient, parameterName=Gradient Density}
  ) {materialFunction=&RadialGradientExponential}
  NamedRerouteDeclaration(None=LinearInterpolate(
    A=Constant3Vector() {constant=<rgba 0.0148 0.1354 0.0188 1>}
    B=Constant3Vector() {constant=<rgba 0.0902 0.3385 0.0247 1>}
    Alpha=Multiply(A=@n8.RadialGradientExponential) {constB=1.158}
   )) {name=Green, nodeColor=<rgba 0.0547 1 0 1>}
  n9 = ScalarParameter() {group=Rufus Gradient, parameterName=X Location}
  n10 = ScalarParameter() {defaultValue=0.15, group=Rufus Gradient, parameterName=Y Location}
  n11 = MaterialFunctionCall(
   UVs=Add(A=NamedRerouteUsage(), B=AppendVector(A=@n9, B=@n10))
   Radius=ScalarParameter() {defaultValue=0.55, group=Rufus Gradient, parameterName=Gradient Radius}
   Density=ScalarParameter() {defaultValue=2, group=Rufus Gradient, parameterName=Gradient Density}
  ) {materialFunction=&RadialGradientExponential}
  NamedRerouteDeclaration(None=LinearInterpolate(
    A=Constant3Vector() {constant=<rgba 0.001 0.0744 0.4271 1>}
    B=Constant3Vector() {constant=<rgba 0 0.1922 0.7157 1>}
    Alpha=Multiply(A=@n11.RadialGradientExponential) {constB=2.5}
   )) {name=Blue, nodeColor=<rgba 0 0.377 1 1>}
  n12 = ScalarParameter() {group=Rufus Gradient, parameterName=X Location}
  n13 = ScalarParameter() {defaultValue=0.15, group=Rufus Gradient, parameterName=Y Location}
  n14 = MaterialFunctionCall(
   UVs=Add(A=NamedRerouteUsage(), B=AppendVector(A=@n12, B=@n13))
   Radius=ScalarParameter() {defaultValue=0.45, group=Rufus Gradient, parameterName=Gradient Radius}
   Density=ScalarParameter() {defaultValue=3.5, group=Rufus Gradient, parameterName=Gradient Density}
  ) {materialFunction=&RadialGradientExponential}
  NamedRerouteDeclaration(None=LinearInterpolate(
    A=Constant3Vector() {constant=<rgba 0.0865 0.0185 0.2582 1>}
    B=Constant3Vector() {constant=<rgba 0.305 0.0595 0.7011 1>}
    Alpha=Multiply(A=@n14.RadialGradientExponential)
   )) {name=Purple, nodeColor=<rgba 0.7578 0 1 1>}
  n15 = ScalarParameter() {group=Rufus Gradient, parameterName=X Location}
  n16 = ScalarParameter() {defaultValue=0.15, group=Rufus Gradient, parameterName=Y Location}
  n17 = MaterialFunctionCall(
   UVs=Add(A=NamedRerouteUsage(), B=AppendVector(A=@n15, B=@n16))
   Radius=ScalarParameter() {defaultValue=0.45, group=Rufus Gradient, parameterName=Gradient Radius}
   Density=ScalarParameter() {defaultValue=1.75, group=Rufus Gradient, parameterName=Gradient Density}
  ) {materialFunction=&RadialGradientExponential}
  NamedRerouteDeclaration(None=LinearInterpolate(
    A=Constant3Vector() {constant=<rgba 0.2502 0.0578 0.0168 1>}
    B=Constant3Vector() {constant=<rgba 0.5781 0.1396 0.0271 1>}
    Alpha=Multiply(A=@n17.RadialGradientExponential) {constB=1.75}
   )) {name=Brown, nodeColor=<rgba 0.3333 0.1476 0 1>}
 ]}
