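# UI material function for the offer catalog's Sahara series look, applied when static switch blsSahara_Series is on; otherwise it outputs its input.
# Flow-mapped noise and clouds, shaped by horizontal and vertical edge masks, form an animated orange-to-gold background.
# Two additive lightning layers flash on separate Frac/Sine time loops, with UVs warped by panning clouds.
# Lightning gradient textures, subtracted from each pulse, drive gold cores fed through named reroutes.
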MaterialFunction /OfferCatalog/Art/Sahara/MF_UI_SaharaSeries
graph {nodes [
  Add(B=MaterialFunctionCall(X=ScalarParameter() {defaultValue=1.5, parameterName=arrowX}, Y=ScalarParameter() {defaultValue=0.04, parameterName=ArrowY}) {materialFunction=&MakeFloat2}.Result)
  n1 = Frac(None=Multiply(A=Time(), B=Constant() {r=0.4}))
  n2 = Sine(None=@n1)
  n3 = Multiply(A=Multiply(A=ConstantBiasScale(None=@n1) {bias=-0.5, scale=2}, B=Multiply(A=@n2, B=@n2))) {constB=1.01}
  n4 = MaterialFunctionCall() {materialFunction=&MF_UI_UVScaling}
  n5 = Subtract(A=TextureSample(UVs=Panner(Coordinate=@n4.Scaled To Fill) {speedX=0.1, speedY=-0.1}) {texture=/Game/Effects/Fort_Effects/Textures/Smoke/T_Clouds_01}.R) {constB=0.75}
  n6 = Constant() {r=-0.3914}
  n7 = Constant() {r=-0.316}
  n8 = Add(
   A=LinearInterpolate(A=@n4.Scaled To Fill, B=Add(A=@n4.Scaled To Fill, B=@n5), Alpha=Constant() {r=0.1})
   B=Add(
    A=Multiply(A=@n4.Scaled To Fill) {constB=0.2952}
    B=MaterialFunctionCall(X=@n6, Y=@n7) {materialFunction=&MakeFloat2}.Result
   )
  )
  n9 = TextureSample(UVs=@n8) {samplerSource=SSM_Clamp_WorldGroupSettings, samplerType=SAMPLERTYPE_Alpha, texture=&/T_UI_LightningLine}
  n10 = Multiply(A=SmoothStep(Value=@n9.R) {constMax=0.9271, constMin=0.28}, B=Constant() {r=9})
  n11 = Multiply(A=SmoothStep(Value=@n9.R) {constMax=1.5}, B=Constant() {r=9.8812})
  n12 = Multiply(A=Constant3Vector() {constant=<rgba 1 0.7389 0.1823 0>}, B=Multiply(A=Saturate(None=@n3), B=Add(A=@n10, B=@n11)))
  n13 = Frac(None=Multiply(A=Time(), B=Constant() {r=0.3}))
  n14 = Sine(None=@n13)
  n15 = Multiply(A=Multiply(A=ConstantBiasScale(None=@n13) {bias=-0.5, scale=2}, B=Multiply(A=@n14, B=@n14))) {constB=0.8}
  n16 = Saturate(None=@n15)
  n17 = MaterialFunctionCall() {materialFunction=&MF_UI_UVScaling}
  n18 = Subtract(A=TextureSample(UVs=Panner(Coordinate=@n17.Scaled To Fill) {speedX=0.1, speedY=-0.1}) {texture=&T_Clouds_01}.R) {constB=0.7}
  n19 = Constant() {r=0.184}
  n20 = Constant() {r=0.1}
  n21 = Add(
   A=LinearInterpolate(A=@n17.Scaled To Fill, B=Add(A=@n17.Scaled To Fill, B=@n18), Alpha=Constant() {r=0.15})
   B=Add(
    A=Multiply(A=@n17.Scaled To Fill) {constB=1.5}
    B=MaterialFunctionCall(X=@n19, Y=@n20) {materialFunction=&MakeFloat2}.Result
   )
  )
  n22 = TextureSample(UVs=@n21) {samplerSource=SSM_Clamp_WorldGroupSettings, samplerType=SAMPLERTYPE_Alpha, texture=&T_UI_LightningLine}
  n23 = Add(
   A=Multiply(A=SmoothStep(Value=@n22.R) {constMax=0.9, constMin=0.14}, B=Constant() {r=7})
   B=Multiply(A=SmoothStep() {constMax=1.5}, B=Constant() {r=4.2801})
  )
  n24 = NamedRerouteUsage()
  n25 = NamedRerouteUsage()
  n26 = TextureObject() {samplerType=SAMPLERTYPE_Grayscale, texture=/Game/UI/Frontend/ItemShop/Art/Rumble/LowResBlurredNoise_Rumble}
  n27 = MaterialFunctionCall() {materialFunction=&MF_UI_UVScaling}
  n28 = Panner(Coordinate=@n27.Scaled To Fill) {speedY=0.1312}
  n29 = TextureObject() {samplerType=SAMPLERTYPE_LinearColor, texture=/Game/UI/Frontend/ItemShop/Art/WaterBackground/T_UI_WaterFlowMap}
  n30 = Constant() {r=0.3536}
  n31 = TextureObject() {texture=&/T_UI_Clouds}
  n32 = Panner(Coordinate=@n27.Scaled To Fill) {speedX=-0.106, speedY=-0.17}
  n33 = TextureObject() {samplerType=SAMPLERTYPE_Masks, texture=/Game/Packages/DS_Fortnite_Terrain_NoLOD/Textures/FlowMap1}
  n34 = Constant() {r=0.7928}
  n35 = MaterialFunctionCall() {materialFunction=&MF_UI_UVScaling}
  n36 = ConstantBiasScale(None=@n35.Scaled To Fill) {bias=-0.5, scale=2}
  n37 = Subtract(A=ComponentMask(None=Abs(None=@n36)) {g=true}, B=ScalarParameter() {defaultValue=0.0511, parameterName=Subtract})
  n38 = ScalarParameter() {defaultValue=1.0721, parameterName=Falloff}
  n39 = Multiply(A=Saturate(None=Multiply(A=@n37, B=@n38)), B=ScalarParameter() {defaultValue=1.085, parameterName=Intensity})
  n40 = ConstantBiasScale(None=@n35.Scaled To Fill) {bias=-0.5, scale=2}
  n41 = Subtract(A=ComponentMask(None=Abs(None=@n40)) {r=true}, B=ScalarParameter() {defaultValue=0.302, parameterName=SubtractH})
  n42 = ScalarParameter() {defaultValue=0.8595, parameterName=FalloffH}
  n43 = Multiply(A=Saturate(None=Multiply(A=@n41, B=@n42)), B=ScalarParameter() {defaultValue=2.3306, parameterName=IntensityH})
  n44 = Multiply(
   A=LinearInterpolate(
    A=MaterialFunctionCall!(Texture 2D=@n26, Texture UVs=@n28, FlowMap Texture 2D=@n29, FlowMap UVs, FlowIntensity=@n30, FlowSpeed) {materialFunction=/FortUILibrary/MaterialFunctions/Effects/MF_UI_FlowMap}.Result
    Alpha=MaterialFunctionCall(Texture 2D=@n31, Texture UVs=@n32, FlowMap Texture 2D=@n33, FlowIntensity=@n34) {materialFunction=&MF_UI_FlowMap}.Result
   )
   B=OneMinus(None=LinearInterpolate(A=@n39, B=@n43))
  )
  n45 = LinearInterpolate(A=Constant3Vector() {constant=<rgba 1 0.1238 0.0142 0>}, B=Constant3Vector() {constant=<rgba 0.9301 0.6308 0.0685 0>}, Alpha=@n44)
  n46 = Constant3Vector()
  n47 = OneMinus(None=@n44)
  n48 = Add(
   A=Multiply(A=Constant3Vector() {constant=<rgba 1 0.7788 0.3073 0>}, B=Multiply(A=@n16, B=@n23))
   B=Add(A=Add(A=@n24, B=@n25), B=LinearInterpolate(A=@n45, B=@n46, Alpha=@n47) {constAlpha=0})
  )
  FunctionOutput(None=StaticSwitchParameter(True=Add(A=@n12, B=@n48), False=FunctionInput()) {parameterName=blsSahara_Series})
  LinearInterpolate(
   A=FunctionInput(Preview=VectorParameter() {group=CorruptSeries, parameterName=BG_Color, defaultValue=<rgba 1 1 1 1>}.RGB) {bUsePreviewValueAsDefault=true, inputName=RGB 1, sortPriority=1}
   B=FunctionInput(Preview=VectorParameter() {group=CorruptSeries, parameterName=CellColor, defaultValue=<rgba 0 0 0 1>}.RGB) {bUsePreviewValueAsDefault=true, inputName=RGB 2, sortPriority=1}
  )
  n49 = Constant3Vector()
  n50 = Constant3Vector() {constant=<rgba 0.9301 0.6308 0.0685 0>}
  n51 = Saturate(None=Subtract(A=@n15, B=TextureSample(UVs=@n21) {samplerSource=SSM_Clamp_WorldGroupSettings, texture=&/T_UI_LightningGradient}.R))
  NamedRerouteDeclaration(None=LinearInterpolate(B=LinearInterpolate(A=@n49, B=@n50, Alpha=@n51), Alpha=SmoothStep(Value=@n22.R) {constMax=0.285, constMin=0.285})) {name=Lightning_2, nodeColor=<rgba 0.1016 1 0 1>}
  n52 = Constant3Vector()
  n53 = Constant3Vector() {constant=<rgba 0.9301 0.6308 0.0685 0>}
  n54 = Saturate(None=Subtract(A=@n3, B=TextureSample(UVs=@n8) {samplerSource=SSM_Clamp_WorldGroupSettings, texture=&T_UI_LightningGradient}.R))
  NamedRerouteDeclaration(None=LinearInterpolate(B=LinearInterpolate(A=@n52, B=@n53, Alpha=@n54), Alpha=SmoothStep(Value=@n9.R) {constMax=0.3, constMin=0.3})) {name=Lightning_3, nodeColor=<rgba 0.1016 1 0 1>}
  NamedRerouteUsage()
 ]}
