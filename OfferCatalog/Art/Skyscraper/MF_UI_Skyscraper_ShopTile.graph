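# Material function that supplies a Skyscraper-themed background for UI shop tiles, toggled by the bls_SkyScraper static switch.
# When on, it builds a rotated, smoothstepped dark grey gradient on fill-scaled UVs, then blends toward the lighter grey using a grunge texture tiled in screen-space UVs.
# When off, it passes its input through.
# A leftover lerp of CorruptSeries BG_Color and CellColor inputs does not feed the output.
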MaterialFunction /OfferCatalog/Art/Skyscraper/MF_UI_Skyscraper_ShopTile
graph {nodes [
  n1 = Constant3Vector() {constant=<rgba 0.1981 0.1981 0.2159 1>}
  n2 = MaterialFunctionCall(TextureSize=Constant2Vector() {g=1080, r=1920}) {materialFunction=&MF_UI_UVScaling}
  n3 = Constant()
  n4 = Constant() {r=0.75}
  n5 = Constant() {r=0.3}
  n6 = LinearInterpolate(
   A=Constant3Vector() {constant=<rgba 0.0401 0.0401 0.0434 1>}
   B=@n1
   Alpha=SmoothStep(Value=MaterialFunctionCall(UVs=@n2.Scaled To Fill, GradientStart=@n3, GradientEnd=@n4, RotationAngle=@n5) {materialFunction=&MF_UI_LinearGradient}.Result) {constMax=1.75}
  )
  n7 = MaterialFunctionCall!(ScreenUV, TextureSize, ScalePivotPoint) {materialFunction=/FortUILibrary/MaterialFunctions/Basics/MF_UI_ScreenSpaceUVScaling}
  n8 = SmoothStep(Value=TextureSample(UVs=Multiply(A=@n7.Scaled to Fit) {constB=2}) {samplerType=SAMPLERTYPE_Normal, texture=&/T_UI_SkyscraperGrungeTile}.R) {constMin=-1.7}
  FunctionOutput(None=StaticSwitchParameter(True=LinearInterpolate(A=@n6, B=@n1, Alpha=@n8), False=FunctionInput()) {parameterName=bls_SkyScraper})
  LinearInterpolate(
   A=FunctionInput(Preview=VectorParameter() {group=CorruptSeries, parameterName=BG_Color, defaultValue=<rgba 1 1 1 1>}.RGB) {bUsePreviewValueAsDefault=true, inputName=RGB 1, sortPriority=1}
   B=FunctionInput(Preview=VectorParameter() {group=CorruptSeries, parameterName=CellColor, defaultValue=<rgba 0 0 0 1>}.RGB) {bUsePreviewValueAsDefault=true, inputName=RGB 2, sortPriority=1}
  )
 ]}
