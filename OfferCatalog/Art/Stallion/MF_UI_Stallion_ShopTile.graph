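# Material function for the Stallion series shop tile UI, which optionally replaces an incoming value with the Stallion background art.
# The bls_StallionSeries static switch outputs the T_UI_StallionBG texture, with UVs scaled to fill through MF_UI_UVScaling, or passes the function input through when off.
# A separate lerp between CorruptSeries background and cell color inputs is not connected to the output.
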MaterialFunction /OfferCatalog/Art/Stallion/MF_UI_Stallion_ShopTile
graph {nodes [
  n1 = MaterialFunctionCall() {materialFunction=&MF_UI_UVScaling}
  FunctionOutput(None=StaticSwitchParameter(True=TextureSample(UVs=@n1.Scaled To Fill) {texture=&/T_UI_StallionBG}.RGB, False=FunctionInput()) {parameterName=bls_StallionSeries})
  LinearInterpolate(
   A=FunctionInput(Preview=VectorParameter() {group=CorruptSeries, parameterName=BG_Color, defaultValue=<rgba 1 1 1 1>}.RGB) {bUsePreviewValueAsDefault=true, inputName=RGB 1, sortPriority=1}
   B=FunctionInput(Preview=VectorParameter() {group=CorruptSeries, parameterName=CellColor, defaultValue=<rgba 0 0 0 1>}.RGB) {bUsePreviewValueAsDefault=true, inputName=RGB 2, sortPriority=1}
  )
 ]}
