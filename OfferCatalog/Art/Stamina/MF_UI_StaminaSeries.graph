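# UI material function that optionally swaps incoming color for the Stamina series sky texture, apparently for Stamina offer art in the offer catalog.
# The bIsStaminaSeries static switch outputs either T_UI_Stamina_Sky RGB or a pass-through function input, resolved at compile time.
# Sky UVs come from the ScaleX output of MF_UI_UVScaling, then get scaled about their center by a constant.
# The RGB 1 and RGB 2 inputs appear unconnected.
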MaterialFunction /OfferCatalog/Art/Stamina/MF_UI_StaminaSeries
graph {nodes [
  FunctionInput() {inputName=RGB 1, sortPriority=1}
  FunctionInput() {inputName=RGB 2, sortPriority=1}
  n1 = MaterialFunctionCall(
   UVs=MaterialFunctionCall() {materialFunction=&MF_UI_UVScaling}.ScaleX
   Texture Scale=Constant() {r=2.3}
  ) {materialFunction=&ScaleUVsByCenter}
  FunctionOutput(None=StaticSwitchParameter(True=TextureSample(UVs=@n1.UVs) {texture=&/T_UI_Stamina_Sky}.RGB, False=FunctionInput()) {parameterName=bIsStaminaSeries})
 ]}
