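# Material function for Undefeated-series UI shop tiles that supplies a green background colour.
# A static switch, bls_UndefeatedSeries, outputs either a rotated linear gradient between two greens or the incoming function input.
# An outline border mask, built from pixel-size-corrected thickness and Step/Min tests on 9-slice UVs, is also present but does not reach the output.
# Leftover CorruptSeries colour inputs feed another unconnected blend.
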MaterialFunction /OfferCatalog/Art/Undefeated/MF_UI_Undefeated_ShopTile
graph {nodes [
  n1 = LinearInterpolate(
   A=Constant3Vector() {constant=<rgba 0.0235 0.4275 0 1>}
   B=Constant3Vector() {constant=<rgba 0.1922 0.8784 0.2118 1>}
   Alpha=MaterialFunctionCall(GradientStart=Constant(), GradientEnd=Constant() {r=1}, RotationAngle=Constant() {r=0.25}) {materialFunction=&MF_UI_LinearGradient}.Result
  )
  FunctionOutput(None=StaticSwitchParameter(True=@n1, False=FunctionInput()) {parameterName=bls_UndefeatedSeries})
  n2 = NamedRerouteUsage()
  n3 = Multiply(
   A=MaterialFunctionCall() {materialFunction=&GetUserInterfaceUV}.9-Slice UV
   B=Constant2Vector() {g=0.35, r=0.35}
  )
  n4 = ComponentMask(None=@n3) {r=true}
  n5 = Step(Y=@n2, X=@n4)
  n6 = Step(Y=@n2, X=OneMinus(None=@n4))
  n7 = NamedRerouteUsage()
  n8 = ComponentMask(None=@n3) {g=true}
  n9 = Step(Y=@n7, X=@n8)
  n10 = Step(Y=@n7, X=OneMinus(None=@n8))
  LinearInterpolate(A=Constant3Vector() {constant=<rgba 0.0704 0.491 0.0452 1>}, B=@n1, Alpha=Min(A=Min(A=@n5, B=@n6), B=Min(A=@n9, B=@n10)))
  LinearInterpolate(
   A=FunctionInput(Preview=VectorParameter() {group=CorruptSeries, parameterName=BG_Color, defaultValue=<rgba 1 1 1 1>}.RGB) {bUsePreviewValueAsDefault=true, inputName=RGB 1, sortPriority=1}
   B=FunctionInput(Preview=VectorParameter() {group=CorruptSeries, parameterName=CellColor, defaultValue=<rgba 0 0 0 1>}.RGB) {bUsePreviewValueAsDefault=true, inputName=RGB 2, sortPriority=1}
  )
  n11 = NamedRerouteUsage()
  n12 = MaterialFunctionCall() {materialFunction=&GetUserInterfaceUV}
  n13 = ComponentMask(None=@n12.9-Slice UV) {r=true}
  n14 = OneMinus(None=@n13)
  n15 = NamedRerouteUsage()
  n16 = ComponentMask(None=@n12.9-Slice UV) {g=true}
  n17 = OneMinus(None=@n16)
  Min(A=Min(A=Step(Y=@n11, X=@n13), B=Step(Y=@n11, X=@n14)), B=Min(A=Step(Y=@n15, X=@n16), B=Step(Y=@n15, X=@n17)))
  n18 = MaterialFunctionCall(Float2=VectorParameter() {parameterName=Native Pixel Size, defaultValue=<rgba 256 256 0 1>}.RGB) {materialFunction=&BreakOutFloat2Components}
  n19 = ScalarParameter() {defaultValue=0.03, parameterName=Outline Thickness}
  NamedRerouteDeclaration(None=Divide(A=Multiply(A=@n19, B=@n18.R), B=NamedRerouteUsage())) {name=Thickness X, nodeColor=<rgba 1 0 0.5102 1>}
  n20 = MaterialFunctionCall() {materialFunction=&GetUserInterfaceUV}
  NamedRerouteDeclaration(None=ComponentMask(None=@n20.Pixel Size) {r=true}) {name=PixelSizeX, nodeColor=<rgba 1 0.0389 0.1086 1>}
  NamedRerouteDeclaration(None=ComponentMask(None=@n20.Pixel Size) {g=true}) {name=PixelSize Y, nodeColor=<rgba 1 0.3984 0 1>}
  NamedRerouteDeclaration(None=Divide(A=Multiply(A=@n18.G, B=@n19), B=NamedRerouteUsage())) {name=Thickness Y, nodeColor=<rgba 1 0.8901 0 1>}
 ]}
