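# Material function that supplies the Winterfest series UI artwork color, or passes its input through when bIsWinterfestSeries is off.
# Nested static switches pick the Blue, Red or Green texture, with Red dimmed to 70%.
# Each texture repeats a UI pixel-size UV scale-and-center block, though as wired its ratios divide a component by itself.
# A lerp between BG_Color and CellColor inputs is left unconnected to the output.
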MaterialFunction /OfferCatalog/Art/Winterfest/WF_2021/MF_UI_Winterfest_Series
graph {nodes [
  n1 = MaterialFunctionCall() {materialFunction=&GetUserInterfaceUV}
  n2 = MaterialFunctionCall(Float2=@n1.Pixel Size) {materialFunction=&BreakOutFloat2Components}
  n3 = Divide(A=@n2.G, B=@n2.G)
  n4 = Multiply(A=@n1.Normalized UV, B=AppendVector(A=Constant() {r=1}, B=@n3))
  n5 = AppendVector(A=Constant(), B=Multiply(A=OneMinus(None=@n3)) {constB=0.5})
  n6 = MaterialFunctionCall(Float2=MaterialFunctionCall() {materialFunction=&GetUserInterfaceUV}.Pixel Size) {materialFunction=&BreakOutFloat2Components}
  n7 = Constant()
  n8 = Constant() {r=1}
  n9 = MaterialFunctionCall() {materialFunction=&GetUserInterfaceUV}
  n10 = MaterialFunctionCall(Float2=@n9.Pixel Size) {materialFunction=&BreakOutFloat2Components}
  n11 = Divide(A=@n10.R, B=@n10.R)
  n12 = Multiply(A=@n9.Normalized UV, B=AppendVector(A=@n11, B=Constant() {r=1}))
  n13 = AppendVector(A=Multiply(A=OneMinus(None=@n11)) {constB=0.5}, B=Constant())
  n14 = TextureSampleParameter2D(UVs=LinearInterpolate(A=Add(A=@n4, B=@n5), B=Add(A=@n12, B=@n13), Alpha=If(A=@n6.R, B=@n6.R, "A > B"=@n7, "A == B"=@n7, "A < B"=@n8))) {parameterName=Winterfest_Blue, texture=&/T_UI_WF_Blue}
  n15 = MaterialFunctionCall() {materialFunction=&GetUserInterfaceUV}
  n16 = MaterialFunctionCall(Float2=@n15.Pixel Size) {materialFunction=&BreakOutFloat2Components}
  n17 = Divide(A=@n16.G, B=@n16.G)
  n18 = AppendVector(A=Constant() {r=1}, B=@n17)
  n19 = Constant()
  n20 = Multiply(A=OneMinus(None=@n17)) {constB=0.5}
  n21 = MaterialFunctionCall(Float2=MaterialFunctionCall() {materialFunction=&GetUserInterfaceUV}.Pixel Size) {materialFunction=&BreakOutFloat2Components}
  n22 = Constant()
  n23 = MaterialFunctionCall() {materialFunction=&GetUserInterfaceUV}
  n24 = MaterialFunctionCall(Float2=@n23.Pixel Size) {materialFunction=&BreakOutFloat2Components}
  n25 = Divide(A=@n24.R, B=@n24.R)
  n26 = AppendVector(A=@n25, B=Constant() {r=1})
  n27 = Multiply(A=OneMinus(None=@n25)) {constB=0.5}
  n28 = Constant()
  n29 = LinearInterpolate(
   A=Add(A=Multiply(A=@n15.Normalized UV, B=@n18), B=AppendVector(A=@n19, B=@n20))
   B=Add(A=Multiply(A=@n23.Normalized UV, B=@n26), B=AppendVector(A=@n27, B=@n28))
   Alpha=If(A=@n21.R, B=@n21.R, "A > B"=@n22, "A == B"=@n22, "A < B"=Constant() {r=1})
  )
  n30 = Constant() {r=1}
  n31 = MaterialFunctionCall() {materialFunction=&GetUserInterfaceUV}
  n32 = MaterialFunctionCall(Float2=@n31.Pixel Size) {materialFunction=&BreakOutFloat2Components}
  n33 = Divide(A=@n32.G, B=@n32.G)
  n34 = OneMinus(None=@n33)
  n35 = Add(A=Multiply(A=@n31.Normalized UV, B=AppendVector(A=@n30, B=@n33)), B=AppendVector(A=Constant(), B=Multiply(A=@n34) {constB=0.5}))
  n36 = MaterialFunctionCall(Float2=MaterialFunctionCall() {materialFunction=&GetUserInterfaceUV}.Pixel Size) {materialFunction=&BreakOutFloat2Components}
  n37 = Constant()
  n38 = If(A=@n36.R, B=@n36.R, "A > B"=@n37, "A == B"=@n37, "A < B"=Constant() {r=1})
  n39 = MaterialFunctionCall() {materialFunction=&GetUserInterfaceUV}
  n40 = MaterialFunctionCall(Float2=@n39.Pixel Size) {materialFunction=&BreakOutFloat2Components}
  n41 = Divide(A=@n40.R, B=@n40.R)
  n42 = Constant() {r=1}
  n43 = OneMinus(None=@n41)
  n44 = Add(A=Multiply(A=@n39.Normalized UV, B=AppendVector(A=@n41, B=@n42)), B=AppendVector(A=Multiply(A=@n43) {constB=0.5}, B=Constant()))
  n45 = StaticSwitchParameter(
   True=LinearInterpolate(B=TextureSampleParameter2D(UVs=@n29) {parameterName=Winterfest_Red, texture=&/T_UI_WF_Red}.RGB) {constAlpha=0.7}
   False=TextureSampleParameter2D(UVs=LinearInterpolate(A=@n35, B=@n44, Alpha=@n38)) {parameterName=Winterfest_Green, texture=&/T_UI_WF_Green}.RGB
  ) {parameterName=Red}
  FunctionOutput(None=StaticSwitchParameter(True=StaticSwitchParameter(True=@n14.RGB, False=@n45) {parameterName=Blue}, False=FunctionInput()) {parameterName=bIsWinterfestSeries})
  LinearInterpolate(
   A=FunctionInput(Preview=VectorParameter() {group=CorruptSeries, parameterName=BG_Color, defaultValue=<rgba 1 1 1 1>}.RGB) {bUsePreviewValueAsDefault=true, inputName=RGB 1, sortPriority=1}
   B=FunctionInput(Preview=VectorParameter() {group=CorruptSeries, parameterName=CellColor, defaultValue=<rgba 0 0 0 1>}.RGB) {bUsePreviewValueAsDefault=true, inputName=RGB 2, sortPriority=1}
  )
 ]}
