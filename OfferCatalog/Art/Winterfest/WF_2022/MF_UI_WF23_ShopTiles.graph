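# UI material function that draws the Winterfest 2023 shop tile look behind the bls_WF23_Series static switch, passing its input through when off.
# A thin rounded-box outline, broken up by blurred noise, blends with a wider glow outline to form an icy cyan border.
# This is added over navy-to-blue rotated linear gradients with faint random grain.
# An unconnected CorruptSeries color lerp remains in the graph.
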MaterialFunction /OfferCatalog/Art/Winterfest/WF_2022/MF_UI_WF23_ShopTiles
graph {nodes [
  Constant2Vector() {g=1080, r=1920}
  n1 = Multiply(A=NamedRerouteUsage()) {constB=0.8}
  n2 = Constant2Vector() {g=0.75, r=0.65}
  n3 = Min(
   A=SmoothStep(Value=TextureSample(UVs=@n1) {samplerType=SAMPLERTYPE_Grayscale, texture=/FortUILibrary/UIKit/Buttons/Textures/LowResBlurredNoise}.R) {constMax=0.5, constMin=0.25}
   B=MaterialFunctionCall(
    UVs=NamedRerouteUsage()
    Size=@n2
    CornerRoundness=Constant() {r=0.05}
    OutlineThickness=Constant() {r=-0.002}
    GlowMax=Constant() {r=0.0125}
   ) {materialFunction=&MF_UI_Box}.Outline
  )
  n4 = Constant3Vector() {constant=<rgba 0.0395 0.5832 0.7344 1>}
  n5 = SmoothStep(Value=MaterialFunctionCall(
    UVs=NamedRerouteUsage()
    Size=@n2
    CornerRoundness=Constant() {r=0.05}
    OutlineThickness=Constant() {r=-0.004}
    GlowMin=Constant()
    GlowMax=Constant() {r=0.04}
   ) {materialFunction=&MF_UI_Box}.Outline) {constMax=3}
  n6 = Multiply(A=LinearInterpolate(A=Multiply(A=@n3, B=@n4), B=Constant3Vector() {constant=<rgba 0 0.5174 0.592 1>}, Alpha=Multiply(A=@n5)))
  n7 = MaterialFunctionCall!(UV=NamedRerouteUsage(), Seed) {materialFunction=&MF_UI_Random}
  n8 = Constant3Vector() {constant=<rgba 0 0.0178 0.1354 1>}
  n9 = Constant3Vector() {constant=<rgba 0 0.4544 0.8646 1>}
  n10 = SmoothStep(Value=MaterialFunctionCall(UVs=NamedRerouteUsage(), GradientStart=Constant() {r=0.25}, GradientEnd=Constant() {r=1}, RotationAngle=Constant() {r=0.35}) {materialFunction=&MF_UI_LinearGradient}.Result) {constMax=1.5}
  n11 = MaterialFunctionCall(UVs=NamedRerouteUsage(), GradientStart=Constant() {r=0.5}, GradientEnd=Constant() {r=1}, RotationAngle=Constant() {r=-0.25}) {materialFunction=&MF_UI_LinearGradient}
  n12 = LinearInterpolate(
   A=Add(A=Multiply(A=@n7.Result) {constB=0.02}, B=LinearInterpolate(A=@n8, B=@n9, Alpha=@n10))
   B=Multiply(A=Constant3Vector() {constant=<rgba 0.0037 0.005 0.0521 1>}, B=@n11.Result)
   Alpha=SmoothStep(Value=@n11.Result) {constMax=1.75}
  )
  FunctionOutput(None=StaticSwitchParameter(True=Add(A=@n6, B=@n12), False=FunctionInput()) {parameterName=bls_WF23_Series})
  LinearInterpolate(
   A=FunctionInput(Preview=VectorParameter() {group=CorruptSeries, parameterName=BG_Color, defaultValue=<rgba 1 1 1 1>}.RGB) {bUsePreviewValueAsDefault=true, inputName=RGB 1, sortPriority=1}
   B=FunctionInput(Preview=VectorParameter() {group=CorruptSeries, parameterName=CellColor, defaultValue=<rgba 0 0 0 1>}.RGB) {bUsePreviewValueAsDefault=true, inputName=RGB 2, sortPriority=1}
  )
  NamedRerouteDeclaration(None=MaterialFunctionCall() {materialFunction=&MF_UI_UVScaling}.ScaleY) {name=UVScalingFit, nodeColor=<rgba 0.9219 1 0 1>}
 ]}
