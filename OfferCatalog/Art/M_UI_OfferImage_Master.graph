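# Master translucent UI material for item shop offer tiles, compositing offer artwork over a background with falloff color.
# One or optional second image, still or curve-driven flipbook, gets aspect-aware UVs, offsets and hover-reactive zoom around a pivot.
# Static switches choose a textured background, radial gradient with color-dodge spotlight, or long chain of per-series background functions.
# Animated swipe mask, rounded corners, backing gradient, off-screen tint and texture-size debug readouts round it out.
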Material /OfferCatalog/Art/M_UI_OfferImage_Master
graph {
 properties {
  bUsedWithStaticMesh=false
  blendMode=BLEND_Translucent
  materialDomain=MD_UI
  pixelDepthOffsetMode=PDOM_Legacy
  usedShadingModels=Default Lit
 }
 nodes [
  n1 = NamedRerouteUsage()
  n2 = ScalarParameter() {defaultValue=50, group=Image, parameterName=ZoomImage_Percent, sliderMax=100, sliderMin=-100}
  n3 = NamedRerouteUsage()
  n4 = StaticSwitchParameter(True=NamedRerouteUsage(), False=@n3) {group=Global, parameterName=IsCharacter}
  Composite!(is vertical ratio=@n1, Zoom image percent=@n2, UVs=@n4) {subgraphName=Handle Image Zoom}
  n5 = NamedRerouteUsage()
  n6 = ScalarParameter() {defaultValue=50, group=Image secondary, parameterName=2-ZoomImage_Percent, sliderMax=100, sliderMin=-100}
  n7 = StaticSwitchParameter(True=NamedRerouteUsage(), False=NamedRerouteUsage()) {group=Global, parameterName=IsCharacter}
  Composite(is vertical ratio=@n5, Zoom image percent=@n6, UVs=@n7) {subgraphName=Handle Image Zoom}
  Constant3Vector() {constant=<rgba 0.1042 0.0977 0.0651 1>}
  LinearInterpolate(A=@n4, B=@n3, Alpha=Floor(None=ScalarParameter() {group=Pivot Point, parameterName=PivotOverCharacter, sortPriority=2})) {desc=Unplugged for now}
  n8 = Rotator(
   Coordinate=TextureCoordinate() {coordinateIndex=4}
   Time=LinearInterpolate(
    A=ScalarParameter() {defaultValue=-0.25, group=Animation, parameterName=Swipe_Rotation}
    B=ScalarParameter() {defaultValue=-0.25, group=Animation, parameterName=Swipe_Rotation_Final}
    Alpha=ScalarParameter() {group=Animation, parameterName=Swipe}
   )
  ) {speed=6.2832}
  n9 = Add(
   A=LinearInterpolate(A=OneMinus(None=@n8), B=@n8, Alpha=ScalarParameter() {group=Animation, parameterName=SwipeOrientation})
   B=LinearInterpolate(
    A=ScalarParameter() {defaultValue=0.05, group=Animation, parameterName=Swipe_Offset}
    B=ScalarParameter() {defaultValue=-1.05, group=Animation, parameterName=Swipe_Offset_Final}
    Alpha=ScalarParameter() {group=Animation, parameterName=Swipe}
   )
  )
  n10 = MaterialFunctionCall!(Native_U_PixelSize=ScalarParameter() {defaultValue=384, group=ResolutionSharpness, parameterName=NativePixelSize}) {materialFunction=/FortUILibrary/Foundation/Art_ToolKit/Tools/M_UI_Resolution_Sharpness_Correction}
  n11 = ScalarParameter() {defaultValue=64, group=Animation, parameterName=Swipe_Sharpness_V}
  n12 = Saturate(None=Multiply(A=ComponentMask(None=@n9) {r=true}, B=Multiply(A=@n10.Result, B=@n11)) {constB=32})
  n13 = ScalarParameter() {defaultValue=1, group=Global, parameterName=FinalAlpha, sliderMax=1}
  n14 = VectorParameter() {group=Tile-driven params, parameterName=CornerRadiusOverride}
  MaterialFunctionCall!(
   Input Alpha=Saturate(None=Multiply(A=@n12, B=@n13))
   bOverrideCornerRadius=ScalarParameter() {group=Tile-driven params, parameterName=bOverrideCornerRadius}
   Corner Radius Override=AppendVector(A=@n14.RGB, B=@n14.RGB)
  ) {materialFunction=/FortUILibrary/Systems/ItemShopTile/Materials/MaterialFunctions/MF_UI_ItemShop_RoundedCorners}
  n15 = TextureObjectParameter() {
   group=Image
   parameterName=OfferImage
   texture=/OfferCatalog/Textures/BattleRoyale/FeaturedItems/Outfit/T-AthenaSoldiers-CID-655-Athena-Commando-F-Barefoot
  }
  NamedRerouteDeclaration(None=TextureProperty(Tex=@n15)) {name=TextureSize, nodeColor=<rgba 1 0 0.1172 1>}
  n16 = TextureObjectParameter() {
   group=Image secondary
   parameterName=2-OfferImage
   sortPriority=1
   texture=/OfferCatalog/Textures/BattleRoyale/FeaturedItems/Outfit/T-AthenaSoldiers-CID-656-Athena-Commando-M-TeriyakiFishFreezerBurn
  }
  NamedRerouteDeclaration(None=TextureProperty(Tex=@n16)) {name=TextureSize_2, nodeColor=<rgba 1 0 0.1172 1>}
  n17 = Time()
  n18 = ScalarParameter() {defaultValue=1, group=FlipBook, parameterName=FlipbookAnimationSpeed}
  n19 = CurveAtlasRowParameter(CurveTime=Fmod(A=Multiply(A=@n17, B=@n18), B=Constant() {r=1})) {
   atlas=&/CA_UI_OfferImage_FlipbookCurves
   curve=&/C_UI_OfferImage_SmoothPingPong
   group=FlipBook
   parameterName=FlipbookCurve
  }
  n20 = Divide(
   A=ScalarParameter() {defaultValue=512, group=FlipBook, parameterName=FrameResolution_Width}
   B=ScalarParameter() {defaultValue=1024, group=FlipBook, parameterName=FrameResolution_Height}
  )
  n21 = AppendVector(A=@n20, B=Constant() {r=1})
  n22 = ScalarParameter() {group=Image, parameterName=OffsetImage_X, sliderMax=100, sliderMin=-100}
  n23 = Add(
   A=LinearInterpolate(B=ScalarParameter() {group=Image, parameterName=OffsetImage_Y_Compensation}, Alpha=NamedRerouteUsage())
   B=ScalarParameter() {group=Image, parameterName=OffsetImage_Y, sliderMax=100, sliderMin=-100}
  )
  n24 = Add(A=NamedRerouteUsage(), B=AppendVector(A=Multiply(A=@n22) {constB=-0.01}, B=Multiply(A=@n23) {constB=-0.01}))
  n25 = TextureSample(
   UVs=MaterialFunctionCall(
    "Animation  Phase (0-1)"=Saturate(None=@n19)
    Number of Rows=ScalarParameter() {defaultValue=3, group=FlipBook, parameterName=Frame_RowCount}
    Number of Columns=ScalarParameter() {defaultValue=6, group=FlipBook, parameterName=Frame_ColumnCount}
    UVs=MaterialFunctionCall(UVs=@n24, Texture Scale=@n21) {materialFunction=&ScaleUVsByCenter}.UVs
   ) {materialFunction=&FlipBook}.UVs
   Tex=@n15
  ) {desc=flipbook image, samplerSource=SSM_Wrap_WorldGroupSettings}
  n26 = Add(A=@n24) {constB=-0.5}
  n27 = Multiply(A=Add(A=Abs(None=@n26), B=Multiply(A=@n20) {constB=-0.5})) {constB=512}
  n28 = ComponentMask(None=OneMinus(None=Saturate(None=@n27))) {r=true}
  n29 = TextureSample(UVs=@n24, Tex=@n15) {desc=normal image, samplerSource=SSM_Clamp_WorldGroupSettings}
  n30 = MaterialFunctionCall!(UVs=@n24, Alpha=@n29.A) {materialFunction=/Game/UI/Foundation/Items/ClampAlphaToTransparentBorder}
  n31 = Multiply(
   A=Subtract(A=ComponentMask(None=@n24) {g=true}, B=ScalarParameter() {defaultValue=0.5, parameterName=Item_GradientMask_OFFSET_DEBUG}) {constB=0.5}
   B=ScalarParameter() {defaultValue=3, parameterName=Item_GradientMask_STRENGTH_DEBUG}
  ) {constB=3}
  n32 = Multiply(A=Saturate(None=OneMinus(None=@n31)), B=@n30.Result)
  n33 = Floor(None=ScalarParameter() {parameterName=Item_GradientMask_OnOff})
  NamedRerouteDeclaration(None=StaticSwitchParameter(True=LinearInterpolate(B=@n25.A, Alpha=@n28), False=LinearInterpolate(A=@n30.Result, B=@n32, Alpha=@n33)) {group=FlipBook, parameterName=IsFlipBook}) {name=Image 1 alpha, nodeColor=<rgba 0 0 0 1>}
  n34 = Time()
  n35 = ScalarParameter() {defaultValue=1, group=Image secondary, parameterName=2-FlipbookAnimationSpeed}
  n36 = CurveAtlasRowParameter(CurveTime=Fmod(A=Multiply(A=@n34, B=@n35), B=Constant() {r=1})) {
   atlas=&CA_UI_OfferImage_FlipbookCurves
   curve=&C_UI_OfferImage_SmoothPingPong
   group=FlipBook
   parameterName=FlipbookCurve
  }
  n37 = Divide(
   A=ScalarParameter() {defaultValue=512, group=Image secondary, parameterName=2-FrameResolution_Width}
   B=ScalarParameter() {defaultValue=1024, group=Image secondary, parameterName=2-FrameResolution_Height}
  )
  n38 = AppendVector(A=@n37, B=Constant() {r=1})
  n39 = ScalarParameter() {defaultValue=10, group=Image secondary, parameterName=2-OffsetImage_X, sliderMax=100, sliderMin=-100, sortPriority=2}
  n40 = Add(
   A=LinearInterpolate(B=ScalarParameter() {group=Image secondary, parameterName=2-OffsetImage_Y_Compensation, sortPriority=4}, Alpha=NamedRerouteUsage())
   B=ScalarParameter() {group=Image secondary, parameterName=2-OffsetImage_Y, sliderMax=100, sliderMin=-100, sortPriority=3}
  )
  n41 = Add(A=NamedRerouteUsage(), B=AppendVector(A=Multiply(A=@n39) {constB=-0.01}, B=Multiply(A=@n40) {constB=-0.01}))
  n42 = TextureSample(
   UVs=MaterialFunctionCall(
    "Animation  Phase (0-1)"=Saturate(None=@n36)
    Number of Rows=ScalarParameter() {defaultValue=3, group=Image secondary, parameterName=2-Frame_RowCount}
    Number of Columns=ScalarParameter() {defaultValue=6, group=Image secondary, parameterName=2-Frame_ColumnCount}
    UVs=MaterialFunctionCall(UVs=@n41, Texture Scale=@n38) {materialFunction=&ScaleUVsByCenter}.UVs
   ) {materialFunction=&FlipBook}.UVs
   Tex=@n16
  ) {desc=flipbook image, samplerSource=SSM_Wrap_WorldGroupSettings}
  n43 = Add(A=@n41) {constB=-0.5}
  n44 = Multiply(A=Add(A=Abs(None=@n43), B=Multiply(A=@n37) {constB=-0.5})) {constB=512}
  n45 = ComponentMask(None=OneMinus(None=Saturate(None=@n44))) {r=true}
  n46 = TextureSample(UVs=@n41, Tex=@n16) {desc=normal image, samplerSource=SSM_Clamp_WorldGroupSettings}
  NamedRerouteDeclaration(None=StaticSwitchParameter(
    True=LinearInterpolate(B=@n42.A, Alpha=@n45)
    False=MaterialFunctionCall(UVs=@n41, Alpha=@n46.A) {materialFunction=&ClampAlphaToTransparentBorder}.Result
   ) {group=Image secondary, parameterName=2-IsFlipBook, sortPriority=10}) {name=Image 2 alpha, nodeColor=<rgba 0 0 0 1>}
  NamedRerouteDeclaration(None=StaticSwitchParameter(True=@n25.RGB, False=@n29.RGB) {group=FlipBook, parameterName=IsFlipBook}) {name=Image 1 color, nodeColor=<rgba 0 0 0 1>}
  NamedRerouteDeclaration(None=StaticSwitchParameter(True=@n42.RGB, False=@n46.RGB) {group=Image secondary, parameterName=2-IsFlipBook, sortPriority=10}) {name=Image 2 color, nodeColor=<rgba 0 0 0 1>}
  n47 = NamedRerouteUsage()
  n48 = NamedRerouteUsage()
  n49 = NamedRerouteUsage()
  NamedRerouteDeclaration(None=StaticSwitchParameter(True=LinearInterpolate(A=@n47, B=@n48, Alpha=@n49), False=NamedRerouteUsage()) {group=Image secondary, parameterName=Use second Image, sortPriority=0}) {name=image color blended, nodeColor=<rgba 0 0 0 1>}
  n50 = NamedRerouteUsage()
  n51 = NamedRerouteUsage()
  NamedRerouteDeclaration(None=StaticSwitchParameter(True=Max(A=@n50, B=@n51), False=NamedRerouteUsage()) {group=Image secondary, parameterName=Use second Image, sortPriority=0}) {name=image alpha blended, nodeColor=<rgba 0 0 0 1>}
  n52 = MaterialFunctionCall(Float2=MaterialFunctionCall() {materialFunction=&GetUserInterfaceUV}.Pixel Size) {materialFunction=&BreakOutFloat2Components}
  n53 = Constant()
  n54 = Constant() {r=1}
  n55 = Divide(A=@n52.G, B=@n52.G)
  n56 = ScalarParameter() {defaultValue=1.19, parameterName=Lower Ratio}
  n57 = ScalarParameter() {defaultValue=1.21, parameterName=Higher Ratio}
  n58 = LinearInterpolate(
   A=@n56
   B=@n57
   Alpha=MaterialFunctionCall!(A=@n56, B=@n57) {materialFunction=/FortUILibrary/MaterialFunctions/Basics/MF_UI_BooleanOperations}.And
  ) {constA=1.19}
  n59 = Constant() {r=1}
  n60 = Constant()
  NamedRerouteDeclaration(None=LinearInterpolate(
    A=If(A=@n52.R, B=@n52.R, "A > B"=@n53, "A == B"=@n53, "A < B"=@n54)
    B=If(A=@n55, B=@n58, "A > B"=@n59, "A == B"=@n59, "A < B"=@n60)
    Alpha=ScalarParameter() {defaultValue=1, parameterName=VerticalAspectRatioFix}
   )) {name=is vertical ratio, nodeColor=<rgba 0 0 0 1>}
  n61 = MaterialFunctionCall() {materialFunction=&MF_UI_UVScaling}
  NamedRerouteDeclaration(None=@n61.Scaled To Fill) {name=Square UVs Scaled to Fill XY, nodeColor=<rgba 0 0 0 1>}
  MaterialExpressionReroute_316 = Reroute(None=@n4)
  n62 = ScalarParameter() {defaultValue=7.5, group=Animation, parameterName=Hover_Zoom_Increment}
  MaterialExpressionReroute_315 = Reroute(None=@n2)
  n63 = ScalarParameter() {group=Image, parameterName=Scale_Compensation}
  MaterialExpressionReroute_314 = Reroute(None=@n1)
  n64 = StaticSwitchParameter(
   True=@MaterialExpressionReroute_315.Zoom image percent
   False=Add(
    A=LinearInterpolate(B=@n63, Alpha=@MaterialExpressionReroute_314.is vertical ratio) {constB=7.5}
    B=@MaterialExpressionReroute_315.Zoom image percent
   )
  ) {group=Global, parameterName=IsCharacter}
  n65 = Multiply(A=LinearInterpolate(A=@n64, B=Add(A=@n64, B=@n62) {constB=10}, Alpha=ScalarParameter() {group=Animation, parameterName=Hover}) {constA=0.95}) {constB=0.01}
  n66 = Divide(
   A=ScalarParameter() {defaultValue=344, group=Global, parameterName=TileHeightSmall}
   B=ScalarParameter() {defaultValue=720, group=Global, parameterName=TileHeightLarge}
  )
  n67 = Multiply(A=LinearInterpolate(B=Subtract(B=@n62) {constA=100}, Alpha=LinearInterpolate(B=@n66, Alpha=@n65) {constB=0.325}) {constAlpha=100, constB=95}) {constB=0.01}
  n68 = Clamp(None=LinearInterpolate(
    A=ScalarParameter() {defaultValue=1, group=ZoomCrop, parameterName=Zoom_Min}
    B=ScalarParameter() {defaultValue=2, group=ZoomCrop, parameterName=Zoom_Max}
    Alpha=@n67
   )) {maxDefault=2}
  n69 = AppendVector(
   A=ScalarParameter() {defaultValue=0.5, group=ZoomCrop, parameterName=ZoomAroundPoint_X}
   B=LinearInterpolate(A=Constant(), B=ScalarParameter() {defaultValue=0.125, group=ZoomCrop, parameterName=ZoomAroundPoint_Y}, Alpha=@n67)
  )
  n70 = Subtract(
   A=Add(A=Divide(A=@MaterialExpressionReroute_316.UVs, B=@n68), B=@n69)
   B=Divide(A=@n69, B=@n68) {desc=Scale the center offset by the scaling factor}
  ) {desc=Re offset to maintain center}
  MaterialExpressionReroute_317 = Reroute(None=@n70)
  NamedRerouteDeclaration(None=@MaterialExpressionReroute_317.UVs) {name=image 1 UVs, nodeColor=<rgba 1 0.6562 0 1>}
  NamedRerouteDeclaration(None=@n61.ScaleY) {name=Square UVs Scaled to Fit Y, nodeColor=<rgba 0 0 0 1>}
  MaterialExpressionReroute_320 = Reroute(None=@n7)
  n71 = ScalarParameter() {defaultValue=7.5, group=Animation, parameterName=Hover_Zoom_Increment}
  MaterialExpressionReroute_319 = Reroute(None=@n6)
  n72 = ScalarParameter() {group=Image, parameterName=Scale_Compensation}
  MaterialExpressionReroute_318 = Reroute(None=@n5)
  n73 = StaticSwitchParameter(
   True=@MaterialExpressionReroute_319.Zoom image percent
   False=Add(
    A=LinearInterpolate(B=@n72, Alpha=@MaterialExpressionReroute_318.is vertical ratio) {constB=7.5}
    B=@MaterialExpressionReroute_319.Zoom image percent
   )
  ) {group=Global, parameterName=IsCharacter}
  n74 = Multiply(A=LinearInterpolate(A=@n73, B=Add(A=@n73, B=@n71) {constB=10}, Alpha=ScalarParameter() {group=Animation, parameterName=Hover}) {constA=0.95}) {constB=0.01}
  n75 = Divide(
   A=ScalarParameter() {defaultValue=344, group=Global, parameterName=TileHeightSmall}
   B=ScalarParameter() {defaultValue=720, group=Global, parameterName=TileHeightLarge}
  )
  n76 = Multiply(A=LinearInterpolate(B=Subtract(B=@n71) {constA=100}, Alpha=LinearInterpolate(B=@n75, Alpha=@n74) {constB=0.325}) {constAlpha=100, constB=95}) {constB=0.01}
  n77 = Clamp(None=LinearInterpolate(
    A=ScalarParameter() {defaultValue=1, group=ZoomCrop, parameterName=Zoom_Min}
    B=ScalarParameter() {defaultValue=2, group=ZoomCrop, parameterName=Zoom_Max}
    Alpha=@n76
   )) {maxDefault=2}
  n78 = AppendVector(
   A=ScalarParameter() {defaultValue=0.5, group=ZoomCrop, parameterName=ZoomAroundPoint_X}
   B=LinearInterpolate(A=Constant(), B=ScalarParameter() {defaultValue=0.125, group=ZoomCrop, parameterName=ZoomAroundPoint_Y}, Alpha=@n76)
  )
  n79 = Subtract(
   A=Add(A=Divide(A=@MaterialExpressionReroute_320.UVs, B=@n77), B=@n78)
   B=Divide(A=@n78, B=@n77) {desc=Scale the center offset by the scaling factor}
  ) {desc=Re offset to maintain center}
  MaterialExpressionReroute_321 = Reroute(None=@n79)
  NamedRerouteDeclaration(None=@MaterialExpressionReroute_321.UVs) {name=image 2 UVs, nodeColor=<rgba 0.8516 1 0 1>}
  n80 = TextureObjectParameter() {group=Image, parameterName=ImageBackground, texture=/OfferCatalog/Art/Armadillo/T_UI_Armadillo_Skyfield}
  NamedRerouteDeclaration(None=TextureProperty(Tex=@n80)) {name=TextureSize_Bg, nodeColor=<rgba 1 0 0.1172 1>}
  n81 = MaterialFunctionCall(Float2=NamedRerouteUsage()) {materialFunction=&BreakOutFloat2Components}
  n82 = Constant() {r=4}
  n83 = NamedRerouteUsage()
  n84 = Add(
   A=MaterialFunctionCall(
    Number=@n81.R
    MaximumNumberOfDigits=@n82
    UVs=@n83
    DebugTextLocation RG_UpperRight BA_LowerLeft=Constant4Vector() {constant=<rgba 0 0 0.45 0.3>}.RGBA
   ) {materialFunction=&DebugScalarValues}.Result
   B=MaterialFunctionCall(
    Number=@n81.G
    MaximumNumberOfDigits=@n82
    UVs=@n83
    DebugTextLocation RG_UpperRight BA_LowerLeft=Constant4Vector() {constant=<rgba 0.55 0 1 0.3>}.RGBA
   ) {materialFunction=&DebugScalarValues}.Result
  )
  n85 = MaterialFunctionCall(Float2=NamedRerouteUsage()) {materialFunction=&BreakOutFloat2Components}
  n86 = StaticSwitchParameter(True=Constant() {r=4}, False=Constant()) {group=Image secondary, parameterName=Use second Image, sortPriority=0}
  n87 = NamedRerouteUsage()
  n88 = Add(
   A=MaterialFunctionCall(
    Number=@n85.R
    MaximumNumberOfDigits=@n86
    UVs=@n87
    DebugTextLocation RG_UpperRight BA_LowerLeft=Constant4Vector() {constant=<rgba 0 0.35 0.45 0.65>}.RGBA
   ) {materialFunction=&DebugScalarValues}.Result
   B=MaterialFunctionCall(
    Number=@n85.G
    MaximumNumberOfDigits=@n86
    UVs=@n87
    DebugTextLocation RG_UpperRight BA_LowerLeft=Constant4Vector() {constant=<rgba 0.55 0.35 1 0.65>}.RGBA
   ) {materialFunction=&DebugScalarValues}.Result
  )
  n89 = MaterialFunctionCall(Float2=NamedRerouteUsage()) {materialFunction=&BreakOutFloat2Components}
  n90 = StaticSwitchParameter(True=Constant() {r=4}, False=Constant()) {group=Background, parameterName=UseTextureBackground}
  n91 = NamedRerouteUsage()
  n92 = MaterialFunctionCall(
   Number=@n89.R
   MaximumNumberOfDigits=@n90
   UVs=@n91
   DebugTextLocation RG_UpperRight BA_LowerLeft=Constant4Vector() {constant=<rgba 0 0.7 0.45 1>}.RGBA
  ) {materialFunction=&DebugScalarValues}
  n93 = MaterialFunctionCall(
   Number=@n89.G
   MaximumNumberOfDigits=@n90
   UVs=@n91
   DebugTextLocation RG_UpperRight BA_LowerLeft=Constant4Vector() {constant=<rgba 0.55 0.7 1 1>}.RGBA
  ) {materialFunction=&DebugScalarValues}
  NamedRerouteDeclaration(None=Add(A=Add(A=@n84, B=@n88), B=Add(A=@n92.Result, B=@n93.Result))) {name=Debug Image Sizes, nodeColor=<rgba 0 0 0 1>}
  n94 = NamedRerouteUsage()
  n95 = MaterialFunctionCall(Float2=MaterialFunctionCall() {materialFunction=&GetUserInterfaceUV}.Pixel Size) {materialFunction=&BreakOutFloat2Components}
  n96 = Subtract(A=Divide(A=@n95.R, B=@n95.R))
  n97 = MaterialFunctionCall(X=Max(A=Divide(A=@n96)) {constB=0}, Y=Constant() {r=-0.4}) {materialFunction=&MakeFloat2}
  n98 = Constant2Vector() {g=8, r=2}
  n99 = MaterialFunctionCall(Float2=MaterialFunctionCall() {materialFunction=&GetUserInterfaceUV}.Pixel Size) {materialFunction=&BreakOutFloat2Components}
  n100 = Divide(A=Min(A=@n99.R, B=@n99.R)) {constA=200, constB=400}
  NamedRerouteDeclaration(None=Multiply(A=Add(A=@n94, B=@n97.Result), B=Multiply(A=@n98, B=@n100))) {name=Debug Image Size UVs, nodeColor=<rgba 0 0 0 1>}
  n101 = NamedRerouteUsage()
  n102 = Subtract(A=Abs(None=Subtract(A=@n101) {constB=0.5})) {constB=0.5}
  n103 = MaterialFunctionCall(Float2=Ceil(None=Min(A=@n102))) {materialFunction=&BreakOutFloat2Components}
  NamedRerouteDeclaration(None=Max(A=@n103.R, B=@n103.R)) {name=Debug Image Sizes Mask, nodeColor=<rgba 0 0 0 1>}
  n104 = MaterialFunctionCall(TextureSize=NamedRerouteUsage(), ScalePivotPoint=NamedRerouteUsage()) {materialFunction=&MF_UI_UVScaling}
  NamedRerouteDeclaration(None=@n104.Scaled To Fill) {name=Image 1 UVs Scaled to Fill XY, nodeColor=<rgba 0 0 0 1>}
  n105 = MaterialFunctionCall(TextureSize=NamedRerouteUsage(), ScalePivotPoint=NamedRerouteUsage()) {materialFunction=&MF_UI_UVScaling}
  NamedRerouteDeclaration(None=@n105.Scaled To Fill) {name=Image 2 UVs Scaled to Fill XY, nodeColor=<rgba 0 0 0 1>}
  NamedRerouteDeclaration(None=MaterialFunctionCall(TextureSize=NamedRerouteUsage(), ScalePivotPoint=NamedRerouteUsage()) {materialFunction=&MF_UI_UVScaling}.Scaled To Fill) {name=Image Bg UVs Scaled to Fill XY, nodeColor=<rgba 0 0 0 1>}
  NamedRerouteDeclaration(None=@n104.ScaleY) {name=Image 1 UVs Scaled to Fit Y, nodeColor=<rgba 0 0 0 1>}
  NamedRerouteDeclaration(None=@n105.ScaleY) {name=Image 2 UVs Scaled to Fit Y, nodeColor=<rgba 0 0 0 1>}
  NamedRerouteDeclaration(None=@n61.Scaled to Fit) {name=Square UVs Scaled to Fit XY, nodeColor=<rgba 0 0 0 1>}
  n106 = ComponentMask(None=MaterialFunctionCall() {materialFunction=&GetUserInterfaceUV}.Pixel Size) {g=true}
  n107 = Add(
   A=ScalarParameter() {defaultValue=150, parameterName=GradientPixelHeight}
   B=Constant() {desc="top padding in pixels,\r\nto go a bit above the text", r=60}
  ) {constB=20}
  n108 = Multiply(
   A=OneMinus(None=MaterialFunctionCall() {materialFunction=&GetUserInterfaceUV}.Normalized UV)
   B=AppendVector(A=Constant() {r=1}, B=Divide(A=@n106, B=@n107))
  )
  n109 = ScalarParameter() {defaultValue=1.15, group=Gradient Settings, parameterName=GradientEnd, sortPriority=2}
  n110 = ScalarParameter() {defaultValue=0.1, group=Gradient Settings, parameterName=GradientStart, sortPriority=1}
  n111 = ScalarParameter() {defaultValue=-0.3, group=Gradient Settings, parameterName=GradientRotation, sortPriority=4}
  n112 = VectorParameter() {
   group=Gradient Settings
   parameterName=GradientRotationCenter
   sortPriority=5
   channelNames {a="", b="", g=Center Y, r=Center X}
   defaultValue=<rgba 0.5 0.5 0 0>
  }
  n113 = ScalarParameter() {group=Gradient Settings, parameterName=GradientOffset, sortPriority=3}
  NamedRerouteDeclaration(None=SmoothStep(Value=MaterialFunctionCall(UVs=@n108, GradientStart=@n109, GradientEnd=@n110, RotationAngle=@n111, RotationCenter=@n112, GradientOffset=@n113) {materialFunction=&MF_UI_LinearGradient}.Result)) {name=backing gradient, nodeColor=<rgba 0 1 0.0391 1>}
  n114 = ScalarParameter() {defaultValue=0.5, group=Pivot Point, parameterName=X_PivotPoint, sortPriority=3}
  n115 = ScalarParameter() {group=Pivot Point, parameterName=Y_PivotPoint, sortPriority=3}
  n116 = Floor(None=ScalarParameter() {group=Pivot Point, parameterName=enablePivotPoint, sortPriority=1})
  NamedRerouteDeclaration(None=LinearInterpolate(A=Constant2Vector() {g=0.5, r=0.5}, B=AppendVector(A=@n114, B=@n115), Alpha=Saturate(None=@n116))) {name=PivotPoint, desc=Unplugged for now, nodeColor=<rgba 0 0 0 1>}
  PinBase() {reroutePins [
    is vertical ratio {expression=&self:MaterialExpressionReroute_314}
    Zoom image percent {expression=&self:MaterialExpressionReroute_315}
    UVs {expression=&self:MaterialExpressionReroute_316}
   ]}
  PinBase!(UVs=@n70) {reroutePins [UVs {expression=&self:MaterialExpressionReroute_317}]}
  PinBase!() {reroutePins [
    is vertical ratio {expression=&self:MaterialExpressionReroute_318}
    Zoom image percent {expression=&self:MaterialExpressionReroute_319}
    UVs {expression=&self:MaterialExpressionReroute_320}
   ]}
  PinBase!(UVs=@n79) {reroutePins [UVs {expression=&self:MaterialExpressionReroute_321}]}
  n117 = MaterialFunctionCall() {materialFunction=&ScreenAlignedUVs}
  n118 = Divide(B=Multiply(A=ScalarParameter() {defaultValue=5, group=Global, parameterName=OffScreen_Scale_ThresholdPercent}) {constB=0.01})
  n119 = ComponentMask(None=Multiply(A=OneMinus(None=@n117."X 100%, Y 100%"), B=@n118)) {r=true}
  n120 = Saturate(None=ComponentMask(None=Multiply(A=@n117."X 100%, Y 100%", B=@n118)) {r=true})
  Saturate(None=DotProduct(A=Saturate(None=@n119), B=@n120))
  n121 = NamedRerouteUsage()
  n122 = VectorParameter() {group=Global, parameterName=OffScreen_Color, defaultValue=<rgba 0 0.0125 0.25 1>}
  n123 = NamedRerouteUsage()
  n124 = NamedRerouteUsage()
  n125 = Multiply(A=ScalarParameter() {defaultValue=25, group=Image, parameterName=FallOffColor_Postion}) {constB=0.01}
  n126 = AppendVector(A=Constant(), B=Multiply(A=LinearInterpolate(Alpha=@n125) {constA=1, constB=0}) {constB=-1})
  n127 = Power(
   Base=ComponentMask(None=Add(A=@n124, B=@n126)) {g=true}
   Exp=ScalarParameter() {defaultValue=1.22, group=Image, parameterName=FallOffColor_DecayRate}
  )
  n128 = ScalarParameter() {defaultValue=40, group=Image, parameterName=FallOffColor_Fill_Percent}
  n129 = Multiply(A=@n128) {constB=0.125}
  n130 = LinearInterpolate(A=@n128, B=Saturate(None=Subtract(A=@n128, B=@n129)), Alpha=ScalarParameter() {group=Animation, parameterName=Hover})
  n131 = LinearInterpolate(
   A=NamedRerouteUsage()
   B=NamedRerouteDeclaration(None=VectorParameter() {group=Image, parameterName=FallOff_Color, defaultValue=<rgba 0.325 0.3047 0.2031 1>}.RGB) {name=FallOff Color, nodeColor=<rgba 1 0.75 0 1>}
   Alpha=Multiply(A=Saturate(None=@n127), B=Multiply(A=@n130) {constB=0.01})
  )
  n132 = Multiply(A=ScalarParameter() {group=Background Gradient, parameterName=Spotlight_Intensity}) {constB=0.01}
  n133 = Multiply(A=ScalarParameter() {defaultValue=50, group=Background Gradient, parameterName=Spotlight_Position_X}) {constB=0.01}
  n134 = Multiply(A=ScalarParameter() {defaultValue=25, group=Background Gradient, parameterName=Spotlight_Position_Y}) {constB=0.01}
  n135 = Multiply(A=ScalarParameter() {defaultValue=50, group=Background Gradient, parameterName=Gradient_Position_X}) {constB=0.01}
  n136 = Multiply(A=ScalarParameter() {defaultValue=25, group=Background Gradient, parameterName=Gradient_Position_Y}) {constB=0.01}
  n137 = ScalarParameter() {defaultValue=66.67, group=Background Gradient, parameterName=Spotlight_Size}
  n138 = ScalarParameter() {defaultValue=66.67, group=Background Gradient, parameterName=Gradient_Size}
  n139 = ScalarParameter() {defaultValue=16, group=Background Gradient, parameterName=Spotlight_Hardness}
  n140 = ScalarParameter() {defaultValue=16, group=Background Gradient, parameterName=Gradient_Hardness}
  n141 = NamedRerouteUsage()
  n142 = MaterialFunctionCall(
   UVs=@n141
   CenterPosition=StaticSwitchParameter(True=AppendVector(A=@n133, B=@n134), False=AppendVector(A=@n135, B=@n136)) {group=Background Gradient, parameterName=UseCustomSpotlightLocation}
   Radius=StaticSwitchParameter(True=Multiply(A=@n137) {constB=0.01}, False=Multiply(A=@n138) {constB=0.01}) {group=Background Gradient, parameterName=UseCustomSpotlightRadius}
   Density=StaticSwitchParameter(True=Multiply(A=@n139) {constB=0.1}, False=Multiply(A=@n140) {constB=0.1}) {group=Background Gradient, parameterName=UseCustomSpotlight_Hardness}
  ) {materialFunction=&RadialGradientExponential}
  n143 = VectorParameter() {group=Background Gradient, parameterName=Background_Color_A, defaultValue=<rgba 0 0.0056 0.0521 1>}
  n144 = VectorParameter() {group=Background Gradient, parameterName=Background_Color_B, defaultValue=<rgba 0.3185 0.3005 0.2016 1>}
  n145 = AppendVector(
   A=Multiply(A=ScalarParameter() {defaultValue=50, group=Background Gradient, parameterName=Gradient_Position_X}) {constB=0.01}
   B=Multiply(A=ScalarParameter() {defaultValue=25, group=Background Gradient, parameterName=Gradient_Position_Y}) {constB=0.01}
  )
  n146 = Multiply(A=ScalarParameter() {defaultValue=66.67, group=Background Gradient, parameterName=Gradient_Size}) {constB=0.01}
  n147 = ScalarParameter() {defaultValue=16, group=Background Gradient, parameterName=Gradient_Hardness}
  n148 = Multiply(A=@n147) {constB=0.25}
  n149 = Multiply(A=LinearInterpolate(A=@n147, B=Add(A=@n147, B=@n148), Alpha=ScalarParameter() {group=Animation, parameterName=HoverBurst})) {constB=0.1}
  n150 = MaterialFunctionCall!(
   Blend=Saturate(None=LinearInterpolate(B=@n132, Alpha=@n142.RadialGradientExponential))
   Base=LinearInterpolate(
    A=NamedRerouteDeclaration(None=@n143.RGB) {name=Background Color A, nodeColor=<rgba 1 0 0.1875 1>}
    B=@n144.RGB
    Alpha=MaterialFunctionCall(UVs=@n141, CenterPosition=@n145, Radius=@n146, Density=@n149) {materialFunction=&RadialGradientExponential}.RadialGradientExponential
   )
  ) {materialFunction=/Engine/Functions/Engine_MaterialFunctions03/Blends/Blend_ColorDodge}
  n151 = VectorParameter() {group=Background Gradient, parameterName=Background_Color_A, defaultValue=<rgba 0 0.0056 0.0521 1>}
  n152 = VectorParameter() {group=Background Gradient, parameterName=Background_Color_B, defaultValue=<rgba 0.3185 0.3005 0.2016 1>}
  n153 = MaterialFunctionCall!(
   RGB 1=@n151.RGB
   RGB 2=@n152.RGB
   In=MaterialFunctionCall!(RGB 2=@n152.RGB, RGB 1=@n151.RGB) {materialFunction=/Game/Athena/UI/Frontend/CosmeticItemCard/Materials/MF_UI_Default}.Result
  ) {materialFunction=/Game/Athena/UI/Frontend/CosmeticItemCard/Materials/MF_UI_LavaSeries}
  n154 = MaterialFunctionCall!(
   In=MaterialFunctionCall!(
    In=MaterialFunctionCall!(In=@n153.Result, RGB 2=@n152.RGB, RGB 1=@n151.RGB) {materialFunction=/Game/Athena/UI/Frontend/CosmeticItemCard/Materials/MF_UI_FrozenSeries}.Result
    RGB 2=@n152.RGB
    RGB 1=@n151.RGB
   ) {materialFunction=/Game/Athena/UI/Frontend/CosmeticItemCard/Materials/MF_UI_MarvelSeries}.Result
   RGB 2=@n152.RGB
   RGB 1=@n151.RGB
  ) {materialFunction=/Game/Athena/UI/Frontend/CosmeticItemCard/Materials/MF_UI_CubeSeries}
  n155 = MaterialFunctionCall!(
   In=MaterialFunctionCall!(
    In=MaterialFunctionCall!(In=@n154.Result, RGB 2=@n152.RGB, RGB 1=@n151.RGB) {materialFunction=/Game/Athena/UI/Frontend/CosmeticItemCard/Materials/MF_UI_ShadowSeries}.Result
    RGB 2=@n152.RGB
    RGB 1=@n151.RGB
   ) {materialFunction=/Game/Athena/UI/Frontend/CosmeticItemCard/Materials/MF_UI_CreatorCollabSeries}.Result
   RGB 2=@n152.RGB
   RGB 1=@n151.RGB
  ) {materialFunction=/Game/Athena/UI/Frontend/CosmeticItemCard/Materials/MF_UI_DCSeries}
  n156 = MaterialFunctionCall!(
   In=MaterialFunctionCall!(
    In=MaterialFunctionCall!(In=@n155.Result, RGB 2=@n152.RGB, RGB 1=@n151.RGB) {materialFunction=/Game/Athena/UI/Frontend/CosmeticItemCard/Materials/MF_UI_SlurpSeries}.Result
    RGB 2=@n152.RGB
    RGB 1=@n151.RGB
   ) {materialFunction=/Game/Athena/UI/Frontend/CosmeticItemCard/Materials/MF_UI_ColumbusSeries}.Result
   RGB 2=@n152.RGB
   RGB 1=@n151.RGB
  ) {materialFunction=/Game/Athena/UI/Frontend/CosmeticItemCard/Materials/MF_UI_PlatformSeries}
  n157 = MaterialFunctionCall!(
   In=MaterialFunctionCall!(
    In=MaterialFunctionCall!(In=@n156.Result, RGB 2=@n152.RGB, RGB 1=@n151.RGB) {materialFunction=/OfferCatalog/Art/Majesty/MF_UI_MajestySet}.Result
    RGB 2=@n152.RGB
    RGB 1=@n151.RGB
   ) {materialFunction=/FortniteCoreUI/Frontend/CosmeticItemCard/Materials/MF_UI_TextileSeries}.Result
   RGB 2=@n152.RGB
   RGB 1=@n151.RGB
  ) {materialFunction=/OfferCatalog/Art/FNM/MF_UI_FNMSeries}
  n158 = MaterialFunctionCall!(
   In=MaterialFunctionCall!(
    In=MaterialFunctionCall!(In=@n157.Result, RGB 2=@n152.RGB, RGB 1=@n151.RGB) {materialFunction=/FortniteCoreUI/Frontend/CosmeticItemCard/Materials/MF_UI_HeadbandSeries}.Result
    RGB 1=@n152.RGB
    RGB 2=@n151.RGB
   ) {materialFunction=/OfferCatalog/Art/Parallel/MF_UI_Parallel_Series_Base}.Result
   RGB 1=@n152.RGB
   RGB 2=@n151.RGB
  ) {materialFunction=/OfferCatalog/Art/Winterfest/WF_2021/MF_UI_Winterfest_Series}
  n159 = MaterialFunctionCall!(
   In=MaterialFunctionCall!(
    In=MaterialFunctionCall!(In=@n158.Result, RGB 2=@n152.RGB, RGB 1=@n151.RGB) {materialFunction=/FortniteCoreUI/Frontend/CosmeticItemCard/Materials/MF_UI_SlitherSeries}.Result
    RGB 1=@n152.RGB
    RGB 2=@n151.RGB
   ) {materialFunction=/OfferCatalog/Art/Logarithm/MF_UI_LogarithmSeries}.Result
   RGB 1=@n152.RGB
   RGB 2=@n151.RGB
  ) {materialFunction=/OfferCatalog/Art/Keen/MF_UI_Keen_Series}
  n160 = MaterialFunctionCall!(
   In=MaterialFunctionCall!(
    In=MaterialFunctionCall!(In=@n159.Result, RGB 1=@n152.RGB, RGB 2=@n151.RGB) {materialFunction=/OfferCatalog/Art/Gimmick/MF_UI_GimmickSeries}.Result
    RGB 1=@n152.RGB
    RGB 2=@n151.RGB
   ) {materialFunction=/OfferCatalog/Art/Thrive/ShopTiles/MF_UI_Thrive_Background}.Result
   RGB 2=@n151.RGB
   RGB 1=@n152.RGB
  ) {materialFunction=/OfferCatalog/Art/IWDTakeover/MF_UI_IWDSeries}
  n161 = MaterialFunctionCall!(
   In=MaterialFunctionCall!(
    In=MaterialFunctionCall!(In=@n160.Result, RGB 2=@n151.RGB, RGB 1=@n152.RGB) {materialFunction=/FortniteCoreUI/Frontend/CosmeticItemCard/Materials/MF_UI_SnowfallSeries}.Result
    RGB 1=@n152.RGB
    RGB 2=@n151.RGB
   ) {materialFunction=/OfferCatalog/Art/LittleEgg/MF_UI_LittleEgg}.Result
   RGB 1=@n152.RGB
   RGB 2=@n151.RGB
  ) {materialFunction=/OfferCatalog/Art/Cactus/MF_UI_Cactus_Base}
  n162 = MaterialFunctionCall!(
   In=MaterialFunctionCall!(
    In=MaterialFunctionCall!(In=@n161.Result, RGB 2=@n151.RGB, RGB 1=@n152.RGB) {materialFunction=/FortniteCoreUI/Frontend/CosmeticItemCard/Materials/MF_UI_RumbleSeries}.Result
    RGB 1=@n151.RGB
    RGB 2=@n152.RGB
   ) {materialFunction=/OfferCatalog/Art/Indigo/ShopTiles/MF_UI_IndigoSeries}.Result
   RGB 1=@n151.RGB
   RGB 2=@n152.RGB
  ) {materialFunction=/OfferCatalog/Art/Alfredo/MF_UI_Alfredo_Series}
  n163 = MaterialFunctionCall!(
   In=MaterialFunctionCall!(
    In=MaterialFunctionCall!(In=@n162.Result, RGB 2=@n152.RGB, RGB 1=@n151.RGB) {materialFunction=/OfferCatalog/Art/Armadillo/MF_UI_ArmadilloSeries}.Result
    RGB 2=@n152.RGB
    RGB 1=@n151.RGB
   ) {materialFunction=/FortniteCoreUI/Frontend/CosmeticItemCard/Materials/MF_UI_OhanaSeries}.Result
   RGB 2=@n152.RGB
   RGB 1=@n151.RGB
  ) {materialFunction=/OfferCatalog/Art/Stamina/MF_UI_StaminaSeries}
  n164 = MaterialFunctionCall!(
   In=MaterialFunctionCall!(
    In=MaterialFunctionCall!(In=@n163.Result, RGB 1=@n152.RGB, RGB 2=@n151.RGB) {materialFunction=/OfferCatalog/Art/PinkLemonade/MF_UI_PinkLemonade}.Result
    RGB 2=@n152.RGB
    RGB 1=@n151.RGB
   ) {materialFunction=/FortniteCoreUI/Frontend/CosmeticItemCard/Materials/MF_UI_HeadsetSeries}.Result
   RGB 1=@n152.RGB
   RGB 2=@n151.RGB
  ) {materialFunction=/OfferCatalog/Art/FNM/S22/MF_UI_FNMares_ShopTile}
  n165 = MaterialFunctionCall!(
   In=MaterialFunctionCall!(
    In=MaterialFunctionCall!(In=@n164.Result, RGB 1=@n152.RGB, RGB 2=@n151.RGB) {materialFunction=/OfferCatalog/Art/Sahara/MF_UI_SaharaSeries}.Result
    RGB 1=@n152.RGB
    RGB 2=@n151.RGB
   ) {materialFunction=/OfferCatalog/Art/Stallion/MF_UI_Stallion_ShopTile}.Result
   RGB 1=@n152.RGB
   RGB 2=@n151.RGB
  ) {materialFunction=/OfferCatalog/Art/FNCS22/MF_UI_FNCS22_Series}
  n166 = MaterialFunctionCall!(
   In=MaterialFunctionCall!(
    In=MaterialFunctionCall!(In=@n165.Result, RGB 1=@n152.RGB, RGB 2=@n151.RGB) {materialFunction=/OfferCatalog/Art/Billy/MF_UI_Billy_ShopTile}.Result
    RGB 1=@n152.RGB
    RGB 2=@n151.RGB
   ) {materialFunction=/OfferCatalog/Art/FFC/MF_UI_ShopTile_RedOasis}.Result
   RGB 1=@n152.RGB
   RGB 2=@n151.RGB
  ) {materialFunction=/OfferCatalog/Art/Coyote_Trail/MF_UI_CoyoteTrail_ShopTiles}
  n167 = MaterialFunctionCall!(
   In=MaterialFunctionCall!(
    In=MaterialFunctionCall!(In=@n166.Result, RGB 1=@n152.RGB, RGB 2=@n151.RGB) {materialFunction=/OfferCatalog/Art/EmeraldGlass/MF_UI_EmeraldGlass_ShopTiles}.Result
    RGB 1=@n152.RGB
    RGB 2=@n151.RGB
   ) {materialFunction=/OfferCatalog/Art/Lettuce/MF_UI_Lettuce_ShopTiles}.Result
   RGB 1=@n152.RGB
   RGB 2=@n151.RGB
  ) {materialFunction=/OfferCatalog/Art/Sunburst/MF_UI_Sunburst_ShopTile}
  n168 = MaterialFunctionCall!(
   In=MaterialFunctionCall!(
    In=MaterialFunctionCall!(In=@n167.Result, RGB 1=@n152.RGB, RGB 2=@n151.RGB) {materialFunction=/OfferCatalog/Art/InspireSpell/MF_UI_InspireSpell_ShopTile}.Result
    RGB 1=@n152.RGB
    RGB 2=@n151.RGB
   ) {materialFunction=/OfferCatalog/Art/Lucky/MF_UI_Lucky_ShopTile}.Result
   RGB 1=@n152.RGB
   RGB 2=@n151.RGB
  ) {materialFunction=/OfferCatalog/Art/Refeatures/MF_UI_Refeature_ShopTile}
  n169 = MaterialFunctionCall!(
   In=MaterialFunctionCall!(
    In=MaterialFunctionCall!(In=@n168.Result, RGB 1=@n152.RGB, RGB 2=@n151.RGB) {materialFunction=/OfferCatalog/Art/Emphasis/MF_UI_Emphasis_ShopTile}.Result
    RGB 1=@n152.RGB
    RGB 2=@n151.RGB
   ) {materialFunction=/OfferCatalog/Art/FearlessFlight/MF_UI_ShopTile_FearlessFlight}.Result
   RGB 1=@n152.RGB
   RGB 2=@n151.RGB
  ) {materialFunction=/OfferCatalog/Art/Pencil/MF_UI_ShopTile_Pencil}
  n170 = MaterialFunctionCall!(
   In=MaterialFunctionCall!(
    In=MaterialFunctionCall!(In=@n169.Result, RGB 1=@n152.RGB, RGB 2=@n151.RGB) {materialFunction=/OfferCatalog/Art/LastVoice/MF_UI_ShopTile_LastVoice}.Result
    RGB 1=@n152.RGB
    RGB 2=@n151.RGB
   ) {materialFunction=/OfferCatalog/Art/Undefeated/MF_UI_Undefeated_ShopTile}.Result
   RGB 1=@n152.RGB
   RGB 2=@n151.RGB
  ) {materialFunction=/OfferCatalog/Art/Quotient/MF_UI_ShopTile_Quotient}
  n171 = MaterialFunctionCall!(
   In=MaterialFunctionCall!(
    In=MaterialFunctionCall!(In=@n170.Result, RGB 1=@n152.RGB, RGB 2=@n151.RGB) {materialFunction=/OfferCatalog/Art/FNM/S26/MF_UI_FNMares_S26_ShopTile}.Result
    RGB 1=@n152.RGB
    RGB 2=@n151.RGB
   ) {materialFunction=/OfferCatalog/Art/FNCS_GC/MF_UI_FNCS_S26_ShopTile}.Result
   RGB 1=@n152.RGB
   RGB 2=@n151.RGB
  ) {materialFunction=/OfferCatalog/Art/Rufus/MF_UI_Rufus_ShopTile}
  n172 = MaterialFunctionCall!(
   In=MaterialFunctionCall!(
    In=MaterialFunctionCall!(In=@n171.Result, RGB 1=@n152.RGB, RGB 2=@n151.RGB) {materialFunction=/OfferCatalog/Art/InstantGravel/MF_UI_InstantGravel_ShopTile}.Result
    RGB 1=@n152.RGB
    RGB 2=@n151.RGB
   ) {materialFunction=/OfferCatalog/Art/Winterfest/WF_2022/MF_UI_WF23_ShopTiles}.Result
   RGB 1=@n152.RGB
   RGB 2=@n151.RGB
  ) {materialFunction=/OfferCatalog/Art/Ch5S1/MF_UI_Ch5S1_ShopTiles}
  n173 = MaterialFunctionCall!(
   In=MaterialFunctionCall!(
    In=MaterialFunctionCall!(In=@n172.Result, RGB 1=@n152.RGB, RGB 2=@n151.RGB) {materialFunction=/OfferCatalog/Art/Durian/MF_UI_Durian_ShopTiles}.Result
    RGB 1=@n152.RGB
    RGB 2=@n151.RGB
   ) {materialFunction=/OfferCatalog/Art/LinedNotebook/MF_UI_LinedNotebook_ShopTile}.Result
   RGB 1=@n152.RGB
   RGB 2=@n151.RGB
  ) {materialFunction=/OfferCatalog/Art/Skyscraper/MF_UI_Skyscraper_ShopTile}
  n174 = MaterialFunctionCall!(
   In=MaterialFunctionCall!(
    In=MaterialFunctionCall!(In=@n173.Result, RGB 1=@n152.RGB, RGB 2=@n151.RGB) {materialFunction=/OfferCatalog/Art/PastaSauce/MF_UI_PastaSauce_ShopTiles}.Result
    RGB 1=@n152.RGB
    RGB 2=@n151.RGB
   ) {materialFunction=/OfferCatalog/Art/ColdDay/MF_UI_ColdDay_Tile_BG}.Result
   RGB 1=@n152.RGB
   RGB 2=@n151.RGB
  ) {materialFunction=/OfferCatalog/Art/RareDelight/MF_UI_RareDelight_ShopTiles}
  n175 = Add(
   A=MaterialFunctionCall!(
    In=MaterialFunctionCall!(In=@n174.Result, RGB 1=@n152.RGB, RGB 2=@n151.RGB) {materialFunction=/OfferCatalog/Art/Stainless/MF_UI_Stainless_ShopTile}.Result
    RGB 1=@n152.RGB
    RGB 2=@n151.RGB
   ) {materialFunction=/OfferCatalog/Art/Greenhouse/MF_UI_Greenhouse_ShopTile}.Result
   B=ScalarParameter() {defaultValue=0.0125, group=Background, parameterName=Background-Balance, sortPriority=1}
  ) {constB=0.25}
  n176 = VectorParameter() {group=Background, parameterName=MF_RadialCoordinates, defaultValue=<rgba 0.5 0 0 1>}
  n177 = ScalarParameter() {defaultValue=1.35, group=Background, parameterName=Vignette Radius}
  n178 = ScalarParameter() {defaultValue=3, group=Background, parameterName=Vignette Density}
  n179 = Multiply(
   A=OneMinus(None=MaterialFunctionCall(UVs=@n141, CenterPosition=@n176.RGB, Radius=@n177, Density=@n178) {materialFunction=&RadialGradientExponential}.RadialGradientExponential)
   B=ScalarParameter() {group=Background, parameterName=Vignette-Strength}
  )
  n180 = StaticSwitchParameter(
   True=LinearInterpolate(A=TextureSample(UVs=@n123, Tex=@n80) {samplerSource=SSM_Clamp_WorldGroupSettings}.RGB, B=@n131, Alpha=NamedRerouteUsage())
   False=StaticSwitchParameter(True=LinearInterpolate(A=@n175, B=@n151.RGB, Alpha=@n179), False=Saturate(None=@n150.Result)) {group=Global, parameterName=IsSeriesCosmetic}
  ) {group=Background, parameterName=UseTextureBackground}
  n181 = LinearInterpolate(A=@n180, B=@n131, Alpha=NamedRerouteUsage())
  n182 = ScalarParameter() {group=Global, parameterName=UseAnimatedRenderTarget}
  n183 = OneMinus(None=Multiply(A=ScalarParameter() {defaultValue=10, group=Global, parameterName=OffScreen_Amount}) {constB=0.01})
  n184 = Power(Base=@n120, Exp=ScalarParameter() {defaultValue=1.5, group=Global, parameterName=OffScreen_ColorDecayRate}) {constExponent=1.22}
  n185 = LinearInterpolate(
   A=@n122.RGB
   B=StaticSwitchParameter(True=NamedRerouteUsage(), False=LinearInterpolate(A=@n181, B=@n180, Alpha=@n182)) {group=Image, parameterName=IsJustImage}
   Alpha=LinearInterpolate(A=Saturate(None=@n183), Alpha=Saturate(None=@n184))
  )
  n186 = NamedRerouteUsage()
  n187 = VectorParameter() {parameterName=Backing Gradient Override Color, defaultValue=<rgba 0 0 0 1>}
  n188 = ScalarParameter() {parameterName=Backing Gradient Black Override}
  n189 = NamedRerouteUsage()
  n190 = LinearInterpolate(A=@n189, B=Multiply(A=@n189, B=ScalarParameter() {defaultValue=1, parameterName=Backing Gradient Override Opacity}), Alpha=@n188)
  n191 = CollectionParameter() {collection=/FortUILibrary/Systems/ItemShopTile/MPC_ItemShop, parameterName=bEnableDynamicBackingColor}
  n192 = ScalarParameter() {defaultValue=85, group=Global, parameterName=Disabled_Color_Percent}
  n193 = LinearInterpolate(
   A=LinearInterpolate(A=@n185, B=LinearInterpolate(A=@n186, B=@n187.RGB, Alpha=@n188), Alpha=Multiply(A=@n190, B=@n191))
   B=LinearInterpolate(A=@n185, B=@n122.RGB, Alpha=Multiply(A=@n192) {constA=0.01, constB=0.01})
   Alpha=ScalarParameter() {group=Global, parameterName=Enable-Disable}
  )
  StaticSwitchParameter(True=Add(A=Multiply(A=@n193, B=@n121), B=NamedRerouteUsage()), False=@n193) {group=Global, parameterName=ShowDebugImageSizes}
 ]
}
