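# Translucent UI master material for an offer tile status panel: a backing color with an optional icon and one angled edge.
# A rotated, offset UV4 gradient, sharpened with resolution correction, forms the shape mask, scaled by FinalAlpha.
# The icon supports offset, rotation and scale, with tint desaturating and fading it into the backing.
# Hover blends edge angle, offset and backing color, enlarges the icon 15% and nearly removes its tint.
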Material /OfferCatalog/Art/M_UI_OfferTile_StatusPanel_Master
graph {
 properties {
  bUsedWithStaticMesh=false
  blendMode=BLEND_Translucent
  materialDomain=MD_UI
  pixelDepthOffsetMode=PDOM_Legacy
  usedShadingModels=Default Lit
 }
 nodes [
  n1 = TextureCoordinate() {coordinateIndex=4}
  n2 = LinearInterpolate(
   A=ScalarParameter() {defaultValue=-0.245, group=OverallShape, parameterName=Shape_Rotation_Top}
   B=ScalarParameter() {defaultValue=-0.245, group=OverallShape, parameterName=Shape_Rotation_Top_Hover}
   Alpha=ScalarParameter() {parameterName=Hover}
  )
  n3 = ScalarParameter() {defaultValue=-0.075, group=OverallShape, parameterName=Shape_Offset_Top}
  n4 = ScalarParameter() {defaultValue=-0.075, group=OverallShape, parameterName=Shape_Offset_Top_Hover}
  n5 = ScalarParameter() {parameterName=Hover}
  n6 = ComponentMask(None=Add(A=Rotator(Coordinate=@n1, Time=@n2) {speed=6.2832}, B=LinearInterpolate(A=@n3, B=@n4, Alpha=@n5))) {r=true}
  n7 = Multiply(
   A=MaterialFunctionCall!(Native_U_PixelSize=ScalarParameter() {defaultValue=384, group=ResolutionSharpness, parameterName=NativePixelSize}) {materialFunction=/FortUILibrary/Foundation/Art_ToolKit/Tools/M_UI_Resolution_Sharpness_Correction}.Result
   B=ScalarParameter() {defaultValue=64, group=OverallShape, parameterName=Shape_Sharpness_V}
  )
  Multiply(A=Saturate(None=Multiply(A=@n6, B=@n7) {constB=32}), B=ScalarParameter() {defaultValue=1, parameterName=FinalAlpha})
  n8 = LinearInterpolate(
   A=VectorParameter() {parameterName=BackingColor, defaultValue=<rgba 0.005 0.005 0.005 1>}.RGB
   B=VectorParameter() {parameterName=BackingColor_Hover, defaultValue=<rgba 0.2 0.8 0.7 1>}.RGB
   Alpha=ScalarParameter() {parameterName=Hover}
  )
  n9 = ScalarParameter() {parameterName=IconRotation}
  n10 = MaterialFunctionCall() {materialFunction=&GetUserInterfaceUV}
  n11 = AppendVector(
   A=Multiply(A=ScalarParameter() {parameterName=OffsetImage_X_Percent}) {constB=0.01}
   B=Multiply(A=ScalarParameter() {parameterName=OffsetImage_Y_Percent}) {constB=0.01}
  )
  n12 = MaterialFunctionCall(Float2=@n10.Pixel Size) {materialFunction=&BreakOutFloat2Components}
  n13 = Divide(A=@n12.R, B=@n12.R)
  n14 = Constant() {r=1}
  n15 = OneMinus(None=@n13)
  n16 = MaterialFunctionCall(Float2=AppendVector(A=Multiply(A=@n15) {constB=0.5}, B=Constant())) {materialFunction=&BreakOutFloat2Components}
  n17 = ScalarParameter() {parameterName=ShiftPixels}
  n18 = Add(A=Multiply(A=Add(A=@n10.Normalized UV, B=@n11), B=AppendVector(A=@n13, B=@n14)), B=AppendVector(A=Add(A=@n16.R, B=@n17), B=@n16.G))
  n19 = ScalarParameter() {defaultValue=100, parameterName=ScaleImage_Percent}
  n20 = LinearInterpolate(A=@n19, B=Add(A=@n19, B=Multiply(A=@n19) {constB=0.15}), Alpha=ScalarParameter() {parameterName=Hover})
  n21 = TextureSample(
   UVs=MaterialFunctionCall(
    UVs=MaterialFunctionCall!(Rotation_Fraction=@n9, TexCoord=@n18) {materialFunction=/FortUILibrary/Foundation/Art_ToolKit/AngledBoxes/MF_UI_Rotator_Fractions}.Result
    Texture Scale=Multiply(A=@n20) {constB=0.01}
   ) {materialFunction=&ScaleUVsByCenter}.UVs
   Tex=TextureObjectParameter() {parameterName=Icon, texture=/Game/UI/Foundation/Textures/Icons/Items/T-Items-MTX}
  ) {samplerSource=SSM_Clamp_WorldGroupSettings}
  n22 = Multiply(A=LinearInterpolate(A=ScalarParameter() {defaultValue=25, parameterName=TintPercent}, Alpha=ScalarParameter() {parameterName=Hover})) {constB=0.01}
  n23 = Multiply(A=@n22, B=ScalarParameter() {defaultValue=1, parameterName=EffectDesatuation})
  n24 = OneMinus(None=@n22)
  StaticSwitchParameter(True=LinearInterpolate(A=@n8, B=Desaturation(None=@n21.RGB, Fraction=@n23), Alpha=Multiply(A=@n21.A, B=@n24)), False=@n8) {defaultValue=true, parameterName=HasIcon}
 ]
}
